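# Modal CommonUI prompt where a spectator types a file name to save or load broadcast spectator settings.
# MenuConfig selects the Save (0) or Load (1) wording and whether Confirm calls SaveSpectatorSettings or LoadSpectatorSettings on FortPlayerControllerSpectating.
# Blank save names and failed loads show an inline error, while save results go unchecked; typed names are trimmed to 30 characters.
# Opening switches to UI-only input with text focus; closing restores game-and-UI input.
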WidgetBlueprint /FortniteCoreUI/UI/Broadcast/FortBroadcastSaveLoadPrompt
animations [Intro {
  binding_count=2
  length_seconds=0.25
  object=&self:Intro
  sequence {bindings [
    CoachMark {
     id=#1
     possessed_class=WBP_UIKit_CoachMark_C
     tracks [{
       type=MovieSceneFloatTrack
       display_name=Render Opacity
       sections [{type=MovieSceneFloatSection, end_frame=5, start_frame=0, channels [None {type=MovieSceneScriptingFloatChannel, keys [[0,0],[5,1]]}]}]
      }]
    }
    BackingPrompt {
     id=#2
     possessed_class=CommonBorder
     tracks [
      {
       type=MovieSceneFloatTrack
       display_name=Render Opacity
       sections [{type=MovieSceneFloatSection, end_frame=5, start_frame=0, channels [None {type=MovieSceneScriptingFloatChannel, keys [[0,0],[2,0],[5,1]]}]}]
      }
      {
       type=MovieScene2DTransformTrack
       sections [{
         type=MovieScene2DTransformSection
         end_frame=5
         start_frame=0
         channels [
          Scale.X {type=MovieSceneScriptingFloatChannel, keys [[2,1.15],[4,1]]}
          Scale.Y {type=MovieSceneScriptingFloatChannel, keys [[2,1.15],[4,1]]}
         ]
        }]
      }
     ]
    }
   ]}
 }]
graphs [
 BP_GetDesiredFocusTarget {nodes [
   Get Desired Focus Target()
   > Return Node!(execute, ReturnValue:Widget Object Reference=Get EditableText_Value() -> (EditableText_Value:Editable Text Object Reference))
  ]}
 ClickAcceptButton {nodes [
   ClickAcceptButton() {is=title K2Node_FunctionEntry}
   > n1 = Branch(Condition="Equal (Integer)"(A=Get MenuConfig() -> (MenuConfig:int=0)))
   n2 = Get EditableText_Value()
   > n3 = Branch(Condition="Equal Exactly (Text)"!(
     A:"Text (by ref)"="GetText (Editable Text)"!(self:Editable Text Object Reference=@n2) -> (ReturnValue:Text)
     B:"Text (by ref)"=Get LocBlankText() -> (LocBlankText:Text)
    ) -> (ReturnValue:bool))
   > ShowErrorMessage!(execute, self, Message:Text="LOCTEXT(\"Please enter a valid file name.\")") -> (then)
   n4 = Cast To FortPlayerControllerSpectating!(execute=@n3.else, Object=GetOwningPlayer()) -> (then, CastFailed, AsFort Player Controller Spectating:Fort Player Controller Spectating Object Reference) {is=title K2Node_DynamicCast}
   > n5 = SaveSpectatorSettings!(execute, self:Fort Player Controller Spectating Object Reference=@n4, FileName:String=ReturnEditableText!() -> (ReturnValue:String)) -> (then, ReturnValue:bool)
   RemoveFromParent(execute=[@n5.then, @n6.CastFailed, @n8.then])
   > Set Input Mode Game And UI!(
    execute
    PlayerController:Player Controller Object Reference=GetPlayerController!(PlayerIndex:int=0) -> (ReturnValue:Player Controller Object Reference)
    InWidgetToFocus:Widget Object Reference=Self-Reference()
    InMouseLockMode:EMouseLockMode Enum=DoNotLock
    bHideCursorDuringCapture:bool=true
    bFlushInput:bool=false
   ) -> (then) {is=title K2Node_CallFunction}
   Branch(execute=@n1.else, Condition="Equal (Integer)"(A=Get MenuConfig() -> (MenuConfig=0), B=1))
   > n6 = Cast To FortPlayerControllerSpectating(Object=GetOwningPlayer())
   > n7 = LoadSpectatorSettings!(execute, self:Fort Player Controller Spectating Object Reference=@n6, FileName:String=ReturnEditableText()) -> (then, ReturnValue:bool)
   > n8 = Branch(Condition=@n7)
   ShowErrorMessage(execute=@n8.else, Message="LOCTEXT(\"File not found. Please try again.\")")
   Reroute Node(InputPin:Wildcard) -> (OutputPin:Wildcard)
  ]}
 EventGraph {nodes [
   "On Clicked (Button_Decline)"!() -> (OutputDelegate, then, Button:Common Button Base Object Reference) {is=title K2Node_ComponentBoundEvent}
   > RemoveFromParent()
   > Set Input Mode Game And UI(
    PlayerController=GetPlayerController(PlayerIndex=0)
    InWidgetToFocus=Self-Reference()
    InMouseLockMode=DoNotLock
    bHideCursorDuringCapture=true
    bFlushInput=false
   )
   "On Clicked (Button_Accept)"!() -> (OutputDelegate, then, Button:Common Button Base Object Reference) {is=title K2Node_ComponentBoundEvent}
   > ClickAcceptButton() {is=title K2Node_CallFunction}
   n1 = "OnTextChanged (Editable Text) (EditableText_Value)"!() -> (OutputDelegate, then, Text:"Text (by ref)") {is=title K2Node_ComponentBoundEvent}
   n2 = "To String (Text)"(InText=@n1.Text)
   > "SetText (Editable Text)"!(
    execute
    self:Editable Text Object Reference=Get EditableText_Value()
    InText:Text="To Text (String)"(InString=GetSubstring!(SourceString:String=@n2, StartIndex:int=0, Length:int=30) -> (ReturnValue:String))
   ) -> (then)
   Event PreConstruct()
   > SetAppearance!(execute, self, Config:int=0) -> (then)
   Event Construct()
   > n3 = Cast To FortPlayerControllerSpectating(Object=GetOwningPlayer())
   > SetAppearance(Config=Get MenuConfig() -> (MenuConfig=0))
   > PlayAnimationForward(InAnimation=Get Intro())
   > SpawnSound2D!(
    execute
    Sound:Sound Base Object Reference=Get IntroSound() -> (IntroSound:Sound Base Object Reference)
    VolumeMultiplier:float=1
    PitchMultiplier:float=1
    StartTime:float=0
    ConcurrencySettings:Sound Concurrency Object Reference
    bPersistAcrossLevelTransition:bool=false
    bAutoDestroy:bool=true
   ) -> (then, ReturnValue:Audio Component Object Reference)
   > Set Input Mode UI Only!(
    execute
    PlayerController:Player Controller Object Reference=@n3
    InWidgetToFocus:Widget Object Reference=Self-Reference()
    InMouseLockMode:EMouseLockMode Enum=DoNotLock
    bFlushInput:bool=false
   ) -> (then) {is=title K2Node_CallFunction}
   > SetKeyboardFocus(self:Widget Object Reference=Get EditableText_Value())
  ]}
 ReturnEditableText {nodes [
   ReturnEditableText() {is=title K2Node_FunctionEntry}
   n1 = Get EditableText_Value()
   > Return Node(ReturnValue:String="To String (Text)"(InText="GetText (Editable Text)"(self=@n1)))
  ]}
 SetAppearance {nodes [
   n1 = SetAppearance!() -> (then, Config:int) {is=title K2Node_FunctionEntry}
   > n2 = Set LocConfig(LocConfig:int=@n1) -> (Output_Get:int=0)
   > n3 = Branch(Condition="Equal (Integer)"(A=@n2))
   > SetVisibility(self=Get EditableText_Box() -> (EditableText_Box:Common Border Object Reference), InVisibility=Visible)
   > "SetText (Text)"(self=Get Title() -> (Title:Common Text Object Reference), InText="LOCTEXT(\"Save Settings\")")
   > "SetText (Text)"(self=Get Description() -> (Description:Common Text Object Reference), InText="LOCTEXT(\"Enter a name for your settings:\")")
   Branch(execute=@n3.else, Condition="Equal (Integer)"(A=Get LocConfig() -> (LocConfig:int=0), B=1))
   > SetVisibility(self=Get EditableText_Box(), InVisibility=Visible)
   > "SetText (Text)"(self=Get Title(), InText="LOCTEXT(\"Load Settings\")")
   > "SetText (Text)"(self=Get Description(), InText="LOCTEXT(\"Enter the name of the file you\\'d like to load:\")")
  ]}
 ShowErrorMessage {nodes [
   n1 = ShowErrorMessage!() -> (then, Message:Text) {is=title K2Node_FunctionEntry}
   > "SetText (Text)"(self=Get ErrorMessage() -> (ErrorMessage:Common Text Object Reference), InText=@n1)
   > SetVisibility(self=Get ErrorMessage(), InVisibility=Visible)
  ]}
]
variables [
 IntroSound {type=SoundBase, default=/Game/Sounds/Fort_UI_Sounds/FrontEnd/Whooshes/Toasts/UI_Whoosh_Generic_Toast_Appear_Cue}
 MenuConfig {type=integer}
 ViewSettings {type=FortBroadcastViewSettings_C, default=null}
]
widgets {
 info {inheritedWidgetCount=0, parentClass=&CommonActivatableWidget, rootWidgetClass=&CanvasPanel, widgetCount=24}
 widgets [
  CanvasPanel_47 {widgetClassPath=&CanvasPanel, widget=CanvasPanel_47}
  Overlay {
   widgetClassPath=&Overlay
   parent=CanvasPanel_47
   widget=Overlay
   slot {class=CanvasPanelSlot, properties {layoutData {alignment=<0 0>, anchors {maximum=<1 1>, minimum=<0 0>}, offsets=<ltrb 0 0 0 0>}}}
  }
  Blur {
   widgetClassPath=&BackgroundBlur
   parent=Overlay
   widget=Blur
   properties {blurStrength=15}
   slot {class=OverlaySlot, properties {horizontalAlignment=HAlign_Fill, verticalAlignment=VAlign_Fill}}
  }
  BackingColor {
   widgetClassPath=&Border
   parent=Overlay
   widget=BackingColor
   properties {
    renderOpacity=0.85
    background {
     bIsDynamicallyLoaded=false
     drawAs=Image
     imageType=NoImage
     mirroring=NoMirror
     resourceName=None
     resourceObject=null
     tiling=NoTile
     imageSize=<32 32>
     margin=<ltrb 0 0 0 0>
     outlineSettings {
      bUseBrushTransparency=false
      roundingType=HalfHeightRadius
      width=0
      color {colorUseRule=UseColor_Specified, specifiedColor=<rgba 0 0 0 0>}
      cornerRadii=<0 0 0 0>
     }
     tintColor {colorUseRule=UseColor_Specified, specifiedColor=<rgba 0 0.0006 0.0075 1>}
     uVRegion {bIsValid=false, max=<0 0>, min=<0 0>}
    }
    contentColorAndOpacity=<rgba 0 0.0006 0.0075 1>
   }
   slot {class=OverlaySlot, properties {horizontalAlignment=HAlign_Fill, verticalAlignment=VAlign_Fill}}
  }
  BackingGradient {
   widgetClassPath=&Border
   parent=Overlay
   widget=BackingGradient
   properties {
    background {
     bIsDynamicallyLoaded=false
     drawAs=Image
     imageType=NoImage
     mirroring=NoMirror
     resourceName=None
     resourceObject=/Game/Athena/UI/Frontend/Art/M_UI_RadialGradientMaster_Inst
     tiling=NoTile
     imageSize=<32 32>
     margin=<ltrb 0 0 0 0>
     outlineSettings {
      bUseBrushTransparency=false
      roundingType=HalfHeightRadius
      width=0
      color {colorUseRule=UseColor_Specified, specifiedColor=<rgba 0 0 0 0>}
      cornerRadii=<0 0 0 0>
     }
     tintColor {colorUseRule=UseColor_Specified, specifiedColor=<rgba 1 1 1 1>}
     uVRegion {bIsValid=false, max=<0 0>, min=<0 0>}
    }
    brushColor=<rgba 0.0052 0.0052 0.0056 0.5>
    padding=<ltrb 0 0 0 0>
   }
   slot {class=OverlaySlot, properties {horizontalAlignment=HAlign_Fill, verticalAlignment=VAlign_Fill}}
  }
  SizeBox_297 {
   widgetClassPath=&SizeBox
   parent=Overlay
   widget=SizeBox_297
   properties {
    bOverride_MaxAspectRatio=true
    bOverride_MinDesiredHeight=true
    bOverride_WidthOverride=true
    maxAspectRatio=2.5
    minDesiredHeight=600
    widthOverride=1000
   }
   slot {class=OverlaySlot, properties {horizontalAlignment=HAlign_Center, verticalAlignment=VAlign_Center}}
  }
  BackingPrompt {
   widgetClassPath=&CommonBorder
   parent=SizeBox_297
   widget=BackingPrompt
   properties {
    style=/FortUILibrary/Styles/Backgrounds/_BorderStyle_Modal._BorderStyle_Modal_C
    background {
     bIsDynamicallyLoaded=false
     drawAs=Image
     imageType=NoImage
     mirroring=NoMirror
     resourceName=None
     resourceObject=null
     tiling=NoTile
     imageSize=<64 64>
     margin=<ltrb 0 0 0 0>
     outlineSettings {
      bUseBrushTransparency=false
      roundingType=HalfHeightRadius
      width=0
      color {colorUseRule=UseColor_Specified, specifiedColor=<rgba 0 0 0 0>}
      cornerRadii=<0 0 0 0>
     }
     tintColor {colorUseRule=UseColor_Specified, specifiedColor=<rgba 0.0024 0.016 0.117 1>}
     uVRegion {bIsValid=false, max=<0 0>, min=<0 0>}
    }
    padding=<ltrb 0 0 0 0>
   }
   slot {class=SizeBoxSlot}
  }
  Overlay_0 {widgetClassPath=&Overlay, parent=BackingPrompt, widget=Overlay_0, slot {class=BorderSlot, properties {padding=<ltrb 0 0 0 0>}}}
  SettingsBox {
   widgetClassPath=&VerticalBox
   parent=Overlay_0
   widget=SettingsBox
   slot {class=OverlaySlot, properties {horizontalAlignment=HAlign_Fill, verticalAlignment=VAlign_Fill}}
  }
  CommonBorder_Modal {
   widgetClassPath=&CommonBorder
   parent=SettingsBox
   widget=CommonBorder_Modal
   properties {
    style=/FortUILibrary/Styles/Backgrounds/_BorderStyle_ModalHeader._BorderStyle_ModalHeader_C
    background {
     bIsDynamicallyLoaded=false
     drawAs=Image
     imageType=NoImage
     mirroring=NoMirror
     resourceName=None
     resourceObject=null
     tiling=NoTile
     imageSize=<64 64>
     margin=<ltrb 0 0 0 0>
     outlineSettings {
      bUseBrushTransparency=false
      roundingType=HalfHeightRadius
      width=0
      color {colorUseRule=UseColor_Specified, specifiedColor=<rgba 0 0 0 0>}
      cornerRadii=<0 0 0 0>
     }
     tintColor {colorUseRule=UseColor_Specified, specifiedColor=<rgba 0.0009 0.006 0.0395 1>}
     uVRegion {bIsValid=false, max=<0 0>, min=<0 0>}
    }
    padding=<ltrb 40 0 4 0>
   }
   slot {class=VerticalBoxSlot}
  }
  HorizontalBox_1 {
   widgetClassPath=&HorizontalBox
   parent=CommonBorder_Modal
   widget=HorizontalBox_1
   slot {class=BorderSlot, properties {padding=<ltrb 40 0 4 0>}}
  }
  Spacer_0 {widgetClassPath=&Spacer, parent=HorizontalBox_1, widget=Spacer_0, properties {size=<0 85>}, slot {class=HorizontalBoxSlot}}
  Title {
   widgetClassPath=&CommonTextBlock
   bIsVariable=true
   parent=HorizontalBox_1
   widget=Title
   properties {
    style=/FortUILibrary/Styles/Typography/_TextStyle_SubH2._TextStyle_SubH2_C
    text=Save Settings
    textTransformPolicy=ToUpper
    font {
     bForceMonospaced=false
     bMaterialIsStencil=false
     fontMaterial=null
     fontObject=&Burbank
     letterSpacing=0
     monospacedWidth=1
     size=33
     skewAmount=0
     typefaceFontName=BurbankRegular-Black
     outlineSettings {
      bApplyOutlineToDropShadows=false
      bMiteredCorners=false
      bSeparateFillAlpha=false
      outlineMaterial=null
      outlineSize=0
      outlineColor=<rgba 0 0 0 1>
     }
    }
    shadowOffset=<0 0>
    strikeBrush {
     bIsDynamicallyLoaded=false
     drawAs=NoDrawType
     imageType=NoImage
     mirroring=NoMirror
     resourceName=None
     resourceObject=null
     tiling=NoTile
     imageSize=<32 32>
     margin=<ltrb 0 0 0 0>
     outlineSettings {
      bUseBrushTransparency=false
      roundingType=HalfHeightRadius
      width=0
      color {colorUseRule=UseColor_Specified, specifiedColor=<rgba 0 0 0 0>}
      cornerRadii=<0 0 0 0>
     }
     tintColor {colorUseRule=UseColor_Specified, specifiedColor=<rgba 1 1 1 1>}
     uVRegion {bIsValid=false, max=<0 0>, min=<0 0>}
    }
   }
   slot {class=HorizontalBoxSlot, properties {horizontalAlignment=HAlign_Left, verticalAlignment=VAlign_Center, padding=<ltrb 8 0 20 0>}}
  }
  VisibilityBorder {
   widgetClassPath=&CommonHardwareVisibilityBorder
   parent=HorizontalBox_1
   widget=VisibilityBorder
   properties {
    horizontalAlignment=HAlign_Right
    visibilityQuery {
     autoDescription=" ANY( Input.Touch )"
     tokenStreamVersion=0
     userDescription=if using touch
     queryTokenStream [0,1,1,1,0]
     tagDictionary [{tagName=Input.Touch}]
    }
   }
   slot {class=HorizontalBoxSlot, properties {horizontalAlignment=HAlign_Right, size {sizeRule=Fill, value=1}}}
  }
  Description {
   widgetClassPath=&CommonTextBlock
   bIsVariable=true
   parent=SettingsBox
   widget=Description
   properties {
    lineHeightPercentage=1.29
    style=/FortUILibrary/Styles/Typography/_TextStyle_Body1._TextStyle_Body1_C
    text=Are you sure you want to respawn?
    font {
     bForceMonospaced=false
     bMaterialIsStencil=false
     fontMaterial=null
     fontObject=&BurbankSmall
     letterSpacing=0
     monospacedWidth=1
     size=18
     skewAmount=0
     typefaceFontName=Bold
     outlineSettings {
      bApplyOutlineToDropShadows=false
      bMiteredCorners=false
      bSeparateFillAlpha=false
      outlineMaterial=null
      outlineSize=0
      outlineColor=<rgba 0 0 0 1>
     }
    }
    shadowOffset=<0 0>
    strikeBrush {
     bIsDynamicallyLoaded=false
     drawAs=NoDrawType
     imageType=NoImage
     mirroring=NoMirror
     resourceName=None
     resourceObject=null
     tiling=NoTile
     imageSize=<32 32>
     margin=<ltrb 0 0 0 0>
     outlineSettings {
      bUseBrushTransparency=false
      roundingType=HalfHeightRadius
      width=0
      color {colorUseRule=UseColor_Specified, specifiedColor=<rgba 0 0 0 0>}
      cornerRadii=<0 0 0 0>
     }
     tintColor {colorUseRule=UseColor_Specified, specifiedColor=<rgba 1 1 1 1>}
     uVRegion {bIsValid=false, max=<0 0>, min=<0 0>}
    }
   }
   slot {class=VerticalBoxSlot, properties {padding=<ltrb 48 34 48 0>}}
  }
  EditableText_Box {
   widgetClassPath=&CommonBorder
   bIsVariable=true
   parent=SettingsBox
   widget=EditableText_Box
   properties {
    verticalAlignment=VAlign_Center
    background {
     bIsDynamicallyLoaded=false
     drawAs=Image
     imageType=NoImage
     mirroring=NoMirror
     resourceName=None
     resourceObject=null
     tiling=NoTile
     imageSize=<32 32>
     margin=<ltrb 0 0 0 0>
     outlineSettings {
      bUseBrushTransparency=false
      roundingType=HalfHeightRadius
      width=0
      color {colorUseRule=UseColor_Specified, specifiedColor=<rgba 0 0 0 0>}
      cornerRadii=<0 0 0 0>
     }
     tintColor {colorUseRule=UseColor_Specified, specifiedColor=<rgba 1 1 1 1>}
     uVRegion {bIsValid=false, max=<0 0>, min=<0 0>}
    }
    padding=<ltrb 16 0 16 0>
   }
   slot {class=VerticalBoxSlot, properties {verticalAlignment=VAlign_Top, padding=<ltrb 48 10 48 0>}}
  }
  HorizontalBox_0 {
   widgetClassPath=&HorizontalBox
   parent=EditableText_Box
   widget=HorizontalBox_0
   slot {class=BorderSlot, properties {verticalAlignment=VAlign_Center, padding=<ltrb 16 0 16 0>}}
  }
  EditableText_Value {
   widgetClassPath=/Script/UMG.EditableText
   bIsVariable=true
   parent=HorizontalBox_0
   widget=EditableText_Value
   properties {
    hintText=Save Name
    navigation=&self:WidgetTree.EditableText_Value.WidgetNavigation_0
    widgetStyle {
     backgroundImageComposing {
      bIsDynamicallyLoaded=false
      drawAs=NoDrawType
      imageType=FullColor
      mirroring=NoMirror
      resourceName=../../../Engine/Content/Slate/Old/HyperlinkDotted.png
      resourceObject=null
      tiling=Both
      imageSize=<0 0>
      margin=<ltrb 0 0 0 0.1875>
      outlineSettings {
       bUseBrushTransparency=false
       roundingType=HalfHeightRadius
       width=0
       color {colorUseRule=UseColor_Specified, specifiedColor=<rgba 0 0 0 0>}
       cornerRadii=<0 0 0 0>
      }
      tintColor {colorUseRule=UseColor_Specified, specifiedColor=<rgba 1 1 1 1>}
      uVRegion {bIsValid=false, max=<0 0>, min=<0 0>}
     }
     backgroundImageSelected {
      bIsDynamicallyLoaded=false
      drawAs=NoDrawType
      imageType=FullColor
      mirroring=NoMirror
      resourceName=../../../Engine/Content/Slate/Common/EditableTextSelectionBackground.png
      resourceObject=null
      tiling=NoTile
      imageSize=<0 0>
      margin=<ltrb 0.25 0.25 0.25 0.25>
      outlineSettings {
       bUseBrushTransparency=false
       roundingType=HalfHeightRadius
       width=0
       color {colorUseRule=UseColor_Specified, specifiedColor=<rgba 0 0 0 0>}
       cornerRadii=<0 0 0 0>
      }
      tintColor {colorUseRule=UseColor_Specified, specifiedColor=<rgba 1 1 1 1>}
      uVRegion {bIsValid=false, max=<0 0>, min=<0 0>}
     }
     caretImage {
      bIsDynamicallyLoaded=false
      drawAs=Image
      imageType=FullColor
      mirroring=NoMirror
      resourceName=../../../Engine/Content/Slate/Old/White.png
      resourceObject=null
      tiling=NoTile
      imageSize=<16 16>
      margin=<ltrb 0 0 0 0>
      outlineSettings {
       bUseBrushTransparency=false
       roundingType=HalfHeightRadius
       width=0
       color {colorUseRule=UseColor_Specified, specifiedColor=<rgba 0 0 0 0>}
       cornerRadii=<0 0 0 0>
      }
      tintColor {colorUseRule=UseColor_Specified, specifiedColor=<rgba 1 1 1 1>}
      uVRegion {bIsValid=false, max=<0 0>, min=<0 0>}
     }
     colorAndOpacity {colorUseRule=UseColor_Specified, specifiedColor=<rgba 0 0 0 1>}
     font {
      bForceMonospaced=false
      bMaterialIsStencil=false
      fontMaterial=null
      fontObject=&BurbankSmall
      letterSpacing=0
      monospacedWidth=1
      size=16
      skewAmount=0
      typefaceFontName=Bold
      outlineSettings {
       bApplyOutlineToDropShadows=false
       bMiteredCorners=false
       bSeparateFillAlpha=false
       outlineMaterial=null
       outlineSize=0
       outlineColor=<rgba 0 0 0 1>
      }
     }
     hintColorAndOpacity {colorUseRule=UseColor_Foreground, specifiedColor=<rgba 1 0 1 1>}
     selectedBackgroundColor {colorUseRule=UseColor_Foreground, specifiedColor=<rgba 1 0 1 1>}
    }
   }
   slot {class=HorizontalBoxSlot, properties {verticalAlignment=VAlign_Center, size {sizeRule=Fill, value=1}}}
  }
  Spacer_EditText {
   widgetClassPath=&Spacer
   parent=HorizontalBox_0
   widget=Spacer_EditText
   properties {size=<0 56>}
   slot {class=HorizontalBoxSlot, properties {horizontalAlignment=HAlign_Left, verticalAlignment=VAlign_Top}}
  }
  ErrorMessage {
   widgetClassPath=&CommonTextBlock
   bIsVariable=true
   parent=SettingsBox
   widget=ErrorMessage
   properties {
    lineHeightPercentage=1.29
    text=Error message appears here for various feedback
    visibility=Collapsed
    colorAndOpacity {colorUseRule=UseColor_Specified, specifiedColor=<rgba 1 0.1981 0.0048 1>}
    font {
     bForceMonospaced=false
     bMaterialIsStencil=false
     fontMaterial=null
     fontObject=&BurbankSmall
     letterSpacing=0
     monospacedWidth=1
     size=13.5
     skewAmount=0
     typefaceFontName=Bold
     outlineSettings {
      bApplyOutlineToDropShadows=false
      bMiteredCorners=false
      bSeparateFillAlpha=false
      outlineMaterial=null
      outlineSize=0
      outlineColor=<rgba 0 0 0 1>
     }
    }
    shadowOffset=<0 0>
    strikeBrush {
     bIsDynamicallyLoaded=false
     drawAs=NoDrawType
     imageType=NoImage
     mirroring=NoMirror
     resourceName=None
     resourceObject=null
     tiling=NoTile
     imageSize=<32 32>
     margin=<ltrb 0 0 0 0>
     outlineSettings {
      bUseBrushTransparency=false
      roundingType=HalfHeightRadius
      width=0
      color {colorUseRule=UseColor_Specified, specifiedColor=<rgba 0 0 0 0>}
      cornerRadii=<0 0 0 0>
     }
     tintColor {colorUseRule=UseColor_Specified, specifiedColor=<rgba 1 1 1 1>}
     uVRegion {bIsValid=false, max=<0 0>, min=<0 0>}
    }
   }
   slot {class=VerticalBoxSlot, properties {padding=<ltrb 48 15 48 0>}}
  }
  HorizontalBox_6 {
   widgetClassPath=&HorizontalBox
   parent=SettingsBox
   widget=HorizontalBox_6
   slot {class=VerticalBoxSlot, properties {verticalAlignment=VAlign_Bottom, padding=<ltrb 48 0 48 32>, size {sizeRule=Fill, value=1}}}
  }
  Button_Accept {
   widgetClassPath=&WBP_UIKit_Button_Regular
   bIsVariable=true
   parent=HorizontalBox_6
   widget=Button_Accept
   properties {
    text=Confirm
    pressedSlateSoundOverride {resourceObject=/Game/Sounds/Fort_UI_Sounds/FrontEnd/Whooshes/Toasts/UI_Whoosh_Generic_Toast_Close_Cue}
   }
   slot {class=HorizontalBoxSlot, properties {size {sizeRule=Fill, value=1}}}
  }
  Button_Decline {
   widgetClassPath=&WBP_UIKit_Button_Regular
   bIsVariable=true
   parent=HorizontalBox_6
   widget=Button_Decline
   properties {
    text=Decline
    pressedSlateSoundOverride {resourceObject=&UI_Whoosh_Generic_Toast_Close_Cue}
   }
   slot {class=HorizontalBoxSlot, properties {padding=<ltrb 48 0 0 0>, size {sizeRule=Fill, value=1}}}
  }
  CoachMark {
   widgetClassPath=/FortUILibrary/UIKit/CoachMark/WBP_UIKit_CoachMark
   bIsVariable=true
   parent=Overlay
   widget=CoachMark
   properties {
    arrowPosition=Top_Center
    dismissButtonClass=null
    justificationDescription=Center
    otherActionButtonClass=null
    shouldShowDismissButton=false
    textDescription="For now, this menu only supports input from mouse and keyboard."
    textDismissButton=""
    textHeader=""
    textOtherActionButton=""
    visibility=Visible
    widget Width=700
   }
   slot {class=OverlaySlot, properties {horizontalAlignment=HAlign_Center, verticalAlignment=VAlign_Bottom, padding=<ltrb 0 0 0 50>}}
  }
 ]
}
