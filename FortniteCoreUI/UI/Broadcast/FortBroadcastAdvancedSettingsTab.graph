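# FortBroadcastAdvancedSettingsTab is the replay spectator's advanced settings panel for drone camera movement, tilt, pendulum swing, bone/throwable attachment and camera shot presets.
# PreConstruct initializes the rotators and sliders, whose events map values to ranges and call the parent tab's setters, some also toggling dependent row visibility.
# Shot buttons and rotators invoke, swap or override hotkey shots, some through the spectating controller.
# LatestHoveredWidget records hovered or focused rows to restore gamepad focus.
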WidgetBlueprint /FortniteCoreUI/UI/Broadcast/FortBroadcastAdvancedSettingsTab
graphs [
 CloseTabGroup {nodes [
   CloseTabGroup() {is=title K2Node_FunctionEntry}
   > SetTabCollapsed!(
    execute
    self:Fort Replay Context Object Reference=Get FortReplayContext!() -> (ReturnValue:Fort Replay Context Object Reference) {is=title K2Node_GetBlueprintContextEx}
    bIsCollapsed:bool=true
   ) -> (then)
  ]}
 EventGraph {nodes [
   n1 = "On Value Commited (DroneDutch)"!() -> (OutputDelegate, then, Value:double) {is=title K2Node_ComponentBoundEvent}
   > SetDroneDutchAngle!(execute, self:Fort Spectator Advance Settings Tab Object Reference, DroneDutchAngle:float=@n1.Value) -> (then)
   n2 = "Selection Changed (DroneDirectMovement)"!() -> (OutputDelegate, then, Selected Index:int=0) {is=title K2Node_ComponentBoundEvent}
   > SetDroneDirectMovementEnabled!(
    execute
    self:Fort Spectator Advance Settings Tab Object Reference
    bDroneDirectMovementEnabled:bool="integer > integer"(A=@n2.Selected Index, B=0) {is=title K2Node_CallFunction}
   ) -> (then)
   > ExecuteConsoleCommand!(execute, Command=Slate.EnableGlobalInvalidation 0, SpecificPlayer=GetOwningPlayer()) -> (then)
   "On Focused (DroneLinearVelocityZmultiplier)"() {is=title K2Node_ComponentBoundEvent}
   > Set LatestHoveredWidget(LatestHoveredWidget:User Widget Object Reference=Get DroneLinearVelocityZmultiplier() -> (DroneLinearVelocityZmultiplier:Options Menu Slider Replay Object Reference)) -> (Output_Get:User Widget Object Reference)
   "On Focused (DroneXInputLock)"() {is=title K2Node_ComponentBoundEvent}
   > Set LatestHoveredWidget(LatestHoveredWidget=Get DroneXInputLock() -> (DroneXInputLock:Rotator Selector Replay Settings Object Reference))
   n3 = "Selection Changed (DroneYInputLock)"!() -> (OutputDelegate, then, Selected Index:int=0) {is=title K2Node_ComponentBoundEvent}
   > SetDroneYInputLockEnabled!(
    execute
    self:Fort Spectator Advance Settings Tab Object Reference
    bDroneYInputLockEnabled:bool="integer > integer"(A=@n3.Selected Index, B=0)
   ) -> (then)
   "On Hovered (DroneYInputLock)"() {is=title K2Node_ComponentBoundEvent}
   > Set LatestHoveredWidget(LatestHoveredWidget=Get DroneYInputLock() -> (DroneYInputLock:Rotator Selector Replay Settings Object Reference))
   "On Focused (DroneYInputLock)"() {is=title K2Node_ComponentBoundEvent}
   > Set LatestHoveredWidget(LatestHoveredWidget=Get DroneYInputLock())
   n4 = "Selection Changed (ShowCameraDummy)"!() -> (OutputDelegate, then, Selected Index:int=0) {is=title K2Node_ComponentBoundEvent}
   > ToggleDummyCamera!(
    execute
    self:Fort Spectator Advance Settings Tab Object Reference
    bActive:bool="integer > integer"(A=@n4.Selected Index) {is=title K2Node_PromotableOperator}
   ) -> (then)
   "On Hovered (ShowCameraDummy)"() {is=title K2Node_ComponentBoundEvent}
   > Set LatestHoveredWidget(LatestHoveredWidget=Get ShowCameraDummy() -> (ShowCameraDummy:Rotator Selector Replay Settings Object Reference))
   n5 = "Slider Changed (DroneDutch)"!() -> (OutputDelegate, then, Slider Value:double) {is=title K2Node_ComponentBoundEvent}
   n6 = GetMinMaxDroneDutchAngle!(self:Fort Spectator Advance Settings Tab Object Reference) -> (MinAngle:float, MaxAngle:float)
   > SetDroneDutchAngle(DroneDutchAngle=MapRangeClamped(Value=@n5.Slider Value, InRangeA=0, InRangeB=1, OutRangeA=@n6, OutRangeB=@n6.MaxAngle))
   "On Hovered (DroneDirectMovement)"() {is=title K2Node_ComponentBoundEvent}
   > Set LatestHoveredWidget(LatestHoveredWidget=Get DroneDirectMovement() -> (DroneDirectMovement:Common User Widget Object Reference))
   "On Hovered (DutchControlScheme)"() {is=title K2Node_ComponentBoundEvent}
   > Set LatestHoveredWidget(LatestHoveredWidget=Get DutchControlScheme() -> (DutchControlScheme:Rotator Selector Replay Settings Object Reference))
   "On Focused (DroneLinearVelocity)"() {is=title K2Node_ComponentBoundEvent}
   > Set LatestHoveredWidget(LatestHoveredWidget=Get DroneLinearVelocity() -> (DroneLinearVelocity:Options Menu Slider Replay Object Reference))
   "On Focused (DroneRotationVelocity)"() {is=title K2Node_ComponentBoundEvent}
   > Set LatestHoveredWidget(LatestHoveredWidget=Get DroneRotationVelocity() -> (DroneRotationVelocity:Options Menu Slider Replay Object Reference))
   "On Focused (DroneDutchSpeed)"() {is=title K2Node_ComponentBoundEvent}
   > Set LatestHoveredWidget(LatestHoveredWidget=Get DroneDutchSpeed() -> (DroneDutchSpeed:Options Menu Slider Replay Object Reference))
   "On Focused (DroneDutch)"() {is=title K2Node_ComponentBoundEvent}
   > Set LatestHoveredWidget(LatestHoveredWidget=Get DroneDutch() -> (DroneDutch:Options Menu Slider Replay Object Reference))
   n7 = "Slider Changed (DroneLinearVelocity)"!() -> (OutputDelegate, then, Slider Value:double) {is=title K2Node_ComponentBoundEvent}
   > SetDroneDirectMovementLinearVelocity!(
    execute
    self:Fort Spectator Advance Settings Tab Object Reference
    DroneLinearVelocity:float=MapRangeClamped(Value=@n7.Slider Value, InRangeA=0, InRangeB=1, OutRangeB=Get DroneLinearVelocityMax() -> (DroneLinearVelocityMax:double))
   ) -> (then)
   n8 = "Selection Changed (DroneAttachToBone)"!() -> (OutputDelegate, then, Selected Index:int=0) {is=title K2Node_ComponentBoundEvent}
   n9 = "integer > integer"(A=@n8.Selected Index, B=0) {is=title K2Node_CallFunction}
   > SetDroneAttachToBonesEnabled!(execute, self:Fort Spectator Advance Settings Tab Object Reference, bDroneAttachToBonesEnabled:bool=@n9) -> (then)
   > n10 = Branch(Condition=@n9)
   n11 = "Get (a ref)"(Array:Array of Texts=Make Bones Names!() -> (BonesNames:Array of Texts) {is=title K2Node_MacroInstance}, Dimension 1=0) -> (Output:"Text (by ref)")
   > SetDroneAttachToBonesName!(
    execute
    self:Fort Spectator Advance Settings Tab Object Reference
    InDroneAttachToBonesName:Name=String To Name(InString=Get Source String from Text!(Text:Text=@n11) -> (ReturnValue:String) {is=title K2Node_CallFunction})
   ) -> (then)
   SetDroneAttachToBonesName(execute=@n10.else, InDroneAttachToBonesName=None)
   n12 = "Slider Changed (DroneDutchAngleStepSlider)"!() -> (OutputDelegate, then, Slider Value:double) {is=title K2Node_ComponentBoundEvent}
   n13 = MapRangeClamped(Value=@n12.Slider Value, InRangeA=0, InRangeB=1, OutRangeA=1, OutRangeB=30)
   > SetDroneDutchAngleStep!(execute, self:Fort Spectator Advance Settings Tab Object Reference, DroneDutchAngleStep:float=@n13) -> (then)
   > UpdateSliderStep!(execute, self:Options Menu Slider Replay Object Reference=Get DroneDutch(), Step:double=@n13) -> (then)
   n14 = "Slider Changed (DroneDutchSpeedStepSlider)"!() -> (OutputDelegate, then, Slider Value:double) {is=title K2Node_ComponentBoundEvent}
   n15 = MapRangeClamped(Value=@n14.Slider Value, InRangeA=0, InRangeB=1, OutRangeA=1, OutRangeB=30)
   > SetDroneDutchSpeedStep!(execute, self:Fort Spectator Advance Settings Tab Object Reference, DroneDutchSpeedStep:float=@n15) -> (then)
   > UpdateSliderStep(self=Get DroneDutchSpeed(), Step=@n15)
   "On Focused (DroneDutchAngleStepSlider)"() {is=title K2Node_ComponentBoundEvent}
   > Set LatestHoveredWidget(LatestHoveredWidget=Get DroneDutchAngleStepSlider() -> (DroneDutchAngleStepSlider:Options Menu Slider Replay Object Reference))
   "On Focused (DroneDutchSpeedStepSlider)"() {is=title K2Node_ComponentBoundEvent}
   > Set LatestHoveredWidget(LatestHoveredWidget=Get DroneDutchSpeedStepSlider() -> (DroneDutchSpeedStepSlider:Options Menu Slider Replay Object Reference))
   n16 = "Slider Changed (DroneDragMultiplier)"!() -> (OutputDelegate, then, Slider Value:double) {is=title K2Node_ComponentBoundEvent}
   > SetDroneDragMultiplier!(
    execute
    self:Fort Spectator Advance Settings Tab Object Reference
    DroneDragMultiplier:float=MapRangeClamped(Value=@n16.Slider Value, InRangeA=0, InRangeB=1, OutRangeB=5)
   ) -> (then)
   n17 = "On Value Commited (DroneDutchAngleStepSlider)"!() -> (OutputDelegate, then, Value:double) {is=title K2Node_ComponentBoundEvent}
   n18 = MapRangeClamped(Value=@n17.Value, InRangeA=0, InRangeB=1, OutRangeA=1, OutRangeB=30)
   > SetDroneDutchAngleStep(DroneDutchAngleStep=@n18)
   > UpdateSliderStep(self=Get DroneDutch(), Step=@n18)
   "On Focused (DroneDragMultiplier)"() {is=title K2Node_ComponentBoundEvent}
   > Set LatestHoveredWidget(LatestHoveredWidget=Get DroneDragMultiplier() -> (DroneDragMultiplier:Options Menu Slider Replay Object Reference))
   n19 = "On Value Commited (DroneDutchSpeed)"!() -> (OutputDelegate, then, Value:double) {is=title K2Node_ComponentBoundEvent}
   > SetDroneDutchSpeed!(execute, self:Fort Spectator Advance Settings Tab Object Reference, DroneDutchSpeed:float=@n19.Value) -> (then)
   n20 = "On Value Commited (DroneDragMultiplier)"!() -> (OutputDelegate, then, Value:double) {is=title K2Node_ComponentBoundEvent}
   > SetDroneDragMultiplier(DroneDragMultiplier=@n20.Value)
   n21 = "Slider Changed (DroneRotationVelocity)"!() -> (OutputDelegate, then, Slider Value:double) {is=title K2Node_ComponentBoundEvent}
   > SetDroneDirectMovementRotationVelocity!(
    execute
    self:Fort Spectator Advance Settings Tab Object Reference
    DroneRotationVelocity:float=MapRangeClamped(Value=@n21.Slider Value, InRangeA=0, InRangeB=1, OutRangeB=Get DroneRotationVelocityMax() -> (DroneRotationVelocityMax:double))
   ) -> (then)
   n22 = "On Value Commited (DroneDutchSpeedStepSlider)"!() -> (OutputDelegate, then, Value:double) {is=title K2Node_ComponentBoundEvent}
   n23 = MapRangeClamped(Value=@n22.Value, InRangeA=0, InRangeB=1, OutRangeA=1, OutRangeB=30)
   > SetDroneDutchSpeedStep(DroneDutchSpeedStep=@n23)
   > UpdateSliderStep(self=Get DroneDutchSpeed(), Step=@n23)
   n24 = "Selection Changed (ShowCameraShots)"!() -> (OutputDelegate, then, Selected Index:int=0) {is=title K2Node_ComponentBoundEvent}
   > SetVisibility(
    self=[
     Get Rotator_Preset() -> (Rotator_Preset:Rotator Selector Replay Settings Object Reference)
     Get Rotator_CurrentShot() -> (Rotator_CurrentShot:Rotator Selector Replay Settings Object Reference)
     Get Button_UndoRedo() -> (Button_UndoRedo:Options Menu Action Replay Object Reference)
     Get Button_OverrideShot() -> (Button_OverrideShot:Options Menu Action Replay Object Reference)
     Get Button_SaveSettings() -> (Button_SaveSettings:Options Menu Action Replay Object Reference)
     Get Button_LoadSettings() -> (Button_LoadSettings:Options Menu Action Replay Object Reference)
    ]
    InVisibility=Select(Option 0=Collapsed, Option 1=Visible, Index:int=@n24.Selected Index)
   )
   > ToggleCameraShotPresets!(
    execute
    self:Fort Spectator Advance Settings Tab Object Reference
    bActive:bool=Select(Option 0:bool=false, Option 1:bool=true, Index=@n24.Selected Index) -> (ReturnValue:bool)
   ) -> (then)
   > SetNavigationRules()
   n25 = "Get (a copy)"(Array:Array of Texts=Get Options() -> (Options:Array of Texts)) -> (Output:Text)
   > InvokeCameraShot!(execute, self, HotKeyNum:int=String To Integer(InString="To String (Text)"(InText=@n25))) -> (then)
   > Update Row Selector!(execute, self:Rotator Selector Replay Settings Object Reference=Get Rotator_CurrentShot(), Currently Selected:int=0) -> (then)
   "On Hovered (ShowCameraShots)"() {is=title K2Node_ComponentBoundEvent}
   > Set LatestHoveredWidget(LatestHoveredWidget=Get ShowCameraShots() -> (ShowCameraShots:Rotator Selector Replay Settings Object Reference))
   n26 = "Selection Changed (Rotator_Preset)"!() -> (OutputDelegate, then, Selected Index:int=0) {is=title K2Node_ComponentBoundEvent}
   > SetCameraShotPreset!(execute, self:Fort Spectator Advance Settings Tab Object Reference, Number:int=@n26.Selected Index) -> (then)
   n27 = "Get (a copy)"(Array=Get Options())
   > InvokeCameraShot(HotKeyNum=String To Integer(InString="To String (Text)"(InText=@n27)))
   > Update Row Selector(self=Get Rotator_CurrentShot(), Currently Selected=0)
   "On Focused (DroneDirectMovement)"() {is=title K2Node_ComponentBoundEvent}
   > n28 = Cast To RotatorSelectorReplaySettings!(execute, Object=Get DroneDirectMovement()) -> (then, CastFailed, AsRotator Selector Replay Settings:Rotator Selector Replay Settings Object Reference) {is=title K2Node_DynamicCast}
   > Set LatestHoveredWidget(LatestHoveredWidget=@n28)
   "On Hovered (Rotator_Preset)"() {is=title K2Node_ComponentBoundEvent}
   > Set LatestHoveredWidget(LatestHoveredWidget=Get Rotator_Preset())
   "On Clicked (Shot0_Button)"!() -> (OutputDelegate, then, Button:Common Button Base Object Reference) {is=title K2Node_ComponentBoundEvent}
   > InvokeCameraShot(HotKeyNum=0)
   "On Clicked (Shot1_Button)"!() -> (OutputDelegate, then, Button:Common Button Base Object Reference) {is=title K2Node_ComponentBoundEvent}
   > InvokeCameraShot(HotKeyNum=1)
   "On Clicked (Shot2_Button)"!() -> (OutputDelegate, then, Button:Common Button Base Object Reference) {is=title K2Node_ComponentBoundEvent}
   > InvokeCameraShot(HotKeyNum=2)
   "On Clicked (Shot3_Button)"!() -> (OutputDelegate, then, Button:Common Button Base Object Reference) {is=title K2Node_ComponentBoundEvent}
   > InvokeCameraShot(HotKeyNum=3)
   n29 = "Selection Changed (DroneAttachToBoneName)"!() -> (OutputDelegate, then, Selected Index:int=0) {is=title K2Node_ComponentBoundEvent}
   n30 = "Get (a ref)"(Array=Make Bones Names(), Dimension 1=@n29.Selected Index)
   > SetDroneAttachToBonesName(InDroneAttachToBonesName=String To Name(InString=Get Source String from Text(Text=@n30)))
   "On Clicked (Shot4_Button)"!() -> (OutputDelegate, then, Button:Common Button Base Object Reference) {is=title K2Node_ComponentBoundEvent}
   > InvokeCameraShot(HotKeyNum=4)
   "On Clicked (Shot5_Button)"!() -> (OutputDelegate, then, Button:Common Button Base Object Reference) {is=title K2Node_ComponentBoundEvent}
   > InvokeCameraShot(HotKeyNum=5)
   "On Clicked (Shot6_Button)"!() -> (OutputDelegate, then, Button:Common Button Base Object Reference) {is=title K2Node_ComponentBoundEvent}
   > InvokeCameraShot(HotKeyNum=6)
   "On Clicked (Shot7_Button)"!() -> (OutputDelegate, then, Button:Common Button Base Object Reference) {is=title K2Node_ComponentBoundEvent}
   > InvokeCameraShot(HotKeyNum=7)
   "On Clicked (Shot8_Button)"!() -> (OutputDelegate, then, Button:Common Button Base Object Reference) {is=title K2Node_ComponentBoundEvent}
   > InvokeCameraShot(HotKeyNum=8)
   "On Focused (Button_ResetDutchAngle)"() {is=title K2Node_ComponentBoundEvent}
   > Set LatestHoveredWidget(LatestHoveredWidget=Get Button_ResetDutchAngle() -> (Button_ResetDutchAngle:Options Menu Action Replay Object Reference))
   "On Focused (Button_ResetDutchSpeed)"() {is=title K2Node_ComponentBoundEvent}
   > Set LatestHoveredWidget(LatestHoveredWidget=Get Button_ResetDutchSpeed() -> (Button_ResetDutchSpeed:Options Menu Action Replay Object Reference))
   "On Focused (DutchControlScheme)"() {is=title K2Node_ComponentBoundEvent}
   > Set LatestHoveredWidget(LatestHoveredWidget=Get DutchControlScheme())
   "On Focused (DroneAttachToBone)"() {is=title K2Node_ComponentBoundEvent}
   > Set LatestHoveredWidget(LatestHoveredWidget=Get DroneAttachToBone() -> (DroneAttachToBone:Rotator Selector Replay Settings Object Reference))
   "On Focused (DroneAttachToBoneName)"() {is=title K2Node_ComponentBoundEvent}
   > Set LatestHoveredWidget(LatestHoveredWidget=Get DroneAttachToBoneName() -> (DroneAttachToBoneName:Rotator Selector Replay Settings Object Reference))
   n31 = "Slider Changed (DroneDutchSpeed)"!() -> (OutputDelegate, then, Slider Value:double) {is=title K2Node_ComponentBoundEvent}
   > SetDroneDutchSpeed(DroneDutchSpeed=MapRangeClamped(
     Value=@n31.Slider Value
     InRangeA=0
     InRangeB=1
     OutRangeA=Get DroneDutchSpeedMin() -> (DroneDutchSpeedMin:double)
     OutRangeB=Get DroneDutchSpeedMax() -> (DroneDutchSpeedMax:double)
    ))
   "On Focused (ShowCameraDummy)"() {is=title K2Node_ComponentBoundEvent}
   > Set LatestHoveredWidget(LatestHoveredWidget=Get ShowCameraDummy())
   "On Focused (ShowCameraShots)"() {is=title K2Node_ComponentBoundEvent}
   > Set LatestHoveredWidget(LatestHoveredWidget=Get ShowCameraShots())
   "On Focused (Rotator_Preset)"() {is=title K2Node_ComponentBoundEvent}
   > Set LatestHoveredWidget(LatestHoveredWidget=Get Rotator_Preset())
   "On Clicked (Shot0_Button_1)"!() -> (OutputDelegate, then, Button:Common Button Base Object Reference) {is=title K2Node_ComponentBoundEvent}
   > ToggleUndoRedoButton!(
    execute
    self
    Button:WBP UIKit Button Regular Object Reference=Get Shot0_Button_1() -> (Shot0_Button_1:WBP UIKit Button Regular Object Reference)
   ) -> (then)
   > SwapCameraShots!(execute, self, HotKeyNum:int=0) -> (then)
   "On Clicked (Shot9_Button)"!() -> (OutputDelegate, then, Button:Common Button Base Object Reference) {is=title K2Node_ComponentBoundEvent}
   > InvokeCameraShot(HotKeyNum=9)
   "On Clicked (Shot1_Button_1)"!() -> (OutputDelegate, then, Button:Common Button Base Object Reference) {is=title K2Node_ComponentBoundEvent}
   > ToggleUndoRedoButton(Button=Get Shot1_Button_1() -> (Shot1_Button_1:WBP UIKit Button Regular Object Reference))
   > SwapCameraShots(HotKeyNum=1)
   "On Clicked (Shot2_Button_1)"!() -> (OutputDelegate, then, Button:Common Button Base Object Reference) {is=title K2Node_ComponentBoundEvent}
   > ToggleUndoRedoButton(Button=Get Shot2_Button_1() -> (Shot2_Button_1:WBP UIKit Button Regular Object Reference))
   > SwapCameraShots(HotKeyNum=2)
   "On Clicked (Shot3_Button_1)"!() -> (OutputDelegate, then, Button:Common Button Base Object Reference) {is=title K2Node_ComponentBoundEvent}
   > ToggleUndoRedoButton(Button=Get Shot3_Button_1() -> (Shot3_Button_1:WBP UIKit Button Regular Object Reference))
   > SwapCameraShots(HotKeyNum=3)
   "On Clicked (Shot4_Button_1)"!() -> (OutputDelegate, then, Button:Common Button Base Object Reference) {is=title K2Node_ComponentBoundEvent}
   > ToggleUndoRedoButton(Button=Get Shot4_Button_1() -> (Shot4_Button_1:WBP UIKit Button Regular Object Reference))
   > SwapCameraShots(HotKeyNum=4)
   "On Clicked (Shot5_Button_1)"!() -> (OutputDelegate, then, Button:Common Button Base Object Reference) {is=title K2Node_ComponentBoundEvent}
   > ToggleUndoRedoButton(Button=Get Shot5_Button_1() -> (Shot5_Button_1:WBP UIKit Button Regular Object Reference))
   > SwapCameraShots(HotKeyNum=5)
   n32 = "Selection Changed (DutchControlScheme)"!() -> (OutputDelegate, then, Selected Index:int=0) {is=title K2Node_ComponentBoundEvent}
   > SetDutchControlSchemaEnabled!(
    execute
    self:Fort Spectator Advance Settings Tab Object Reference
    bDutchControlSchemaEnabled:bool="integer > integer"(A=@n32.Selected Index, B=0)
   ) -> (then)
   "On Clicked (Shot6_Button_1)"!() -> (OutputDelegate, then, Button:Common Button Base Object Reference) {is=title K2Node_ComponentBoundEvent}
   > ToggleUndoRedoButton(Button=Get Shot6_Button_1() -> (Shot6_Button_1:WBP UIKit Button Regular Object Reference))
   > SwapCameraShots(HotKeyNum=6)
   "On Clicked (Shot7_Button_1)"!() -> (OutputDelegate, then, Button:Common Button Base Object Reference) {is=title K2Node_ComponentBoundEvent}
   > ToggleUndoRedoButton(Button=Get Shot7_Button_1() -> (Shot7_Button_1:WBP UIKit Button Regular Object Reference))
   > SwapCameraShots(HotKeyNum=7)
   "On Clicked (Shot8_Button_1)"!() -> (OutputDelegate, then, Button:Common Button Base Object Reference) {is=title K2Node_ComponentBoundEvent}
   > ToggleUndoRedoButton(Button=Get Shot8_Button_1() -> (Shot8_Button_1:WBP UIKit Button Regular Object Reference))
   > SwapCameraShots(HotKeyNum=8)
   "On Clicked (Shot9_Button_1)"!() -> (OutputDelegate, then, Button:Common Button Base Object Reference) {is=title K2Node_ComponentBoundEvent}
   > ToggleUndoRedoButton(Button=Get Shot9_Button_1() -> (Shot9_Button_1:WBP UIKit Button Regular Object Reference))
   > SwapCameraShots(HotKeyNum=9)
   n33 = "Selection Changed (Rotator_CurrentShot)"!() -> (OutputDelegate, then, Selected Index:int=0) {is=title K2Node_ComponentBoundEvent}
   n34 = "Get (a copy)"(Array=Get Options(), Dimension 1=@n33.Selected Index)
   > InvokeCameraShot(HotKeyNum=String To Integer(InString="To String (Text)"(InText=@n34)))
   > Update Row Selector(self=Get Rotator_CurrentShot(), Currently Selected=@n33.Selected Index)
   > SetFocus(self=Get Rotator_CurrentShot())
   "Button Clicked (Button_UndoRedo)"() {is=title K2Node_ComponentBoundEvent}
   n35 = "Get (a copy)"(
    Array=Get Options()
    Dimension 1=Get SelectedIndex!(self:Rotator Selector Replay Settings Object Reference=Get Rotator_CurrentShot()) -> (SelectedIndex:int=0)
   )
   > SwapCameraShots(HotKeyNum=String To Integer(InString="To String (Text)"(InText=@n35)))
   > SetFocus(self=Get Button_UndoRedo())
   n36 = "On Value Commited (DroneLinearVelocity)"!() -> (OutputDelegate, then, Value:double) {is=title K2Node_ComponentBoundEvent}
   > SetDroneDirectMovementLinearVelocity(DroneLinearVelocity=@n36.Value)
   n37 = "On Value Commited (DroneRotationVelocity)"!() -> (OutputDelegate, then, Value:double) {is=title K2Node_ComponentBoundEvent}
   > SetDroneDirectMovementRotationVelocity(DroneRotationVelocity=@n37.Value)
   n38 = "On Value Commited (DroneLinearVelocityZmultiplier)"!() -> (OutputDelegate, then, Value:double) {is=title K2Node_ComponentBoundEvent}
   > SetDroneDirectMovementZLinearVelocityMultiplier!(execute, self:Fort Spectator Advance Settings Tab Object Reference, DroneLinearVelocity:float=@n38.Value) -> (then)
   n39 = "Selection Changed (DroneAttachToThrowable)"!() -> (OutputDelegate, then, Selected Index:int=0) {is=title K2Node_ComponentBoundEvent}
   > SetDroneAttachToThrowableEnabled!(
    execute
    self:Fort Spectator Advance Settings Tab Object Reference
    bDroneAttachToThrowableEnabled:bool="integer > integer"(A=@n39.Selected Index, B=0)
   ) -> (then)
   "Button Clicked (Button_ResetDutchSpeed)"() {is=title K2Node_ComponentBoundEvent}
   > SetDroneDutchSpeed(DroneDutchSpeed=0)
   "On Focused (DroneAttachToThrowable)"() {is=title K2Node_ComponentBoundEvent}
   > Set LatestHoveredWidget(LatestHoveredWidget=Get DroneAttachToThrowable() -> (DroneAttachToThrowable:Rotator Selector Replay Settings Object Reference))
   n40 = "Selection Changed (DroneDirectMovementSaveState)"!() -> (OutputDelegate, then, Selected Index:int=0) {is=title K2Node_ComponentBoundEvent}
   > SetDroneDirectMovementStateSaved!(
    execute
    self:Fort Spectator Advance Settings Tab Object Reference
    bDroneDirectMovementStateSaved:bool="integer > integer"(A=@n40.Selected Index, B=0)
   ) -> (then)
   "On Hovered (DroneDirectMovementSaveState)"() {is=title K2Node_ComponentBoundEvent}
   > Set LatestHoveredWidget(LatestHoveredWidget=Get DroneDirectMovementSaveState() -> (DroneDirectMovementSaveState:Rotator Selector Replay Settings Object Reference))
   "On Focused (DroneDirectMovementSaveState)"() {is=title K2Node_ComponentBoundEvent}
   > Set LatestHoveredWidget(LatestHoveredWidget=Get DroneDirectMovementSaveState())
   n41 = "Selection Changed (DroneGazeDirectedMovement)"!() -> (OutputDelegate, then, Selected Index:int=0) {is=title K2Node_ComponentBoundEvent}
   > SetDroneGazeDirectedVerticalMovementEnabled!(
    execute
    self:Fort Spectator Advance Settings Tab Object Reference
    bDroneGazeDirectedVerticalMovementEnabled:bool="integer > integer"(A=@n41.Selected Index, B=0)
   ) -> (then)
   "Button Clicked (Button_OverrideShot)"() {is=title K2Node_ComponentBoundEvent}
   > n42 = Cast To FortPlayerControllerSpectating!(execute, Object=GetOwningPlayer()) -> (then, CastFailed, AsFort Player Controller Spectating:Fort Player Controller Spectating Object Reference) {is=title K2Node_DynamicCast}
   n43 = "Get (a copy)"(Array=Get Options(), Dimension 1=Get SelectedIndex(self=Get Rotator_CurrentShot()) -> (SelectedIndex=0))
   > AssignHotKeyByNumber!(
    execute
    self:Fort Player Controller Spectating Object Reference=@n42
    HotKeyNum:int=String To Integer(InString="To String (Text)"(InText=@n43))
   ) -> (then)
   > SetFocus(self=Get Button_OverrideShot())
   "On Focused (DroneGazeDirectedMovement)"() {is=title K2Node_ComponentBoundEvent}
   > Set LatestHoveredWidget(LatestHoveredWidget=Get DroneGazeDirectedMovement() -> (DroneGazeDirectedMovement:Rotator Selector Replay Settings Object Reference))
   n44 = "Selection Changed (DronePendulum)"!() -> (OutputDelegate, then, Selected Index:int=0) {is=title K2Node_ComponentBoundEvent}
   > SetDronePendulumEnabled!(
    execute
    self:Fort Spectator Advance Settings Tab Object Reference
    bDronePendulumEnabled:bool="integer > integer"(A=@n44.Selected Index) {is=title K2Node_PromotableOperator}
   ) -> (then)
   > SetVisibility(
    self=[
     Get DronePendulumArmLength() -> (DronePendulumArmLength:Options Menu Slider Replay Object Reference)
     Get DronePendulumCameraRoll() -> (DronePendulumCameraRoll:Rotator Selector Replay Settings Object Reference)
     Get DronePendulumAmplitude() -> (DronePendulumAmplitude:Options Menu Slider Replay Object Reference)
     Get DronePendulumFrequency() -> (DronePendulumFrequency:Options Menu Slider Replay Object Reference)
    ]
    InVisibility=Select(Option 0:ESlateVisibility Enum=Collapsed, Option 1:ESlateVisibility Enum=Visible, Index=@n44.Selected Index) -> (ReturnValue:ESlateVisibility Enum)
   )
   "On Hovered (DroneGazeDirectedMovement)"() {is=title K2Node_ComponentBoundEvent}
   > Set LatestHoveredWidget(LatestHoveredWidget=Get DroneGazeDirectedMovement())
   n45 = "Slider Changed (DronePendulumFrequency)"!() -> (OutputDelegate, then, Slider Value:double) {is=title K2Node_ComponentBoundEvent}
   n46 = Get!(self:Options Menu Slider Replay Object Reference=Get DronePendulumFrequency()) -> (MinMaxValue_X:double, MinMaxValue_Y:double)
   > SetDronePendulumFrequency!(
    execute
    self:Fort Spectator Advance Settings Tab Object Reference
    DronePendulumFrequency:float=MapRangeClamped(Value=@n45.Slider Value, InRangeA=0, InRangeB=1, OutRangeA=@n46, OutRangeB=@n46.MinMaxValue_Y)
   ) -> (then)
   "Button Clicked (Button_ResetDutchAngle)"() {is=title K2Node_ComponentBoundEvent}
   > SetDroneDutchAngle(DroneDutchAngle=0)
   n47 = "Slider Changed (DronePendulumArmLength)"!() -> (OutputDelegate, then, Slider Value:double) {is=title K2Node_ComponentBoundEvent}
   n48 = Get(self=Get DronePendulumArmLength())
   > SetDronePendulumArmLength!(
    execute
    self:Fort Spectator Advance Settings Tab Object Reference
    DronePendulumArmLength:float=float * float(A=MapRangeClamped(Value=@n47.Slider Value, InRangeA=0, InRangeB=1, OutRangeA=@n48, OutRangeB=@n48.MinMaxValue_Y), B=100) {is=title K2Node_PromotableOperator}
   ) -> (then)
   n49 = "Slider Changed (DronePendulumAmplitude)"!() -> (OutputDelegate, then, Slider Value:double) {is=title K2Node_ComponentBoundEvent}
   n50 = Get(self=Get DronePendulumAmplitude())
   > SetDronePendulumAmplitude!(
    execute
    self:Fort Spectator Advance Settings Tab Object Reference
    DronePendulumAmplitude:float=MapRangeClamped(Value=@n49.Slider Value, InRangeA=0, InRangeB=1, OutRangeA=@n50, OutRangeB=@n50.MinMaxValue_Y)
   ) -> (then)
   n51 = "Selection Changed (DronePendulumCameraRoll)"!() -> (OutputDelegate, then, Selected Index:int=0) {is=title K2Node_ComponentBoundEvent}
   > SetDronePendulumRollEnabled!(
    execute
    self:Fort Spectator Advance Settings Tab Object Reference
    bDronePendulumRollEnabled:bool="integer > integer"(A=@n51.Selected Index)
   ) -> (then)
   "On Focused (DronePendulum)"() {is=title K2Node_ComponentBoundEvent}
   > Set LatestHoveredWidget(LatestHoveredWidget=Get DronePendulum() -> (DronePendulum:Rotator Selector Replay Settings Object Reference))
   "On Hovered (DronePendulum)"() {is=title K2Node_ComponentBoundEvent}
   > Set LatestHoveredWidget(LatestHoveredWidget=Get DronePendulum())
   "On Focused (DronePendulumArmLength)"() {is=title K2Node_ComponentBoundEvent}
   > Set LatestHoveredWidget(LatestHoveredWidget=Get DronePendulumArmLength())
   "On Focused (DronePendulumAmplitude)"() {is=title K2Node_ComponentBoundEvent}
   > Set LatestHoveredWidget(LatestHoveredWidget=Get DronePendulumAmplitude())
   "On Focused (DronePendulumFrequency)"() {is=title K2Node_ComponentBoundEvent}
   > Set LatestHoveredWidget(LatestHoveredWidget=Get DronePendulumFrequency())
   "On Hovered (DronePendulumCameraRoll)"() {is=title K2Node_ComponentBoundEvent}
   > Set LatestHoveredWidget(LatestHoveredWidget=Get DronePendulumCameraRoll())
   "On Focused (DronePendulumCameraRoll)"() {is=title K2Node_ComponentBoundEvent}
   > Set LatestHoveredWidget(LatestHoveredWidget=Get DronePendulumCameraRoll())
   n52 = "Slider Changed (DroneLinearVelocityZmultiplier)"!() -> (OutputDelegate, then, Slider Value:double) {is=title K2Node_ComponentBoundEvent}
   > SetDroneDirectMovementZLinearVelocityMultiplier(DroneLinearVelocity=MapRangeClamped(
     Value=@n52.Slider Value
     InRangeA=0
     InRangeB=1
     OutRangeB=Get DroneLinearVelocityZMultiplierMax() -> (DroneLinearVelocityZMultiplierMax:double)
    ))
   n53 = "On Value Commited (DronePendulumArmLength)"!() -> (OutputDelegate, then, Value:double) {is=title K2Node_ComponentBoundEvent}
   > SetDronePendulumArmLength(DronePendulumArmLength=float * float(A=@n53.Value, B=100))
   n54 = "On Value Commited (DronePendulumAmplitude)"!() -> (OutputDelegate, then, Value:double) {is=title K2Node_ComponentBoundEvent}
   > SetDronePendulumAmplitude(DronePendulumAmplitude=@n54.Value)
   n55 = "On Value Commited (DronePendulumFrequency)"!() -> (OutputDelegate, then, Value:double) {is=title K2Node_ComponentBoundEvent}
   > SetDronePendulumFrequency(DronePendulumFrequency=@n55.Value)
   "Button Clicked (Button_SaveSettings)"() {is=title K2Node_ComponentBoundEvent}
   > ShowSaveLoadPrompt!(execute, self, Model:int=0) -> (then)
   > SetFocus(self=Get Button_SaveSettings())
   "Button Clicked (Button_LoadSettings)"() {is=title K2Node_ComponentBoundEvent}
   > ShowSaveLoadPrompt(Model=1)
   > SetFocus(self=Get Button_LoadSettings())
   n56 = "Selection Changed (DroneZAxisLock)"!() -> (OutputDelegate, then, Selected Index:int=0) {is=title K2Node_ComponentBoundEvent}
   > SetDroneZAxisLockEnabled!(
    execute
    self:Fort Spectator Advance Settings Tab Object Reference
    bDroneZAxisLockEnabled:bool="integer > integer"(A=@n56.Selected Index, B=0) {is=title K2Node_CallFunction}
   ) -> (then)
   "On Hovered (DroneZAxisLock)"() {is=title K2Node_ComponentBoundEvent}
   > Set LatestHoveredWidget(LatestHoveredWidget=Get DroneZAxisLock() -> (DroneZAxisLock:Rotator Selector Replay Settings Object Reference))
   "On Focused (DroneZAxisLock)"() {is=title K2Node_ComponentBoundEvent}
   > Set LatestHoveredWidget(LatestHoveredWidget=Get DroneZAxisLock())
   n57 = "Selection Changed (DroneXInputLock)"!() -> (OutputDelegate, then, Selected Index:int=0) {is=title K2Node_ComponentBoundEvent}
   > SetDroneXInputLockEnabled!(
    execute
    self:Fort Spectator Advance Settings Tab Object Reference
    bDroneXInputLockEnabled:bool="integer > integer"(A=@n57.Selected Index, B=0)
   ) -> (then)
   "On Hovered (DroneXInputLock)"() {is=title K2Node_ComponentBoundEvent}
   > Set LatestHoveredWidget(LatestHoveredWidget=Get DroneXInputLock())
   n58 = SwapCameraShots!() -> (OutputDelegate, then, HotKeyNum:int) {is=title K2Node_CustomEvent}
   > n59 = Cast To FortPlayerControllerSpectating(Object=GetOwningPlayer())
   > SwapBackHotkey!(execute, self:Fort Player Controller Spectating Object Reference=@n59, HotKeyNum:int=@n58.HotKeyNum) -> (then)
   > InvokeHotKeyByNumber!(execute, self:Fort Player Controller Spectating Object Reference=@n59, HotKeyNum:int=@n58.HotKeyNum) -> (then)
   n60 = OnHotkeyAssigned_Event!() -> (OutputDelegate, then, HotKeyNum:int) {is=title K2Node_CustomEvent}
   > MarkShots!(execute, self, ShotNumber:int=@n60.HotKeyNum) -> (then)
   > Update Row Selector(
    self=Get Rotator_CurrentShot()
    Currently Selected=Find Item!(
     TargetArray:Array of Texts=Get Options()
     ItemToFind:"Text (by ref)"="To Text (Integer)"(Value=@n60.HotKeyNum, bUseGrouping=false, MaximumIntegralDigits=3)
    ) -> (ReturnValue)
   )
   n61 = ScrollBarBehavior() {is=title K2Node_CustomEvent}
   n62 = Get CommonInputSubsystem!() -> (ReturnValue:Common Input Object Reference) {is=title K2Node_GetBlueprintContextEx}
   > Is Valid(InputObject=@n62)
   n63 = GetCurrentInputType(self=@n62)
   > n64 = Branch(Condition="Equal (Enum)"(A=@n63, B=Gamepad))
   > SetScrollWhenFocusChanges!(
    execute
    self:Scroll Box Object Reference=Get ScrollBox_Tabs() -> (ScrollBox_Tabs:Scroll Box Object Reference)
    NewScrollWhenFocusChanges:EScrollWhenFocusChanges Enum=AnimatedScroll
   ) -> (then)
   n65 = Get LatestHoveredWidget() -> (LatestHoveredWidget:User Widget Object Reference)
   > n66 = Is Valid(InputObject=@n65)
   n67 = Cast To RotatorSelectorReplaySettings(execute=@n66.Is Not Valid, Object=Get DroneDirectMovement())
   > SetFocus(self=@n67)
   SetFocus(execute=@n66.Is Valid, self=@n65)
   Branch(execute=@n64.else, Condition="Equal (Enum)"(A=@n63, B=MouseAndKeyboard))
   > SetScrollWhenFocusChanges(self=Get ScrollBox_Tabs(), NewScrollWhenFocusChanges=NoScroll)
   n68 = CloseonSelectionChanged!() -> (OutputDelegate, then, Selected Index:int) {is=title K2Node_CustomEvent}
   > CloseTabGroup() {is=title K2Node_CallFunction}
   n69 = CloseOnSlider!() -> (OutputDelegate, then, Slider Value:double=~) {is=title K2Node_CustomEvent}
   > CloseTabGroup()
   n70 = CloseOnButtonClicked() {is=title K2Node_CustomEvent}
   > CloseTabGroup()
   n71 = InvokeCameraShot!() -> (OutputDelegate, then, HotKeyNum:int) {is=title K2Node_CustomEvent}
   > n72 = Cast To FortPlayerControllerSpectating(Object=GetOwningPlayer())
   > InvokeHotKeyByNumber(self=@n72, HotKeyNum=@n71.HotKeyNum)
   n73 = OnHotkeyInvoked!() -> (OutputDelegate, then, HotKeyNum:int) {is=title K2Node_CustomEvent}
   > Update Row Selector(
    self=Get Rotator_CurrentShot()
    Currently Selected=Find Item(TargetArray=Get Options(), ItemToFind="To Text (Integer)"(Value=@n73.HotKeyNum, bUseGrouping=false, MaximumIntegralDigits=3))
   )
   Event PreConstruct()
   > Initialize!(
    execute
    self:Rotator Selector Replay Settings Object Reference=Cast To RotatorSelectorReplaySettings!(Object=Get DroneDirectMovement()) -> (AsRotator Selector Replay Settings, bSuccess:bool=~)
    Display Text:Text="LOCTEXT(\"Direct Movement\")"
    Options:Array of Texts=Make Off on Array!() -> (OffOnArray:Array of Texts) {is=title K2Node_MacroInstance}
    Hover Text:Text="LOCTEXT(\"Enable/Disable drone direct movement\")"
    Tab Tooltip:Common Text Object Reference
   ) -> (then)
   > Initialize(
    self=Get DroneGazeDirectedMovement()
    Display Text="LOCTEXT(\"Gaze Vertical Movement\")"
    Options=Make Off on Array()
    Hover Text="LOCTEXT(\"Enable/Disable gaze vertical drone movement\")"
   )
   > Initialize(
    self=Get DroneXInputLock()
    Display Text="LOCTEXT(\"X Input Lock\")"
    Options=Make Off on Array()
    Hover Text="LOCTEXT(\"Enable/Disable forward/backward drone movement\")"
   )
   > Initialize(
    self=Get DroneYInputLock()
    Display Text="LOCTEXT(\"Y Input Lock\")"
    Options=Make Off on Array()
    Hover Text="LOCTEXT(\"Enable/Disable drone side movement\")"
   )
   > Initialize(
    self=Get DroneZAxisLock()
    Display Text="LOCTEXT(\"Z Axis Lock\")"
    Options=Make Off on Array()
    Hover Text="LOCTEXT(\"Enable/Disable vertical  lock drone movement\")"
   )
   > Initialize(
    self=Get DroneDirectMovementSaveState()
    Display Text="LOCTEXT(\"Save speed state settings\")"
    Options=Make Off on Array()
    Hover Text="LOCTEXT(\"Enable/Disable saving of drone direct movement speed settings\")"
   )
   > Update Slider!(
    execute
    self:Options Menu Slider Replay Object Reference=Get DroneLinearVelocity()
    Slider Text:Text="LOCTEXT(\"Linear Speed\")"
    Slider Value:double=0.2
    Hover Text:Text
    Tooltip Text Block:Common Text Object Reference
    Min Max Value_X:double=0
    Min Max Value_Y:double=Get DroneLinearVelocityMax()
   ) -> (then)
   > Update Slider(
    self=Get DroneLinearVelocityZmultiplier()
    Slider Text="LOCTEXT(\"Z Multiplier\")"
    Slider Value=0.2
    Min Max Value_X=0
    Min Max Value_Y=Get DroneLinearVelocityZMultiplierMax()
   )
   > Update Slider(
    self=Get DroneRotationVelocity()
    Slider Text="LOCTEXT(\"Rotational Speed\")"
    Slider Value=0.167
    Min Max Value_X=0
    Min Max Value_Y=Get DroneRotationVelocityMax()
   )
   n74 = GetMinMaxDroneDutchAngle()
   > Update Slider(self=Get DroneDutch(), Slider Text="LOCTEXT(\"Tilt\")", Slider Value=0.5, Min Max Value_X=@n74, Min Max Value_Y=@n74.MaxAngle)
   > UpdateSliderStep(self=Get DroneDutch(), Step=1)
   > Update Slider(
    self=Get DroneDutchSpeed()
    Slider Text="LOCTEXT(\"Tilt Speed\")"
    Slider Value=0.5
    Min Max Value_X=Get DroneDutchSpeedMin()
    Min Max Value_Y=Get DroneDutchSpeedMax()
   )
   > UpdateSliderStep(self=Get DroneDutchSpeed(), Step=1)
   > Initialize(
    self=Get DutchControlScheme()
    Display Text="LOCTEXT(\"Control Scheme\")"
    Options=Make Array!("[0]":Text="LOCTEXT(\"Default\")", "[1]":Text="LOCTEXT(\"Tilt\")") -> (Array:Array of Texts)
    Hover Text="LOCTEXT(\"Enable/Disable expert control scheme\")"
   )
   > n75 = Sequence!(execute) -> (then_0, then_1, then_2:Exec, then_3:Exec)
   > Initialize(
    self=Get ShowCameraShots()
    Display Text="LOCTEXT(\"Camera Sets\")"
    Options=Make Off on Array()
    Hover Text="LOCTEXT(\"Enable/Disable Camera Sets\")"
   )
   > Initialize(
    self=Get Rotator_Preset()
    Display Text="LOCTEXT(\"Preset\")"
    Options=Get ShotFiles() -> (ShotFiles:Array of Texts)
    Hover Text="LOCTEXT(\"Enable/Disable Switch set\")"
   )
   > Initialize(
    self=Get ShowCameraDummy()
    Display Text="LOCTEXT(\"Camera Dummy\")"
    Options=Make Array("[0]"="LOCTEXT(\"Hidden\")", "[1]"="LOCTEXT(\"Visible\")")
    Hover Text="LOCTEXT(\"Enable/Disable Camera Dummy\")"
   )
   Initialize(
    execute=@n75.then_1
    self=Get DroneAttachToThrowable()
    Display Text="LOCTEXT(\"Attach to Throwable\")"
    Options=Make Off on Array()
    Hover Text="LOCTEXT(\"Enable/Disable attaching to throwable\")"
   )
   > Initialize(
    self=Get DroneAttachToBone()
    Display Text="LOCTEXT(\"Attach to Bone\")"
    Options=Make Off on Array()
    Hover Text="LOCTEXT(\"Enable/Disable attaching to bone\")"
   )
   > Initialize(
    self=Get DroneAttachToBoneName()
    Display Text="LOCTEXT(\"Bone\")"
    Options=Make Bones Text!() -> (BonesText:Array of Texts) {is=title K2Node_MacroInstance}
    Hover Text="LOCTEXT(\"Bone Name\")"
   )
   > Update Slider(
    self=Get DroneDutchAngleStepSlider()
    Slider Text="LOCTEXT(\"Tilt Slider Step\")"
    Slider Value=0
    Hover Text="LOCTEXT(\"Dutch Angle Step\")"
    Min Max Value_X=1
    Min Max Value_Y=30
   )
   > UpdateSliderStep(self=Get DroneDutchAngleStepSlider(), Step=1)
   > Update Slider(
    self=Get DroneDutchSpeedStepSlider()
    Slider Text="LOCTEXT(\"Tilt Slider Step\")"
    Slider Value=0
    Hover Text="LOCTEXT(\"Dutch Speed Step\")"
    Min Max Value_X=1
    Min Max Value_Y=30
   )
   > UpdateSliderStep(self=Get DroneDutchSpeedStepSlider(), Step=1)
   > Update Slider!(
    execute
    self=Get DroneDragMultiplier()
    Slider Text="LOCTEXT(\"Drag Multiplier\")"
    Slider Value=0.2
    Hover Text
    Tooltip Text Block
    Min Max Value:Vector 2D Structure=(X=0,Y=5)
   ) -> (then)
   Initialize(
    execute=@n75.then_2
    self=Get DronePendulum()
    Display Text="LOCTEXT(\"Pendulum Swing\")"
    Options=Make Off on Array()
    Hover Text="LOCTEXT(\"Enable pendulum-like swings in a plane, perpendicular to camera view direction\")"
   )
   > Initialize(
    self=Get DronePendulumCameraRoll()
    Display Text="LOCTEXT(\"Roll camera\")"
    Options=Make Off on Array()
    Hover Text="LOCTEXT(\"Update camera roll rotation to align with movement\")"
   )
   n76 = GetMinMaxDronePendulumFrequency!(self:Fort Spectator Advance Settings Tab Object Reference) -> (MinFrequency:float, MaxFrequency:float)
   > Set!(
    execute
    MinMaxValue_X:double=@n76
    MinMaxValue_Y:double=@n76.MaxFrequency
    self:Options Menu Slider Replay Object Reference=Get DronePendulumFrequency()
   ) -> (then, Output_Get:Vector 2D Structure)
   n77 = GetMinMaxDronePendulumArmLength!(self:Fort Spectator Advance Settings Tab Object Reference) -> (MinArmLength:float, MaxArmLength:float)
   > Set MinMaxValue!(
    execute
    MinMaxValue:Vector 2D Structure=vector2d * vector2d(A:Vector 2D Structure=MakeVector2D(X=@n77, Y=@n77.MaxArmLength), B=0.01) -> (ReturnValue:Vector 2D Structure) {is=title K2Node_PromotableOperator}
    self:Options Menu Slider Replay Object Reference=Get DronePendulumArmLength()
   ) -> (then, Output_Get:Vector 2D Structure)
   n78 = GetMinMaxDronePendulumAmplitude!(self:Fort Spectator Advance Settings Tab Object Reference) -> (MinAmplitude:float, MaxAmplitude:float)
   > Set(MinMaxValue_X=@n78, MinMaxValue_Y=@n78.MaxAmplitude, self=Get DronePendulumAmplitude())
   SetPanelPadding(execute=@n75.then_3)
   n79 = Event OnDroneAttachToBonesEnabledChanged!() -> (OutputDelegate, then, bDroneAttachToBonesEnabled:bool) {is=title K2Node_Event}
   > Update Row Selector(
    self=Get DroneAttachToBone()
    Currently Selected=Select(Option 0:int=0, Option 1:int=1, Index:bool=@n79.bDroneAttachToBonesEnabled) -> (ReturnValue:int)
   )
   > SetVisibility(
    self=Get DroneAttachToBoneName()
    InVisibility=Select(Option 0:ESlateVisibility Enum=Collapsed, Option 1:ESlateVisibility Enum=Visible, Index=@n79.bDroneAttachToBonesEnabled) -> (ReturnValue:ESlateVisibility Enum)
   )
   n80 = Event OnCameraTypeChanged!() -> (
    OutputDelegate
    then
    FortPlayerControllerSpectating:Fort Player Controller Spectating Object Reference
    CameraType:ESpectatorCameraType Enum=ThirdPerson
   ) {is=title K2Node_Event}
   > n81 = Set CurrentSpectatorCameraType(CurrentSpectatorCameraType:ESpectatorCameraType Enum=@n80.CameraType) -> (Output_Get:ESpectatorCameraType Enum=ThirdPerson)
   n82 = GetConsoleVariableBoolValue!(execute=[@n81.then, @n89.else, @n92.then, @n91.CastFailed], VariableName:String=drone.AttachToBones) -> (then, ReturnValue:bool)
   n83 = Select!(
    ThirdPerson:bool=false
    DroneFree:bool=true
    Gameplay:bool=false
    ReverseShot:bool=false
    FollowShot:bool=false
    DroneFollow:bool=false
    DroneAttach:bool=false
    BattleMap:bool=false
    ARDrone:bool=false
    DroneFreeDynamic:bool=false
    Sequencer:bool=false
    Index:ESpectatorCameraType Enum=@n81
   ) -> (ReturnValue:bool)
   > SetVisibility(
    self=Get DroneAttachToBone()
    InVisibility=Select!(Option 0:ESlateVisibility Enum=Collapsed, Option 1:ESlateVisibility Enum=Visible, Index:bool=AND Boolean(A=@n82, B=@n83)) -> (ReturnValue:ESlateVisibility Enum)
   )
   n84 = NOT Boolean(A=IsShippingBuild!(self:Fort Runtime Options Object Reference=GetRuntimeOptions()) -> (ReturnValue:bool))
   n85 = Select!(
    ThirdPerson:bool=false
    DroneFree:bool=false
    Gameplay:bool=false
    ReverseShot:bool=false
    FollowShot:bool=false
    DroneFollow:bool=false
    DroneAttach:bool=true
    BattleMap:bool=false
    ARDrone:bool=false
    DroneFreeDynamic:bool=false
    Sequencer:bool=false
    Index:ESpectatorCameraType Enum=@n81
   ) -> (ReturnValue:bool)
   > SetVisibility(
    self=Get DroneZAxisLock()
    InVisibility=Select!(Option 0:ESlateVisibility Enum=Collapsed, Option 1:ESlateVisibility Enum=Visible, Index:bool=AND Boolean(A=@n84, B=@n85)) -> (ReturnValue:ESlateVisibility Enum)
   )
   n86 = GetRuntimeOptions()
   > SetVisibility(
    self=[Get DroneAttachToThrowable(), Get DroneDragMultiplier()]
    InVisibility=Select(Option 0=Visible, Option 1=Collapsed, Index=IsShippingBuild(self=@n86))
   )
   > SetVisibility(self=Get ShowCameraDummy(), InVisibility=Select(Option 0=Visible, Option 1=Collapsed, Index=GetIsPlayingReplay!() -> (ReturnValue:bool)))
   > ScrollBarBehavior() {is=title K2Node_CallFunction}
   > Bind Event to On Keybinds Changed(
    self:Fort Global UIContext Object Reference=Get FortGlobalUIContext!() -> (ReturnValue:Fort Global UIContext Object Reference) {is=title K2Node_GetBlueprintContextEx}
    Delegate=@n61
   ) {is=title K2Node_AddDelegate}
   > SetNavigationRules()
   n87 = Event OnDroneDutchAngleStepChanged!() -> (OutputDelegate, then, DroneDutchAngleStep:float) {is=title K2Node_Event}
   > UpdateSliderValueUnmapped!(execute, self:Options Menu Slider Replay Object Reference=Get DroneDutchAngleStepSlider(), UnmappedValue:double=@n87.DroneDutchAngleStep) -> (then)
   n88 = Event OnDroneDutchSpeedStepChanged!() -> (OutputDelegate, then, DroneDutchSpeedStep:float) {is=title K2Node_Event}
   > UpdateSliderValueUnmapped(self=Get DroneDutchSpeedStepSlider(), UnmappedValue=@n88.DroneDutchSpeedStep)
   Event On Activated()
   > Parent: On Activated() {is=title K2Node_CallParentFunction}
   > n89 = Branch(Condition=Get bFirstActivation() -> (bFirstActivation:bool))
   > n90 = Cast To RotatorSelectorReplaySettings!(execute:Exec, Object=Get DroneDirectMovement()) -> (then:Exec, CastFailed:Exec, AsRotator Selector Replay Settings)
   > SetFocus(self=@n90)
   > Set bFirstActivation(bFirstActivation:bool=false) -> (Output_Get:bool)
   > n91 = Cast To FortReplaySpectator!(execute, Object=GetOwningPlayer()) -> (then, CastFailed, AsFort Replay Spectator:Fort Replay Spectator Object Reference) {is=title K2Node_DynamicCast}
   > n92 = Bind Event to On Hotkey Invoked(self:Fort Player Controller Spectating Object Reference=@n91, Delegate=@n73) {is=title K2Node_AddDelegate}
   Event Construct()
   > n93 = Cast To RotatorSelectorReplaySettings(Object=Get DroneDirectMovement())
   > Bind Event to Selection Changed(
    self:Rotator Selector Replay Settings Object Reference=[
     @n93
     Get DutchControlScheme()
     Get DroneAttachToBone()
     Get DroneAttachToBoneName()
     Get ShowCameraDummy()
     Get ShowCameraShots()
     Get Rotator_Preset()
     Get DroneAttachToThrowable()
     Get DroneGazeDirectedMovement()
     Get DronePendulum()
     Get DronePendulumCameraRoll()
     Get DroneZAxisLock()
     Get DroneXInputLock()
     Get DroneYInputLock()
     Get DroneDirectMovementSaveState()
    ]
    Delegate=@n68
   ) {is=title K2Node_AddDelegate}
   > Bind Event to Slider Changed(
    self:Options Menu Slider Replay Object Reference=[
     Get DroneLinearVelocity()
     Get DroneRotationVelocity()
     Get DroneLinearVelocityZmultiplier()
     Get DroneDragMultiplier()
     Get DroneDutch()
     Get DroneDutchAngleStepSlider()
     Get DroneDutchSpeed()
     Get DroneDutchSpeedStepSlider()
     Get DronePendulumFrequency()
     Get DronePendulumArmLength()
     Get DronePendulumAmplitude()
    ]
    Delegate=@n69
   ) {is=title K2Node_AddDelegate}
   > Bind Event to Button Clicked(self:Options Menu Action Replay Object Reference=[Get Button_ResetDutchAngle(), Get Button_ResetDutchSpeed()], Delegate=@n70) {is=title K2Node_AddDelegate}
   > n94 = Branch(Condition=IsShippingBuild(self=GetRuntimeOptions()))
   > SetVisibility(
    self=[
     Get Rotator_Preset()
     Get Rotator_CurrentShot()
     Get Button_UndoRedo()
     Get DronePendulumFrequency()
     Get DronePendulumCameraRoll()
     Get DronePendulumArmLength()
     Get DronePendulumAmplitude()
     Get DroneZAxisLock()
     Get DroneXInputLock()
     Get DroneYInputLock()
     Get DroneDutchAngleStepSlider()
     Get DroneDutchSpeedStepSlider()
     Get DroneGazeDirectedMovement()
     Get DronePendulum()
     Get DroneDirectMovementSaveState()
     Get ShowCameraShots()
     Get ShowCameraDummy()
    ]
    InVisibility=Collapsed
   )
   n95 = Cast To FortPlayerControllerSpectating(execute=@n94.else, Object=GetOwningPlayer())
   > Assign On Hotkey Assigned!(execute:Exec, self:Fort Player Controller Spectating Object Reference=@n95, Delegate:"Delegate (by ref)"=@n60) -> (then:Exec) {is=title K2Node_AssignDelegate}
   n96 = Event OnDroneAttachToBonesNameChanged!() -> (OutputDelegate, then, DroneAttachToBonesName:Name=None) {is=title K2Node_Event}
   > n97 = For Each Loop with Break!(Exec:Exec, Array:Array of Texts=Make Bones Names(), Break:Exec=@n98.then) -> (LoopBody:Exec, Array Element:Text, Array Index:int, Completed:Exec)
   > n98 = Branch(Condition="Equal, Case Insensitive (Text)"!(A:"Text (by ref)"=@n97, B:"Text (by ref)"="To Text (Name)"!(InName:Name=@n96.DroneAttachToBonesName) -> (ReturnValue:Text)) -> (ReturnValue:bool))
   Update Row Selector(execute=@n97.Completed, self=Get DroneAttachToBoneName(), Currently Selected=@n97.Array Index)
   n99 = Event OnDroneDragMultiplierChanged!() -> (OutputDelegate, then, DroneDragMultiplier:float) {is=title K2Node_Event}
   > UpdateSliderValueUnmapped(self=Get DroneDragMultiplier(), UnmappedValue=@n99.DroneDragMultiplier)
   n100 = Event OnCameraShotPresetChanged!() -> (OutputDelegate, then, Number:int=0) {is=title K2Node_Event}
   > Branch(Condition="integer >= integer"(A=@n100.Number))
   > Update Row Selector(self=Get Rotator_Preset(), Currently Selected=@n100.Number)
   > n101 = Cast To FortPlayerControllerSpectating(Object=GetOwningPlayer())
   n102 = Get ShotFiles()
   > ShowNotification!(
    execute
    self:Fort Player Controller Spectating Object Reference=@n101
    NotificationText:Text=Format Text!(Format="LOCTEXT(\"Camera Preset {Preset} \")", Preset:"Text (by ref)"="Get (a ref)"(Array=@n102, Dimension 1=@n100.Number)) -> (Result)
   ) -> (then)
   n103 = Event OnDroneAttachToThrowableEnabledChanged!() -> (OutputDelegate, then, bDroneAttachToThrowableEnabled:bool) {is=title K2Node_Event}
   > Update Row Selector(
    self=Get DroneAttachToThrowable()
    Currently Selected=Select(Option 0:int=0, Option 1:int=1, Index=@n103.bDroneAttachToThrowableEnabled) -> (ReturnValue:int)
   )
   > Branch(Condition=@n103.bDroneAttachToThrowableEnabled)
   > SetDroneAttachToBonesEnabled(bDroneAttachToBonesEnabled=NOT Boolean(A=@n103.bDroneAttachToThrowableEnabled))
   n104 = Event OnDroneGazeDirectedVerticalMovementEnabledChanged!() -> (OutputDelegate, then, bDroneGazeDirectedVerticalMovementEnabled:bool) {is=title K2Node_Event}
   > Update Row Selector(
    self=Get DroneGazeDirectedMovement()
    Currently Selected=Select(Option 0=0, Option 1=1, Index=@n104.bDroneGazeDirectedVerticalMovementEnabled)
   )
   n105 = Event OnDroneDirectMovementEnabledChanged!() -> (OutputDelegate, then, bDroneDirectMovementEnabled:bool) {is=title K2Node_Event}
   > Update Row Selector(
    self=Cast To RotatorSelectorReplaySettings!(Object=Get DroneDirectMovement()) -> (AsRotator Selector Replay Settings, bSuccess:bool=~)
    Currently Selected=Select(Option 0=0, Option 1=1, Index=@n105.bDroneDirectMovementEnabled)
   )
   > SetVisibility(
    self=[Get DroneLinearVelocity(), Get DroneRotationVelocity(), Get DroneLinearVelocityZmultiplier()]
    InVisibility=Select(Option 0:ESlateVisibility Enum=Collapsed, Option 1:ESlateVisibility Enum=Visible, Index=@n105.bDroneDirectMovementEnabled) -> (ReturnValue:ESlateVisibility Enum)
   )
   n106 = NOT Boolean(A=IsShippingBuild(self=GetRuntimeOptions()))
   > SetVisibility(
    self=Get DroneDirectMovementSaveState()
    InVisibility=Select(Option 0=Collapsed, Option 1=Visible, Index=AND Boolean(A=@n105.bDroneDirectMovementEnabled, B=@n106))
   )
   n107 = Event OnDroneDirectMovementStateSavedChanged!() -> (OutputDelegate, then, bDroneDirectMovementStateSaved:bool) {is=title K2Node_Event}
   > Update Row Selector(
    self=Get DroneDirectMovementSaveState()
    Currently Selected=Select(Option 0:int=0, Option 1:int=1, Index=@n107.bDroneDirectMovementStateSaved) -> (ReturnValue:int)
   )
   Event Destruct()
   > Branch(Condition=GetIsPlayingReplay())
   > ResetToDefault(self:Fort Replay View Settings Tab Base Object Reference)
   n108 = Event OnDronePendulumEnabledChanged!() -> (OutputDelegate, then, bDronePendulumEnabled:bool) {is=title K2Node_Event}
   > Update Row Selector(self=Get DronePendulum(), Currently Selected=Select(Option 0=0, Option 1=1, Index=@n108.bDronePendulumEnabled))
   n109 = Event OnDronePendulumAmplitudeChanged!() -> (OutputDelegate, then, DronePendulumAmplitude:float) {is=title K2Node_Event}
   > UpdateSliderValueUnmapped(self=Get DronePendulumAmplitude(), UnmappedValue=@n109.DronePendulumAmplitude)
   n110 = Event OnDronePendulumArmLengthChanged!() -> (OutputDelegate, then, DronePendulumArmLength:float) {is=title K2Node_Event}
   > UpdateSliderValueUnmapped(self=Get DronePendulumArmLength(), UnmappedValue=float * float(A=@n110.DronePendulumArmLength, B=0.01))
   n111 = Event OnDronePendulumFrequencyChanged!() -> (OutputDelegate, then, DronePendulumFrequency:float) {is=title K2Node_Event}
   > UpdateSliderValueUnmapped(self=Get DronePendulumFrequency(), UnmappedValue=@n111.DronePendulumFrequency)
   n112 = Event OnDronePendulumRollEnabledChanged!() -> (OutputDelegate, then, bDronePendulumRollEnabled:bool) {is=title K2Node_Event}
   > Update Row Selector(self=Get DronePendulumCameraRoll(), Currently Selected=Select(Option 0=0, Option 1=1, Index=@n112.bDronePendulumRollEnabled))
   n113 = Event OnDroneZAxisLockEnabledChanged!() -> (OutputDelegate, then, bDroneZAxisLockEnabled:bool) {is=title K2Node_Event}
   > Update Row Selector(self=Get DroneZAxisLock(), Currently Selected=Select(Option 0=0, Option 1=1, Index=@n113.bDroneZAxisLockEnabled))
   n114 = Event OnDroneXInputLockEnabledChanged!() -> (OutputDelegate, then, bDroneXInputLockEnabled:bool) {is=title K2Node_Event}
   > Update Row Selector(self=Get DroneXInputLock(), Currently Selected=Select(Option 0=0, Option 1=1, Index=@n114.bDroneXInputLockEnabled))
   n115 = Event OnDroneYInputLockEnabledChanged!() -> (OutputDelegate, then, bDroneYInputLockEnabled:bool) {is=title K2Node_Event}
   > Update Row Selector(self=Get DroneYInputLock(), Currently Selected=Select(Option 0=0, Option 1=1, Index=@n115.bDroneYInputLockEnabled))
   n116 = Event OnDutchControlSchemaEnabledChanged!() -> (OutputDelegate, then, bDutchControlSchemaEnabled:bool) {is=title K2Node_Event}
   > Update Row Selector(self=Get DutchControlScheme(), Currently Selected=Select(Option 0=0, Option 1=1, Index=@n116.bDutchControlSchemaEnabled))
   n117 = Event OnDroneDirectMovementLinearVelocityChanged!() -> (OutputDelegate, then, DroneLinearVelocity:float) {is=title K2Node_Event}
   > UpdateSliderValueUnmapped(self=Get DroneLinearVelocity(), UnmappedValue=@n117.DroneLinearVelocity)
   n118 = Event OnDroneDirectMovementRotationVelocityChanged!() -> (OutputDelegate, then, DroneRotationVelocity:float) {is=title K2Node_Event}
   > UpdateSliderValueUnmapped(self=Get DroneRotationVelocity(), UnmappedValue=@n118.DroneRotationVelocity)
   n119 = Event OnDroneDutchSpeedChanged!() -> (OutputDelegate, then, DroneDutchSpeed:float) {is=title K2Node_Event}
   > UpdateSliderValueUnmapped(self=Get DroneDutchSpeed(), UnmappedValue=@n119.DroneDutchSpeed)
   n120 = Event OnDroneDirectMovementZLinearVelocityMultiplierChanged!() -> (OutputDelegate, then, DroneZLinearVelocityMultiplier:float) {is=title K2Node_Event}
   > UpdateSliderValueUnmapped(self=Get DroneLinearVelocityZmultiplier(), UnmappedValue=@n120.DroneZLinearVelocityMultiplier)
   n121 = Event OnDroneDutchAngleChanged!() -> (OutputDelegate, then, DroneDutchAngle:float) {is=title K2Node_Event}
   > UpdateSliderValueUnmapped(self=Get DroneDutch(), UnmappedValue=@n121.DroneDutchAngle)
   Event ResetFocusedElement() {is=title K2Node_Event}
   > Is Valid(InputObject=Get LatestHoveredWidget())
   > n122 = Branch(Condition="Equal (Object)"(A=Get DroneDirectMovement(), B=Get LatestHoveredWidget()))
   > ResetDroneDirectMovementEnabled(self:Fort Spectator Advance Settings Tab Object Reference)
   n123 = Branch(execute=@n122.else, Condition="Equal (Object)"(A=Get DutchControlScheme(), B=Get LatestHoveredWidget()))
   > ResetDutchControlSchemaEnabled(self:Fort Spectator Advance Settings Tab Object Reference)
   n124 = Branch(execute=@n123.else, Condition="Equal (Object)"(A=Get DroneLinearVelocity(), B=Get LatestHoveredWidget()))
   > ResetDroneDirectMovementLinearVelocity(self:Fort Spectator Advance Settings Tab Object Reference)
   n125 = Branch(execute=@n124.else, Condition="Equal (Object)"(A=Get DroneLinearVelocityZmultiplier(), B=Get LatestHoveredWidget()))
   > ResetDroneDirectMovementZLinearVelocityMultiplier(self:Fort Spectator Advance Settings Tab Object Reference)
   n126 = Branch(execute=@n125.else, Condition="Equal (Object)"(A=Get DroneRotationVelocity(), B=Get LatestHoveredWidget()))
   > ResetDroneDirectMovementRotationVelocity(self:Fort Spectator Advance Settings Tab Object Reference)
   n127 = Branch(execute=@n126.else, Condition="Equal (Object)"(A=Get DroneDutchSpeed(), B=Get LatestHoveredWidget()))
   > ResetDroneDutchSpeed(self:Fort Spectator Advance Settings Tab Object Reference)
   n128 = Branch(execute=@n127.else, Condition="Equal (Object)"(A=Get DroneDutch(), B=Get LatestHoveredWidget()))
   > ResetDroneDutchAngle(self:Fort Spectator Advance Settings Tab Object Reference)
   n129 = Branch(execute=@n128.else, Condition="Equal (Object)"(A=Get DroneDutchAngleStepSlider(), B=Get LatestHoveredWidget()))
   > ResetDroneDutchAngleStep(self:Fort Spectator Advance Settings Tab Object Reference)
   n130 = Branch(execute=@n129.else, Condition="Equal (Object)"(A=Get DroneDutchSpeedStepSlider(), B=Get LatestHoveredWidget()))
   > ResetDroneDutchSpeedStep(self:Fort Spectator Advance Settings Tab Object Reference)
   n131 = Branch(execute=@n130.else, Condition="Equal (Object)"(A=Get DroneDragMultiplier(), B=Get LatestHoveredWidget()))
   > ResetDroneDragMultiplier(self:Fort Spectator Advance Settings Tab Object Reference)
   n132 = Branch(execute=@n131.else, Condition="Equal (Object)"(A=Get DroneDirectMovementSaveState(), B=Get LatestHoveredWidget()))
   > ResetDroneDirectMovementStateSaved(self:Fort Spectator Advance Settings Tab Object Reference)
   n133 = Branch(execute=@n132.else, Condition="Equal (Object)"(A=Get DroneGazeDirectedMovement(), B=Get LatestHoveredWidget()))
   > ResetDroneGazeDirectedVerticalMovementEnabled(self:Fort Spectator Advance Settings Tab Object Reference)
   n134 = Branch(execute=@n133.else, Condition="Equal (Object)"(A=Get DronePendulum(), B=Get LatestHoveredWidget()))
   > ResetDronePendulumEnabled(self:Fort Spectator Advance Settings Tab Object Reference)
   n135 = Branch(execute=@n134.else, Condition="Equal (Object)"(A=Get DronePendulumFrequency(), B=Get LatestHoveredWidget()))
   > ResetDronePendulumFrequency(self:Fort Spectator Advance Settings Tab Object Reference)
   n136 = Branch(execute=@n135.else, Condition="Equal (Object)"(A=Get DronePendulumArmLength(), B=Get LatestHoveredWidget()))
   > ResetDronePendulumArmLength(self:Fort Spectator Advance Settings Tab Object Reference)
   n137 = Branch(execute=@n136.else, Condition="Equal (Object)"(A=Get DronePendulumAmplitude(), B=Get LatestHoveredWidget()))
   > ResetDronePendulumAmplitude(self:Fort Spectator Advance Settings Tab Object Reference)
   n138 = Branch(execute=@n137.else, Condition="Equal (Object)"(A=Get DronePendulumCameraRoll(), B=Get LatestHoveredWidget()))
   > ResetDronePendulumRollEnabled(self:Fort Spectator Advance Settings Tab Object Reference)
   n139 = Branch(execute=@n138.else, Condition="Equal (Object)"(A=Get DroneZAxisLock(), B=Get LatestHoveredWidget()))
   > ResetDroneZAxisLockEnabled(self:Fort Spectator Advance Settings Tab Object Reference)
   n140 = Branch(execute=@n139.else, Condition="Equal (Object)"(A=Get DroneXInputLock(), B=Get LatestHoveredWidget()))
   > ResetDroneXInputLockEnabled(self:Fort Spectator Advance Settings Tab Object Reference)
   Branch(execute=@n140.else, Condition="Equal (Object)"(A=Get DroneYInputLock(), B=Get LatestHoveredWidget()))
   > ResetDroneYInputLockEnabled(self:Fort Spectator Advance Settings Tab Object Reference)
  ]}
 GetAllVisibleChilds {nodes [
   n1 = GetAllVisibleChilds!() -> (then, Panel:Panel Widget Object Reference, IsVisible:bool=~) {is=title K2Node_FunctionEntry}
   > n2 = For Each Loop(Array:Array of Widget Object References=GetAllChildren(self=@n1)) -> (Array Element:Widget Object Reference)
   Return Node!(execute=@n2.Completed, ReturnValue:Array of Widget Object References=Get Childs() -> (Childs:Array of Widget Object References))
   > Branch(Condition="Equal (Boolean)"(A:bool=IsVisible(self=@n2), B:bool=@n1.IsVisible) -> (ReturnValue:bool=~) {is=title K2Node_PromotableOperator})
   > Add(TargetArray:Array of Widget Object References=Get Childs(), NewItem:"Widget Object Reference (by ref)"=@n2) {is=title K2Node_CallArrayFunction}
  ]}
 IsGamepadInputType {nodes [
   IsGamepadInputType() {is=title K2Node_FunctionEntry}
   n1 = Get CommonInputSubsystem()
   > Return Node(ReturnValue:bool="Equal (Enum)"(A=GetCurrentInputType(self=@n1), B=Gamepad))
  ]}
 MakeBonesNames {nodes [
   Inputs!()
   Outputs!(BonesNames:Array of Texts=Make Array!(
     "[0]"="LOCTEXT(\"head\")"
     "[1]"="LOCTEXT(\"hand_r\")"
     "[2]":Text="LOCTEXT(\"hand_l\")"
     "[3]":Text="LOCTEXT(\"pelvis\")"
     "[4]":Text="LOCTEXT(\"weapon_r\")"
     "[5]":Text="LOCTEXT(\"weapon_l\")"
     "[6]":Text="LOCTEXT(\"foot_r\")"
     "[7]":Text="LOCTEXT(\"foot_l\")"
    ) -> (Array))
  ]}
 MakeBonesText {nodes [
   Inputs()
   Outputs!(BonesText:Array of Texts=Make Array(
     "[0]"="LOCTEXT(\"Head\")"
     "[1]"="LOCTEXT(\"Right Arm\")"
     "[2]"="LOCTEXT(\"Left Arm\")"
     "[3]"="LOCTEXT(\"Pelvis\")"
     "[4]"="LOCTEXT(\"Weapon Right\")"
     "[5]"="LOCTEXT(\"Weapon Left\")"
     "[6]"="LOCTEXT(\"Foot Right\")"
     "[7]"="LOCTEXT(\"Foot Left\")"
    ))
  ]}
 MakeOffOnArray {nodes [
   Inputs()
   Outputs!(OffOnArray:Array of Texts=Make Array!("[0]"=Get OffText() -> (OffText:Text), "[1]"=Get OnText() -> (OnText:Text)) -> (Array))
  ]}
 MarkShots {nodes [
   n1 = MarkShots!() -> (then, ShotNumber:int) {is=title K2Node_FunctionEntry}
   > Clear(TargetArray:Array of Texts=Get Options())
   > n2 = Cast To FortPlayerControllerSpectating(Object=GetOwningPlayer())
   > n3 = For Each Loop(Array:Array of WBP UIKit Button Regular Object References=Make Array!(
     "[0]":WBP UIKit Button Regular Object Reference=Get Shot0_Button() -> (Shot0_Button:WBP UIKit Button Regular Object Reference)
     "[1]":WBP UIKit Button Regular Object Reference=Get Shot1_Button() -> (Shot1_Button:WBP UIKit Button Regular Object Reference)
     "[2]":WBP UIKit Button Regular Object Reference=Get Shot2_Button() -> (Shot2_Button:WBP UIKit Button Regular Object Reference)
     "[3]":WBP UIKit Button Regular Object Reference=Get Shot3_Button() -> (Shot3_Button:WBP UIKit Button Regular Object Reference)
     "[4]":WBP UIKit Button Regular Object Reference=Get Shot4_Button() -> (Shot4_Button:WBP UIKit Button Regular Object Reference)
     "[5]":WBP UIKit Button Regular Object Reference=Get Shot5_Button() -> (Shot5_Button:WBP UIKit Button Regular Object Reference)
     "[6]":WBP UIKit Button Regular Object Reference=Get Shot6_Button() -> (Shot6_Button:WBP UIKit Button Regular Object Reference)
     "[7]":WBP UIKit Button Regular Object Reference=Get Shot7_Button() -> (Shot7_Button:WBP UIKit Button Regular Object Reference)
     "[8]":WBP UIKit Button Regular Object Reference=Get Shot8_Button() -> (Shot8_Button:WBP UIKit Button Regular Object Reference)
     "[9]":WBP UIKit Button Regular Object Reference=Get Shot9_Button() -> (Shot9_Button:WBP UIKit Button Regular Object Reference)
    ) -> (Array:Array of WBP UIKit Button Regular Object References)) -> (Array Element:WBP UIKit Button Regular Object Reference)
   n4 = For Each Loop(
    Exec=@n3.Completed
    Array=Make Array(
     "[0]"=Get Shot0_Button_1()
     "[1]"=Get Shot1_Button_1()
     "[2]"=Get Shot2_Button_1()
     "[3]"=Get Shot3_Button_1()
     "[4]"=Get Shot4_Button_1()
     "[5]"=Get Shot5_Button_1()
     "[6]"=Get Shot6_Button_1()
     "[7]"=Get Shot7_Button_1()
     "[8]"=Get Shot8_Button_1()
     "[9]"=Get Shot9_Button_1()
    )
   )
   > n5 = IsBackHotkeySet!(execute, self:Fort Player Controller Spectating Object Reference=@n2, HotKeyNum:int=@n4.Array Index) -> (then, ReturnValue:bool)
   > SetIsEnabled(self=@n4, bInIsEnabled=@n5)
   > SetIcon!(
    execute
    self:WBP UIKit Button CTA Base Object Reference=@n4
    Icon:Object Reference=/Game/UI/Foundation/Textures/Icons/Replay/T-Icon-Replay-Rewind-L
   ) -> (then)
   > SetVisibility(
    self=@n4
    InVisibility=Select(Option 0:ESlateVisibility Enum=Collapsed, Option 1:ESlateVisibility Enum=Visible, Index=@n5) -> (ReturnValue:ESlateVisibility Enum)
   )
   n6 = Get Rotator_CurrentShot()
   n7 = Update Options!(execute=@n4.Completed, self:Rotator Selector Replay Settings Object Reference=@n6, Options:Array of Texts=Get Options()) -> (then, bOptionsEmpty:bool)
   > n8 = Branch(Condition=@n7)
   > n9 = UpdateOptionsText!(execute, self:Rotator Selector Replay Settings Object Reference=@n6, NewText:Text="LOCTEXT(\"No Shots Found\")") -> (then)
   Return Node!(execute=[@n10.then, @n9.then])
   n10 = UpdateOptionsText(execute=@n8.else, self=@n6, NewText="To Text (Integer)"(Value=@n1, bUseGrouping=false, MaximumIntegralDigits=3))
   n11 = IsHotkeySet!(execute=@n3.LoopBody, self:Fort Player Controller Spectating Object Reference=@n2, HotKeyNum:int=@n3.Array Index) -> (then, ReturnValue:bool)
   > SetIsEnabled(self=@n3, bInIsEnabled=@n11)
   > Branch(Condition=@n11)
   > Add(
    TargetArray:Array of Texts=Get Options()
    NewItem:"Text (by ref)"="To Text (Integer)"(Value=@n3.Array Index, bUseGrouping=false, MaximumIntegralDigits=3)
   )
  ]}
 SetNavigationRules {nodes [
   SetNavigationRules() {is=title K2Node_FunctionEntry}
   > n1 = Switch on Int!(execute:Exec, Selection:int=Get SelectedIndex(self=Get ShowCameraShots()) -> (SelectedIndex=0)) -> (Default:Exec, 0:Exec, 1:Exec)
   Set Loc_BottomWidget(execute=@n1.0, Loc_BottomWidget:Widget Object Reference=Get DutchControlScheme()) -> (Output_Get:Widget Object Reference)
   > n2 = GetConsoleVariableBoolValue(VariableName=drone.AttachToBones)
   > n3 = Branch(Condition=@n2)
   > n4 = Set Loc_BottomWidget(Loc_BottomWidget=Get DroneAttachToBone())
   n5 = Branch(execute=[@n4.then, @n3.else], Condition=GetIsPlayingReplay())
   n6 = Branch(execute=[@n10.then, @n5.then], Condition=IsShippingBuild(self=GetRuntimeOptions()))
   n7 = GetAllVisibleChilds!(Panel:Panel Widget Object Reference=Get ScrollBox_Tabs(), IsVisible:bool=true) -> (ReturnValue:Array of Widget Object References) {is=title K2Node_CallFunction}
   Set Loc_BottomWidget(
    execute=[@n13.then, @n9.then, @n6.then]
    Loc_BottomWidget="Get (a copy)"(Array:Array of Widget Object References=@n7, Dimension 1=Last Index!(TargetArray:Array of Widget Object References=@n7) -> (ReturnValue=0)) -> (Output:Widget Object Reference)
   )
   n8 = "Get (a copy)"(Array=GetAllChildren(self=Get ScrollBox_Tabs()))
   > SetNavigationRuleExplicit!(
    execute
    self:Widget Object Reference=@n8
    Direction:EUINavigation Enum=Up
    InWidget:Widget Object Reference=Get Loc_BottomWidget() -> (Loc_BottomWidget:Widget Object Reference)
   ) -> (then)
   > SetNavigationRuleExplicit(self=Get Loc_BottomWidget(), Direction=Down, InWidget=@n8)
   n9 = Set Loc_BottomWidget(execute=@n6.else, Loc_BottomWidget=Get ShowCameraShots())
   n10 = Set Loc_BottomWidget(execute=@n5.else, Loc_BottomWidget=Get ShowCameraDummy())
   n11 = "Get (a copy)"(Array=GetAllChildren(self=Get ScrollBox_Tabs()))
   n12 = Get Button_UndoRedo()
   SetNavigationRuleExplicit(execute=@n1.1, self=@n11, Direction=Up, InWidget=@n12)
   > SetNavigationRuleExplicit(self=@n12, Direction=Down, InWidget=@n11)
   > SetNavigationRuleExplicit(self=Get ShowCameraShots(), Direction=Down, InWidget=Get Rotator_Preset())
   n13 = Set Loc_BottomWidget(Loc_BottomWidget=Get DroneAttachToThrowable())
  ]}
 SetPanelPadding {nodes [
   SetPanelPadding() {is=title K2Node_FunctionEntry}
   > n1 = For Each Loop(Array:Array of Widget Object References=GetAllChildren(self=Get ScrollBox_Tabs())) -> (Array Element:Widget Object Reference)
   > Set Loc_ScrollSlot(Loc_ScrollSlot:Scroll Box Slot Object Reference=SlotAsScrollBoxSlot!(Widget:Widget Object Reference=@n1) -> (ReturnValue:Scroll Box Slot Object Reference)) -> (Output_Get:Scroll Box Slot Object Reference)
   > n2 = Branch(Condition="Equal (Integer)"(A=@n1.Array Index))
   > SetPadding!(
    execute
    self:Scroll Box Slot Object Reference=Get Loc_ScrollSlot() -> (Loc_ScrollSlot:Scroll Box Slot Object Reference)
    InPadding_Left:float=15
    InPadding_Top:float=15
    InPadding_Right:float=15
    InPadding_Bottom:float=12
   ) -> (then)
   n3 = Length(TargetArray:Array of Widget Object References=GetAllChildren(self=Get ScrollBox_Tabs()))
   n4 = Branch(execute=@n2.else, Condition="Equal (Integer)"(A=@n1.Array Index, B=int - int(A=@n3)))
   > SetPadding(self=Get Loc_ScrollSlot(), InPadding_Left=15, InPadding_Top=0, InPadding_Right=15, InPadding_Bottom=15)
   SetPadding(execute=@n4.else, self=Get Loc_ScrollSlot(), InPadding_Left=15, InPadding_Top=0, InPadding_Right=15, InPadding_Bottom=12)
  ]}
 ShowSaveLoadPrompt {nodes [
   n1 = ShowSaveLoadPrompt!() -> (then, Model:int) {is=title K2Node_FunctionEntry}
   > Set LocConfig(LocConfig:int=@n1) -> (Output_Get:int=0)
   > n2 = Branch(Condition=IsShippingBuild(self=GetRuntimeOptions()))
   n3 = IsDesktopPlatform!(execute=@n2.else, self:Fort Global UIContext Object Reference=Get FortGlobalUIContext()) -> (then, ReturnValue:bool)
   > Branch(Condition=@n3)
   > n4 = Create Widget!(
    execute:Exec
    Class:User Widget Class Reference=&/FortBroadcastSaveLoadPrompt.FortBroadcastSaveLoadPrompt_C
    OwningPlayer:Player Controller Object Reference
    MenuConfig:int=Get LocConfig() -> (LocConfig:int=0)
    ViewSettings:Fort Broadcast View Settings Object Reference
   ) -> (then:Exec, ReturnValue:Fort Broadcast Save Load Prompt Object Reference)
   > AddToViewport!(execute, self:User Widget Object Reference=@n4, ZOrder:int=0) -> (then)
  ]}
 ToggleUndoRedoButton {nodes [
   n1 = ToggleUndoRedoButton!() -> (then, Button:WBP UIKit Button Regular Object Reference) {is=title K2Node_FunctionEntry}
   > n2 = Branch(Condition="Equal (Object)"(
     A=Get Icon!(self:WBP UIKit Button CTA Base Object Reference=@n1) -> (Icon:Texture 2D Object Reference)
     B=/Game/UI/Foundation/Textures/Icons/Replay/T-Icon-Replay-Forward-L
    ))
   > SetIcon(self=@n1, Icon=&T-Icon-Replay-Rewind-L)
   SetIcon(execute=@n2.else, self=@n1, Icon=&T-Icon-Replay-Forward-L)
  ]}
]
variables [
 bFirstActivation {type=boolean, default=true}
 CurrentSpectatorCameraType {type=ESpectatorCameraType, default=ThirdPerson}
 DroneDutchSpeedMax {type=number, default=30}
 DroneDutchSpeedMin {type=number, default=-30}
 DroneLinearVelocityMax {type=number, default=5}
 DroneLinearVelocityZMultiplierMax {type=number, default=5}
 DroneRotationVelocityMax {type=number, default=3}
 LatestHoveredWidget {type=UserWidget, default=null}
 OffText {type=string, default=Off}
 OnText {type=string, default=On}
 Options {type="array<string>", default []}
 ShotFiles {type="array<string>", default [0000,0005,0010,0015,0020,0025,0030,0035,0040,0045]}
 StoredDragMultiplier {type=number, default=1}
]
widgets {
 info {inheritedWidgetCount=1, parentClass=/Script/FortniteUI.FortSpectatorAdvanceSettingsTab, rootWidgetClass=&Overlay, widgetCount=63}
 widgets [
  Overlay_0 {widgetClassPath=&Overlay, widget=Overlay_0, properties {bOverride_Cursor=true}}
  SizeBox_0 {
   widgetClassPath=&SizeBox
   parent=Overlay_0
   widget=SizeBox_0
   properties {bOverride_HeightOverride=true, bOverride_WidthOverride=true, heightOverride=470, maxDesiredHeight=410, widthOverride=700}
   slot {class=OverlaySlot}
  }
  ScrollBox_Tabs {
   widgetClassPath=&ScrollBox
   bIsVariable=true
   parent=SizeBox_0
   widget=ScrollBox_Tabs
   properties {
    allowOverscroll=false
    bAnimateWheelScrolling=true
    consumeMouseWheel=Always
    navigationDestination=Center
    scrollbarPadding=<ltrb 0 10 10 10>
    scrollbarThickness=<15 15>
    widgetBarStyle {
     thickness=12
     draggedThumbImage {
      bIsDynamicallyLoaded=false
      drawAs=RoundedBox
      imageType=FullColor
      mirroring=NoMirror
      resourceName=../../../Engine/Content/Slate/Common/Scrollbar_Thumb.png
      resourceObject=null
      tiling=NoTile
      imageSize=<12 12>
      margin=<ltrb 0.25 0.25 0.25 0.25>
      outlineSettings {
       bUseBrushTransparency=false
       roundingType=FixedRadius
       width=2
       color {colorUseRule=UseColor_Specified, specifiedColor=<rgba 0 0 0 0>}
       cornerRadii=<6 6 6 6>
      }
      tintColor {colorUseRule=UseColor_Specified, specifiedColor=<rgba 1 1 1 1>}
      uVRegion {bIsValid=false, max=<0 0>, min=<0 0>}
     }
     horizontalBackgroundImage {
      bIsDynamicallyLoaded=false
      drawAs=Box
      imageType=FullColor
      mirroring=NoMirror
      resourceName=None
      resourceObject=/FortUILibrary/UIKit/ScrollBar/MaterialInstances/MI_UIKit_ScrollbarBackground
      tiling=NoTile
      imageSize=<12 12>
      margin=<ltrb 0.5 0.5 0.5 0.5>
      outlineSettings {
       bUseBrushTransparency=false
       roundingType=HalfHeightRadius
       width=0
       color {colorUseRule=UseColor_Specified, specifiedColor=<rgba 0 0 0 0>}
       cornerRadii=<0 0 0 0>
      }
      tintColor {colorUseRule=UseColor_Specified, specifiedColor=<rgba 1 1 1 1>}
      uVRegion {bIsValid=false, max=<0 0>, min=<0 0>}
     }
     horizontalBottomSlotImage {
      bIsDynamicallyLoaded=false
      drawAs=NoDrawType
      imageType=FullColor
      mirroring=NoMirror
      resourceName=../../../Engine/Content/Slate/Common/Scrollbar_Background_Horizontal.png
      resourceObject=null
      tiling=NoTile
      imageSize=<8 8>
      margin=<ltrb 0 0 0 0>
      outlineSettings {
       bUseBrushTransparency=false
       roundingType=HalfHeightRadius
       width=0
       color {colorUseRule=UseColor_Specified, specifiedColor=<rgba 0 0 0 0>}
       cornerRadii=<0 0 0 0>
      }
      tintColor {colorUseRule=UseColor_Specified, specifiedColor=<rgba 1 1 1 1>}
      uVRegion {bIsValid=false, max=<0 0>, min=<0 0>}
     }
     horizontalTopSlotImage {
      bIsDynamicallyLoaded=false
      drawAs=NoDrawType
      imageType=FullColor
      mirroring=NoMirror
      resourceName=../../../Engine/Content/Slate/Common/Scrollbar_Background_Horizontal.png
      resourceObject=null
      tiling=NoTile
      imageSize=<8 8>
      margin=<ltrb 0 0 0 0>
      outlineSettings {
       bUseBrushTransparency=false
       roundingType=HalfHeightRadius
       width=0
       color {colorUseRule=UseColor_Specified, specifiedColor=<rgba 0 0 0 0>}
       cornerRadii=<0 0 0 0>
      }
      tintColor {colorUseRule=UseColor_Specified, specifiedColor=<rgba 1 1 1 1>}
      uVRegion {bIsValid=false, max=<0 0>, min=<0 0>}
     }
     hoveredThumbImage {
      bIsDynamicallyLoaded=false
      drawAs=RoundedBox
      imageType=FullColor
      mirroring=NoMirror
      resourceName=../../../Engine/Content/Slate/Common/Scrollbar_Thumb.png
      resourceObject=null
      tiling=NoTile
      imageSize=<12 12>
      margin=<ltrb 0.25 0.25 0.25 0.25>
      outlineSettings {
       bUseBrushTransparency=false
       roundingType=FixedRadius
       width=2
       color {colorUseRule=UseColor_Specified, specifiedColor=<rgba 0 0 0 0>}
       cornerRadii=<6 6 6 6>
      }
      tintColor {colorUseRule=UseColor_Specified, specifiedColor=<rgba 1 1 1 1>}
      uVRegion {bIsValid=false, max=<0 0>, min=<0 0>}
     }
     normalThumbImage {
      bIsDynamicallyLoaded=false
      drawAs=RoundedBox
      imageType=FullColor
      mirroring=NoMirror
      resourceName=../../../Engine/Content/Slate/Common/Scrollbar_Thumb.png
      resourceObject=null
      tiling=NoTile
      imageSize=<12 12>
      margin=<ltrb 0.25 0.25 0.25 0.25>
      outlineSettings {
       bUseBrushTransparency=false
       roundingType=FixedRadius
       width=2
       color {colorUseRule=UseColor_Specified, specifiedColor=<rgba 0 0 0 0>}
       cornerRadii=<6 6 6 6>
      }
      tintColor {colorUseRule=UseColor_Specified, specifiedColor=<rgba 1 1 1 0.75>}
      uVRegion {bIsValid=false, max=<0 0>, min=<0 0>}
     }
     verticalBackgroundImage {
      bIsDynamicallyLoaded=false
      drawAs=Box
      imageType=NoImage
      mirroring=NoMirror
      resourceName=None
      resourceObject=&MI_UIKit_ScrollbarBackground
      tiling=NoTile
      imageSize=<12 12>
      margin=<ltrb 0.5 0.5 0.5 0.5>
      outlineSettings {
       bUseBrushTransparency=false
       roundingType=HalfHeightRadius
       width=0
       color {colorUseRule=UseColor_Specified, specifiedColor=<rgba 0 0 0 0>}
       cornerRadii=<0 0 0 0>
      }
      tintColor {colorUseRule=UseColor_Specified, specifiedColor=<rgba 1 1 1 1>}
      uVRegion {bIsValid=false, max=<0 0>, min=<0 0>}
     }
     verticalBottomSlotImage {
      bIsDynamicallyLoaded=false
      drawAs=NoDrawType
      imageType=FullColor
      mirroring=NoMirror
      resourceName=../../../Engine/Content/Slate/Common/Scrollbar_Background_Vertical.png
      resourceObject=null
      tiling=NoTile
      imageSize=<8 8>
      margin=<ltrb 0 0 0 0>
      outlineSettings {
       bUseBrushTransparency=false
       roundingType=HalfHeightRadius
       width=0
       color {colorUseRule=UseColor_Specified, specifiedColor=<rgba 0 0 0 0>}
       cornerRadii=<0 0 0 0>
      }
      tintColor {colorUseRule=UseColor_Specified, specifiedColor=<rgba 1 1 1 1>}
      uVRegion {bIsValid=false, max=<0 0>, min=<0 0>}
     }
     verticalTopSlotImage {
      bIsDynamicallyLoaded=false
      drawAs=NoDrawType
      imageType=FullColor
      mirroring=NoMirror
      resourceName=../../../Engine/Content/Slate/Common/Scrollbar_Background_Vertical.png
      resourceObject=null
      tiling=NoTile
      imageSize=<8 8>
      margin=<ltrb 0 0 0 0>
      outlineSettings {
       bUseBrushTransparency=false
       roundingType=HalfHeightRadius
       width=0
       color {colorUseRule=UseColor_Specified, specifiedColor=<rgba 0 0 0 0>}
       cornerRadii=<0 0 0 0>
      }
      tintColor {colorUseRule=UseColor_Specified, specifiedColor=<rgba 1 1 1 1>}
      uVRegion {bIsValid=false, max=<0 0>, min=<0 0>}
     }
    }
    widgetStyle {
     barThickness=12
     bottomShadowBrush {
      bIsDynamicallyLoaded=false
      drawAs=NoDrawType
      imageType=FullColor
      mirroring=NoMirror
      resourceName=../../../Engine/Content/Slate/Common/ScrollBoxShadowBottom.png
      resourceObject=null
      tiling=NoTile
      imageSize=<16 8>
      margin=<ltrb 0.5 0 0.5 1>
      outlineSettings {
       bUseBrushTransparency=false
       roundingType=HalfHeightRadius
       width=0
       color {colorUseRule=UseColor_Specified, specifiedColor=<rgba 0 0 0 0>}
       cornerRadii=<0 0 0 0>
      }
      tintColor {colorUseRule=UseColor_Specified, specifiedColor=<rgba 1 1 1 1>}
      uVRegion {bIsValid=false, max=<0 0>, min=<0 0>}
     }
     horizontalScrolledContentPadding=<ltrb 0 0 1 0>
     leftShadowBrush {
      bIsDynamicallyLoaded=false
      drawAs=NoDrawType
      imageType=FullColor
      mirroring=NoMirror
      resourceName=../../../Engine/Content/Slate/Common/ScrollBoxShadowLeft.png
      resourceObject=null
      tiling=NoTile
      imageSize=<8 16>
      margin=<ltrb 1 0.5 0 0.5>
      outlineSettings {
       bUseBrushTransparency=false
       roundingType=HalfHeightRadius
       width=0
       color {colorUseRule=UseColor_Specified, specifiedColor=<rgba 0 0 0 0>}
       cornerRadii=<0 0 0 0>
      }
      tintColor {colorUseRule=UseColor_Specified, specifiedColor=<rgba 1 1 1 1>}
      uVRegion {bIsValid=false, max=<0 0>, min=<0 0>}
     }
     rightShadowBrush {
      bIsDynamicallyLoaded=false
      drawAs=NoDrawType
      imageType=FullColor
      mirroring=NoMirror
      resourceName=../../../Engine/Content/Slate/Common/ScrollBoxShadowRight.png
      resourceObject=null
      tiling=NoTile
      imageSize=<8 16>
      margin=<ltrb 0 0.5 1 0.5>
      outlineSettings {
       bUseBrushTransparency=false
       roundingType=HalfHeightRadius
       width=0
       color {colorUseRule=UseColor_Specified, specifiedColor=<rgba 0 0 0 0>}
       cornerRadii=<0 0 0 0>
      }
      tintColor {colorUseRule=UseColor_Specified, specifiedColor=<rgba 1 1 1 1>}
      uVRegion {bIsValid=false, max=<0 0>, min=<0 0>}
     }
     topShadowBrush {
      bIsDynamicallyLoaded=false
      drawAs=NoDrawType
      imageType=FullColor
      mirroring=NoMirror
      resourceName=../../../Engine/Content/Slate/Common/ScrollBoxShadowTop.png
      resourceObject=null
      tiling=NoTile
      imageSize=<16 8>
      margin=<ltrb 0.5 1 0.5 0>
      outlineSettings {
       bUseBrushTransparency=false
       roundingType=HalfHeightRadius
       width=0
       color {colorUseRule=UseColor_Specified, specifiedColor=<rgba 0 0 0 0>}
       cornerRadii=<0 0 0 0>
      }
      tintColor {colorUseRule=UseColor_Specified, specifiedColor=<rgba 1 1 1 1>}
      uVRegion {bIsValid=false, max=<0 0>, min=<0 0>}
     }
     verticalScrolledContentPadding=<ltrb 0 0 0 1>
    }
   }
   slot {class=SizeBoxSlot}
  }
  DroneDirectMovement {
   widgetClassPath=/FortniteCoreUI/UI/Replay/RotatorSelectorReplaySettings
   bInherited=true
   bIsVariable=true
   parent=ScrollBox_Tabs
   widget=DroneDirectMovement
   properties {description="Toggles the ability to control the camera's movement directly, removing all drag and stabilization."}
   slot {class=ScrollBoxSlot, properties {padding=<ltrb 0 0 0 10>}}
  }
  DroneDirectMovementSaveState {
   widgetClassPath=&RotatorSelectorReplaySettings
   bIsVariable=true
   parent=ScrollBox_Tabs
   widget=DroneDirectMovementSaveState
   properties {description=Toggles the ability to save direct movement speed settings or not, title=Save speed state settings}
   slot {class=ScrollBoxSlot, properties {padding=<ltrb 0 0 0 10>}}
  }
  DroneLinearVelocity {
   widgetClassPath=/FortniteCoreUI/UI/Replay/OptionsMenuSliderReplay
   bIsVariable=true
   parent=ScrollBox_Tabs
   widget=DroneLinearVelocity
   properties {bUseCheckBox=true, description="Adjusts the speed of the camera's linear movement (Left Thumbstick and WASD).", formatting=Single Indent}
   slot {class=ScrollBoxSlot, properties {padding=<ltrb 0 0 0 10>}}
  }
  DroneRotationVelocity {
   widgetClassPath=&OptionsMenuSliderReplay
   bIsVariable=true
   parent=ScrollBox_Tabs
   widget=DroneRotationVelocity
   properties {
    bUseCheckBox=true
    description="Adjusts the speed of the camera's rotational movement (Right Thumbstick and Mouse)."
    formatting=Single Indent
   }
   slot {class=ScrollBoxSlot, properties {padding=<ltrb 0 0 0 10>}}
  }
  DroneLinearVelocityZmultiplier {
   widgetClassPath=&OptionsMenuSliderReplay
   bIsVariable=true
   parent=ScrollBox_Tabs
   widget=DroneLinearVelocityZmultiplier
   properties {
    bUseCheckBox=true
    description="Adjusts the speed of the camera's vertical movement (Gamepad Triggers and Q/E)."
    formatting=Single Indent
    valueSuffixText=x
   }
   slot {class=ScrollBoxSlot, properties {padding=<ltrb 0 0 0 10>}}
  }
  DroneGazeDirectedMovement {
   widgetClassPath=&RotatorSelectorReplaySettings
   bIsVariable=true
   parent=ScrollBox_Tabs
   widget=DroneGazeDirectedMovement
   properties {description=Toggles the ability to control the camera's movement directly to gaze direction., title=Gaze Directed Movement}
   slot {class=ScrollBoxSlot, properties {padding=<ltrb 0 0 0 10>}}
  }
  DroneXInputLock {
   widgetClassPath=&RotatorSelectorReplaySettings
   bIsVariable=true
   parent=ScrollBox_Tabs
   widget=DroneXInputLock
   properties {title=X Input Lock, options [Off,On]}
   slot {class=ScrollBoxSlot, properties {padding=<ltrb 0 0 0 10>}}
  }
  DroneYInputLock {
   widgetClassPath=&RotatorSelectorReplaySettings
   bIsVariable=true
   parent=ScrollBox_Tabs
   widget=DroneYInputLock
   properties {title=Y Input Lock, options [Off,On]}
   slot {class=ScrollBoxSlot, properties {padding=<ltrb 0 0 0 10>}}
  }
  DroneZAxisLock {
   widgetClassPath=&RotatorSelectorReplaySettings
   bIsVariable=true
   parent=ScrollBox_Tabs
   widget=DroneZAxisLock
   properties {title=Z Axis Lock}
   slot {class=ScrollBoxSlot, properties {padding=<ltrb 0 0 0 10>}}
  }
  DroneDragMultiplier {
   widgetClassPath=&OptionsMenuSliderReplay
   bIsVariable=true
   parent=ScrollBox_Tabs
   widget=DroneDragMultiplier
   properties {bUseCheckBox=true, description="Adjust the amount of drag when moving (not compatible with Direct Movement).", valueSuffixText=x}
   slot {class=ScrollBoxSlot, properties {padding=<ltrb 0 0 0 10>}}
  }
  DroneDutch {
   widgetClassPath=&OptionsMenuSliderReplay
   bIsVariable=true
   parent=ScrollBox_Tabs
   widget=DroneDutch
   properties {
    bUseCheckBox=true
    description="Tilts the camera left or right, also referred to as a Dutch Angle."
    maxFractionalDigits=0
    maxIntegralDigits=3
    minFractionalDigits=0
    valueSuffixText="°"
   }
   slot {class=ScrollBoxSlot, properties {padding=<ltrb 0 0 0 10>}}
  }
  DroneDutchAngleStepSlider {
   widgetClassPath=&OptionsMenuSliderReplay
   bIsVariable=true
   parent=ScrollBox_Tabs
   widget=DroneDutchAngleStepSlider
   properties {description=Alters how many degrees the Tilt setting changes per controller input., formatting=Single Indent, maxIntegralDigits=3}
   slot {class=ScrollBoxSlot, properties {padding=<ltrb 0 0 0 10>}}
  }
  Button_ResetDutchAngle {
   widgetClassPath=/FortniteCoreUI/UI/Replay/OptionsMenuActionReplay
   bIsVariable=true
   parent=ScrollBox_Tabs
   widget=Button_ResetDutchAngle
   properties {buttonText=Confirm, formatting=Single Indent, title=Reset Tilt}
   slot {class=ScrollBoxSlot, properties {padding=<ltrb 0 0 0 10>}}
  }
  DroneDutchSpeed {
   widgetClassPath=&OptionsMenuSliderReplay
   bIsVariable=true
   parent=ScrollBox_Tabs
   widget=DroneDutchSpeed
   properties {
    bUseCheckBox=true
    description="Speed at which the camera tilts while operating it (timeline must be playing to see this effect)."
    maxIntegralDigits=3
   }
   slot {class=ScrollBoxSlot, properties {padding=<ltrb 0 0 0 10>}}
  }
  DroneDutchSpeedStepSlider {
   widgetClassPath=&OptionsMenuSliderReplay
   bIsVariable=true
   parent=ScrollBox_Tabs
   widget=DroneDutchSpeedStepSlider
   properties {description=Alters how many degrees the Tilt Speed setting changes per controller input., formatting=Single Indent, maxIntegralDigits=3}
   slot {class=ScrollBoxSlot, properties {padding=<ltrb 0 0 0 10>}}
  }
  Button_ResetDutchSpeed {
   widgetClassPath=&OptionsMenuActionReplay
   bIsVariable=true
   parent=ScrollBox_Tabs
   widget=Button_ResetDutchSpeed
   properties {buttonText=Confirm, formatting=Single Indent, title=Reset Tilt Speed}
   slot {class=ScrollBoxSlot, properties {padding=<ltrb 0 0 0 10>}}
  }
  DutchControlScheme {
   widgetClassPath=&RotatorSelectorReplaySettings
   bIsVariable=true
   parent=ScrollBox_Tabs
   widget=DutchControlScheme
   properties {description=Switches Drone Speed controls to Tilt.}
   slot {class=ScrollBoxSlot, properties {padding=<ltrb 0 0 0 10>}}
  }
  DroneAttachToThrowable {
   widgetClassPath=&RotatorSelectorReplaySettings
   bIsVariable=true
   parent=ScrollBox_Tabs
   widget=DroneAttachToThrowable
   properties {description="Attach the camera to a throwable item, such as a grenade."}
   slot {class=ScrollBoxSlot, properties {padding=<ltrb 0 0 0 10>}}
  }
  DroneAttachToBone {
   widgetClassPath=&RotatorSelectorReplaySettings
   bIsVariable=true
   parent=ScrollBox_Tabs
   widget=DroneAttachToBone
   properties {description="Attach the camera to a specific bone on the selected character (must be in Drone Free)."}
   slot {class=ScrollBoxSlot, properties {padding=<ltrb 0 0 0 10>}}
  }
  DroneAttachToBoneName {
   widgetClassPath=&RotatorSelectorReplaySettings
   bIsVariable=true
   parent=ScrollBox_Tabs
   widget=DroneAttachToBoneName
   properties {description=The name of the bone the camera is currently attached to., formatting=Single Indent}
   slot {class=ScrollBoxSlot, properties {padding=<ltrb 0 0 0 10>}}
  }
  ShowCameraDummy {
   widgetClassPath=&RotatorSelectorReplaySettings
   bIsVariable=true
   parent=ScrollBox_Tabs
   widget=ShowCameraDummy
   properties {description=Toggles visibility of the camera actor.}
   slot {class=ScrollBoxSlot, properties {padding=<ltrb 0 0 0 10>}}
  }
  ShowCameraShots {
   widgetClassPath=&RotatorSelectorReplaySettings
   bIsVariable=true
   parent=ScrollBox_Tabs
   widget=ShowCameraShots
   properties {description="Enable to save up to 100 Shots in this Replay (10 Shots per Preset)! On the Timeline, pick a Shot, then choose to save either your camera, or the camera and the time. "}
   slot {class=ScrollBoxSlot, properties {padding=<ltrb 0 0 0 10>}}
  }
  Rotator_Preset {
   widgetClassPath=&RotatorSelectorReplaySettings
   bIsVariable=true
   parent=ScrollBox_Tabs
   widget=Rotator_Preset
   properties {
    description="Each Preset has 10 available Shots you can assign. To assign the camera or the camera and time to a Shot, use the buttons on the Timeline."
    formatting=Single Indent
    visibility=Collapsed
   }
   slot {class=ScrollBoxSlot, properties {padding=<ltrb 0 0 0 12>}}
  }
  Rotator_CurrentShot {
   widgetClassPath=&RotatorSelectorReplaySettings
   bIsVariable=true
   parent=ScrollBox_Tabs
   widget=Rotator_CurrentShot
   properties {
    description="Shows all Shots that have been saved. You may also quickly switch between them here!"
    formatting=Single Indent
    title=Shot
    visibility=Collapsed
    options [0,1,2,3,4,5,6,7,8,9]
   }
   slot {class=ScrollBoxSlot, properties {padding=<ltrb 15 0 15 15>}}
  }
  Button_UndoRedo {
   widgetClassPath=&OptionsMenuActionReplay
   bIsVariable=true
   parent=ScrollBox_Tabs
   widget=Button_UndoRedo
   properties {
    buttonText=Confirm
    description="Clear the camera data from the Shot selected above, or bring back a previously cleared Shot. "
    formatting=Single Indent
    title="Undo/Redo "
    visibility=Collapsed
   }
   slot {class=ScrollBoxSlot, properties {padding=<ltrb 0 0 0 10>}}
  }
  Button_OverrideShot {
   widgetClassPath=&OptionsMenuActionReplay
   bIsVariable=true
   parent=ScrollBox_Tabs
   widget=Button_OverrideShot
   properties {buttonText=Confirm, description=Override current shot, formatting=Single Indent, title=Override Shot, visibility=Collapsed}
   slot {class=ScrollBoxSlot, properties {padding=<ltrb 0 0 0 10>}}
  }
  Button_SaveSettings {
   widgetClassPath=&OptionsMenuActionReplay
   bIsVariable=true
   parent=ScrollBox_Tabs
   widget=Button_SaveSettings
   properties {buttonText=Confirm, description=Save spectator settings, formatting=Single Indent, title=Save Settings, visibility=Collapsed}
   slot {class=ScrollBoxSlot, properties {padding=<ltrb 0 0 0 10>}}
  }
  Button_LoadSettings {
   widgetClassPath=&OptionsMenuActionReplay
   bIsVariable=true
   parent=ScrollBox_Tabs
   widget=Button_LoadSettings
   properties {buttonText=Confirm, description=Load spectator settings, formatting=Single Indent, title=Load Settings, visibility=Collapsed}
   slot {class=ScrollBoxSlot, properties {padding=<ltrb 0 0 0 10>}}
  }
  DronePendulum {
   widgetClassPath=&RotatorSelectorReplaySettings
   bIsVariable=true
   parent=ScrollBox_Tabs
   widget=DronePendulum
   properties {description="Enable pendulum-like swings in a plane, perpendicular to camera view direction", title=Pendulum swing, options [Off,On]}
   slot {class=ScrollBoxSlot, properties {padding=<ltrb 0 0 0 10>}}
  }
  DronePendulumArmLength {
   widgetClassPath=&OptionsMenuSliderReplay
   bIsVariable=true
   parent=ScrollBox_Tabs
   widget=DronePendulumArmLength
   properties {
    bUseCheckBox=true
    description=Length of swing arm
    formatting=Single Indent
    maxIntegralDigits=3
    slider Value=0.1
    title=Arm Length
    valueSuffixText=m
    visibility=Collapsed
    minMaxValue=<0 10>
   }
   slot {class=ScrollBoxSlot, properties {padding=<ltrb 0 0 0 10>}}
  }
  DronePendulumAmplitude {
   widgetClassPath=&OptionsMenuSliderReplay
   bIsVariable=true
   parent=ScrollBox_Tabs
   widget=DronePendulumAmplitude
   properties {
    bUseCheckBox=true
    description=Max displacement angle
    formatting=Single Indent
    maxIntegralDigits=3
    slider Value=0.1111
    title=Amplitude
    valueSuffixText="°"
    visibility=Collapsed
    minMaxValue=<0 90>
   }
   slot {class=ScrollBoxSlot, properties {padding=<ltrb 0 0 0 10>}}
  }
  DronePendulumFrequency {
   widgetClassPath=&OptionsMenuSliderReplay
   bIsVariable=true
   parent=ScrollBox_Tabs
   widget=DronePendulumFrequency
   properties {
    bUseCheckBox=true
    description="Swing frequency (swings per second)"
    formatting=Single Indent
    maxIntegralDigits=3
    slider Value=0.5
    title=Frequency
    valueSuffixText=Hz
    visibility=Collapsed
   }
   slot {class=ScrollBoxSlot, properties {padding=<ltrb 0 0 0 10>}}
  }
  DronePendulumCameraRoll {
   widgetClassPath=&RotatorSelectorReplaySettings
   bIsVariable=true
   parent=ScrollBox_Tabs
   widget=DronePendulumCameraRoll
   properties {
    description=Update camera roll rotation to align with movement
    formatting=Single Indent
    title=Roll camera
    visibility=Collapsed
    options [False,True]
   }
   slot {class=ScrollBoxSlot, properties {padding=<ltrb 15 0 15 15>}}
  }
  CameraSetsPanel {
   widgetClassPath=&VerticalBox
   bIsVariable=true
   parent=ScrollBox_Tabs
   widget=CameraSetsPanel
   properties {visibility=Collapsed}
   slot {class=ScrollBoxSlot}
  }
  CameraSetHorizontalBox {
   widgetClassPath=&HorizontalBox
   bIsVariable=true
   parent=CameraSetsPanel
   widget=CameraSetHorizontalBox
   properties {visibility=Collapsed}
   slot {class=VerticalBoxSlot, properties {horizontalAlignment=HAlign_Center, padding=<ltrb 0 8 0 0>}}
  }
  SizeBox_1 {
   widgetClassPath=&SizeBox
   parent=CameraSetHorizontalBox
   widget=SizeBox_1
   properties {bOverride_WidthOverride=true, widthOverride=200}
   slot {class=HorizontalBoxSlot}
  }
  VerticalBox_0 {widgetClassPath=&VerticalBox, parent=SizeBox_1, widget=VerticalBox_0, slot {class=SizeBoxSlot}}
  Shot0_Button {
   widgetClassPath=&WBP_UIKit_Button_Regular
   bIsVariable=true
   parent=VerticalBox_0
   widget=Shot0_Button
   properties {
    buttonSize=Small
    navigation=&self:WidgetTree.Shot0_Button.WidgetNavigation_0
    text=Shot 0
   }
   slot {class=VerticalBoxSlot, properties {padding=<ltrb 0 0 0 15>}}
  }
  Shot1_Button {
   widgetClassPath=&WBP_UIKit_Button_Regular
   bIsVariable=true
   parent=VerticalBox_0
   widget=Shot1_Button
   properties {buttonSize=Small, text=Shot 1}
   slot {class=VerticalBoxSlot, properties {padding=<ltrb 0 0 0 15>}}
  }
  Shot2_Button {
   widgetClassPath=&WBP_UIKit_Button_Regular
   bIsVariable=true
   parent=VerticalBox_0
   widget=Shot2_Button
   properties {buttonSize=Small, text=Shot 2}
   slot {class=VerticalBoxSlot, properties {padding=<ltrb 0 0 0 15>}}
  }
  Shot3_Button {
   widgetClassPath=&WBP_UIKit_Button_Regular
   bIsVariable=true
   parent=VerticalBox_0
   widget=Shot3_Button
   properties {buttonSize=Small, text=Shot 3}
   slot {class=VerticalBoxSlot, properties {padding=<ltrb 0 0 0 15>}}
  }
  Shot4_Button {
   widgetClassPath=&WBP_UIKit_Button_Regular
   bIsVariable=true
   parent=VerticalBox_0
   widget=Shot4_Button
   properties {buttonSize=Small, text=Shot 4}
   slot {class=VerticalBoxSlot, properties {padding=<ltrb 0 0 0 12>}}
  }
  Shot5_Button {
   widgetClassPath=&WBP_UIKit_Button_Regular
   bIsVariable=true
   parent=VerticalBox_0
   widget=Shot5_Button
   properties {buttonSize=Small, text=Shot 5}
   slot {class=VerticalBoxSlot, properties {padding=<ltrb 0 0 0 15>}}
  }
  Shot6_Button {
   widgetClassPath=&WBP_UIKit_Button_Regular
   bIsVariable=true
   parent=VerticalBox_0
   widget=Shot6_Button
   properties {buttonSize=Small, text=Shot 6}
   slot {class=VerticalBoxSlot, properties {padding=<ltrb 0 0 0 15>}}
  }
  Shot7_Button {
   widgetClassPath=&WBP_UIKit_Button_Regular
   bIsVariable=true
   parent=VerticalBox_0
   widget=Shot7_Button
   properties {buttonSize=Small, text=Shot 7}
   slot {class=VerticalBoxSlot, properties {padding=<ltrb 0 0 0 15>}}
  }
  Shot8_Button {
   widgetClassPath=&WBP_UIKit_Button_Regular
   bIsVariable=true
   parent=VerticalBox_0
   widget=Shot8_Button
   properties {buttonSize=Small, text=Shot 8}
   slot {class=VerticalBoxSlot, properties {padding=<ltrb 0 0 0 15>}}
  }
  Shot9_Button {
   widgetClassPath=&WBP_UIKit_Button_Regular
   bIsVariable=true
   parent=VerticalBox_0
   widget=Shot9_Button
   properties {
    buttonSize=Small
    navigation=&self:WidgetTree.Shot9_Button.WidgetNavigation_0
    text=Shot 9
   }
   slot {class=VerticalBoxSlot, properties {padding=<ltrb 0 0 0 12>}}
  }
  Spacer_134 {
   widgetClassPath=&Spacer
   parent=CameraSetHorizontalBox
   widget=Spacer_134
   properties {size=<50 0>}
   slot {class=HorizontalBoxSlot}
  }
  SizeBox {
   widgetClassPath=&SizeBox
   parent=CameraSetHorizontalBox
   widget=SizeBox
   properties {bOverride_WidthOverride=true, widthOverride=150}
   slot {class=HorizontalBoxSlot}
  }
  VerticalBox {widgetClassPath=&VerticalBox, parent=SizeBox, widget=VerticalBox, slot {class=SizeBoxSlot}}
  Shot0_Button_1 {
   widgetClassPath=&WBP_UIKit_Button_Regular
   bIsVariable=true
   parent=VerticalBox
   widget=Shot0_Button_1
   properties {
    buttonSize=Small
    icon=&T-Icon-Replay-Rewind-L
    iconMaterial=/FortUILibrary/UIKit/Buttons/MaterialInstances/GenericButtons/MI_GenericIconDeepBlue
    navigation=&self:WidgetTree.Shot0_Button_1.WidgetNavigation_0
    text=""
   }
   slot {class=VerticalBoxSlot, properties {padding=<ltrb 0 0 0 15>}}
  }
  Shot1_Button_1 {
   widgetClassPath=&WBP_UIKit_Button_Regular
   bIsVariable=true
   parent=VerticalBox
   widget=Shot1_Button_1
   properties {
    buttonSize=Small
    icon=&T-Icon-Replay-Rewind-L
    iconMaterial=&MI_GenericIconDeepBlue
    text=""
   }
   slot {class=VerticalBoxSlot, properties {padding=<ltrb 0 0 0 15>}}
  }
  Shot2_Button_1 {
   widgetClassPath=&WBP_UIKit_Button_Regular
   bIsVariable=true
   parent=VerticalBox
   widget=Shot2_Button_1
   properties {
    buttonSize=Small
    icon=&T-Icon-Replay-Rewind-L
    iconMaterial=&MI_GenericIconDeepBlue
    text=""
   }
   slot {class=VerticalBoxSlot, properties {padding=<ltrb 0 0 0 15>}}
  }
  Shot3_Button_1 {
   widgetClassPath=&WBP_UIKit_Button_Regular
   bIsVariable=true
   parent=VerticalBox
   widget=Shot3_Button_1
   properties {
    buttonSize=Small
    icon=&T-Icon-Replay-Rewind-L
    iconMaterial=&MI_GenericIconDeepBlue
    text=""
   }
   slot {class=VerticalBoxSlot, properties {padding=<ltrb 0 0 0 15>}}
  }
  Shot4_Button_1 {
   widgetClassPath=&WBP_UIKit_Button_Regular
   bIsVariable=true
   parent=VerticalBox
   widget=Shot4_Button_1
   properties {
    buttonSize=Small
    icon=&T-Icon-Replay-Rewind-L
    iconMaterial=&MI_GenericIconDeepBlue
    text=""
   }
   slot {class=VerticalBoxSlot, properties {padding=<ltrb 0 0 0 12>}}
  }
  Shot5_Button_1 {
   widgetClassPath=&WBP_UIKit_Button_Regular
   bIsVariable=true
   parent=VerticalBox
   widget=Shot5_Button_1
   properties {
    buttonSize=Small
    icon=&T-Icon-Replay-Rewind-L
    iconMaterial=&MI_GenericIconDeepBlue
    text=""
   }
   slot {class=VerticalBoxSlot, properties {padding=<ltrb 0 0 0 15>}}
  }
  Shot6_Button_1 {
   widgetClassPath=&WBP_UIKit_Button_Regular
   bIsVariable=true
   parent=VerticalBox
   widget=Shot6_Button_1
   properties {
    buttonSize=Small
    icon=&T-Icon-Replay-Rewind-L
    iconMaterial=&MI_GenericIconDeepBlue
    text=""
   }
   slot {class=VerticalBoxSlot, properties {padding=<ltrb 0 0 0 15>}}
  }
  Shot7_Button_1 {
   widgetClassPath=&WBP_UIKit_Button_Regular
   bIsVariable=true
   parent=VerticalBox
   widget=Shot7_Button_1
   properties {
    buttonSize=Small
    icon=&T-Icon-Replay-Rewind-L
    iconMaterial=&MI_GenericIconDeepBlue
    text=""
   }
   slot {class=VerticalBoxSlot, properties {padding=<ltrb 0 0 0 15>}}
  }
  Shot8_Button_1 {
   widgetClassPath=&WBP_UIKit_Button_Regular
   bIsVariable=true
   parent=VerticalBox
   widget=Shot8_Button_1
   properties {
    buttonSize=Small
    icon=&T-Icon-Replay-Rewind-L
    iconMaterial=&MI_GenericIconDeepBlue
    text=""
   }
   slot {class=VerticalBoxSlot, properties {padding=<ltrb 0 0 0 15>}}
  }
  Shot9_Button_1 {
   widgetClassPath=&WBP_UIKit_Button_Regular
   bIsVariable=true
   parent=VerticalBox
   widget=Shot9_Button_1
   properties {
    buttonSize=Small
    icon=&T-Icon-Replay-Rewind-L
    iconMaterial=&MI_GenericIconDeepBlue
    navigation=&self:WidgetTree.Shot9_Button_1.WidgetNavigation_1
    text=""
   }
   slot {class=VerticalBoxSlot, properties {padding=<ltrb 0 0 0 12>}}
  }
 ]
}
