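# Broadcast/replay spectator tab for ambient settings: time of day, pause, sun light angle and cloud shadow time.
# Controls call the native parent's setters, except Cloud Time, which runs a `ke*` console command and is hidden in shipping builds.
# Looping timers poll time and sun yaw to resync sliders, while save/load events persist values by key.
# The last focused widget drives resets and gamepad refocus, with explicit wrap-around navigation.
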WidgetBlueprint /FortniteCoreUI/UI/Broadcast/FortBroadcastAmbientSettingsTab
graphs [
 CloseTabGroup {nodes [
   CloseTabGroup() {is=title K2Node_FunctionEntry}
   > SetTabCollapsed!(
    execute
    self:Fort Replay Context Object Reference=Get FortReplayContext!() -> (ReturnValue:Fort Replay Context Object Reference) {is=title K2Node_GetBlueprintContextEx}
    bIsCollapsed:bool=true
   ) -> (then)
  ]}
 EventGraph {nodes [
   n1 = "Slider Changed (TimeOfDaySlider)"!() -> (OutputDelegate, then, Slider Value:double) {is=title K2Node_ComponentBoundEvent}
   > SetTimeOfDay!(
    execute
    self:Fort Spectator Ambient Settings Tab Object Reference
    TimeOfDay:float=float * float(A=@n1.Slider Value, B=24) {is=title K2Node_PromotableOperator}
   ) -> (then)
   > CloseTabGroup() {is=title K2Node_CallFunction}
   n2 = "Selection Changed (TimeOfDayPauseToggle)"!() -> (OutputDelegate, then, Selected Index:int=0) {is=title K2Node_ComponentBoundEvent}
   > SetTimeOfDayPaused!(execute, self:Fort Spectator Ambient Settings Tab Object Reference, bPaused:bool="integer > integer"(A=@n2.Selected Index)) -> (then)
   > CloseTabGroup()
   n3 = "On Slider End (CloudTimeRotation)"!() -> (OutputDelegate, then, SliderValue:double) {is=title K2Node_ComponentBoundEvent}
   n4 = float * float(A=@n3.SliderValue, B=24)
   > ExecuteConsoleCommand!(
    execute
    Command=Append(A=MakeLiteralString(Value="ke* SetCloudsCustomTime "), B="To String (Float)"(InDouble=@n4))
    SpecificPlayer=GetOwningPlayer()
   ) -> (then)
   n5 = "On Value Commited (CloudTimeRotation)"!() -> (OutputDelegate, then, Value:double) {is=title K2Node_ComponentBoundEvent}
   > ExecuteConsoleCommand(
    Command=Append(A=MakeLiteralString(Value="ke* SetCloudsCustomTime "), B="To String (Float)"(InDouble=@n5.Value))
    SpecificPlayer=GetOwningPlayer()
   )
   n6 = "Slider Changed (TimeOfDayLightRotation)"!() -> (OutputDelegate, then, Slider Value:double) {is=title K2Node_ComponentBoundEvent}
   > SetOverridenSunYawAngle!(execute, self:Fort Spectator Ambient Settings Tab Object Reference, InYaw:float=float * float(A=@n6.Slider Value, B=360)) -> (then)
   > CloseTabGroup()
   "On Focused (TimeOfDaySlider)"() {is=title K2Node_ComponentBoundEvent}
   > Set LatestHoveredWidget(LatestHoveredWidget:User Widget Object Reference=Cast To OptionsMenuSliderReplay!(Object=Get TimeOfDaySlider() -> (TimeOfDaySlider:Widget Object Reference)) -> (AsOptions Menu Slider Replay:Options Menu Slider Replay Object Reference, bSuccess:bool=~) {is=title K2Node_DynamicCast}) -> (Output_Get:User Widget Object Reference)
   "On Hovered (TimeOfDayPauseToggle)"() {is=title K2Node_ComponentBoundEvent}
   > Set LatestHoveredWidget(LatestHoveredWidget=Get TimeOfDayPauseToggle() -> (TimeOfDayPauseToggle:Rotator Selector Replay Settings Object Reference))
   "On Focused (TimeOfDayLightRotation)"() {is=title K2Node_ComponentBoundEvent}
   > Set LatestHoveredWidget(LatestHoveredWidget=Get TimeOfDayLightRotation() -> (TimeOfDayLightRotation:Options Menu Slider Replay Object Reference))
   n7 = "On Value Commited (TimeOfDayLightRotation)"!() -> (OutputDelegate, then, Value:double) {is=title K2Node_ComponentBoundEvent}
   > SetOverridenSunYawAngle(InYaw=@n7.Value)
   > Update Slider Value!(
    execute
    self:Options Menu Slider Replay Object Reference=Get TimeOfDayLightRotation()
    NormalizedValue:double=float / float(A=@n7.Value, B=360)
   ) -> (then)
   n8 = "On Value Commited (TimeOfDaySlider)"!() -> (OutputDelegate, then, Value:double) {is=title K2Node_ComponentBoundEvent}
   > SetTimeOfDay(TimeOfDay=@n8.Value)
   > n9 = Cast To OptionsMenuSliderReplay!(execute:Exec, Object=Get TimeOfDaySlider()) -> (then:Exec, CastFailed:Exec, AsOptions Menu Slider Replay)
   > Update Slider Value(self=@n9, NormalizedValue=float / float(A=@n8.Value, B=24))
   "On Focused (CloudTimeRotation)"() {is=title K2Node_ComponentBoundEvent}
   > Set LatestHoveredWidget(LatestHoveredWidget=Get CloudTimeRotation() -> (CloudTimeRotation:Options Menu Slider Replay Object Reference))
   n10 = KeybindsChanged() {is=title K2Node_CustomEvent}
   n11 = Get CommonInputSubsystem!() -> (ReturnValue:Common Input Object Reference) {is=title K2Node_GetBlueprintContextEx}
   > Is Valid(InputObject=@n11)
   > Branch(Condition="Equal (Enum)"(A=GetCurrentInputType(self=@n11), B=Gamepad))
   n12 = Get LatestHoveredWidget() -> (LatestHoveredWidget:User Widget Object Reference)
   > n13 = Is Valid(InputObject=@n12)
   n14 = Cast To OptionsMenuSliderReplay(execute=@n13.Is Not Valid, Object=Get TimeOfDaySlider())
   > SetFocus(self=@n14)
   SetFocus(execute=@n13.Is Valid, self=@n12)
   Event PreConstruct()
   > InitializeWidgets()
   > n15 = Cast To OptionsMenuSliderReplay(Object=Get TimeOfDaySlider())
   > Set LatestHoveredWidget(LatestHoveredWidget=@n15)
   Event Construct()
   > Set Timer by Function Name!(
    execute
    Object:Object Reference
    FunctionName:String=UpdateTimeOfDay
    Time:float=0.5
    bLooping:bool=true
    bMaxOncePerFrame:bool=false
    InitialStartDelay:float=0
    InitialStartDelayVariance:float=0
   ) -> (then, ReturnValue:Timer Handle Structure) {is=title K2Node_CallFunction}
   > Set Timer by Function Name(FunctionName=UpdateLightAngle, Time=0.5, bLooping=true, bMaxOncePerFrame=false, InitialStartDelay=0, InitialStartDelayVariance=0)
   > SetNavigationRules()
   Event On Activated()
   > Parent: On Activated() {is=title K2Node_CallParentFunction}
   n16 = GetRuntimeOptions()
   > SetVisibility(
    self=Get CloudTimeRotation()
    InVisibility=Select(Option 0=Visible, Option 1=Collapsed, Index=IsShippingBuild!(self:Fort Runtime Options Object Reference=@n16) -> (ReturnValue:bool))
   )
   > n17 = Branch(Condition=Get bIsFirstActivation() -> (bIsFirstActivation:bool))
   > n18 = Cast To OptionsMenuSliderReplay(Object=Get TimeOfDaySlider())
   > SetFocus(self=@n18)
   > n19 = Set bIsFirstActivation(bIsFirstActivation:bool=false) -> (Output_Get:bool)
   SetNavigationRules(execute=[@n19.then, @n17.else])
   > Bind Event to On Keybinds Changed(
    self:Fort Global UIContext Object Reference=Get FortGlobalUIContext!() -> (ReturnValue:Fort Global UIContext Object Reference) {is=title K2Node_GetBlueprintContextEx}
    Delegate=@n10
   ) {is=title K2Node_AddDelegate}
   n20 = Event OnTimeOfDayPauseChanged!() -> (OutputDelegate, then, bPaused:bool) {is=title K2Node_Event}
   > Update Row Selector!(
    execute
    self:Rotator Selector Replay Settings Object Reference=Get TimeOfDayPauseToggle()
    Currently Selected:int=SelectInt(A=1, B=0, bPickA=@n20.bPaused)
   ) -> (then)
   Event Destruct()
   > Clear Timer by Function Name!(execute, Object:Object Reference, FunctionName:String=UpdateTimeOfDay) -> (then)
   > Clear Timer by Function Name(FunctionName=UpdateLightAngle)
   n21 = Event OnSaveTabSettings!() -> (OutputDelegate, then, SaveData:Fort Replay Tab Settings Handle Object Reference) {is=title K2Node_Event}
   > SetValue!(
    execute
    self:Fort Replay Tab Settings Handle Object Reference=@n21.SaveData
    Key:Name=Get TimeOfDayKey() -> (TimeOfDayKey:Name=None)
    Value:double=Get LastTimeOfDay() -> (LastTimeOfDay:double)
   ) -> (then)
   n22 = Get TimeOfDayPauseToggle()
   > SetValue(
    self=@n21.SaveData
    Key=Get TODPausedKey() -> (TODPausedKey:Name=None)
    Value:bool="integer > integer"(A=Get SelectedIndex!(self:Rotator Selector Replay Settings Object Reference=@n22) -> (SelectedIndex:int=0))
   )
   > SetValue(self=@n21.SaveData, Key=Get LightAngleKey() -> (LightAngleKey:Name=None), Value:double=Get LastLightAngle() -> (LastLightAngle:double))
   > SetValue(
    self=@n21.SaveData
    Key=Get CloudTimeKey() -> (CloudTimeKey:Name=None)
    Value=GetMappedValue!(self:Options Menu Slider Replay Object Reference=Get CloudTimeRotation()) -> (MappedValue:double)
   )
   n23 = Event OnLoadTabSettings!() -> (OutputDelegate, then, SaveData:Fort Replay Tab Settings Handle Object Reference) {is=title K2Node_Event}
   > n24 = Sequence!(execute) -> (then_0, then_1, then_2:Exec, then_3:Exec)
   n25 = GetValue!(self:Fort Replay Tab Settings Handle Object Reference=@n23.SaveData, Key:Name=Get TimeOfDayKey() -> (TimeOfDayKey=None)) -> (OutValue:double=~, ReturnValue:bool)
   > Branch(Condition=@n25.ReturnValue)
   > n26 = Cast To OptionsMenuSliderReplay(Object=Get TimeOfDaySlider())
   > UpdateSliderValueUnmapped!(execute, self:Options Menu Slider Replay Object Reference=@n26, UnmappedValue:double=@n25) -> (then)
   > Set LastTimeOfDay(LastTimeOfDay:double=@n25) -> (Output_Get:double)
   n27 = GetValue(self=@n23.SaveData, Key=Get TODPausedKey() -> (TODPausedKey=None)) -> (OutValue:bool=~)
   Branch(execute=@n24.then_1, Condition=@n27.ReturnValue)
   > Update Row Selector(self=Get TimeOfDayPauseToggle(), Currently Selected=SelectInt(A=0, B=1, bPickA=@n27))
   n28 = GetValue(self=@n23.SaveData, Key=Get LightAngleKey() -> (LightAngleKey=None)) -> (OutValue:double=~)
   Branch(execute=@n24.then_2, Condition=@n28.ReturnValue)
   > UpdateSliderValueUnmapped(self=Get TimeOfDayLightRotation(), UnmappedValue=@n28)
   > Set LastLightAngle(LastLightAngle:double=@n28) -> (Output_Get:double)
   n29 = GetValue(self=@n23.SaveData, Key=Get CloudTimeKey() -> (CloudTimeKey=None)) -> (OutValue=~)
   Branch(execute=@n24.then_3, Condition=@n29.ReturnValue)
   > UpdateSliderValueUnmapped(self=Get CloudTimeRotation(), UnmappedValue=@n29)
   Event ResetFocusedElement() {is=title K2Node_Event}
   > Is Valid(InputObject=Get LatestHoveredWidget())
   > n30 = Branch(Condition="Equal (Object)"(A=Get TimeOfDaySlider(), B=Get LatestHoveredWidget()))
   > ResetTimeOfDay(self:Fort Spectator Ambient Settings Tab Object Reference)
   Branch(execute=@n30.else, Condition="Equal (Object)"(A=Get TimeOfDayPauseToggle(), B=Get LatestHoveredWidget()))
   > SetTimeOfDayPaused(bPaused=false)
   Get CloudTimeKey() -> (CloudTimeKey=None)
  ]}
 InitializeWidgets {nodes [
   InitializeWidgets() {is=title K2Node_FunctionEntry}
   > n1 = Sequence!(execute) -> (then_0, then_1)
   n2 = Cast To OptionsMenuSliderReplay!(Object=Get TimeOfDaySlider()) -> (AsOptions Menu Slider Replay, bSuccess:bool=~)
   > Update Slider!(
    execute
    self:Options Menu Slider Replay Object Reference=@n2
    Slider Text:Text="LOCTEXT(\"Time of Day\")"
    Slider Value:double=float / float(A=GetCurrentTimeOfDay!(self:Fort Spectator Ambient Settings Tab Object Reference) -> (ReturnValue:float), B=24)
    Hover Text:Text
    Tooltip Text Block:Common Text Object Reference
    Min Max Value:Vector 2D Structure=(X=0,Y=24)
   ) -> (then)
   > UpdateSliderStep!(execute, self:Options Menu Slider Replay Object Reference=@n2, Step:double=0.25) -> (then)
   > Initialize!(
    execute
    self:Rotator Selector Replay Settings Object Reference=Get TimeOfDayPauseToggle()
    Display Text:Text="LOCTEXT(\"Pause Time of Day\")"
    Options:Array of Texts=Make Array!("[0]":Text="LOCTEXT(\"Off\")", "[1]":Text="LOCTEXT(\"On\")") -> (Array:Array of Texts)
    Hover Text:Text
    Tab Tooltip:Common Text Object Reference
   ) -> (then)
   Update Slider(
    execute=@n1.then_1
    self=Get TimeOfDayLightRotation()
    Slider Text="LOCTEXT(\"Light Angle\")"
    Slider Value=float / float(A=GetOverridenSunYawAngle!(self:Fort Spectator Ambient Settings Tab Object Reference) -> (ReturnValue:float), B=360)
    Min Max Value=(X=0,Y=360)
   )
   > UpdateSliderStep(self=Get TimeOfDayLightRotation(), Step=1)
  ]}
 SetNavigationRules {nodes [
   SetNavigationRules() {is=title K2Node_FunctionEntry}
   n1 = GetAllChildren(self=Get ScrollBox_Tabs() -> (ScrollBox_Tabs:Scroll Box Object Reference))
   n2 = "Get (a copy)"(Array:Array of Widget Object References=@n1, Dimension 1=int - int(A=Length(TargetArray:Array of Widget Object References=@n1))) -> (Output:Widget Object Reference)
   n3 = "Get (a copy)"(Array=GetAllChildren(self=Get ScrollBox_Tabs()))
   > SetNavigationRuleExplicit!(execute, self:Widget Object Reference=@n3, Direction:EUINavigation Enum=Up, InWidget:Widget Object Reference=@n2) -> (then)
   > SetNavigationRuleExplicit(self=@n2, Direction=Down, InWidget=@n3)
  ]}
 UpdateLightAngle {nodes [
   UpdateLightAngle() {is=title K2Node_FunctionEntry}
   n1 = GetOverridenSunYawAngle()
   > Branch(Condition="Not Equal (Float)"(A=Get LastLightAngle(), B=@n1))
   > n2 = Set LastLightAngle(LastLightAngle=@n1)
   > Update Slider Value(self=Get TimeOfDayLightRotation(), NormalizedValue=float / float(A=@n2, B=360))
  ]}
 UpdateTimeOfDay {nodes [
   UpdateTimeOfDay() {is=title K2Node_FunctionEntry}
   n1 = GetCurrentTimeOfDay()
   > Branch(Condition="Not Equal (Float)"(A=Get LastTimeOfDay(), B=@n1))
   n2 = "float > float"(A=@n1, B=-1)
   > SetIsEnabled(self=Get TimeOfDaySlider(), bInIsEnabled=@n2)
   > SetIsEnabled(
    self=Get TimeOfDayPauseToggle()
    bInIsEnabled=AND Boolean(A=IsGeneralTimeOfDayManagerActive!(self:Fort Spectator Ambient Settings Tab Object Reference) -> (ReturnValue:bool), B=@n2)
   )
   > n3 = Set LastTimeOfDay(LastTimeOfDay=@n1)
   > Update Slider Value(self=Cast To OptionsMenuSliderReplay(Object=Get TimeOfDaySlider()) -> (bSuccess=~), NormalizedValue=float / float(A=@n3, B=24))
  ]}
]
variables [
 bIsFirstActivation {type=boolean, default=true}
 CloudTimeKey {type=string, default=cloudTime}
 LastLightAngle {type=number}
 LastTimeOfDay {type=number}
 LatestHoveredWidget {type=UserWidget, default=null}
 LightAngleKey {type=string, default=lightAngle}
 TimeOfDayKey {type=string, default=timeOfDay}
 TODPausedKey {type=string, default=todPaused}
]
widgets {
 info {inheritedWidgetCount=1, parentClass=/Script/FortniteUI.FortSpectatorAmbientSettingsTab, rootWidgetClass=&Overlay, widgetCount=8}
 widgets [
  Overlay_0 {widgetClassPath=&Overlay, widget=Overlay_0}
  SizeBox_0 {
   widgetClassPath=&SizeBox
   parent=Overlay_0
   widget=SizeBox_0
   properties {bOverride_HeightOverride=true, bOverride_WidthOverride=true, heightOverride=470, minDesiredWidth=650, widthOverride=700}
   slot {class=OverlaySlot}
  }
  CommonBorder_22 {
   widgetClassPath=&CommonBorder
   parent=SizeBox_0
   widget=CommonBorder_22
   properties {padding=<ltrb 0 0 0 0>}
   slot {class=SizeBoxSlot}
  }
  ScrollBox_Tabs {
   widgetClassPath=&ScrollBox
   bIsVariable=true
   parent=CommonBorder_22
   widget=ScrollBox_Tabs
   properties {
    allowOverscroll=false
    bAnimateWheelScrolling=true
    consumeMouseWheel=Always
    navigationDestination=Center
    scrollWhenFocusChanges=AnimatedScroll
    scrollbarPadding=<ltrb 0 10 10 10>
    scrollbarThickness=<15 15>
    widgetBarStyle {
     thickness=12
     draggedThumbImage {
      bIsDynamicallyLoaded=false
      drawAs=RoundedBox
      imageType=FullColor
      mirroring=NoMirror
      resourceName=../../../Engine/Content/Slate/Common/Scrollbar_Thumb.png
      resourceObject=null
      tiling=NoTile
      imageSize=<12 12>
      margin=<ltrb 0.25 0.25 0.25 0.25>
      outlineSettings {
       bUseBrushTransparency=false
       roundingType=FixedRadius
       width=2
       color {colorUseRule=UseColor_Specified, specifiedColor=<rgba 0 0 0 0>}
       cornerRadii=<6 6 6 6>
      }
      tintColor {colorUseRule=UseColor_Specified, specifiedColor=<rgba 1 1 1 1>}
      uVRegion {bIsValid=false, max=<0 0>, min=<0 0>}
     }
     horizontalBackgroundImage {
      bIsDynamicallyLoaded=false
      drawAs=Box
      imageType=FullColor
      mirroring=NoMirror
      resourceName=None
      resourceObject=/FortUILibrary/UIKit/ScrollBar/MaterialInstances/MI_UIKit_ScrollbarBackground
      tiling=NoTile
      imageSize=<12 12>
      margin=<ltrb 0.5 0.5 0.5 0.5>
      outlineSettings {
       bUseBrushTransparency=false
       roundingType=HalfHeightRadius
       width=0
       color {colorUseRule=UseColor_Specified, specifiedColor=<rgba 0 0 0 0>}
       cornerRadii=<0 0 0 0>
      }
      tintColor {colorUseRule=UseColor_Specified, specifiedColor=<rgba 1 1 1 1>}
      uVRegion {bIsValid=false, max=<0 0>, min=<0 0>}
     }
     horizontalBottomSlotImage {
      bIsDynamicallyLoaded=false
      drawAs=NoDrawType
      imageType=FullColor
      mirroring=NoMirror
      resourceName=../../../Engine/Content/Slate/Common/Scrollbar_Background_Horizontal.png
      resourceObject=null
      tiling=NoTile
      imageSize=<8 8>
      margin=<ltrb 0 0 0 0>
      outlineSettings {
       bUseBrushTransparency=false
       roundingType=HalfHeightRadius
       width=0
       color {colorUseRule=UseColor_Specified, specifiedColor=<rgba 0 0 0 0>}
       cornerRadii=<0 0 0 0>
      }
      tintColor {colorUseRule=UseColor_Specified, specifiedColor=<rgba 1 1 1 1>}
      uVRegion {bIsValid=false, max=<0 0>, min=<0 0>}
     }
     horizontalTopSlotImage {
      bIsDynamicallyLoaded=false
      drawAs=NoDrawType
      imageType=FullColor
      mirroring=NoMirror
      resourceName=../../../Engine/Content/Slate/Common/Scrollbar_Background_Horizontal.png
      resourceObject=null
      tiling=NoTile
      imageSize=<8 8>
      margin=<ltrb 0 0 0 0>
      outlineSettings {
       bUseBrushTransparency=false
       roundingType=HalfHeightRadius
       width=0
       color {colorUseRule=UseColor_Specified, specifiedColor=<rgba 0 0 0 0>}
       cornerRadii=<0 0 0 0>
      }
      tintColor {colorUseRule=UseColor_Specified, specifiedColor=<rgba 1 1 1 1>}
      uVRegion {bIsValid=false, max=<0 0>, min=<0 0>}
     }
     hoveredThumbImage {
      bIsDynamicallyLoaded=false
      drawAs=RoundedBox
      imageType=FullColor
      mirroring=NoMirror
      resourceName=../../../Engine/Content/Slate/Common/Scrollbar_Thumb.png
      resourceObject=null
      tiling=NoTile
      imageSize=<12 12>
      margin=<ltrb 0.25 0.25 0.25 0.25>
      outlineSettings {
       bUseBrushTransparency=false
       roundingType=FixedRadius
       width=2
       color {colorUseRule=UseColor_Specified, specifiedColor=<rgba 0 0 0 0>}
       cornerRadii=<6 6 6 6>
      }
      tintColor {colorUseRule=UseColor_Specified, specifiedColor=<rgba 1 1 1 1>}
      uVRegion {bIsValid=false, max=<0 0>, min=<0 0>}
     }
     normalThumbImage {
      bIsDynamicallyLoaded=false
      drawAs=RoundedBox
      imageType=FullColor
      mirroring=NoMirror
      resourceName=../../../Engine/Content/Slate/Common/Scrollbar_Thumb.png
      resourceObject=null
      tiling=NoTile
      imageSize=<12 12>
      margin=<ltrb 0.25 0.25 0.25 0.25>
      outlineSettings {
       bUseBrushTransparency=false
       roundingType=FixedRadius
       width=2
       color {colorUseRule=UseColor_Specified, specifiedColor=<rgba 0 0 0 0>}
       cornerRadii=<6 6 6 6>
      }
      tintColor {colorUseRule=UseColor_Specified, specifiedColor=<rgba 1 1 1 0.75>}
      uVRegion {bIsValid=false, max=<0 0>, min=<0 0>}
     }
     verticalBackgroundImage {
      bIsDynamicallyLoaded=false
      drawAs=Box
      imageType=NoImage
      mirroring=NoMirror
      resourceName=None
      resourceObject=&MI_UIKit_ScrollbarBackground
      tiling=NoTile
      imageSize=<12 12>
      margin=<ltrb 0.5 0.5 0.5 0.5>
      outlineSettings {
       bUseBrushTransparency=false
       roundingType=HalfHeightRadius
       width=0
       color {colorUseRule=UseColor_Specified, specifiedColor=<rgba 0 0 0 0>}
       cornerRadii=<0 0 0 0>
      }
      tintColor {colorUseRule=UseColor_Specified, specifiedColor=<rgba 1 1 1 1>}
      uVRegion {bIsValid=false, max=<0 0>, min=<0 0>}
     }
     verticalBottomSlotImage {
      bIsDynamicallyLoaded=false
      drawAs=NoDrawType
      imageType=FullColor
      mirroring=NoMirror
      resourceName=../../../Engine/Content/Slate/Common/Scrollbar_Background_Vertical.png
      resourceObject=null
      tiling=NoTile
      imageSize=<8 8>
      margin=<ltrb 0 0 0 0>
      outlineSettings {
       bUseBrushTransparency=false
       roundingType=HalfHeightRadius
       width=0
       color {colorUseRule=UseColor_Specified, specifiedColor=<rgba 0 0 0 0>}
       cornerRadii=<0 0 0 0>
      }
      tintColor {colorUseRule=UseColor_Specified, specifiedColor=<rgba 1 1 1 1>}
      uVRegion {bIsValid=false, max=<0 0>, min=<0 0>}
     }
     verticalTopSlotImage {
      bIsDynamicallyLoaded=false
      drawAs=NoDrawType
      imageType=FullColor
      mirroring=NoMirror
      resourceName=../../../Engine/Content/Slate/Common/Scrollbar_Background_Vertical.png
      resourceObject=null
      tiling=NoTile
      imageSize=<8 8>
      margin=<ltrb 0 0 0 0>
      outlineSettings {
       bUseBrushTransparency=false
       roundingType=HalfHeightRadius
       width=0
       color {colorUseRule=UseColor_Specified, specifiedColor=<rgba 0 0 0 0>}
       cornerRadii=<0 0 0 0>
      }
      tintColor {colorUseRule=UseColor_Specified, specifiedColor=<rgba 1 1 1 1>}
      uVRegion {bIsValid=false, max=<0 0>, min=<0 0>}
     }
    }
    widgetStyle {
     barThickness=12
     bottomShadowBrush {
      bIsDynamicallyLoaded=false
      drawAs=NoDrawType
      imageType=FullColor
      mirroring=NoMirror
      resourceName=../../../Engine/Content/Slate/Common/ScrollBoxShadowBottom.png
      resourceObject=null
      tiling=NoTile
      imageSize=<16 8>
      margin=<ltrb 0.5 0 0.5 1>
      outlineSettings {
       bUseBrushTransparency=false
       roundingType=HalfHeightRadius
       width=0
       color {colorUseRule=UseColor_Specified, specifiedColor=<rgba 0 0 0 0>}
       cornerRadii=<0 0 0 0>
      }
      tintColor {colorUseRule=UseColor_Specified, specifiedColor=<rgba 1 1 1 1>}
      uVRegion {bIsValid=false, max=<0 0>, min=<0 0>}
     }
     horizontalScrolledContentPadding=<ltrb 0 0 1 0>
     leftShadowBrush {
      bIsDynamicallyLoaded=false
      drawAs=NoDrawType
      imageType=FullColor
      mirroring=NoMirror
      resourceName=../../../Engine/Content/Slate/Common/ScrollBoxShadowLeft.png
      resourceObject=null
      tiling=NoTile
      imageSize=<8 16>
      margin=<ltrb 1 0.5 0 0.5>
      outlineSettings {
       bUseBrushTransparency=false
       roundingType=HalfHeightRadius
       width=0
       color {colorUseRule=UseColor_Specified, specifiedColor=<rgba 0 0 0 0>}
       cornerRadii=<0 0 0 0>
      }
      tintColor {colorUseRule=UseColor_Specified, specifiedColor=<rgba 1 1 1 1>}
      uVRegion {bIsValid=false, max=<0 0>, min=<0 0>}
     }
     rightShadowBrush {
      bIsDynamicallyLoaded=false
      drawAs=NoDrawType
      imageType=FullColor
      mirroring=NoMirror
      resourceName=../../../Engine/Content/Slate/Common/ScrollBoxShadowRight.png
      resourceObject=null
      tiling=NoTile
      imageSize=<8 16>
      margin=<ltrb 0 0.5 1 0.5>
      outlineSettings {
       bUseBrushTransparency=false
       roundingType=HalfHeightRadius
       width=0
       color {colorUseRule=UseColor_Specified, specifiedColor=<rgba 0 0 0 0>}
       cornerRadii=<0 0 0 0>
      }
      tintColor {colorUseRule=UseColor_Specified, specifiedColor=<rgba 1 1 1 1>}
      uVRegion {bIsValid=false, max=<0 0>, min=<0 0>}
     }
     topShadowBrush {
      bIsDynamicallyLoaded=false
      drawAs=NoDrawType
      imageType=FullColor
      mirroring=NoMirror
      resourceName=../../../Engine/Content/Slate/Common/ScrollBoxShadowTop.png
      resourceObject=null
      tiling=NoTile
      imageSize=<16 8>
      margin=<ltrb 0.5 1 0.5 0>
      outlineSettings {
       bUseBrushTransparency=false
       roundingType=HalfHeightRadius
       width=0
       color {colorUseRule=UseColor_Specified, specifiedColor=<rgba 0 0 0 0>}
       cornerRadii=<0 0 0 0>
      }
      tintColor {colorUseRule=UseColor_Specified, specifiedColor=<rgba 1 1 1 1>}
      uVRegion {bIsValid=false, max=<0 0>, min=<0 0>}
     }
     verticalScrolledContentPadding=<ltrb 0 0 0 1>
    }
   }
   slot {class=BorderSlot, properties {padding=<ltrb 0 0 0 0>}}
  }
  TimeOfDaySlider {
   widgetClassPath=/FortniteCoreUI/UI/Replay/OptionsMenuSliderReplay
   bInherited=true
   bIsVariable=true
   parent=ScrollBox_Tabs
   widget=TimeOfDaySlider
   properties {bUseCheckBox=true, maxIntegralDigits=2}
   slot {class=ScrollBoxSlot, properties {padding=<ltrb 15 15 15 12>}}
  }
  TimeOfDayPauseToggle {
   widgetClassPath=/FortniteCoreUI/UI/Replay/RotatorSelectorReplaySettings
   bIsVariable=true
   parent=ScrollBox_Tabs
   widget=TimeOfDayPauseToggle
   slot {class=ScrollBoxSlot, properties {padding=<ltrb 15 0 15 12>}}
  }
  TimeOfDayLightRotation {
   widgetClassPath=&OptionsMenuSliderReplay
   bIsVariable=true
   parent=ScrollBox_Tabs
   widget=TimeOfDayLightRotation
   properties {
    bUseCheckBox=true
    description=Adjusts the Directional Light on it's current axis.
    maxFractionalDigits=0
    maxIntegralDigits=3
    minFractionalDigits=0
    valueSuffixText="°"
   }
   slot {class=ScrollBoxSlot, properties {padding=<ltrb 15 0 15 15>}}
  }
  CloudTimeRotation {
   widgetClassPath=&OptionsMenuSliderReplay
   bIsVariable=true
   parent=ScrollBox_Tabs
   widget=CloudTimeRotation
   properties {
    bGroupEditableText=false
    bUseCheckBox=true
    description=Adjusts the shadows cast by the clouds. Takes effect once the slider is released.
    maxFractionalDigits=0
    maxIntegralDigits=4
    minFractionalDigits=0
    slider Value=0
    title=Cloud Time
    minMaxValue=<0 2400>
   }
   slot {class=ScrollBoxSlot, properties {padding=<ltrb 15 0 15 15>}}
  }
 ]
}
