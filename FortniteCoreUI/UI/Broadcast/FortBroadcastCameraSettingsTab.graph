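# Spectator camera settings tab for Fortnite replay and broadcast UI, listing rotators and sliders for visual effects, third-person and drone camera, battle map and watermark options.
# Widget changes call parent-class setters, and On...Changed events sync the controls back.
# Replay state, playlist and camera type control row visibility.
# The last hovered or focused row drives reset-to-default and gamepad refocus, with wrap-around navigation.
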WidgetBlueprint /FortniteCoreUI/UI/Broadcast/FortBroadcastCameraSettingsTab
graphs [
 CloseTabGroup {nodes [
   CloseTabGroup() {is=title K2Node_FunctionEntry}
   > SetTabCollapsed!(
    execute
    self:Fort Replay Context Object Reference=Get FortReplayContext!() -> (ReturnValue:Fort Replay Context Object Reference) {is=title K2Node_GetBlueprintContextEx}
    bIsCollapsed:bool=true
   ) -> (then)
  ]}
 EventGraph {nodes [
   n1 = "Slider Changed (TPDistanceToSubject)"!() -> (OutputDelegate, then, Slider Value:double) {is=title K2Node_ComponentBoundEvent}
   > SetThirdPersonDistanceToSubject!(execute, self:Fort Spectator Camera Settings Tab Object Reference, ThirdPersonDistanceToSubject:float=@n1.Slider Value) -> (then)
   n2 = "Selection Changed (DroneIgnoreJump)"!() -> (OutputDelegate, then, Selected Index:int=0) {is=title K2Node_ComponentBoundEvent}
   > SetDroneIgnoreJump!(
    execute
    self:Fort Spectator Camera Settings Tab Object Reference
    bDroneIgnoreJump:bool="integer > integer"(A=@n2.Selected Index, B=0) {is=title K2Node_CallFunction}
   ) -> (then)
   n3 = "Selection Changed (FirstPlaceHighlight)"!() -> (OutputDelegate, then, Selected Index:int=0) {is=title K2Node_ComponentBoundEvent}
   > SetFirstPlaceHighlightEnabled!(
    execute
    self:Fort Spectator Camera Settings Tab Object Reference
    bFirstPlaceHighlightEnabled:bool="integer > integer"(A=@n3.Selected Index, B=0)
   ) -> (then)
   n4 = "Selection Changed (BattleMapShowTimeIndicator)"!() -> (OutputDelegate, then, Selected Index:int=0) {is=title K2Node_ComponentBoundEvent}
   > SetBattleMapShowTimeIndicator!(
    execute
    self:Fort Spectator Camera Settings Tab Object Reference
    bBattleMapShowTimeIndicator:bool="integer > integer"(A=@n4.Selected Index, B=0)
   ) -> (then)
   n5 = "Selection Changed (SessionIDwatermark)"!() -> (OutputDelegate, then, Selected Index:int=0) {is=title K2Node_ComponentBoundEvent}
   n6 = "integer > integer"(A=@n5.Selected Index, B=0)
   > SetSessionIdWatermarkVisible!(execute, self:Fort Replay Context Object Reference=Get FortReplayContext(), bIsVisible:bool=@n6) -> (then)
   > SetSessionIdWatermark!(execute, self:Fort Spectator Camera Settings Tab Object Reference, bShowSessionIdWatermark:bool=@n6) -> (then)
   "On Hovered (Rotator_PlayerOutlines)"() {is=title K2Node_ComponentBoundEvent}
   > Set LatestHoveredWidget(LatestHoveredWidget:User Widget Object Reference=Get Rotator_PlayerOutlines() -> (Rotator_PlayerOutlines:Common User Widget Object Reference)) -> (Output_Get:User Widget Object Reference)
   "On Hovered (FirstPlaceHighlight)"() {is=title K2Node_ComponentBoundEvent}
   > Set LatestHoveredWidget(LatestHoveredWidget=Get FirstPlaceHighlight() -> (FirstPlaceHighlight:Rotator Selector Replay Settings Object Reference))
   n7 = "Selection Changed (Rotator_PlayerOutlines)"!() -> (OutputDelegate, then, Selected Index:int=0) {is=title K2Node_ComponentBoundEvent}
   > SetPlayerOutlinesEnabled!(
    execute
    self:Fort Spectator Camera Settings Tab Object Reference
    bPlayerOutlinesEnabled:bool="integer > integer"(A=@n7.Selected Index, B=0)
   ) -> (then)
   "On Hovered (StormEffects)"() {is=title K2Node_ComponentBoundEvent}
   > Set LatestHoveredWidget(LatestHoveredWidget=Get StormEffects() -> (StormEffects:Rotator Selector Replay Settings Object Reference))
   "On Focused (OutsideSafeZoneOpacity)"() {is=title K2Node_ComponentBoundEvent}
   > Set LatestHoveredWidget(LatestHoveredWidget=Get OutsideSafeZoneOpacity() -> (OutsideSafeZoneOpacity:Options Menu Slider Replay Object Reference))
   "On Hovered (RelevancyZone)"() {is=title K2Node_ComponentBoundEvent}
   > Set LatestHoveredWidget(LatestHoveredWidget=Get RelevancyZone() -> (RelevancyZone:Rotator Selector Replay Settings Object Reference))
   "On Hovered (HighQualityFX)"() {is=title K2Node_ComponentBoundEvent}
   > Set LatestHoveredWidget(LatestHoveredWidget=Get HighQualityFX() -> (HighQualityFX:Rotator Selector Replay Settings Object Reference))
   "On Hovered (DamageEffects)"() {is=title K2Node_ComponentBoundEvent}
   > Set LatestHoveredWidget(LatestHoveredWidget=Get DamageEffects() -> (DamageEffects:Rotator Selector Replay Settings Object Reference))
   n8 = "Selection Changed (RelevancyZone)"!() -> (OutputDelegate, then, Selected Index:int=0) {is=title K2Node_ComponentBoundEvent}
   > SetRelevancyZoneVisible!(execute, self:Fort Spectator Camera Settings Tab Object Reference, bRelevancyZoneVisible:bool="integer > integer"(A=@n8.Selected Index, B=0)) -> (then)
   n9 = "Selection Changed (BlendSettings)"!() -> (OutputDelegate, then, Selected Index:int=0) {is=title K2Node_ComponentBoundEvent}
   > SetBlendType!(
    execute
    self:Fort Player Controller Spectating Object Reference=Cast To FortPlayerControllerSpectating!(Object=GetOwningPlayer()) -> (AsFort Player Controller Spectating:Fort Player Controller Spectating Object Reference, bSuccess:bool=~) {is=title K2Node_DynamicCast}
    NewBlendType:EFortSpectatorBlendType Enum=Byte to Enum EFortSpectatorBlendType!(Byte:Byte="To Byte (Integer)"!(InInt:int=@n9.Selected Index) -> (ReturnValue:Byte=0)) -> (ReturnValue:EFortSpectatorBlendType Enum) {is=title K2Node_CastByteToEnum}
   ) -> (then)
   "On Hovered (TPAutoFollow)"() {is=title K2Node_ComponentBoundEvent}
   > Set LatestHoveredWidget(LatestHoveredWidget=Get TPAutoFollow() -> (TPAutoFollow:Rotator Selector Replay Settings Object Reference))
   "On Focused (TPDistanceToSubject)"() {is=title K2Node_ComponentBoundEvent}
   > Set LatestHoveredWidget(LatestHoveredWidget=Get TPDistanceToSubject() -> (TPDistanceToSubject:Options Menu Slider Replay Object Reference))
   "On Hovered (ShareLensSettings)"() {is=title K2Node_ComponentBoundEvent}
   > Set LatestHoveredWidget(LatestHoveredWidget=Get ShareLensSettings() -> (ShareLensSettings:Rotator Selector Replay Settings Object Reference))
   "On Hovered (TPCameraCollision)"() {is=title K2Node_ComponentBoundEvent}
   > Set LatestHoveredWidget(LatestHoveredWidget=Get TPCameraCollision() -> (TPCameraCollision:Rotator Selector Replay Settings Object Reference))
   n10 = "Selection Changed (TPAutoFollow)"!() -> (OutputDelegate, then, Selected Index:int=0) {is=title K2Node_ComponentBoundEvent}
   > SetThirdPersonAutoFollowMode!(
    execute
    self:Fort Spectator Camera Settings Tab Object Reference
    ThirdPersonAutoFollowMode:EThirdPersonAutoFollowMode Enum=Byte to Enum EThirdPersonAutoFollowMode!(Byte:Byte="To Byte (Integer)"(InInt=@n10.Selected Index) -> (ReturnValue=0)) -> (ReturnValue:EThirdPersonAutoFollowMode Enum) {is=title K2Node_CastByteToEnum}
   ) -> (then)
   "On Hovered (DroneIgnoreJump)"() {is=title K2Node_ComponentBoundEvent}
   > Set LatestHoveredWidget(LatestHoveredWidget=Get DroneIgnoreJump() -> (DroneIgnoreJump:Rotator Selector Replay Settings Object Reference))
   "On Hovered (SessionIDwatermark)"() {is=title K2Node_ComponentBoundEvent}
   > Set LatestHoveredWidget(LatestHoveredWidget=Get SessionIDwatermark() -> (SessionIDwatermark:Rotator Selector Replay Settings Object Reference))
   "On Hovered (BattleMapShowTimeIndicator)"() {is=title K2Node_ComponentBoundEvent}
   > Set LatestHoveredWidget(LatestHoveredWidget=Get BattleMapShowTimeIndicator() -> (BattleMapShowTimeIndicator:Rotator Selector Replay Settings Object Reference))
   "On Hovered (BattleMapShowNearbyChests)"() {is=title K2Node_ComponentBoundEvent}
   > Set LatestHoveredWidget(LatestHoveredWidget=Get BattleMapShowNearbyChests() -> (BattleMapShowNearbyChests:Rotator Selector Replay Settings Object Reference))
   "On Hovered (BattleMapShowAllSquadMembers)"() {is=title K2Node_ComponentBoundEvent}
   > Set LatestHoveredWidget(LatestHoveredWidget=Get BattleMapShowAllSquadMembers() -> (BattleMapShowAllSquadMembers:Rotator Selector Replay Settings Object Reference))
   n11 = "Selection Changed (DamageEffects)"!() -> (OutputDelegate, then, Selected Index:int=0) {is=title K2Node_ComponentBoundEvent}
   > SetDamageFxEnabled!(execute, self:Fort Spectator Camera Settings Tab Object Reference, bDamageFxEnabled:bool="integer > integer"(A=@n11.Selected Index, B=0)) -> (then)
   n12 = "Selection Changed (BattleMapShowNearbyChests)"!() -> (OutputDelegate, then, Selected Index:int=0) {is=title K2Node_ComponentBoundEvent}
   > SetBattleMapShowNearbyChests!(
    execute
    self:Fort Spectator Camera Settings Tab Object Reference
    bBattleMapShowNearbyChests:bool="integer > integer"(A=@n12.Selected Index, B=0)
   ) -> (then)
   "On Hovered (BattleMapAutomaticCameraPlacement)"() {is=title K2Node_ComponentBoundEvent}
   > Set LatestHoveredWidget(LatestHoveredWidget=Get BattleMapAutomaticCameraPlacement() -> (BattleMapAutomaticCameraPlacement:Rotator Selector Replay Settings Object Reference))
   n13 = "Selection Changed (StormEffects)"!() -> (OutputDelegate, then, Selected Index:int=0) {is=title K2Node_ComponentBoundEvent}
   > SetStormEffectsEnabled!(execute, self:Fort Spectator Camera Settings Tab Object Reference, bStormEffectsEnabled:bool="integer > integer"(A=@n13.Selected Index, B=0)) -> (then)
   n14 = "Selection Changed (ShareLensSettings)"!() -> (OutputDelegate, then, Selected Index:int=0) {is=title K2Node_ComponentBoundEvent}
   > SetShareLensSettings!(execute, self:Fort Spectator Camera Settings Tab Object Reference, bShareLensSettings:bool="integer > integer"(A=@n14.Selected Index, B=0)) -> (then)
   n15 = "Selection Changed (TPCameraCollision)"!() -> (OutputDelegate, then, Selected Index:int=0) {is=title K2Node_ComponentBoundEvent}
   > SetThirdPersonCameraCollision!(
    execute
    self:Fort Spectator Camera Settings Tab Object Reference
    bThirdPersonCameraCollision:bool="integer > integer"(A=@n15.Selected Index, B=0)
   ) -> (then)
   n16 = "Slider Changed (OutsideSafeZoneOpacity)"!() -> (OutputDelegate, then, Slider Value:double) {is=title K2Node_ComponentBoundEvent}
   > SetSafeZoneOpacity!(execute, self:Fort Spectator Camera Settings Tab Object Reference, SafeZoneOpacity:float=@n16.Slider Value) -> (then)
   n17 = "Selection Changed (BattleMapAutomaticCameraPlacement)"!() -> (OutputDelegate, then, Selected Index:int=0) {is=title K2Node_ComponentBoundEvent}
   > SetBattleMapAutomaticCameraPlacement!(
    execute
    self:Fort Spectator Camera Settings Tab Object Reference
    bBattleMapAutomaticCameraPlacement:bool="integer > integer"(A=@n17.Selected Index, B=0)
   ) -> (then)
   n18 = "Selection Changed (HighQualityFX)"!() -> (OutputDelegate, then, Selected Index:int=0) {is=title K2Node_ComponentBoundEvent}
   > SetHighQualityFxEnabled!(
    execute
    self:Fort Spectator Camera Settings Tab Object Reference
    bHighQualityFxEnabled:bool="integer > integer"(A=@n18.Selected Index, B=0)
   ) -> (then)
   n19 = "Selection Changed (BattleMapShowAllSquadMembers)"!() -> (OutputDelegate, then, Selected Index:int=0) {is=title K2Node_ComponentBoundEvent}
   > SetBattleMapShowAllSquadMembers!(
    execute
    self:Fort Spectator Camera Settings Tab Object Reference
    bBattleMapShowAllSquadMembers:bool="integer > integer"(A=@n19.Selected Index, B=0)
   ) -> (then)
   n20 = HandleCloseTab!() -> (OutputDelegate, then, Selected Index:int) {is=title K2Node_CustomEvent}
   > CloseTabGroup() {is=title K2Node_CallFunction}
   n21 = HandleCloseTabSlider!() -> (OutputDelegate, then, Slider Value:double=~) {is=title K2Node_CustomEvent}
   > CloseTabGroup()
   n22 = AdjustScrollBar() {is=title K2Node_CustomEvent}
   n23 = Get CommonInputSubsystem!() -> (ReturnValue:Common Input Object Reference) {is=title K2Node_GetBlueprintContextEx}
   > Is Valid(InputObject=@n23)
   n24 = GetCurrentInputType(self=@n23)
   > n25 = Branch(Condition="Equal (Enum)"(A=@n24, B=Gamepad))
   > SetScrollWhenFocusChanges!(
    execute
    self:Scroll Box Object Reference=Get ScrollBox_Tabs() -> (ScrollBox_Tabs:Scroll Box Object Reference)
    NewScrollWhenFocusChanges:EScrollWhenFocusChanges Enum=AnimatedScroll
   ) -> (then)
   n26 = Get LatestHoveredWidget() -> (LatestHoveredWidget:User Widget Object Reference)
   > n27 = Is Valid(InputObject=@n26)
   n28 = Cast To RotatorSelectorReplaySettings!(execute=@n27.Is Not Valid, Object=Get Rotator_PlayerOutlines()) -> (then, CastFailed, AsRotator Selector Replay Settings:Rotator Selector Replay Settings Object Reference) {is=title K2Node_DynamicCast}
   > SetFocus(self=@n28)
   SetFocus(execute=@n27.Is Valid, self=@n26)
   Branch(execute=@n25.else, Condition="Equal (Enum)"(A=@n24, B=MouseAndKeyboard))
   > SetScrollWhenFocusChanges(self=Get ScrollBox_Tabs(), NewScrollWhenFocusChanges=NoScroll)
   Event Construct()
   n29 = Get PlaybackState!(self:Fort Replay Context Object Reference=Get FortReplayContext()) -> (PlaybackState:Fort Replay Playback State Structure)
   > SetVisibility(
    self=Get RelevancyZone()
    InVisibility=Select(
     Option 0=Collapsed
     Option 1=Visible
     Index=Break Fort Replay Playback State!(FortReplayPlaybackState:"Fort Replay Playback State Structure (by ref)"=@n29) -> (
      StartTime:float=~
      EndTime:float=~
      TimeNow:float=~
      bIsPaused:bool=~
      PlaybackSpeedMultiplier:float=~
      HUDVisibility:EHudVisibilityState Enum
      bLevelStreaming:bool=~
      bHasRelevancyZone:bool=~
      bHasPlayInterval:bool=~
      PlayIntervalStartTime:float=~
      PlayIntervalEndTime:float=~
      bLoopInterval:bool=~
      IntervalLoopBufferTime:float=~
     ) {is=title K2Node_BreakStruct}.bHasRelevancyZone
    )
   )
   > SetVisibility(
    self=Get FirstPlaceHighlight()
    InVisibility=Select(Option 0=Collapsed, Option 1=Visible, Index=IsCompetitivePlaylist!() -> (ReturnValue:bool))
   )
   Event PreConstruct()
   > Initialize!(
    execute
    self:Rotator Selector Replay Settings Object Reference=Cast To RotatorSelectorReplaySettings!(Object=Get Rotator_PlayerOutlines()) -> (AsRotator Selector Replay Settings, bSuccess:bool=~)
    Display Text:Text="LOCTEXT(\"Player Outlines\")"
    Options:Array of Texts=Make Off on Array!() -> (OffOnArray:Array of Texts) {is=title K2Node_MacroInstance}
    Hover Text:Text="LOCTEXT(\"Enable/Disable player outlines\")"
    Tab Tooltip:Common Text Object Reference
   ) -> (then)
   > Initialize(
    self=Get FirstPlaceHighlight()
    Display Text="LOCTEXT(\"First Place Highlight\")"
    Options=Make Off on Array()
    Hover Text="LOCTEXT(\"Enable/Disable first place highlight\")"
   )
   > Initialize(
    self=Get StormEffects()
    Display Text="LOCTEXT(\"Storm Effects\")"
    Options=Make Off on Array()
    Hover Text="LOCTEXT(\"Enable/Disable storm visual effects\")"
   )
   > Update Slider!(
    execute
    self:Options Menu Slider Replay Object Reference=Get OutsideSafeZoneOpacity()
    Slider Text:Text="LOCTEXT(\"Safe Zone Opacity\")"
    Slider Value:double=0.5
    Hover Text:Text
    Tooltip Text Block:Common Text Object Reference
    Min Max Value:Vector 2D Structure=(X=0,Y=1)
   ) -> (then)
   > Initialize(
    self=Get RelevancyZone()
    Display Text="LOCTEXT(\"Replay Region\")"
    Options=Make Off on Array()
    Hover Text="LOCTEXT(\"Enable/disable replay region visibility\")"
   )
   > Initialize(
    self=Get TPAutoFollow()
    Display Text="LOCTEXT(\"Auto Follow\")"
    Options=Make Array!("[0]":Text=Get OffText() -> (OffText:Text), "[1]":Text="LOCTEXT(\"Auto\")", "[2]":Text="LOCTEXT(\"Lazy\")") -> (Array:Array of Texts)
    Hover Text="LOCTEXT(\"Set auto follow mode\")"
   )
   > Update Slider(self=Get TPDistanceToSubject(), Slider Text="LOCTEXT(\"Distance to Subject\")", Slider Value=0.5, Min Max Value=(X=0,Y=1))
   > Initialize(
    self=Get DamageEffects()
    Display Text="LOCTEXT(\"Damage Effects\")"
    Options=Make Off on Array()
    Hover Text="LOCTEXT(\"Enable/disable damage effects\")"
   )
   > Initialize(
    self=Get ShareLensSettings()
    Display Text="LOCTEXT(\"Share Lens Settings\")"
    Options=Make Array!("[0]"="LOCTEXT(\"No\")", "[1]"="LOCTEXT(\"Yes\")") -> (Array)
    Hover Text="LOCTEXT(\"Keep Lens Settings when changing camera type\")"
   )
   > Initialize(
    self=Get BlendSettings() -> (BlendSettings:Rotator Selector Replay Settings Object Reference)
    Display Text="LOCTEXT(\"Camera Blends\")"
    Options=Make Array!("[0]"=Get OffText(), "[1]"="LOCTEXT(\"Orbit\")", "[2]":Text="LOCTEXT(\"Default\")") -> (Array)
    Hover Text="LOCTEXT(\"This does not force blends, only allows them\")"
   )
   > Initialize(
    self=Get HighQualityFX()
    Display Text="LOCTEXT(\"High Quality FX\")"
    Options=Make Off on Array()
    Hover Text="LOCTEXT(\"This setting may impact performance\")"
   )
   > Initialize(self=Get TPCameraCollision(), Display Text="LOCTEXT(\"Static Collision Enabled\")", Options=Make Off on Array())
   > Initialize(self=Get DroneIgnoreJump(), Display Text="LOCTEXT(\"Ignore Jump\")", Options=Make Off on Array())
   > Initialize(self=Get SessionIDwatermark(), Display Text="LOCTEXT(\"Session ID Watermark\")", Options=Make Off on Array())
   > Initialize(
    self=Get BattleMapShowTimeIndicator()
    Display Text="LOCTEXT(\"Time Indicator\")"
    Options=Make Array!("[0]"="LOCTEXT(\"Hide\")", "[1]"="LOCTEXT(\"Show\")") -> (Array)
   )
   > Initialize(
    self=Get BattleMapShowNearbyChests()
    Display Text="LOCTEXT(\"Nearby Chests\")"
    Options=Make Array("[0]"="LOCTEXT(\"Hide\")", "[1]"="LOCTEXT(\"Show\")")
   )
   > Initialize(
    self=Get BattleMapShowAllSquadMembers()
    Display Text="LOCTEXT(\"All Squad Members\")"
    Options=Make Array("[0]"="LOCTEXT(\"Hide\")", "[1]"="LOCTEXT(\"Show\")")
   )
   > Initialize(self=Get BattleMapAutomaticCameraPlacement(), Display Text="LOCTEXT(\"Automatic Camera Placement\")", Options=Make Off on Array())
   > SetPanelPadding()
   > n30 = Cast To RotatorSelectorReplaySettings!(execute:Exec, Object=Get Rotator_PlayerOutlines()) -> (then:Exec, CastFailed:Exec, AsRotator Selector Replay Settings)
   > Bind Event to Selection Changed(
    self:Rotator Selector Replay Settings Object Reference=[
     @n30
     Get RelevancyZone()
     Get HighQualityFX()
     Get DamageEffects()
     Get TPAutoFollow()
     Get ShareLensSettings()
     Get BlendSettings()
     Get TPCameraCollision()
     Get DroneIgnoreJump()
     Get SessionIDwatermark()
     Get FirstPlaceHighlight()
     Get StormEffects()
    ]
    Delegate=@n20
   ) {is=title K2Node_AddDelegate}
   > Bind Event to Slider Changed(self:Options Menu Slider Replay Object Reference=[Get OutsideSafeZoneOpacity(), Get TPDistanceToSubject()], Delegate=@n21) {is=title K2Node_AddDelegate}
   > n31 = Cast To RotatorSelectorReplaySettings(Object=Get Rotator_PlayerOutlines())
   > Set LatestHoveredWidget(LatestHoveredWidget=@n31)
   n32 = Event OnDamageFxEnabledChanged!() -> (OutputDelegate, then, bDamageFxEnabled:bool) {is=title K2Node_Event}
   > Update Row Selector!(
    execute
    self:Rotator Selector Replay Settings Object Reference=Get DamageEffects()
    Currently Selected:int=Select(Option 0:int=0, Option 1:int=1, Index=@n32.bDamageFxEnabled) -> (ReturnValue:int)
   ) -> (then)
   n33 = Event OnThirdPersonAutoFollowModeChanged!() -> (OutputDelegate, then, ThirdPersonAutoFollowMode:EThirdPersonAutoFollowMode Enum=Off) {is=title K2Node_Event}
   > Update Row Selector(self=Get TPAutoFollow(), Currently Selected="To Integer (Byte)"!(InByte:Byte=@n33.ThirdPersonAutoFollowMode) -> (ReturnValue:int=0))
   n34 = Event OnThirdPersonDistanceToSubjectChanged!() -> (OutputDelegate, then, ThirdPersonDistanceToSubject:float) {is=title K2Node_Event}
   > Update Slider Value!(
    execute
    self:Options Menu Slider Replay Object Reference=Get TPDistanceToSubject()
    NormalizedValue:double=@n34.ThirdPersonDistanceToSubject
   ) -> (then)
   n35 = Event OnThirdPersonCameraCollisionChanged!() -> (OutputDelegate, then, bThirdPersonCameraCollision:bool) {is=title K2Node_Event}
   > Update Row Selector(self=Get TPCameraCollision(), Currently Selected=Select(Option 0=0, Option 1=1, Index=@n35.bThirdPersonCameraCollision))
   n36 = Event OnShareLensSettingsChanged!() -> (OutputDelegate, then, bShareLensSettings:bool) {is=title K2Node_Event}
   > Update Row Selector(self=Get ShareLensSettings(), Currently Selected=Select(Option 0=0, Option 1=1, Index=@n36.bShareLensSettings))
   n37 = Event OnDroneIgnoreJumpChanged!() -> (OutputDelegate, then, bDroneIgnoreJump:bool) {is=title K2Node_Event}
   > Update Row Selector(self=Get DroneIgnoreJump(), Currently Selected=Select(Option 0=0, Option 1=1, Index=@n37.bDroneIgnoreJump))
   n38 = Event OnBattleMapShowTimeIndicatorChanged!() -> (OutputDelegate, then, bBattleMapShowTimeIndicator:bool) {is=title K2Node_Event}
   > Update Row Selector(self=Get BattleMapShowTimeIndicator(), Currently Selected=Select(Option 0=0, Option 1=1, Index=@n38.bBattleMapShowTimeIndicator))
   n39 = Event OnBattleMapShowNearbyChestsChanged!() -> (OutputDelegate, then, bBattleMapShowNearbyChests:bool) {is=title K2Node_Event}
   > Update Row Selector(self=Get BattleMapShowNearbyChests(), Currently Selected=Select(Option 0=0, Option 1=1, Index=@n39.bBattleMapShowNearbyChests))
   n40 = Event OnBattleMapShowAllSquadMembersChanged!() -> (OutputDelegate, then, bBattleMapShowAllSquadMembers:bool) {is=title K2Node_Event}
   > Update Row Selector(self=Get BattleMapShowAllSquadMembers(), Currently Selected=Select(Option 0=0, Option 1=1, Index=@n40.bBattleMapShowAllSquadMembers))
   n41 = Event OnBattleMapAutomaticCameraPlacementChanged!() -> (OutputDelegate, then, bInBattleMapAutomaticCameraPlacement:bool) {is=title K2Node_Event}
   > Update Row Selector(
    self=Get BattleMapAutomaticCameraPlacement()
    Currently Selected=Select(Option 0=0, Option 1=1, Index=@n41.bInBattleMapAutomaticCameraPlacement)
   )
   n42 = Event OnSessionIDWatermarkChanged!() -> (OutputDelegate, then, bShowSessionIDWatermark:bool) {is=title K2Node_Event}
   > Update Row Selector(self=Get SessionIDwatermark(), Currently Selected=Select(Option 0=0, Option 1=1, Index=@n42.bShowSessionIDWatermark))
   Event On Activated()
   > Parent: On Activated() {is=title K2Node_CallParentFunction}
   > n43 = Branch(Condition=Get bIsFirstActivation() -> (bIsFirstActivation:bool))
   > n44 = Cast To RotatorSelectorReplaySettings(Object=Get Rotator_PlayerOutlines())
   > SetFocus(self=@n44)
   > n45 = Set bIsFirstActivation(bIsFirstActivation:bool=false) -> (Output_Get:bool)
   SetNavigationRules(execute=[@n45.then, @n43.else])
   > AdjustScrollBar() {is=title K2Node_CallFunction}
   > Bind Event to On Keybinds Changed(
    self:Fort Global UIContext Object Reference=Get FortGlobalUIContext!() -> (ReturnValue:Fort Global UIContext Object Reference) {is=title K2Node_GetBlueprintContextEx}
    Delegate=@n22
   ) {is=title K2Node_AddDelegate}
   > SetNavigationRules()
   n46 = Event OnPlayerOutlinesEnabledChanged!() -> (OutputDelegate, then, bPlayerOutlinesEnabled:bool) {is=title K2Node_Event}
   > Update Row Selector(
    self=Cast To RotatorSelectorReplaySettings!(Object=Get Rotator_PlayerOutlines()) -> (AsRotator Selector Replay Settings, bSuccess:bool=~)
    Currently Selected=Select(Option 0=0, Option 1=1, Index=@n46.bPlayerOutlinesEnabled)
   )
   Event CenterOnTab() {is=title K2Node_Event}
   > Center On Widget(self:Rotator Selector Replay Settings Object Reference=Cast To RotatorSelectorReplaySettings(Object=Get Rotator_PlayerOutlines()) -> (bSuccess=~))
   n47 = Event OnCameraTypeChanged!() -> (
    OutputDelegate
    then
    FortPlayerControllerSpectating:Fort Player Controller Spectating Object Reference
    CameraType:ESpectatorCameraType Enum=ThirdPerson
   ) {is=title K2Node_Event}
   n48 = "Equal (Enum)"(A:ESpectatorCameraType Enum=@n47.CameraType, B:ESpectatorCameraType Enum=ThirdPerson)
   > SetVisibility(
    self=[Get TPAutoFollow(), Get TPDistanceToSubject(), Get TPCameraCollision()]
    InVisibility=Select(Option 0:ESlateVisibility Enum=Collapsed, Option 1:ESlateVisibility Enum=Visible, Index=@n48) -> (ReturnValue:ESlateVisibility Enum)
   )
   > SetVisibility(
    self=Get DroneIgnoreJump()
    InVisibility=Select(Option 0=Collapsed, Option 1=Visible, Index="Equal (Enum)"(A=@n47.CameraType, B=DroneFollow))
   )
   n49 = HasBroadcasterToken!(self:Fort Player Controller Spectating Object Reference=@n47.FortPlayerControllerSpectating) -> (ReturnValue:bool)
   > SetVisibility(self=Get TPCameraCollision(), InVisibility=Select(Option 0=Collapsed, Option 1=Visible, Index=AND Boolean(A=@n48, B=@n49)))
   n50 = Event OnStormEffectsEnabledChanged!() -> (OutputDelegate, then, bStormEffectsEnabled:bool) {is=title K2Node_Event}
   > Update Row Selector(self=Get StormEffects(), Currently Selected=Select(Option 0:int=0, Option 1:int=1, Index=@n50.bStormEffectsEnabled) -> (ReturnValue:int))
   n51 = Event OnFirstPlaceHighlightEnabledChanged!() -> (OutputDelegate, then, bFirstPlaceHighlightEnabled:bool) {is=title K2Node_Event}
   > Update Row Selector(self=Get FirstPlaceHighlight(), Currently Selected=Select(Option 0=0, Option 1=1, Index=@n51.bFirstPlaceHighlightEnabled))
   n52 = Event OnSafeZoneOpacityChanged!() -> (OutputDelegate, then, SafeZoneOpacity:float) {is=title K2Node_Event}
   > Update Slider Value(self=Get OutsideSafeZoneOpacity(), NormalizedValue=@n52.SafeZoneOpacity)
   n53 = Event OnHighQualityFxEnabledChanged!() -> (OutputDelegate, then, bHighQualityFxEnabled:bool) {is=title K2Node_Event}
   > Update Row Selector(self=Get HighQualityFX(), Currently Selected=Select(Option 0=0, Option 1=1, Index=@n53.bHighQualityFxEnabled))
   n54 = Event OnRelevancyZoneVisibleChanged!() -> (OutputDelegate, then, bRelevancyZoneVisible:bool) {is=title K2Node_Event}
   > Update Row Selector(self=Get RelevancyZone(), Currently Selected=Select(Option 0=0, Option 1=1, Index=@n54.bRelevancyZoneVisible))
   Event ResetFocusedElement() {is=title K2Node_Event}
   > Is Valid(InputObject=Get LatestHoveredWidget())
   > n55 = Branch(Condition="Equal (Object)"(A=Get Rotator_PlayerOutlines(), B=Get LatestHoveredWidget()))
   > ResetPlayerOutlinesEnabled(self:Fort Spectator Camera Settings Tab Object Reference)
   n56 = Branch(execute=@n55.else, Condition="Equal (Object)"(A=Get FirstPlaceHighlight(), B=Get LatestHoveredWidget()))
   > ResetFirstPlaceHighlightEnabled(self:Fort Spectator Camera Settings Tab Object Reference)
   n57 = Branch(execute=@n56.else, Condition="Equal (Object)"(A=Get StormEffects(), B=Get LatestHoveredWidget()))
   > ResetStormEffectsEnabled(self:Fort Spectator Camera Settings Tab Object Reference)
   n58 = Branch(execute=@n57.else, Condition="Equal (Object)"(A=Get OutsideSafeZoneOpacity(), B=Get LatestHoveredWidget()))
   > ResetSafeZoneOpacity(self:Fort Spectator Camera Settings Tab Object Reference)
   n59 = Branch(execute=@n58.else, Condition="Equal (Object)"(A=Get RelevancyZone(), B=Get LatestHoveredWidget()))
   > ResetRelevancyZoneVisible(self:Fort Spectator Camera Settings Tab Object Reference)
   n60 = Branch(execute=@n59.else, Condition="Equal (Object)"(A=Get HighQualityFX(), B=Get LatestHoveredWidget()))
   > ResetHighQualityFxEnabled(self:Fort Spectator Camera Settings Tab Object Reference)
   n61 = Branch(execute=@n60.else, Condition="Equal (Object)"(A=Get DamageEffects(), B=Get LatestHoveredWidget()))
   > ResetDamageFxEnabled(self:Fort Spectator Camera Settings Tab Object Reference)
   n62 = Branch(execute=@n61.else, Condition="Equal (Object)"(A=Get TPAutoFollow(), B=Get LatestHoveredWidget()))
   > ResetThirdPersonAutoFollowMode(self:Fort Spectator Camera Settings Tab Object Reference)
   n63 = Branch(execute=@n62.else, Condition="Equal (Object)"(A=Get TPDistanceToSubject(), B=Get LatestHoveredWidget()))
   > ResetThirdPersonDistanceToSubject(self:Fort Spectator Camera Settings Tab Object Reference)
   n64 = Branch(execute=@n63.else, Condition="Equal (Object)"(A=Get ShareLensSettings(), B=Get LatestHoveredWidget()))
   > ResetShareLensSettings(self:Fort Spectator Camera Settings Tab Object Reference)
   n65 = Branch(execute=@n64.else, Condition="Equal (Object)"(A=Get TPCameraCollision(), B=Get LatestHoveredWidget()))
   > ResetThirdPersonCameraCollision(self:Fort Spectator Camera Settings Tab Object Reference)
   n66 = Branch(execute=@n65.else, Condition="Equal (Object)"(A=Get DroneIgnoreJump(), B=Get LatestHoveredWidget()))
   > ResetDroneIgnoreJump(self:Fort Spectator Camera Settings Tab Object Reference)
   n67 = Branch(execute=@n66.else, Condition="Equal (Object)"(A=Get SessionIDwatermark(), B=Get LatestHoveredWidget()))
   > ResetSessionIdWatermark(self:Fort Spectator Camera Settings Tab Object Reference)
   n68 = Branch(execute=@n67.else, Condition="Equal (Object)"(A=Get BattleMapShowTimeIndicator(), B=Get LatestHoveredWidget()))
   > ResetBattleMapShowTimeIndicator(self:Fort Spectator Camera Settings Tab Object Reference)
   n69 = Branch(execute=@n68.else, Condition="Equal (Object)"(A=Get BattleMapShowNearbyChests(), B=Get LatestHoveredWidget()))
   > ResetBattleMapShowNearbyChests(self:Fort Spectator Camera Settings Tab Object Reference)
   n70 = Branch(execute=@n69.else, Condition="Equal (Object)"(A=Get BattleMapShowAllSquadMembers(), B=Get LatestHoveredWidget()))
   > ResetBattleMapShowAllSquadMembers(self:Fort Spectator Camera Settings Tab Object Reference)
   Branch(execute=@n70.else, Condition="Equal (Object)"(A=Get BattleMapAutomaticCameraPlacement(), B=Get LatestHoveredWidget()))
   > ResetBattleMapAutomaticCameraPlacement(self:Fort Spectator Camera Settings Tab Object Reference)
  ]}
 MakeOffOnArray {nodes [
   Inputs!()
   Outputs!(OffOnArray:Array of Texts=Make Array("[0]"=Get OffText(), "[1]"=Get OnText() -> (OnText:Text)))
  ]}
 SetNavigationRules {nodes [
   SetNavigationRules() {is=title K2Node_FunctionEntry}
   n1 = GetAllChildren(self=Get ScrollBox_Tabs())
   n2 = "Get (a copy)"(Array:Array of Widget Object References=@n1, Dimension 1=int - int(A=Length(TargetArray:Array of Widget Object References=@n1))) -> (Output:Widget Object Reference)
   n3 = "Get (a copy)"(Array=GetAllChildren(self=Get ScrollBox_Tabs()))
   > SetNavigationRuleExplicit!(execute, self:Widget Object Reference=@n3, Direction:EUINavigation Enum=Up, InWidget:Widget Object Reference=@n2) -> (then)
   > SetNavigationRuleExplicit(self=@n2, Direction=Down, InWidget=@n3)
  ]}
 SetPanelPadding {nodes [
   SetPanelPadding() {is=title K2Node_FunctionEntry}
   > n1 = For Each Loop(Array:Array of Widget Object References=GetAllChildren(self=Get ScrollBox_Tabs())) -> (Array Element:Widget Object Reference)
   > Set Loc_ScrollSlot(Loc_ScrollSlot:Scroll Box Slot Object Reference=SlotAsScrollBoxSlot!(Widget:Widget Object Reference=@n1) -> (ReturnValue:Scroll Box Slot Object Reference)) -> (Output_Get:Scroll Box Slot Object Reference)
   > n2 = Branch(Condition="Equal (Integer)"(A=@n1.Array Index))
   > SetPadding!(
    execute
    self:Scroll Box Slot Object Reference=Get Loc_ScrollSlot() -> (Loc_ScrollSlot:Scroll Box Slot Object Reference)
    InPadding_Left:float=15
    InPadding_Top:float=15
    InPadding_Right:float=15
    InPadding_Bottom:float=12
   ) -> (then)
   n3 = Length(TargetArray=GetAllChildren(self=Get ScrollBox_Tabs()))
   n4 = Branch(execute=@n2.else, Condition="Equal (Integer)"(A=@n1.Array Index, B=int - int(A=@n3)))
   > SetPadding(self=Get Loc_ScrollSlot(), InPadding_Left=15, InPadding_Top=0, InPadding_Right=15, InPadding_Bottom=15)
   SetPadding(execute=@n4.else, self=Get Loc_ScrollSlot(), InPadding_Left=15, InPadding_Top=0, InPadding_Right=15, InPadding_Bottom=12)
  ]}
]
variables [
 bIsFirstActivation {type=boolean, default=true}
 LatestHoveredWidget {type=UserWidget, default=null}
 OffText {type=string, default=Off}
 OnText {type=string, default=On}
]
widgets {
 info {
  inheritedWidgetCount=1
  namedSlotCount=3
  parentClass=/Script/FortniteUI.FortSpectatorCameraSettingsTab
  rootWidgetClass=&Overlay
  widgetCount=27
 }
 widgets [
  Overlay_0 {widgetClassPath=&Overlay, widget=Overlay_0}
  CommonBorder_0 {
   widgetClassPath=&CommonBorder
   parent=Overlay_0
   widget=CommonBorder_0
   properties {horizontalAlignment=HAlign_Center, padding=<ltrb 0 0 0 0>}
   slot {class=OverlaySlot}
  }
  SizeBox_0 {
   widgetClassPath=&SizeBox
   parent=CommonBorder_0
   widget=SizeBox_0
   properties {bOverride_HeightOverride=true, bOverride_WidthOverride=true, heightOverride=470, maxDesiredHeight=410, widthOverride=700}
   slot {class=BorderSlot, properties {horizontalAlignment=HAlign_Center, padding=<ltrb 0 0 0 0>}}
  }
  ScrollBox_Tabs {
   widgetClassPath=&ScrollBox
   bIsVariable=true
   parent=SizeBox_0
   widget=ScrollBox_Tabs
   properties {
    allowOverscroll=false
    bAnimateWheelScrolling=true
    consumeMouseWheel=Always
    navigationDestination=Center
    scrollWhenFocusChanges=AnimatedScroll
    scrollbarPadding=<ltrb 0 10 10 10>
    scrollbarThickness=<15 15>
    widgetBarStyle {
     thickness=12
     draggedThumbImage {
      bIsDynamicallyLoaded=false
      drawAs=RoundedBox
      imageType=FullColor
      mirroring=NoMirror
      resourceName=../../../Engine/Content/Slate/Common/Scrollbar_Thumb.png
      resourceObject=null
      tiling=NoTile
      imageSize=<12 12>
      margin=<ltrb 0.25 0.25 0.25 0.25>
      outlineSettings {
       bUseBrushTransparency=false
       roundingType=FixedRadius
       width=2
       color {colorUseRule=UseColor_Specified, specifiedColor=<rgba 0 0 0 0>}
       cornerRadii=<6 6 6 6>
      }
      tintColor {colorUseRule=UseColor_Specified, specifiedColor=<rgba 1 1 1 1>}
      uVRegion {bIsValid=false, max=<0 0>, min=<0 0>}
     }
     horizontalBackgroundImage {
      bIsDynamicallyLoaded=false
      drawAs=Box
      imageType=FullColor
      mirroring=NoMirror
      resourceName=None
      resourceObject=/FortUILibrary/UIKit/ScrollBar/MaterialInstances/MI_UIKit_ScrollbarBackground
      tiling=NoTile
      imageSize=<12 12>
      margin=<ltrb 0.5 0.5 0.5 0.5>
      outlineSettings {
       bUseBrushTransparency=false
       roundingType=HalfHeightRadius
       width=0
       color {colorUseRule=UseColor_Specified, specifiedColor=<rgba 0 0 0 0>}
       cornerRadii=<0 0 0 0>
      }
      tintColor {colorUseRule=UseColor_Specified, specifiedColor=<rgba 1 1 1 1>}
      uVRegion {bIsValid=false, max=<0 0>, min=<0 0>}
     }
     horizontalBottomSlotImage {
      bIsDynamicallyLoaded=false
      drawAs=NoDrawType
      imageType=FullColor
      mirroring=NoMirror
      resourceName=../../../Engine/Content/Slate/Common/Scrollbar_Background_Horizontal.png
      resourceObject=null
      tiling=NoTile
      imageSize=<8 8>
      margin=<ltrb 0 0 0 0>
      outlineSettings {
       bUseBrushTransparency=false
       roundingType=HalfHeightRadius
       width=0
       color {colorUseRule=UseColor_Specified, specifiedColor=<rgba 0 0 0 0>}
       cornerRadii=<0 0 0 0>
      }
      tintColor {colorUseRule=UseColor_Specified, specifiedColor=<rgba 1 1 1 1>}
      uVRegion {bIsValid=false, max=<0 0>, min=<0 0>}
     }
     horizontalTopSlotImage {
      bIsDynamicallyLoaded=false
      drawAs=NoDrawType
      imageType=FullColor
      mirroring=NoMirror
      resourceName=../../../Engine/Content/Slate/Common/Scrollbar_Background_Horizontal.png
      resourceObject=null
      tiling=NoTile
      imageSize=<8 8>
      margin=<ltrb 0 0 0 0>
      outlineSettings {
       bUseBrushTransparency=false
       roundingType=HalfHeightRadius
       width=0
       color {colorUseRule=UseColor_Specified, specifiedColor=<rgba 0 0 0 0>}
       cornerRadii=<0 0 0 0>
      }
      tintColor {colorUseRule=UseColor_Specified, specifiedColor=<rgba 1 1 1 1>}
      uVRegion {bIsValid=false, max=<0 0>, min=<0 0>}
     }
     hoveredThumbImage {
      bIsDynamicallyLoaded=false
      drawAs=RoundedBox
      imageType=FullColor
      mirroring=NoMirror
      resourceName=../../../Engine/Content/Slate/Common/Scrollbar_Thumb.png
      resourceObject=null
      tiling=NoTile
      imageSize=<12 12>
      margin=<ltrb 0.25 0.25 0.25 0.25>
      outlineSettings {
       bUseBrushTransparency=false
       roundingType=FixedRadius
       width=2
       color {colorUseRule=UseColor_Specified, specifiedColor=<rgba 0 0 0 0>}
       cornerRadii=<6 6 6 6>
      }
      tintColor {colorUseRule=UseColor_Specified, specifiedColor=<rgba 1 1 1 1>}
      uVRegion {bIsValid=false, max=<0 0>, min=<0 0>}
     }
     normalThumbImage {
      bIsDynamicallyLoaded=false
      drawAs=RoundedBox
      imageType=FullColor
      mirroring=NoMirror
      resourceName=../../../Engine/Content/Slate/Common/Scrollbar_Thumb.png
      resourceObject=null
      tiling=NoTile
      imageSize=<12 12>
      margin=<ltrb 0.25 0.25 0.25 0.25>
      outlineSettings {
       bUseBrushTransparency=false
       roundingType=FixedRadius
       width=2
       color {colorUseRule=UseColor_Specified, specifiedColor=<rgba 0 0 0 0>}
       cornerRadii=<6 6 6 6>
      }
      tintColor {colorUseRule=UseColor_Specified, specifiedColor=<rgba 1 1 1 0.75>}
      uVRegion {bIsValid=false, max=<0 0>, min=<0 0>}
     }
     verticalBackgroundImage {
      bIsDynamicallyLoaded=false
      drawAs=Box
      imageType=NoImage
      mirroring=NoMirror
      resourceName=None
      resourceObject=&MI_UIKit_ScrollbarBackground
      tiling=NoTile
      imageSize=<12 12>
      margin=<ltrb 0.5 0.5 0.5 0.5>
      outlineSettings {
       bUseBrushTransparency=false
       roundingType=HalfHeightRadius
       width=0
       color {colorUseRule=UseColor_Specified, specifiedColor=<rgba 0 0 0 0>}
       cornerRadii=<0 0 0 0>
      }
      tintColor {colorUseRule=UseColor_Specified, specifiedColor=<rgba 1 1 1 1>}
      uVRegion {bIsValid=false, max=<0 0>, min=<0 0>}
     }
     verticalBottomSlotImage {
      bIsDynamicallyLoaded=false
      drawAs=NoDrawType
      imageType=FullColor
      mirroring=NoMirror
      resourceName=../../../Engine/Content/Slate/Common/Scrollbar_Background_Vertical.png
      resourceObject=null
      tiling=NoTile
      imageSize=<8 8>
      margin=<ltrb 0 0 0 0>
      outlineSettings {
       bUseBrushTransparency=false
       roundingType=HalfHeightRadius
       width=0
       color {colorUseRule=UseColor_Specified, specifiedColor=<rgba 0 0 0 0>}
       cornerRadii=<0 0 0 0>
      }
      tintColor {colorUseRule=UseColor_Specified, specifiedColor=<rgba 1 1 1 1>}
      uVRegion {bIsValid=false, max=<0 0>, min=<0 0>}
     }
     verticalTopSlotImage {
      bIsDynamicallyLoaded=false
      drawAs=NoDrawType
      imageType=FullColor
      mirroring=NoMirror
      resourceName=../../../Engine/Content/Slate/Common/Scrollbar_Background_Vertical.png
      resourceObject=null
      tiling=NoTile
      imageSize=<8 8>
      margin=<ltrb 0 0 0 0>
      outlineSettings {
       bUseBrushTransparency=false
       roundingType=HalfHeightRadius
       width=0
       color {colorUseRule=UseColor_Specified, specifiedColor=<rgba 0 0 0 0>}
       cornerRadii=<0 0 0 0>
      }
      tintColor {colorUseRule=UseColor_Specified, specifiedColor=<rgba 1 1 1 1>}
      uVRegion {bIsValid=false, max=<0 0>, min=<0 0>}
     }
    }
    widgetStyle {
     barThickness=12
     bottomShadowBrush {
      bIsDynamicallyLoaded=false
      drawAs=NoDrawType
      imageType=FullColor
      mirroring=NoMirror
      resourceName=../../../Engine/Content/Slate/Common/ScrollBoxShadowBottom.png
      resourceObject=null
      tiling=NoTile
      imageSize=<16 8>
      margin=<ltrb 0.5 0 0.5 1>
      outlineSettings {
       bUseBrushTransparency=false
       roundingType=HalfHeightRadius
       width=0
       color {colorUseRule=UseColor_Specified, specifiedColor=<rgba 0 0 0 0>}
       cornerRadii=<0 0 0 0>
      }
      tintColor {colorUseRule=UseColor_Specified, specifiedColor=<rgba 1 1 1 1>}
      uVRegion {bIsValid=false, max=<0 0>, min=<0 0>}
     }
     horizontalScrolledContentPadding=<ltrb 0 0 1 0>
     leftShadowBrush {
      bIsDynamicallyLoaded=false
      drawAs=NoDrawType
      imageType=FullColor
      mirroring=NoMirror
      resourceName=../../../Engine/Content/Slate/Common/ScrollBoxShadowLeft.png
      resourceObject=null
      tiling=NoTile
      imageSize=<8 16>
      margin=<ltrb 1 0.5 0 0.5>
      outlineSettings {
       bUseBrushTransparency=false
       roundingType=HalfHeightRadius
       width=0
       color {colorUseRule=UseColor_Specified, specifiedColor=<rgba 0 0 0 0>}
       cornerRadii=<0 0 0 0>
      }
      tintColor {colorUseRule=UseColor_Specified, specifiedColor=<rgba 1 1 1 1>}
      uVRegion {bIsValid=false, max=<0 0>, min=<0 0>}
     }
     rightShadowBrush {
      bIsDynamicallyLoaded=false
      drawAs=NoDrawType
      imageType=FullColor
      mirroring=NoMirror
      resourceName=../../../Engine/Content/Slate/Common/ScrollBoxShadowRight.png
      resourceObject=null
      tiling=NoTile
      imageSize=<8 16>
      margin=<ltrb 0 0.5 1 0.5>
      outlineSettings {
       bUseBrushTransparency=false
       roundingType=HalfHeightRadius
       width=0
       color {colorUseRule=UseColor_Specified, specifiedColor=<rgba 0 0 0 0>}
       cornerRadii=<0 0 0 0>
      }
      tintColor {colorUseRule=UseColor_Specified, specifiedColor=<rgba 1 1 1 1>}
      uVRegion {bIsValid=false, max=<0 0>, min=<0 0>}
     }
     topShadowBrush {
      bIsDynamicallyLoaded=false
      drawAs=NoDrawType
      imageType=FullColor
      mirroring=NoMirror
      resourceName=../../../Engine/Content/Slate/Common/ScrollBoxShadowTop.png
      resourceObject=null
      tiling=NoTile
      imageSize=<16 8>
      margin=<ltrb 0.5 1 0.5 0>
      outlineSettings {
       bUseBrushTransparency=false
       roundingType=HalfHeightRadius
       width=0
       color {colorUseRule=UseColor_Specified, specifiedColor=<rgba 0 0 0 0>}
       cornerRadii=<0 0 0 0>
      }
      tintColor {colorUseRule=UseColor_Specified, specifiedColor=<rgba 1 1 1 1>}
      uVRegion {bIsValid=false, max=<0 0>, min=<0 0>}
     }
     verticalScrolledContentPadding=<ltrb 0 0 0 1>
    }
   }
   slot {class=SizeBoxSlot}
  }
  Rotator_PlayerOutlines {
   widgetClassPath=/FortniteCoreUI/UI/Replay/RotatorSelectorReplaySettings
   bInherited=true
   bIsVariable=true
   parent=ScrollBox_Tabs
   widget=Rotator_PlayerOutlines
   slot {class=ScrollBoxSlot, properties {padding=<ltrb 15 15 15 12>}}
  }
  FirstPlaceHighlight {
   widgetClassPath=&RotatorSelectorReplaySettings
   bIsVariable=true
   parent=ScrollBox_Tabs
   widget=FirstPlaceHighlight
   slot {class=ScrollBoxSlot, properties {padding=<ltrb 15 0 15 12>}}
  }
  FortHUDElementWrapper_127 {
   widgetClassPath=/Game/Athena/HUD/FortHUDElementWrapper
   parent=ScrollBox_Tabs
   widget=FortHUDElementWrapper_127
   properties {elementTag {tagName=Athena.Replay.CameraMode.BattleMap}}
   slot {class=ScrollBoxSlot, properties {padding=<ltrb 15 0 15 12>}}
  }
  StormEffects {
   widgetClassPath=&RotatorSelectorReplaySettings
   bIsVariable=true
   namedSlotHost=FortHUDElementWrapper_127
   widget=StormEffects
  }
  OutsideSafeZoneOpacity {
   widgetClassPath=/FortniteCoreUI/UI/Replay/OptionsMenuSliderReplay
   bIsVariable=true
   parent=ScrollBox_Tabs
   widget=OutsideSafeZoneOpacity
   slot {class=ScrollBoxSlot, properties {padding=<ltrb 15 0 15 12>}}
  }
  RelevancyZone {
   widgetClassPath=&RotatorSelectorReplaySettings
   bIsVariable=true
   parent=ScrollBox_Tabs
   widget=RelevancyZone
   slot {class=ScrollBoxSlot, properties {padding=<ltrb 15 0 15 12>}}
  }
  HighQualityFX {
   widgetClassPath=&RotatorSelectorReplaySettings
   bIsVariable=true
   parent=ScrollBox_Tabs
   widget=HighQualityFX
   slot {class=ScrollBoxSlot, properties {padding=<ltrb 15 0 15 12>}}
  }
  DamageEffects {
   widgetClassPath=&RotatorSelectorReplaySettings
   bIsVariable=true
   parent=ScrollBox_Tabs
   widget=DamageEffects
   slot {class=ScrollBoxSlot, properties {padding=<ltrb 15 0 15 12>}}
  }
  TPAutoFollow {
   widgetClassPath=&RotatorSelectorReplaySettings
   bIsVariable=true
   parent=ScrollBox_Tabs
   widget=TPAutoFollow
   slot {class=ScrollBoxSlot, properties {padding=<ltrb 15 0 15 12>}}
  }
  TPDistanceToSubject {
   widgetClassPath=&OptionsMenuSliderReplay
   bIsVariable=true
   parent=ScrollBox_Tabs
   widget=TPDistanceToSubject
   slot {class=ScrollBoxSlot, properties {padding=<ltrb 15 0 15 12>}}
  }
  ShareLensSettings {
   widgetClassPath=&RotatorSelectorReplaySettings
   bIsVariable=true
   parent=ScrollBox_Tabs
   widget=ShareLensSettings
   slot {class=ScrollBoxSlot, properties {padding=<ltrb 15 0 15 12>}}
  }
  BlendSettings {
   widgetClassPath=&RotatorSelectorReplaySettings
   bIsVariable=true
   parent=ScrollBox_Tabs
   widget=BlendSettings
   properties {visibility=Collapsed}
   slot {class=ScrollBoxSlot, properties {padding=<ltrb 15 0 15 12>}}
  }
  TPCameraCollision {
   widgetClassPath=&RotatorSelectorReplaySettings
   bIsVariable=true
   parent=ScrollBox_Tabs
   widget=TPCameraCollision
   slot {class=ScrollBoxSlot, properties {padding=<ltrb 15 0 15 12>}}
  }
  DroneIgnoreJump {
   widgetClassPath=&RotatorSelectorReplaySettings
   bIsVariable=true
   parent=ScrollBox_Tabs
   widget=DroneIgnoreJump
   slot {class=ScrollBoxSlot, properties {padding=<ltrb 15 0 15 12>}}
  }
  BattleMapWrapper {
   widgetClassPath=&FortHUDElementWrapper
   parent=ScrollBox_Tabs
   widget=BattleMapWrapper
   properties {
    visibility=Visible
    elementTag {tagName=Athena.Replay.HUD}
    hUDElementTag {gameplayTags [
      {tagName=Athena.Replay.CameraMode.ThirdPerson}
      {tagName=Athena.Replay.CameraMode.ARDrone}
      {tagName=Athena.Replay.CameraMode.DroneAttach}
      {tagName=Athena.Replay.CameraMode.DroneFollow}
      {tagName=Athena.Replay.CameraMode.DroneFree}
      {tagName=Athena.Replay.CameraMode.GamePlay}
      {tagName=Athena.Replay.CameraMode.FollowShot}
      {tagName=Athena.Replay.CameraMode.ReverseShot}
     ]}
   }
   slot {class=ScrollBoxSlot}
  }
  VerticalBox_0 {widgetClassPath=&VerticalBox, namedSlotHost=BattleMapWrapper, widget=VerticalBox_0}
  BattleMapHeader {
   widgetClassPath=&CommonTextBlock
   parent=VerticalBox_0
   widget=BattleMapHeader
   properties {
    text=BATTLE MAP
    font {
     bForceMonospaced=false
     bMaterialIsStencil=false
     fontMaterial=null
     fontObject=&HeadingNow
     letterSpacing=0
     monospacedWidth=1
     size=21.75
     skewAmount=0
     typefaceFontName=86 Bold Italic
     outlineSettings {
      bApplyOutlineToDropShadows=false
      bMiteredCorners=false
      bSeparateFillAlpha=false
      outlineMaterial=null
      outlineSize=0
      outlineColor=<rgba 0 0 0 1>
     }
    }
    shadowOffset=<0 0>
   }
   slot {class=VerticalBoxSlot, properties {padding=<ltrb 15 10 0 10>}}
  }
  BattleMapShowTimeIndicator {
   widgetClassPath=&RotatorSelectorReplaySettings
   bIsVariable=true
   parent=VerticalBox_0
   widget=BattleMapShowTimeIndicator
   slot {class=VerticalBoxSlot, properties {padding=<ltrb 15 0 15 12>}}
  }
  BattleMapShowNearbyChests {
   widgetClassPath=&RotatorSelectorReplaySettings
   bIsVariable=true
   parent=VerticalBox_0
   widget=BattleMapShowNearbyChests
   slot {class=VerticalBoxSlot, properties {padding=<ltrb 15 0 15 12>}}
  }
  FortHUDElementWrapper {
   widgetClassPath=&FortHUDElementWrapper
   parent=VerticalBox_0
   widget=FortHUDElementWrapper
   properties {elementTag {tagName=Athena.BattleMap.NoSquadMember}}
   slot {class=VerticalBoxSlot, properties {padding=<ltrb 15 0 15 12>}}
  }
  BattleMapShowAllSquadMembers {
   widgetClassPath=&RotatorSelectorReplaySettings
   bIsVariable=true
   namedSlotHost=FortHUDElementWrapper
   widget=BattleMapShowAllSquadMembers
  }
  BattleMapAutomaticCameraPlacement {
   widgetClassPath=&RotatorSelectorReplaySettings
   bIsVariable=true
   parent=VerticalBox_0
   widget=BattleMapAutomaticCameraPlacement
   slot {class=VerticalBoxSlot, properties {padding=<ltrb 15 0 15 12>}}
  }
  SessionIDwatermark {
   widgetClassPath=&RotatorSelectorReplaySettings
   bIsVariable=true
   parent=ScrollBox_Tabs
   widget=SessionIDwatermark
   slot {class=ScrollBoxSlot, properties {padding=<ltrb 15 0 15 15>}}
  }
 ]
}
