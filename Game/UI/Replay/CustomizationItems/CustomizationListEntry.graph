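# CustomizationListEntry is a CommonUserWidget list row that shows the name of one customization item in the replay UI.
# When its list item object is set, it casts to CustomizationListItemObject, caches it in ObjectDataHolder and sets the centered text to the DisplayName from its ItemStructure.
# The layout also holds a right-side named slot and an input action widget, neither driven by the graph.
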WidgetBlueprint /Game/UI/Replay/CustomizationItems/CustomizationListEntry
graphs [EventGraph {nodes [
   n1 = Event OnListItemObjectSet()
   > n2 = Cast To CustomizationListItemObject!(execute, Object=@n1.ListItemObject) -> (then, CastFailed, AsCustomization List Item Object:Customization List Item Object Object Reference) {is=title K2Node_DynamicCast}
   > n3 = Set ObjectDataHolder(ObjectDataHolder:Customization List Item Object Object Reference=@n2) -> (Output_Get:Customization List Item Object Object Reference)
   > Set Text(
    Text=Break Customization Data Structure!(CustomizationDataStructure:"Customization Data Structure Structure (by ref)"=Get ItemStructure!(self:Customization List Item Object Object Reference=@n3) -> (ItemStructure:Customization Data Structure Structure)) -> (
     DisplayName_7_#1:Text
     PersistentName_10_#2:String
     Asset_13_#3:Asset Data Structure
    ) {is=title K2Node_BreakStruct}
    self=Get CenterButtonTextWidget() -> (CenterButtonTextWidget:Common Text Object Reference)
   )
   Event PreConstruct()
   Event Construct()
   Event Tick()
  ]}]
variables [ObjectDataHolder {type=CustomizationListItemObject_C, default=null}]
widgets {
 info {inheritedWidgetCount=0, parentClass=&CommonUserWidget, rootWidgetClass=&Border, widgetCount=7}
 widgets [
  ContentBorder {
   widgetClassPath=&Border
   bIsVariable=true
   widget=ContentBorder
   properties {
    verticalAlignment=VAlign_Center
    background {
     bIsDynamicallyLoaded=false
     drawAs=NoDrawType
     imageType=NoImage
     mirroring=NoMirror
     resourceName=None
     resourceObject=null
     tiling=NoTile
     imageSize=<32 32>
     margin=<ltrb 0 0 0 0>
     outlineSettings {
      bUseBrushTransparency=false
      roundingType=HalfHeightRadius
      width=0
      color {colorUseRule=UseColor_Specified, specifiedColor=<rgba 0 0 0 0>}
      cornerRadii=<0 0 0 0>
     }
     tintColor {colorUseRule=UseColor_Specified, specifiedColor=<rgba 1 1 1 1>}
     uVRegion {bIsValid=false, max=<0 0>, min=<0 0>}
    }
    padding=<ltrb 0 0 0 0>
   }
  }
  HorizontalBox_261 {
   widgetClassPath=&HorizontalBox
   parent=ContentBorder
   widget=HorizontalBox_261
   slot {class=BorderSlot, properties {verticalAlignment=VAlign_Center, padding=<ltrb 0 0 0 0>}}
  }
  Background {
   widgetClassPath=&Border
   bIsVariable=true
   parent=HorizontalBox_261
   widget=Background
   properties {brushColor=<rgba 0.0176 0.0232 0.0331 0>, padding=<ltrb 12 4 12 4>}
   slot {class=HorizontalBoxSlot, properties {size {sizeRule=Fill, value=1}}}
  }
  HorizontalBox_0 {
   widgetClassPath=&HorizontalBox
   parent=Background
   widget=HorizontalBox_0
   slot {class=BorderSlot, properties {padding=<ltrb 12 4 12 4>}}
  }
  CenterButtonTextWidget {
   widgetClassPath=&CommonTextBlock
   bIsVariable=true
   parent=HorizontalBox_0
   widget=CenterButtonTextWidget
   properties {
    justification=Center
    scrollStyle=&TextScrollStyle-Base_C
    style=/FortUILibrary/Styles/Typography/Foundation/TextStyle-Base-S-B.TextStyle-Base-S-B_C
    text=Asset Name
    font {
     bForceMonospaced=false
     bMaterialIsStencil=false
     fontMaterial=null
     fontObject=&NotoSans
     letterSpacing=0
     monospacedWidth=1
     size=18
     skewAmount=0
     typefaceFontName=NotoSans-Bold
     outlineSettings {
      bApplyOutlineToDropShadows=false
      bMiteredCorners=false
      bSeparateFillAlpha=false
      outlineMaterial=null
      outlineSize=0
      outlineColor=<rgba 0 0 0 1>
     }
    }
    shadowOffset=<0 0>
   }
   slot {class=HorizontalBoxSlot, properties {padding=<ltrb 4 0 0 0>, size {sizeRule=Fill, value=1}}}
  }
  RightExtraContentSlot {
   widgetClassPath=&NamedSlot
   bIsVariable=true
   parent=HorizontalBox_261
   widget=RightExtraContentSlot
   properties {bExposeOnInstanceOnly=true}
   slot {class=HorizontalBoxSlot}
  }
  InputActionWidget {
   widgetClassPath=&CommonActionWidget
   bIsVariable=true
   parent=HorizontalBox_261
   widget=InputActionWidget
   slot {class=HorizontalBoxSlot, properties {horizontalAlignment=HAlign_Center, verticalAlignment=VAlign_Center}}
  }
 ]
}
