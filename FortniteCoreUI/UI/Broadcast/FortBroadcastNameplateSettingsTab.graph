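# Broadcast and replay spectator settings tab for nameplates, squad ID, session ranks, Augments and HUD preset.
# Rotators and mapped sliders call native parent setters, whose change events refresh the controls and collapse dependent rows.
# The last hovered or focused row drives per-setting reset and gamepad focus.
# A CVar-gated, non-shipping replay emote search filters assets and applies the clicked emote.
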WidgetBlueprint /FortniteCoreUI/UI/Broadcast/FortBroadcastNameplateSettingsTab
graphs [
 CloseTabGroup {nodes [
   CloseTabGroup() {is=title K2Node_FunctionEntry}
   > SetTabCollapsed!(
    execute
    self:Fort Replay Context Object Reference=Get FortReplayContext!() -> (ReturnValue:Fort Replay Context Object Reference) {is=title K2Node_GetBlueprintContextEx}
    bIsCollapsed:bool=true
   ) -> (then)
  ]}
 EventGraph {nodes [
   n1 = "Selection Changed (SquadID)"!() -> (OutputDelegate, then, Selected Index:int=0) {is=title K2Node_ComponentBoundEvent}
   > SetSquadIdMode!(
    execute
    self:Fort Spectator Nameplate Settings Tab Object Reference
    SquadIdMode:ESpectatorSquadIdMode Enum=Byte to Enum ESpectatorSquadIdMode!(Byte:Byte="To Byte (Integer)"!(InInt:int=@n1.Selected Index) -> (ReturnValue:Byte=0)) -> (ReturnValue:ESpectatorSquadIdMode Enum) {is=title K2Node_CastByteToEnum}
   ) -> (then)
   n2 = "Selection Changed (Rotator_NamePlatesEnabled)"!() -> (OutputDelegate, then, Selected Index:int=0) {is=title K2Node_ComponentBoundEvent}
   > SetNameplatesEnabled!(
    execute
    self:Fort Spectator Nameplate Settings Tab Object Reference
    bNameplatesEnabled:bool="integer > integer"(A=@n2.Selected Index, B=0) {is=title K2Node_CallFunction}
   ) -> (then)
   n3 = "Selection Changed (TeamColor)"!() -> (OutputDelegate, then, Selected Index:int=0) {is=title K2Node_ComponentBoundEvent}
   > SetShowTeamColor!(execute, self:Fort Spectator Nameplate Settings Tab Object Reference, bShowTeamColor:bool="integer > integer"(A=@n3.Selected Index, B=0)) -> (then)
   n4 = "Selection Changed (ArrowColorEnabled)"!() -> (OutputDelegate, then, Selected Index:int=0) {is=title K2Node_ComponentBoundEvent}
   > SetArrowColorEnabled!(execute, self:Fort Spectator Nameplate Settings Tab Object Reference, bArrowColorEnabled:bool="integer > integer"(A=@n4.Selected Index, B=0)) -> (then)
   n5 = "Slider Changed (LowDetailScaleMax)"!() -> (OutputDelegate, then, Slider Value:double) {is=title K2Node_ComponentBoundEvent}
   > SetLowDetailScaleMax!(
    execute
    self:Fort Spectator Nameplate Settings Tab Object Reference
    LowDetailScaleMax:float=Slider Value to Scale!(SliderValue:double=@n5.Slider Value) -> (MappedScale:double=~) {is=title K2Node_MacroInstance}
   ) -> (then)
   n6 = "Slider Changed (ArrowScaleMax)"!() -> (OutputDelegate, then, Slider Value:double) {is=title K2Node_ComponentBoundEvent}
   > SetArrowScaleMax!(
    execute
    self:Fort Spectator Nameplate Settings Tab Object Reference
    ArrowScaleMax:float=Slider Value to Scale(SliderValue=@n6.Slider Value) -> (MappedScale=~)
   ) -> (then)
   n7 = "Selection Changed (Rotator_DisplaySessionRanks)"!() -> (OutputDelegate, then, Selected Index:int=0) {is=title K2Node_ComponentBoundEvent}
   > SetDisplaySessionRanks!(
    execute
    self:Fort Spectator Nameplate Settings Tab Object Reference
    bDisplaySessionRanks:bool="integer > integer"(A=@n7.Selected Index, B=0)
   ) -> (then)
   n8 = "Selection Changed (Rotator_DisplayAugmentsHUD)"!() -> (OutputDelegate, then, Selected Index:int=0) {is=title K2Node_ComponentBoundEvent}
   > Update Row Selector!(
    execute
    self:Rotator Selector Replay Settings Object Reference=Get Rotator_DisplayAugmentsHUD() -> (Rotator_DisplayAugmentsHUD:Rotator Selector Replay Settings Object Reference)
    Currently Selected:int=@n8.Selected Index
   ) -> (then)
   > SetAugmentsHUDVisibility!(
    execute
    self:Fort Spectator Nameplate Settings Tab Object Reference
    bAugmentsHUDVisibility:bool="To Boolean (Integer)"!(InInt:int=@n8.Selected Index) -> (ReturnValue:bool)
   ) -> (then)
   "On Hovered (Rotator_NameplatesEnabled)"() {is=title K2Node_ComponentBoundEvent}
   > Set LatestHoveredWidget(LatestHoveredWidget:User Widget Object Reference=Get Rotator_NamePlatesEnabled() -> (Rotator_NamePlatesEnabled:Common User Widget Object Reference)) -> (Output_Get:User Widget Object Reference)
   "On Hovered (TeamColor)"() {is=title K2Node_ComponentBoundEvent}
   > Set LatestHoveredWidget(LatestHoveredWidget=Get TeamColor() -> (TeamColor:Rotator Selector Replay Settings Object Reference))
   "On Focused (LowDetailDistance)"() {is=title K2Node_ComponentBoundEvent}
   > Set LatestHoveredWidget(LatestHoveredWidget=Get LowDetailDistance() -> (LowDetailDistance:Options Menu Slider Replay Object Reference))
   "On Focused (ArrowDistance)"() {is=title K2Node_ComponentBoundEvent}
   > Set LatestHoveredWidget(LatestHoveredWidget=Get ArrowDistance() -> (ArrowDistance:Options Menu Slider Replay Object Reference))
   n9 = "Slider Changed (OffDistance)"!() -> (OutputDelegate, then, Slider Value:double) {is=title K2Node_ComponentBoundEvent}
   > SetViewDistance!(
    execute
    self:Fort Spectator Nameplate Settings Tab Object Reference
    ViewDistance:float=Slider Value to Distance!(SliderValue:double=@n9.Slider Value) -> (MappedDistance:double=~) {is=title K2Node_MacroInstance}
   ) -> (then)
   "On Focused (OffDistance)"() {is=title K2Node_ComponentBoundEvent}
   > Set LatestHoveredWidget(LatestHoveredWidget=Get OffDistance() -> (OffDistance:Options Menu Slider Replay Object Reference))
   "On Hovered (ScalingOptionsEnabled)"() {is=title K2Node_ComponentBoundEvent}
   > Set LatestHoveredWidget(LatestHoveredWidget=Get ScalingOptionsEnabled() -> (ScalingOptionsEnabled:Rotator Selector Replay Settings Object Reference))
   "On Focused (HighDetailScaleMin)"() {is=title K2Node_ComponentBoundEvent}
   > Set LatestHoveredWidget(LatestHoveredWidget=Get HighDetailScaleMin() -> (HighDetailScaleMin:Options Menu Slider Replay Object Reference))
   "On Focused (HighDetailScaleMax)"() {is=title K2Node_ComponentBoundEvent}
   > Set LatestHoveredWidget(LatestHoveredWidget=Get HighDetailScaleMax() -> (HighDetailScaleMax:Options Menu Slider Replay Object Reference))
   "On Focused (LowDetailScaleMin)"() {is=title K2Node_ComponentBoundEvent}
   > Set LatestHoveredWidget(LatestHoveredWidget=Get LowDetailScaleMin() -> (LowDetailScaleMin:Options Menu Slider Replay Object Reference))
   "On Focused (LowDetailScaleMax)"() {is=title K2Node_ComponentBoundEvent}
   > Set LatestHoveredWidget(LatestHoveredWidget=Get LowDetailScaleMax() -> (LowDetailScaleMax:Options Menu Slider Replay Object Reference))
   "On Focused (ArrowScaleMin)"() {is=title K2Node_ComponentBoundEvent}
   > Set LatestHoveredWidget(LatestHoveredWidget=Get ArrowScaleMin() -> (ArrowScaleMin:Options Menu Slider Replay Object Reference))
   "On Focused (ArrowScaleMax)"() {is=title K2Node_ComponentBoundEvent}
   > Set LatestHoveredWidget(LatestHoveredWidget=Get ArrowScaleMax() -> (ArrowScaleMax:Options Menu Slider Replay Object Reference))
   "On Hovered (ArrowColorEnabled)"() {is=title K2Node_ComponentBoundEvent}
   > Set LatestHoveredWidget(LatestHoveredWidget=Get ArrowColorEnabled() -> (ArrowColorEnabled:Rotator Selector Replay Settings Object Reference))
   "On Hovered (SquadID)"() {is=title K2Node_ComponentBoundEvent}
   > Set LatestHoveredWidget(LatestHoveredWidget=Get SquadID() -> (SquadID:Rotator Selector Replay Settings Object Reference))
   n10 = "Slider Changed (HighDetailScaleMin)"!() -> (OutputDelegate, then, Slider Value:double) {is=title K2Node_ComponentBoundEvent}
   > SetHighDetailScaleMin!(
    execute
    self:Fort Spectator Nameplate Settings Tab Object Reference
    HighDetailScaleMin:float=Slider Value to Scale(SliderValue=@n10.Slider Value) -> (MappedScale=~)
   ) -> (then)
   "On Hovered (Rotator_DisplaySessionRanks)"() {is=title K2Node_ComponentBoundEvent}
   > Set LatestHoveredWidget(LatestHoveredWidget=Get Rotator_DisplaySessionRanks() -> (Rotator_DisplaySessionRanks:Rotator Selector Replay Settings Object Reference))
   "On Hovered (Rotator_DisplayAugmentsHUD)"() {is=title K2Node_ComponentBoundEvent}
   > Set LatestHoveredWidget(LatestHoveredWidget=Get Rotator_DisplayAugmentsHUD())
   n11 = "On Item Clicked (EmotesList)"!() -> (OutputDelegate, then, Item:Object Reference) {is=title K2Node_ComponentBoundEvent}
   > n12 = Cast To CustomizationListItemObject!(execute, Object=@n11.Item) -> (then, CastFailed, AsCustomization List Item Object:Customization List Item Object Object Reference) {is=title K2Node_DynamicCast}
   n13 = Break Customization Data Structure!(CustomizationDataStructure:"Customization Data Structure Structure (by ref)"=Get ItemStructure!(self:Customization List Item Object Object Reference=@n12) -> (ItemStructure:Customization Data Structure Structure)) -> (
    DisplayName_7_#1:Text
    PersistentName_10_#2:String
    Asset_13_#3:Asset Data Structure
   ) {is=title K2Node_BreakStruct}
   n14 = Get Search_SwapEmote() -> (Search_SwapEmote:WBP UIKit Input Field Object Reference)
   > SetHintText!(execute, self:WBP UIKit Input Field Base Object Reference=@n14, Hint Text:Text=@n13) -> (then)
   > Set Text(Text=@n13, self:WBP UIKit Input Field Base Object Reference=@n14)
   > ApplyEmote!(execute, self:Fort Spectator Nameplate Settings Tab Object Reference, EmoteName:String="To String (Text)"(InText=@n13)) -> (then)
   > ClearListItems(self=Get EmotesList() -> (EmotesList:Fort Item List View Object Reference))
   > SetVisibility(self=Get EmoteExtension() -> (EmoteExtension:Size Box Object Reference), InVisibility=Collapsed)
   n15 = "Selection Changed (Rotator_HudVisibility)"!() -> (OutputDelegate, then, Selected Index:int=0) {is=title K2Node_ComponentBoundEvent}
   > n16 = Switch on Int!(execute:Exec, Selection:int=@n15.Selected Index) -> (Default:Exec, 0:Exec, 1:Exec)
   SetHudVisibilityToggleMode!(execute=@n16.0, self:Fort Spectator Nameplate Settings Tab Object Reference, NewMode:EHudVisibilityToggleMode Enum=Replay) -> (then)
   SetHudVisibilityToggleMode(execute=@n16.1, NewMode=Default)
   n17 = "On Text Changed (Search_SwapEmote)"!() -> (OutputDelegate, then, Text:Text) {is=title K2Node_ComponentBoundEvent}
   > ClearListItems(self=Get EmotesList())
   > SetVisibility(self=Get EmoteExtension(), InVisibility=SelfHitTestInvisible)
   n18 = "To String (Text)"(InText=@n17.Text)
   > n19 = Branch(Condition=IsEmpty(InString=@n18))
   n20 = Do Once Then Do!(execute:Exec=@n19.else, Reset:Exec, StartClosed:bool) -> (DoOnce:Exec, ThenDo:Exec) {is=title K2Node_MacroInstance}
   > n21 = PopulateEmoteAssetList(self:Fort Spectator Nameplate Settings Tab Object Reference)
   FilterEmoteAssetList!(execute=[@n21.then, @n20.ThenDo], self:Fort Spectator Nameplate Settings Tab Object Reference, SearchText:String=@n18) -> (then)
   > n22 = For Each Loop(Array:Array of Asset Data Structures=Get FilteredEmoteItemAssetDataArray() -> (FilteredEmoteItemAssetDataArray:Array of Asset Data Structures)) -> (Array Element:Asset Data Structure)
   > n23 = Cast To FortItemDefinition!(execute, Object=GetAsset(InAssetData=@n22)) -> (then, CastFailed, AsFort Item Definition:Fort Item Definition Object Reference) {is=title K2Node_DynamicCast}
   > n24 = Construct Object from Class!(
    execute
    Class=/Game/UI/Replay/CustomizationItems/CustomizationListItemObject.CustomizationListItemObject_C
    self
    ItemStructure:Customization Data Structure Structure=Make Customization Data Structure!(
     DisplayName_7_#1:Text=GetDisplayName!(self:Item Definition Base Object Reference=@n23) -> (ReturnValue:Text)
     PersistentName_10_#2:String
     Asset_13_#3:Asset Data Structure=()
    ) -> (CustomizationDataStructure:Customization Data Structure Structure) {is=title K2Node_MakeStruct}
   ) -> (then, ReturnValue:Customization List Item Object Object Reference)
   > AddItem!(execute, self:List View Object Reference=Get EmotesList(), Item:Object Reference=@n24) -> (then)
   "On Focused (Rotator_HudVisibility)"() {is=title K2Node_ComponentBoundEvent}
   > Set LatestHoveredWidget(LatestHoveredWidget=Get Rotator_HudVisibility() -> (Rotator_HudVisibility:Rotator Selector Replay Settings Object Reference))
   n25 = "On Text Commited (Search_SwapEmote)"!() -> (OutputDelegate, then, Text:Text, Commit method:ETextCommit Enum=Default) {is=title K2Node_ComponentBoundEvent}
   > n26 = Switch on ETextCommit!(execute, Selection:ETextCommit Enum=@n25.Commit method) -> (Default:Exec, OnEnter:Exec, OnUserMovedFocus:Exec, OnCleared:Exec) {is=title K2Node_SwitchEnum}
   Branch(execute=@n26.OnEnter, Condition=TextIsEmpty(InText=@n25.Text))
   > SetVisibility(self=Get EmoteExtension(), InVisibility=Collapsed)
   n27 = "Slider Changed (HighDetailScaleMax)"!() -> (OutputDelegate, then, Slider Value:double) {is=title K2Node_ComponentBoundEvent}
   > SetHighDetailScaleMax!(
    execute
    self:Fort Spectator Nameplate Settings Tab Object Reference
    HighDetailScaleMax:float=Slider Value to Scale(SliderValue=@n27.Slider Value) -> (MappedScale=~)
   ) -> (then)
   n28 = "Selection Changed (ScalingOptionsEnabled)"!() -> (OutputDelegate, then, Selected Index:int=0) {is=title K2Node_ComponentBoundEvent}
   > SetScalingOptionsEnabled!(
    execute
    self:Fort Spectator Nameplate Settings Tab Object Reference
    bScalingOptionsEnabled:bool="integer > integer"(A=@n28.Selected Index, B=0)
   ) -> (then)
   n29 = "Slider Changed (LowDetailDistance)"!() -> (OutputDelegate, then, Slider Value:double) {is=title K2Node_ComponentBoundEvent}
   > SetLowDetailDistance!(
    execute
    self:Fort Spectator Nameplate Settings Tab Object Reference
    LowDetailDistance:float=Slider Value to Distance(SliderValue=@n29.Slider Value) -> (MappedDistance=~)
   ) -> (then)
   n30 = "Slider Changed (ArrowDistance)"!() -> (OutputDelegate, then, Slider Value:double) {is=title K2Node_ComponentBoundEvent}
   > SetArrowDistance!(
    execute
    self:Fort Spectator Nameplate Settings Tab Object Reference
    ArrowDistance:float=Slider Value to Distance(SliderValue=@n30.Slider Value) -> (MappedDistance=~)
   ) -> (then)
   n31 = "Slider Changed (LowDetailScaleMin)"!() -> (OutputDelegate, then, Slider Value:double) {is=title K2Node_ComponentBoundEvent}
   > SetLowDetailScaleMin!(
    execute
    self:Fort Spectator Nameplate Settings Tab Object Reference
    LowDetailScaleMin:float=Slider Value to Scale(SliderValue=@n31.Slider Value) -> (MappedScale=~)
   ) -> (then)
   n32 = "Slider Changed (ArrowScaleMin)"!() -> (OutputDelegate, then, Slider Value:double) {is=title K2Node_ComponentBoundEvent}
   > SetArrowScaleMin!(
    execute
    self:Fort Spectator Nameplate Settings Tab Object Reference
    ArrowScaleMin:float=Slider Value to Scale(SliderValue=@n32.Slider Value) -> (MappedScale=~)
   ) -> (then)
   n33 = HandleCloseTab!() -> (OutputDelegate, then, Selected Index:int) {is=title K2Node_CustomEvent}
   > CloseTabGroup() {is=title K2Node_CallFunction}
   n34 = HandleCloseTabSlider!() -> (OutputDelegate, then, Slider Value:double=~) {is=title K2Node_CustomEvent}
   > CloseTabGroup()
   n35 = AdjustScrollBar() {is=title K2Node_CustomEvent}
   n36 = Get CommonInputSubsystem!() -> (ReturnValue:Common Input Object Reference) {is=title K2Node_GetBlueprintContextEx}
   > Is Valid(InputObject=@n36)
   n37 = GetCurrentInputType(self=@n36)
   > n38 = Branch(Condition="Equal (Enum)"(A=@n37, B=Gamepad))
   > SetScrollWhenFocusChanges!(
    execute
    self:Scroll Box Object Reference=Get ScrollBox_Tabs() -> (ScrollBox_Tabs:Scroll Box Object Reference)
    NewScrollWhenFocusChanges:EScrollWhenFocusChanges Enum=AnimatedScroll
   ) -> (then)
   n39 = Get LatestHoveredWidget() -> (LatestHoveredWidget:User Widget Object Reference)
   > n40 = Is Valid(InputObject=@n39)
   n41 = Cast To RotatorSelectorReplaySettings!(execute=@n40.Is Not Valid, Object=Get Rotator_NamePlatesEnabled()) -> (then, CastFailed, AsRotator Selector Replay Settings:Rotator Selector Replay Settings Object Reference) {is=title K2Node_DynamicCast}
   > SetFocus(self=@n41)
   SetFocus(execute=@n40.Is Valid, self=@n39)
   Branch(execute=@n38.else, Condition="Equal (Enum)"(A=@n37, B=MouseAndKeyboard))
   > SetScrollWhenFocusChanges(self=Get ScrollBox_Tabs(), NewScrollWhenFocusChanges=NoScroll)
   n42 = Event OnNameplatesEnabledChanged!() -> (OutputDelegate, then, bNameplatesEnabled:bool) {is=title K2Node_Event}
   > UpdateNameplateSettingsVisibilities!(execute, self, bNameplateSettingsEnabled:bool=@n42.bNameplatesEnabled) -> (then)
   > Update Row Selector(
    self=Cast To RotatorSelectorReplaySettings!(Object=Get Rotator_NamePlatesEnabled()) -> (AsRotator Selector Replay Settings, bSuccess:bool=~)
    Currently Selected=Select(Option 0:int=0, Option 1:int=1, Index=@n42.bNameplatesEnabled) -> (ReturnValue:int)
   )
   Event On Activated()
   > Parent: On Activated() {is=title K2Node_CallParentFunction}
   > n43 = Branch(Condition=Get bIsFirstActivation() -> (bIsFirstActivation:bool))
   > n44 = Cast To RotatorSelectorReplaySettings!(execute:Exec, Object=Get Rotator_NamePlatesEnabled()) -> (then:Exec, CastFailed:Exec, AsRotator Selector Replay Settings)
   > SetFocus(self=@n44)
   > n45 = Set bIsFirstActivation(bIsFirstActivation:bool=false) -> (Output_Get:bool)
   n46 = GetConsoleVariableBoolValue!(execute=[@n45.then, @n43.else], VariableName:String=replay.EnableEmoteCustomization) -> (then, ReturnValue:bool)
   n47 = IsPlayingReplay!() -> (ReturnValue:bool)
   n48 = NOT Boolean(A=IsShippingBuild!(self:Fort Runtime Options Object Reference=GetRuntimeOptions()) -> (ReturnValue:bool))
   > SetVisibility(
    self=Get Search_SwapEmote()
    InVisibility=Select(
     Option 0:ESlateVisibility Enum=Collapsed
     Option 1:ESlateVisibility Enum=Visible
     Index=AND Boolean!(A=@n46, B=@n47, C:bool=@n48) -> (ReturnValue)
    ) -> (ReturnValue:ESlateVisibility Enum)
   )
   > SetNavigationRules()
   > AdjustScrollBar() {is=title K2Node_CallFunction}
   > Bind Event to On Keybinds Changed(
    self:Fort Global UIContext Object Reference=Get FortGlobalUIContext!() -> (ReturnValue:Fort Global UIContext Object Reference) {is=title K2Node_GetBlueprintContextEx}
    Delegate=@n35
   ) {is=title K2Node_AddDelegate}
   n49 = Event OnLowDetailScaleMaxChanged!() -> (OutputDelegate, then, LowDetailScaleMax:float) {is=title K2Node_Event}
   > UpdateSliderValueUnmapped!(execute, self:Options Menu Slider Replay Object Reference=Get LowDetailScaleMax(), UnmappedValue:double=@n49.LowDetailScaleMax) -> (then)
   n50 = Event OnArrowScaleMinChanged!() -> (OutputDelegate, then, ArrowScaleMin:float) {is=title K2Node_Event}
   > UpdateSliderValueUnmapped(self=Get ArrowScaleMin(), UnmappedValue=@n50.ArrowScaleMin)
   n51 = Event OnArrowScaleMaxChanged!() -> (OutputDelegate, then, ArrowScaleMax:float) {is=title K2Node_Event}
   > UpdateSliderValueUnmapped(self=Get ArrowScaleMax(), UnmappedValue=@n51.ArrowScaleMax)
   n52 = Event OnArrowColorEnabledChanged!() -> (OutputDelegate, then, bArrowColorEnabled:bool) {is=title K2Node_Event}
   > Update Row Selector(self=Get ArrowColorEnabled(), Currently Selected=Select(Option 0:int=0, Option 1:int=1, Index=@n52.bArrowColorEnabled) -> (ReturnValue:int))
   n53 = Event OnShowTeamColorChanged!() -> (OutputDelegate, then, bShowTeamColor:bool) {is=title K2Node_Event}
   > Update Row Selector(self=Get TeamColor(), Currently Selected=Select(Option 0=0, Option 1=1, Index=@n53.bShowTeamColor))
   n54 = Event OnSquadIdModeChanged!() -> (OutputDelegate, then, SquadIdMode:ESpectatorSquadIdMode Enum=AlwaysOff) {is=title K2Node_Event}
   > Update Row Selector(self=Get SquadID(), Currently Selected="To Integer (Byte)"!(InByte:Byte=@n54.SquadIdMode) -> (ReturnValue:int=0))
   n55 = Event OnAugmentsHUDVisibilityChanged!() -> (OutputDelegate, then, bShowAugmentsHUD:bool) {is=title K2Node_Event}
   > Update Row Selector(self=Get Rotator_DisplayAugmentsHUD(), Currently Selected="To Integer (Boolean)"!(InBool:bool=@n55.bShowAugmentsHUD) -> (ReturnValue:int=0))
   n56 = Event OnLowDetailScaleMinChanged!() -> (OutputDelegate, then, LowDetailScaleMin:float) {is=title K2Node_Event}
   > UpdateSliderValueUnmapped(self=Get LowDetailScaleMin(), UnmappedValue=@n56.LowDetailScaleMin)
   n57 = Event OnHudVisibilityToggleModeChanged!() -> (OutputDelegate, then, HudVisibilityToggleMode:EHudVisibilityToggleMode Enum=Default) {is=title K2Node_Event}
   > n58 = Switch on EHudVisibilityToggleMode!(execute, Selection:EHudVisibilityToggleMode Enum=@n57.HudVisibilityToggleMode) -> (Default:Exec, Replay:Exec) {is=title K2Node_SwitchEnum}
   n59 = Get Rotator_HudVisibility()
   > Update Row Selector(self=@n59, Currently Selected=1)
   Update Row Selector(execute=@n58.Replay, self=@n59, Currently Selected=0)
   Event PreConstruct()
   > InitializeWidgets()
   > Set NameplateSettingsWidgets(NameplateSettingsWidgets:Array of User Widget Object References=Make Array!(
     "[0]":User Widget Object Reference=Get TeamColor()
     "[1]":User Widget Object Reference=Get LowDetailDistance()
     "[2]":User Widget Object Reference=Get ArrowDistance()
     "[3]":User Widget Object Reference=Get OffDistance()
     "[4]":User Widget Object Reference=Get ScalingOptionsEnabled()
     "[5]":User Widget Object Reference=Get ArrowColorEnabled()
    ) -> (Array:Array of User Widget Object References)) -> (Output_Get:Array of User Widget Object References)
   > Set ScaleSettingsWidgets(ScaleSettingsWidgets:Array of User Widget Object References=Make Array(
     "[0]"=Get ArrowScaleMax()
     "[1]"=Get ArrowScaleMin()
     "[2]"=Get HighDetailScaleMax()
     "[3]"=Get HighDetailScaleMin()
     "[4]"=Get LowDetailScaleMax()
     "[5]"=Get LowDetailScaleMin()
    )) -> (Output_Get:Array of User Widget Object References)
   > n60 = Cast To RotatorSelectorReplaySettings(Object=Get Rotator_NamePlatesEnabled())
   > Set LatestHoveredWidget(LatestHoveredWidget=@n60)
   n61 = Event OnDisplaySessionRanksChanged!() -> (OutputDelegate, then, bDisplaySessionRanks:bool) {is=title K2Node_Event}
   > Update Row Selector(self=Get Rotator_DisplaySessionRanks(), Currently Selected=Select(Option 0=0, Option 1=1, Index=@n61.bDisplaySessionRanks))
   Event Construct()
   > SetNavigationRules()
   > n62 = Cast To RotatorSelectorReplaySettings(Object=Get Rotator_NamePlatesEnabled())
   > Bind Event to Selection Changed(
    self:Rotator Selector Replay Settings Object Reference=[
     @n62
     Get TeamColor()
     Get ScalingOptionsEnabled()
     Get ArrowColorEnabled()
     Get Rotator_HudVisibility()
     Get SquadID()
     Get Rotator_DisplaySessionRanks()
     Get Rotator_DisplayAugmentsHUD()
    ]
    Delegate=@n33
   ) {is=title K2Node_AddDelegate}
   > Bind Event to Slider Changed(
    self:Options Menu Slider Replay Object Reference=[
     Get OffDistance()
     Get HighDetailScaleMin()
     Get HighDetailScaleMax()
     Get LowDetailScaleMin()
     Get LowDetailScaleMax()
     Get ArrowScaleMin()
     Get ArrowScaleMax()
     Get LowDetailDistance()
     Get ArrowDistance()
    ]
    Delegate=@n34
   ) {is=title K2Node_AddDelegate}
   n63 = Event OnViewDistanceChanged!() -> (OutputDelegate, then, ViewDistance:float) {is=title K2Node_Event}
   > UpdateSliderValueUnmapped(self=Get OffDistance(), UnmappedValue=float / float(A=@n63.ViewDistance, B=100) {is=title K2Node_CallFunction})
   Event CenterOnTab() {is=title K2Node_Event}
   > Center On Widget(self:Rotator Selector Replay Settings Object Reference=Cast To RotatorSelectorReplaySettings!(Object=Get Rotator_NamePlatesEnabled()) -> (AsRotator Selector Replay Settings, bSuccess:bool=~))
   n64 = Event OnLowDetailDistanceChanged!() -> (OutputDelegate, then, LowDetailDistance:float) {is=title K2Node_Event}
   > UpdateSliderValueUnmapped(self=Get LowDetailDistance(), UnmappedValue=float / float(A=@n64.LowDetailDistance, B=100))
   n65 = Event OnArrowDistanceChanged!() -> (OutputDelegate, then, ArrowDistance:float) {is=title K2Node_Event}
   > UpdateSliderValueUnmapped(self=Get ArrowDistance(), UnmappedValue=float / float(A=@n65.ArrowDistance, B=100))
   n66 = Event OnScalingOptionsEnabledChanged!() -> (OutputDelegate, then, bScalingOptionsEnabled:bool) {is=title K2Node_Event}
   > UpdateScaleSettingsVisibilities!(
    execute
    self
    bScaleSettingsEnabled:bool=AND Boolean!(A=GetNameplatesEnabled!(self:Fort Spectator Nameplate Settings Tab Object Reference) -> (ReturnValue:bool), B=@n66.bScalingOptionsEnabled) -> (ReturnValue)
   ) -> (then)
   > Update Row Selector(self=Get ScalingOptionsEnabled(), Currently Selected=Select(Option 0=0, Option 1=1, Index=@n66.bScalingOptionsEnabled))
   n67 = Event OnHighDetailScaleMinChanged!() -> (OutputDelegate, then, HighDetailScaleMin:float) {is=title K2Node_Event}
   > UpdateSliderValueUnmapped(self=Get HighDetailScaleMin(), UnmappedValue=@n67.HighDetailScaleMin)
   n68 = Event OnHighDetailScaleMaxChanged!() -> (OutputDelegate, then, HighDetailScaleMax:float) {is=title K2Node_Event}
   > UpdateSliderValueUnmapped(self=Get HighDetailScaleMax(), UnmappedValue=@n68.HighDetailScaleMax)
   Event ResetFocusedElement() {is=title K2Node_Event}
   > Is Valid(InputObject=Get LatestHoveredWidget())
   > n69 = Branch(Condition="Equal (Object)"(A=Get Rotator_NamePlatesEnabled(), B=Get LatestHoveredWidget()))
   > ResetNameplatesEnabled(self:Fort Spectator Nameplate Settings Tab Object Reference)
   n70 = Branch(execute=@n69.else, Condition="Equal (Object)"(A=Get TeamColor(), B=Get LatestHoveredWidget()))
   > ResetShowTeamColor(self:Fort Spectator Nameplate Settings Tab Object Reference)
   n71 = Branch(execute=@n70.else, Condition="Equal (Object)"(A=Get LowDetailDistance(), B=Get LatestHoveredWidget()))
   > ResetLowDetailDistance(self:Fort Spectator Nameplate Settings Tab Object Reference)
   n72 = Branch(execute=@n71.else, Condition="Equal (Object)"(A=Get ArrowDistance(), B=Get LatestHoveredWidget()))
   > ResetArrowDistance(self:Fort Spectator Nameplate Settings Tab Object Reference)
   n73 = Branch(execute=@n72.else, Condition="Equal (Object)"(A=Get OffDistance(), B=Get LatestHoveredWidget()))
   > ResetViewDistance(self:Fort Spectator Nameplate Settings Tab Object Reference)
   n74 = Branch(execute=@n73.else, Condition="Equal (Object)"(A=Get ScalingOptionsEnabled(), B=Get LatestHoveredWidget()))
   > ResetScalingOptionsEnabled(self:Fort Spectator Nameplate Settings Tab Object Reference)
   n75 = Branch(execute=@n74.else, Condition="Equal (Object)"(A=Get HighDetailScaleMin(), B=Get LatestHoveredWidget()))
   > ResetHighDetailScaleMin(self:Fort Spectator Nameplate Settings Tab Object Reference)
   n76 = Branch(execute=@n75.else, Condition="Equal (Object)"(A=Get HighDetailScaleMax(), B=Get LatestHoveredWidget()))
   > ResetHighDetailScaleMax(self:Fort Spectator Nameplate Settings Tab Object Reference)
   n77 = Branch(execute=@n76.else, Condition="Equal (Object)"(A=Get LowDetailScaleMax(), B=Get LatestHoveredWidget()))
   > ResetLowDetailScaleMax(self:Fort Spectator Nameplate Settings Tab Object Reference)
   n78 = Branch(execute=@n77.else, Condition="Equal (Object)"(A=Get LowDetailScaleMin(), B=Get LatestHoveredWidget()))
   > ResetLowDetailScaleMin(self:Fort Spectator Nameplate Settings Tab Object Reference)
   n79 = Branch(execute=@n78.else, Condition="Equal (Object)"(A=Get ArrowScaleMin(), B=Get LatestHoveredWidget()))
   > ResetArrowScaleMin(self:Fort Spectator Nameplate Settings Tab Object Reference)
   n80 = Branch(execute=@n79.else, Condition="Equal (Object)"(A=Get ArrowScaleMax(), B=Get LatestHoveredWidget()))
   > ResetArrowScaleMax(self:Fort Spectator Nameplate Settings Tab Object Reference)
   n81 = Branch(execute=@n80.else, Condition="Equal (Object)"(A=Get ArrowColorEnabled(), B=Get LatestHoveredWidget()))
   > ResetArrowColorEnabled(self:Fort Spectator Nameplate Settings Tab Object Reference)
   n82 = Branch(execute=@n81.else, Condition="Equal (Object)"(A=Get SquadID(), B=Get LatestHoveredWidget()))
   > ResetSquadIdMode(self:Fort Spectator Nameplate Settings Tab Object Reference)
   n83 = Branch(execute=@n82.else, Condition="Equal (Object)"(A=Get Rotator_DisplaySessionRanks(), B=Get LatestHoveredWidget()))
   > ResetDisplaySessionRanks(self:Fort Spectator Nameplate Settings Tab Object Reference)
   Branch(execute=@n83.else, Condition="Equal (Object)"(A=Get Rotator_DisplayAugmentsHUD(), B=Get LatestHoveredWidget()))
   > ResetAugmentsHUDVisibility(self:Fort Spectator Nameplate Settings Tab Object Reference)
  ]}
 InitializeWidgets {nodes [
   InitializeWidgets() {is=title K2Node_FunctionEntry}
   > n1 = Sequence()
   n2 = GetMinMaxNameplateDistance!(self:Fort Spectator Nameplate Settings Tab Object Reference) -> (MinDistance:float, MaxDistance:float)
   > Set Local_MinDistanceMeters(Local_MinDistanceMeters:double=float / float(A=@n2, B=100)) -> (Output_Get:double)
   > Set Local_MaxDistanceMeters(Local_MaxDistanceMeters:double=float / float(A=@n2.MaxDistance, B=100)) -> (Output_Get:double)
   n3 = GetMinMaxNameplateScale!(self:Fort Spectator Nameplate Settings Tab Object Reference) -> (MinScale:float, MaxScale:float)
   > Set Local_MinScale(Local_MinScale:double=@n3) -> (Output_Get:double)
   > Set Local_MaxScale(Local_MaxScale:double=@n3.MaxScale) -> (Output_Get:double)
   Initialize!(
    execute=@n1.then_1
    self:Rotator Selector Replay Settings Object Reference=Cast To RotatorSelectorReplaySettings(Object=Get Rotator_NamePlatesEnabled()) -> (bSuccess=~)
    Display Text:Text="LOCTEXT(\"Name Plates\")"
    Options:Array of Texts=Make Array!("[0]":Text="LOCTEXT(\"Off\")", "[1]":Text="LOCTEXT(\"On\")") -> (Array:Array of Texts)
    Hover Text:Text="LOCTEXT(\"Enable/Disable name plates above players\")"
    Tab Tooltip:Common Text Object Reference
   ) -> (then)
   > Update Slider!(
    execute
    self:Options Menu Slider Replay Object Reference=Get OffDistance()
    Slider Text:Text="LOCTEXT(\"Off Distance\")"
    Slider Value:double=0.5
    Hover Text:Text
    Tooltip Text Block:Common Text Object Reference
    Min Max Value_X:double=Get Local_MinDistanceMeters() -> (Local_MinDistanceMeters:double)
    Min Max Value_Y:double=Get Local_MaxDistanceMeters() -> (Local_MaxDistanceMeters:double)
   ) -> (then)
   > Update Slider(
    self=Get LowDetailDistance()
    Slider Text="LOCTEXT(\"Low Detail Distance\")"
    Slider Value=0.5
    Min Max Value_X=Get Local_MinDistanceMeters()
    Min Max Value_Y=Get Local_MaxDistanceMeters()
   )
   > Update Slider(
    self=Get ArrowDistance()
    Slider Text="LOCTEXT(\"Arrow Distance\")"
    Slider Value=0.5
    Min Max Value_X=Get Local_MinDistanceMeters()
    Min Max Value_Y=Get Local_MaxDistanceMeters()
   )
   > Initialize(
    self=Get ScalingOptionsEnabled()
    Display Text="LOCTEXT(\"Adjust Scaling\")"
    Options=Make Array("[0]"="LOCTEXT(\"Off\")", "[1]"="LOCTEXT(\"On\")")
    Hover Text="LOCTEXT(\"Provides more options for scaling nameplate types\")"
   )
   > Update Slider(
    self=Get HighDetailScaleMin()
    Slider Text="LOCTEXT(\"High Detail Min\")"
    Slider Value=0.5
    Min Max Value_X=Get Local_MinScale() -> (Local_MinScale:double)
    Min Max Value_Y=Get Local_MaxScale() -> (Local_MaxScale:double)
   )
   > Update Slider(
    self=Get HighDetailScaleMax()
    Slider Text="LOCTEXT(\"High Detail Max\")"
    Slider Value=0.5
    Min Max Value_X=Get Local_MinScale()
    Min Max Value_Y=Get Local_MaxScale()
   )
   > Update Slider(
    self=Get LowDetailScaleMin()
    Slider Text="LOCTEXT(\"Low Detail Min\")"
    Slider Value=0.5
    Min Max Value_X=Get Local_MinScale()
    Min Max Value_Y=Get Local_MaxScale()
   )
   > Update Slider(
    self=Get LowDetailScaleMax()
    Slider Text="LOCTEXT(\"Low Detail Max\")"
    Slider Value=0.5
    Min Max Value_X=Get Local_MinScale()
    Min Max Value_Y=Get Local_MaxScale()
   )
   > Update Slider(
    self=Get ArrowScaleMin()
    Slider Text="LOCTEXT(\"Arrow Scale Min\")"
    Slider Value=0.5
    Min Max Value_X=Get Local_MinScale()
    Min Max Value_Y=Get Local_MaxScale()
   )
   > Update Slider(
    self=Get ArrowScaleMax()
    Slider Text="LOCTEXT(\"Arrow Scale Max\")"
    Slider Value=0.5
    Min Max Value_X=Get Local_MinScale()
    Min Max Value_Y=Get Local_MaxScale()
   )
   > Initialize(
    self=Get SquadID()
    Display Text="LOCTEXT(\"Squad ID\")"
    Options=Make Array!("[0]"="LOCTEXT(\"Off\")", "[1]"="LOCTEXT(\"On\")", "[2]":Text="LOCTEXT(\"Hold\")") -> (Array)
    Hover Text="LOCTEXT(\"Enable Team Info on NamePlate\")"
   )
   > Initialize(
    self=Get TeamColor()
    Display Text="LOCTEXT(\"Team Color\")"
    Options=Make Array!("[0]"="LOCTEXT(\"Off\")", "[1]"="LOCTEXT(\"On\")") -> (Array)
    Hover Text="LOCTEXT(\"Show team color on name plate\")"
   )
   > Initialize(
    self=Get ArrowColorEnabled()
    Display Text="LOCTEXT(\"Arrow Team Color\")"
    Options=Make Array("[0]"="LOCTEXT(\"Off\")", "[1]"="LOCTEXT(\"On\")")
    Hover Text="LOCTEXT(\"Show team color on arrow\")"
   )
   > Initialize(
    self=Get Rotator_DisplaySessionRanks()
    Display Text="LOCTEXT(\"Session Ranks\")"
    Options=Make Array("[0]"="LOCTEXT(\"Hide\")", "[1]"="LOCTEXT(\"Show\")")
    Hover Text="LOCTEXT(\"Display player ranks within the event session\")"
   )
   > Update Row Selector(
    self=Get Rotator_DisplaySessionRanks()
    Currently Selected=Select(Option 0=0, Option 1=1, Index=GetDisplaySessionRanks!(self:Fort Spectator Nameplate Settings Tab Object Reference) -> (ReturnValue:bool))
   )
   n4 = Get Rotator_DisplayAugmentsHUD()
   > Initialize(
    self=@n4
    Display Text="LOCTEXT(\"Augments\")"
    Options=Make Array("[0]"="LOCTEXT(\"Hide\")", "[1]"="LOCTEXT(\"Show\")")
    Hover Text="LOCTEXT(\"Display Augments HUD\")"
   )
   > Update Row Selector(
    self=@n4
    Currently Selected=Select(Option 0=0, Option 1=1, Index=GetAugmentsHUDVisibility!(self:Fort Spectator Nameplate Settings Tab Object Reference) -> (ReturnValue:bool))
   )
   > Initialize(
    self=Get Rotator_HudVisibility()
    Display Text="LOCTEXT(\"HUD Preset\")"
    Options=Make Array("[0]"="LOCTEXT(\"Replay\")", "[1]"="LOCTEXT(\"Broadcast\")")
   )
  ]}
 SetNavigationRules {nodes [
   SetNavigationRules() {is=title K2Node_FunctionEntry}
   n1 = GetAllChildren(self=Get ScrollBox_Tabs())
   n2 = "Get (a copy)"(Array:Array of Widget Object References=@n1, Dimension 1=int - int(A=Length(TargetArray:Array of Widget Object References=@n1))) -> (Output:Widget Object Reference)
   n3 = "Get (a copy)"(Array=GetAllChildren(self=Get ScrollBox_Tabs()))
   > SetNavigationRuleExplicit!(execute, self:Widget Object Reference=@n3, Direction:EUINavigation Enum=Up, InWidget:Widget Object Reference=@n2) -> (then)
   > SetNavigationRuleExplicit(self=@n2, Direction=Down, InWidget=@n3)
  ]}
 SetPanelPadding {nodes [
   SetPanelPadding() {is=title K2Node_FunctionEntry}
   > n1 = For Each Loop(Array:Array of Widget Object References=GetAllChildren(self=Get ScrollBox_Tabs())) -> (Array Element:Widget Object Reference)
   > Set Loc_ScrollSlot(Loc_ScrollSlot:Scroll Box Slot Object Reference=SlotAsScrollBoxSlot!(Widget:Widget Object Reference=@n1) -> (ReturnValue:Scroll Box Slot Object Reference)) -> (Output_Get:Scroll Box Slot Object Reference)
   > n2 = Branch(Condition="Equal (Integer)"(A=@n1.Array Index))
   > SetPadding!(
    execute
    self:Scroll Box Slot Object Reference=Get Loc_ScrollSlot() -> (Loc_ScrollSlot:Scroll Box Slot Object Reference)
    InPadding_Left:float=15
    InPadding_Top:float=15
    InPadding_Right:float=15
    InPadding_Bottom:float=12
   ) -> (then)
   n3 = Length(TargetArray=GetAllChildren(self=Get ScrollBox_Tabs()))
   n4 = Branch(execute=@n2.else, Condition="Equal (Integer)"(A=@n1.Array Index, B=int - int(A=@n3)))
   > SetPadding(self=Get Loc_ScrollSlot(), InPadding_Left=15, InPadding_Top=0, InPadding_Right=15, InPadding_Bottom=15)
   SetPadding(execute=@n4.else, self=Get Loc_ScrollSlot(), InPadding_Left=15, InPadding_Top=0, InPadding_Right=15, InPadding_Bottom=12)
  ]}
 SliderValueToDistance {nodes [
   n1 = GetMinMaxNameplateDistance()
   Outputs!(MappedDistance:double=MapRangeClamped(Value=Inputs!() -> (SliderValue:double=~), InRangeA=0, InRangeB=1, OutRangeA=@n1, OutRangeB=@n1.MaxDistance))
  ]}
 SliderValueToScale {nodes [
   n1 = GetMinMaxNameplateScale()
   Outputs!(MappedScale:double=MapRangeClamped(Value=Inputs() -> (SliderValue=~), InRangeA=0, InRangeB=1, OutRangeA=@n1, OutRangeB=@n1.MaxScale))
  ]}
 UpdateNameplateSettingsVisibilities {nodes [
   n1 = UpdateNameplateSettingsVisibilities!() -> (then, bNameplateSettingsEnabled:bool=~) {is=title K2Node_FunctionEntry}
   > n2 = For Each Loop(Array:Array of User Widget Object References=Get NameplateSettingsWidgets() -> (NameplateSettingsWidgets:Array of User Widget Object References)) -> (Array Element:User Widget Object Reference)
   UpdateScaleSettingsVisibilities(
    execute=@n2.Completed
    bScaleSettingsEnabled=AND Boolean(A=@n1, B=GetScalingOptionsEnabled!(self:Fort Spectator Nameplate Settings Tab Object Reference) -> (ReturnValue:bool))
   )
   SetVisibility(
    execute=@n2.LoopBody
    self=@n2
    InVisibility=Select(Option 0:ESlateVisibility Enum=Collapsed, Option 1:ESlateVisibility Enum=Visible, Index=@n1) -> (ReturnValue:ESlateVisibility Enum)
   )
  ]}
 UpdateScaleSettingsVisibilities {nodes [
   n1 = UpdateScaleSettingsVisibilities!() -> (then, bScaleSettingsEnabled:bool=~) {is=title K2Node_FunctionEntry}
   > n2 = For Each Loop(Array=Get ScaleSettingsWidgets() -> (ScaleSettingsWidgets:Array of User Widget Object References))
   > SetVisibility(self=@n2, InVisibility=Select(Option 0=Collapsed, Option 1=Visible, Index=@n1))
  ]}
]
variables [
 bIsFirstActivation {type=boolean, default=true}
 LatestHoveredWidget {type=UserWidget, default=null}
 NameplateSettingsWidgets {type="array<UserWidget>", default []}
 ScaleSettingsWidgets {type="array<UserWidget>", default []}
]
widgets {
 info {
  inheritedWidgetCount=1
  parentClass=/Script/FortniteUI.FortSpectatorNameplateSettingsTab
  rootWidgetClass=&Overlay
  widgetCount=25
 }
 widgets [
  Overlay_0 {widgetClassPath=&Overlay, widget=Overlay_0}
  CommonBorder_0 {
   widgetClassPath=&CommonBorder
   parent=Overlay_0
   widget=CommonBorder_0
   properties {horizontalAlignment=HAlign_Center, padding=<ltrb 0 0 0 0>}
   slot {class=OverlaySlot}
  }
  SizeBox {
   widgetClassPath=&SizeBox
   bIsVariable=true
   parent=CommonBorder_0
   widget=SizeBox
   properties {
    bOverride_HeightOverride=true
    bOverride_MinDesiredWidth=true
    heightOverride=470
    maxDesiredHeight=410
    minDesiredHeight=410
    minDesiredWidth=700
    widthOverride=700
   }
   slot {class=BorderSlot, properties {horizontalAlignment=HAlign_Center, padding=<ltrb 0 0 0 0>}}
  }
  HorizontalBox_0 {widgetClassPath=&HorizontalBox, parent=SizeBox, widget=HorizontalBox_0, slot {class=SizeBoxSlot}}
  ScrollBox_Tabs {
   widgetClassPath=&ScrollBox
   bIsVariable=true
   parent=HorizontalBox_0
   widget=ScrollBox_Tabs
   properties {
    allowOverscroll=false
    bAnimateWheelScrolling=true
    consumeMouseWheel=Always
    navigationDestination=Center
    scrollWhenFocusChanges=AnimatedScroll
    scrollbarPadding=<ltrb 0 10 10 10>
    scrollbarThickness=<15 15>
    widgetBarStyle {
     thickness=12
     draggedThumbImage {
      bIsDynamicallyLoaded=false
      drawAs=RoundedBox
      imageType=FullColor
      mirroring=NoMirror
      resourceName=../../../Engine/Content/Slate/Common/Scrollbar_Thumb.png
      resourceObject=null
      tiling=NoTile
      imageSize=<12 12>
      margin=<ltrb 0.25 0.25 0.25 0.25>
      outlineSettings {
       bUseBrushTransparency=false
       roundingType=FixedRadius
       width=2
       color {colorUseRule=UseColor_Specified, specifiedColor=<rgba 0 0 0 0>}
       cornerRadii=<6 6 6 6>
      }
      tintColor {colorUseRule=UseColor_Specified, specifiedColor=<rgba 1 1 1 1>}
      uVRegion {bIsValid=false, max=<0 0>, min=<0 0>}
     }
     horizontalBackgroundImage {
      bIsDynamicallyLoaded=false
      drawAs=Box
      imageType=FullColor
      mirroring=NoMirror
      resourceName=None
      resourceObject=/FortUILibrary/UIKit/ScrollBar/MaterialInstances/MI_UIKit_ScrollbarBackground
      tiling=NoTile
      imageSize=<12 12>
      margin=<ltrb 0.5 0.5 0.5 0.5>
      outlineSettings {
       bUseBrushTransparency=false
       roundingType=HalfHeightRadius
       width=0
       color {colorUseRule=UseColor_Specified, specifiedColor=<rgba 0 0 0 0>}
       cornerRadii=<0 0 0 0>
      }
      tintColor {colorUseRule=UseColor_Specified, specifiedColor=<rgba 1 1 1 1>}
      uVRegion {bIsValid=false, max=<0 0>, min=<0 0>}
     }
     horizontalBottomSlotImage {
      bIsDynamicallyLoaded=false
      drawAs=NoDrawType
      imageType=FullColor
      mirroring=NoMirror
      resourceName=../../../Engine/Content/Slate/Common/Scrollbar_Background_Horizontal.png
      resourceObject=null
      tiling=NoTile
      imageSize=<8 8>
      margin=<ltrb 0 0 0 0>
      outlineSettings {
       bUseBrushTransparency=false
       roundingType=HalfHeightRadius
       width=0
       color {colorUseRule=UseColor_Specified, specifiedColor=<rgba 0 0 0 0>}
       cornerRadii=<0 0 0 0>
      }
      tintColor {colorUseRule=UseColor_Specified, specifiedColor=<rgba 1 1 1 1>}
      uVRegion {bIsValid=false, max=<0 0>, min=<0 0>}
     }
     horizontalTopSlotImage {
      bIsDynamicallyLoaded=false
      drawAs=NoDrawType
      imageType=FullColor
      mirroring=NoMirror
      resourceName=../../../Engine/Content/Slate/Common/Scrollbar_Background_Horizontal.png
      resourceObject=null
      tiling=NoTile
      imageSize=<8 8>
      margin=<ltrb 0 0 0 0>
      outlineSettings {
       bUseBrushTransparency=false
       roundingType=HalfHeightRadius
       width=0
       color {colorUseRule=UseColor_Specified, specifiedColor=<rgba 0 0 0 0>}
       cornerRadii=<0 0 0 0>
      }
      tintColor {colorUseRule=UseColor_Specified, specifiedColor=<rgba 1 1 1 1>}
      uVRegion {bIsValid=false, max=<0 0>, min=<0 0>}
     }
     hoveredThumbImage {
      bIsDynamicallyLoaded=false
      drawAs=RoundedBox
      imageType=FullColor
      mirroring=NoMirror
      resourceName=../../../Engine/Content/Slate/Common/Scrollbar_Thumb.png
      resourceObject=null
      tiling=NoTile
      imageSize=<12 12>
      margin=<ltrb 0.25 0.25 0.25 0.25>
      outlineSettings {
       bUseBrushTransparency=false
       roundingType=FixedRadius
       width=2
       color {colorUseRule=UseColor_Specified, specifiedColor=<rgba 0 0 0 0>}
       cornerRadii=<6 6 6 6>
      }
      tintColor {colorUseRule=UseColor_Specified, specifiedColor=<rgba 1 1 1 1>}
      uVRegion {bIsValid=false, max=<0 0>, min=<0 0>}
     }
     normalThumbImage {
      bIsDynamicallyLoaded=false
      drawAs=RoundedBox
      imageType=FullColor
      mirroring=NoMirror
      resourceName=../../../Engine/Content/Slate/Common/Scrollbar_Thumb.png
      resourceObject=null
      tiling=NoTile
      imageSize=<12 12>
      margin=<ltrb 0.25 0.25 0.25 0.25>
      outlineSettings {
       bUseBrushTransparency=false
       roundingType=FixedRadius
       width=2
       color {colorUseRule=UseColor_Specified, specifiedColor=<rgba 0 0 0 0>}
       cornerRadii=<6 6 6 6>
      }
      tintColor {colorUseRule=UseColor_Specified, specifiedColor=<rgba 1 1 1 0.75>}
      uVRegion {bIsValid=false, max=<0 0>, min=<0 0>}
     }
     verticalBackgroundImage {
      bIsDynamicallyLoaded=false
      drawAs=Box
      imageType=NoImage
      mirroring=NoMirror
      resourceName=None
      resourceObject=&MI_UIKit_ScrollbarBackground
      tiling=NoTile
      imageSize=<12 12>
      margin=<ltrb 0.5 0.5 0.5 0.5>
      outlineSettings {
       bUseBrushTransparency=false
       roundingType=HalfHeightRadius
       width=0
       color {colorUseRule=UseColor_Specified, specifiedColor=<rgba 0 0 0 0>}
       cornerRadii=<0 0 0 0>
      }
      tintColor {colorUseRule=UseColor_Specified, specifiedColor=<rgba 1 1 1 1>}
      uVRegion {bIsValid=false, max=<0 0>, min=<0 0>}
     }
     verticalBottomSlotImage {
      bIsDynamicallyLoaded=false
      drawAs=NoDrawType
      imageType=FullColor
      mirroring=NoMirror
      resourceName=../../../Engine/Content/Slate/Common/Scrollbar_Background_Vertical.png
      resourceObject=null
      tiling=NoTile
      imageSize=<8 8>
      margin=<ltrb 0 0 0 0>
      outlineSettings {
       bUseBrushTransparency=false
       roundingType=HalfHeightRadius
       width=0
       color {colorUseRule=UseColor_Specified, specifiedColor=<rgba 0 0 0 0>}
       cornerRadii=<0 0 0 0>
      }
      tintColor {colorUseRule=UseColor_Specified, specifiedColor=<rgba 1 1 1 1>}
      uVRegion {bIsValid=false, max=<0 0>, min=<0 0>}
     }
     verticalTopSlotImage {
      bIsDynamicallyLoaded=false
      drawAs=NoDrawType
      imageType=FullColor
      mirroring=NoMirror
      resourceName=../../../Engine/Content/Slate/Common/Scrollbar_Background_Vertical.png
      resourceObject=null
      tiling=NoTile
      imageSize=<8 8>
      margin=<ltrb 0 0 0 0>
      outlineSettings {
       bUseBrushTransparency=false
       roundingType=HalfHeightRadius
       width=0
       color {colorUseRule=UseColor_Specified, specifiedColor=<rgba 0 0 0 0>}
       cornerRadii=<0 0 0 0>
      }
      tintColor {colorUseRule=UseColor_Specified, specifiedColor=<rgba 1 1 1 1>}
      uVRegion {bIsValid=false, max=<0 0>, min=<0 0>}
     }
    }
    widgetStyle {
     barThickness=12
     bottomShadowBrush {
      bIsDynamicallyLoaded=false
      drawAs=NoDrawType
      imageType=FullColor
      mirroring=NoMirror
      resourceName=../../../Engine/Content/Slate/Common/ScrollBoxShadowBottom.png
      resourceObject=null
      tiling=NoTile
      imageSize=<16 8>
      margin=<ltrb 0.5 0 0.5 1>
      outlineSettings {
       bUseBrushTransparency=false
       roundingType=HalfHeightRadius
       width=0
       color {colorUseRule=UseColor_Specified, specifiedColor=<rgba 0 0 0 0>}
       cornerRadii=<0 0 0 0>
      }
      tintColor {colorUseRule=UseColor_Specified, specifiedColor=<rgba 1 1 1 1>}
      uVRegion {bIsValid=false, max=<0 0>, min=<0 0>}
     }
     horizontalScrolledContentPadding=<ltrb 0 0 1 0>
     leftShadowBrush {
      bIsDynamicallyLoaded=false
      drawAs=NoDrawType
      imageType=FullColor
      mirroring=NoMirror
      resourceName=../../../Engine/Content/Slate/Common/ScrollBoxShadowLeft.png
      resourceObject=null
      tiling=NoTile
      imageSize=<8 16>
      margin=<ltrb 1 0.5 0 0.5>
      outlineSettings {
       bUseBrushTransparency=false
       roundingType=HalfHeightRadius
       width=0
       color {colorUseRule=UseColor_Specified, specifiedColor=<rgba 0 0 0 0>}
       cornerRadii=<0 0 0 0>
      }
      tintColor {colorUseRule=UseColor_Specified, specifiedColor=<rgba 1 1 1 1>}
      uVRegion {bIsValid=false, max=<0 0>, min=<0 0>}
     }
     rightShadowBrush {
      bIsDynamicallyLoaded=false
      drawAs=NoDrawType
      imageType=FullColor
      mirroring=NoMirror
      resourceName=../../../Engine/Content/Slate/Common/ScrollBoxShadowRight.png
      resourceObject=null
      tiling=NoTile
      imageSize=<8 16>
      margin=<ltrb 0 0.5 1 0.5>
      outlineSettings {
       bUseBrushTransparency=false
       roundingType=HalfHeightRadius
       width=0
       color {colorUseRule=UseColor_Specified, specifiedColor=<rgba 0 0 0 0>}
       cornerRadii=<0 0 0 0>
      }
      tintColor {colorUseRule=UseColor_Specified, specifiedColor=<rgba 1 1 1 1>}
      uVRegion {bIsValid=false, max=<0 0>, min=<0 0>}
     }
     topShadowBrush {
      bIsDynamicallyLoaded=false
      drawAs=NoDrawType
      imageType=FullColor
      mirroring=NoMirror
      resourceName=../../../Engine/Content/Slate/Common/ScrollBoxShadowTop.png
      resourceObject=null
      tiling=NoTile
      imageSize=<16 8>
      margin=<ltrb 0.5 1 0.5 0>
      outlineSettings {
       bUseBrushTransparency=false
       roundingType=HalfHeightRadius
       width=0
       color {colorUseRule=UseColor_Specified, specifiedColor=<rgba 0 0 0 0>}
       cornerRadii=<0 0 0 0>
      }
      tintColor {colorUseRule=UseColor_Specified, specifiedColor=<rgba 1 1 1 1>}
      uVRegion {bIsValid=false, max=<0 0>, min=<0 0>}
     }
     verticalScrolledContentPadding=<ltrb 0 0 0 1>
    }
   }
   slot {class=HorizontalBoxSlot, properties {size {sizeRule=Fill, value=1}}}
  }
  Rotator_NamePlatesEnabled {
   widgetClassPath=/FortniteCoreUI/UI/Replay/RotatorSelectorReplaySettings
   bInherited=true
   bIsVariable=true
   parent=ScrollBox_Tabs
   widget=Rotator_NamePlatesEnabled
   slot {class=ScrollBoxSlot, properties {padding=<ltrb 15 15 15 12>}}
  }
  TeamColor {
   widgetClassPath=&RotatorSelectorReplaySettings
   bIsVariable=true
   parent=ScrollBox_Tabs
   widget=TeamColor
   properties {formatting=Single Indent}
   slot {class=ScrollBoxSlot, properties {padding=<ltrb 15 0 15 12>}}
  }
  LowDetailDistance {
   widgetClassPath=/FortniteCoreUI/UI/Replay/OptionsMenuSliderReplay
   bIsVariable=true
   parent=ScrollBox_Tabs
   widget=LowDetailDistance
   properties {formatting=Single Indent, maxFractionalDigits=0, maxIntegralDigits=10, minFractionalDigits=0, valueSuffixText=" m", visibility=Collapsed}
   slot {class=ScrollBoxSlot, properties {padding=<ltrb 15 0 15 12>}}
  }
  ArrowDistance {
   widgetClassPath=&OptionsMenuSliderReplay
   bIsVariable=true
   parent=ScrollBox_Tabs
   widget=ArrowDistance
   properties {formatting=Single Indent, maxFractionalDigits=0, maxIntegralDigits=10, minFractionalDigits=0, valueSuffixText=" m", visibility=Collapsed}
   slot {class=ScrollBoxSlot, properties {padding=<ltrb 15 0 15 12>}}
  }
  OffDistance {
   widgetClassPath=&OptionsMenuSliderReplay
   bIsVariable=true
   parent=ScrollBox_Tabs
   widget=OffDistance
   properties {formatting=Single Indent, maxFractionalDigits=0, maxIntegralDigits=10, minFractionalDigits=0, valueSuffixText=" m", visibility=Collapsed}
   slot {class=ScrollBoxSlot, properties {padding=<ltrb 15 0 15 12>}}
  }
  ScalingOptionsEnabled {
   widgetClassPath=&RotatorSelectorReplaySettings
   bIsVariable=true
   parent=ScrollBox_Tabs
   widget=ScalingOptionsEnabled
   properties {formatting=Single Indent}
   slot {class=ScrollBoxSlot, properties {padding=<ltrb 15 0 15 12>}}
  }
  HighDetailScaleMin {
   widgetClassPath=&OptionsMenuSliderReplay
   bIsVariable=true
   parent=ScrollBox_Tabs
   widget=HighDetailScaleMin
   properties {formatting=Single Indent, visibility=Collapsed}
   slot {class=ScrollBoxSlot, properties {padding=<ltrb 15 0 15 12>}}
  }
  HighDetailScaleMax {
   widgetClassPath=&OptionsMenuSliderReplay
   bIsVariable=true
   parent=ScrollBox_Tabs
   widget=HighDetailScaleMax
   properties {formatting=Single Indent, visibility=Collapsed}
   slot {class=ScrollBoxSlot, properties {padding=<ltrb 15 0 15 12>}}
  }
  LowDetailScaleMin {
   widgetClassPath=&OptionsMenuSliderReplay
   bIsVariable=true
   parent=ScrollBox_Tabs
   widget=LowDetailScaleMin
   properties {formatting=Single Indent, visibility=Collapsed}
   slot {class=ScrollBoxSlot, properties {padding=<ltrb 15 0 15 12>}}
  }
  LowDetailScaleMax {
   widgetClassPath=&OptionsMenuSliderReplay
   bIsVariable=true
   parent=ScrollBox_Tabs
   widget=LowDetailScaleMax
   properties {formatting=Single Indent, visibility=Collapsed}
   slot {class=ScrollBoxSlot, properties {padding=<ltrb 15 0 15 12>}}
  }
  ArrowScaleMin {
   widgetClassPath=&OptionsMenuSliderReplay
   bIsVariable=true
   parent=ScrollBox_Tabs
   widget=ArrowScaleMin
   properties {formatting=Single Indent, visibility=Collapsed}
   slot {class=ScrollBoxSlot, properties {padding=<ltrb 15 0 15 12>}}
  }
  ArrowScaleMax {
   widgetClassPath=&OptionsMenuSliderReplay
   bIsVariable=true
   parent=ScrollBox_Tabs
   widget=ArrowScaleMax
   properties {formatting=Single Indent, visibility=Collapsed}
   slot {class=ScrollBoxSlot, properties {padding=<ltrb 15 0 15 12>}}
  }
  ArrowColorEnabled {
   widgetClassPath=&RotatorSelectorReplaySettings
   bIsVariable=true
   parent=ScrollBox_Tabs
   widget=ArrowColorEnabled
   properties {formatting=Single Indent}
   slot {class=ScrollBoxSlot, properties {padding=<ltrb 15 0 15 12>}}
  }
  Rotator_HudVisibility {
   widgetClassPath=&RotatorSelectorReplaySettings
   bIsVariable=true
   parent=ScrollBox_Tabs
   widget=Rotator_HudVisibility
   properties {description="Toggle between the Replay preset (which will show minimal HUD elements and hide everything with one button press) and the Broadcast preset, showing all elements including player health, the minimap, and more!"}
   slot {class=ScrollBoxSlot, properties {padding=<ltrb 15 0 15 12>}}
  }
  SquadID {
   widgetClassPath=&RotatorSelectorReplaySettings
   bIsVariable=true
   parent=ScrollBox_Tabs
   widget=SquadID
   slot {class=ScrollBoxSlot, properties {padding=<ltrb 15 0 15 12>}}
  }
  Rotator_DisplaySessionRanks {
   widgetClassPath=&RotatorSelectorReplaySettings
   bIsVariable=true
   parent=ScrollBox_Tabs
   widget=Rotator_DisplaySessionRanks
   slot {class=ScrollBoxSlot, properties {padding=<ltrb 15 0 15 12>}}
  }
  Search_SwapEmote {
   widgetClassPath=/FortUILibrary/UIKit/InputField/WBP_UIKit_InputField
   bIsVariable=true
   parent=ScrollBox_Tabs
   widget=Search_SwapEmote
   properties {
    backgroundMaterial=/FortUILibrary/Systems/InputField/MaterialInstances/MI_UIKit_InputField_Background_Fallback
    hint Text=Swap Emote
    holdData=&CommonUI_HoldData_Base_C
    icon=/FortUILibrary/Textures/Icons/T_UI_Icon_UIKit_Search
    useBackground=false
    heightOverride {enable=true, value=64}
    iconProperties {
     material=/FortUILibrary/Systems/InputField/MaterialInstances/MI_UIKit_InputFieldIcon
     image Size=<32 32>
     placement {
      horizontal Alignment=HAlign_Right
      vertical Alignment=VAlign_Center
      grid {column=1, column Span=0, layer=0, row=1, row Span=0, nudge=<0 0>}
      padding=<ltrb 0 0 0 0>
     }
    }
   }
   slot {class=ScrollBoxSlot, properties {padding=<ltrb 31 0 31 12>}}
  }
  Rotator_DisplayAugmentsHUD {
   widgetClassPath=&RotatorSelectorReplaySettings
   bIsVariable=true
   parent=ScrollBox_Tabs
   widget=Rotator_DisplayAugmentsHUD
   slot {class=ScrollBoxSlot, properties {padding=<ltrb 15 0 15 15>}}
  }
  EmoteExtension {
   widgetClassPath=&SizeBox
   bIsVariable=true
   parent=HorizontalBox_0
   widget=EmoteExtension
   properties {bOverride_WidthOverride=true, visibility=Collapsed, widthOverride=500}
   slot {class=HorizontalBoxSlot}
  }
  EmotesList {
   widgetClassPath=/Script/FortniteUI.FortItemListView
   bIsVariable=true
   parent=EmoteExtension
   widget=EmotesList
   properties {
    consumeMouseWheel=Always
    entryWidgetClass=/Game/UI/Replay/CustomizationItems/CustomizationListEntry.CustomizationListEntry_C
    numDesignerPreviewEntries=1
    scrollingAnimationInterpolationSpeed=15
    scrollBarPadding=<ltrb 8 0 2 0>
    scrollBarStyle {
     thickness=8
     draggedThumbImage {
      bIsDynamicallyLoaded=false
      drawAs=Box
      imageType=FullColor
      mirroring=NoMirror
      resourceName=../../../Engine/Content/Slate/Common/Scrollbar_Thumb.png
      resourceObject=null
      tiling=NoTile
      imageSize=<0 0>
      margin=<ltrb 0.25 0.25 0.25 0.25>
      outlineSettings {
       bUseBrushTransparency=false
       roundingType=HalfHeightRadius
       width=0
       color {colorUseRule=UseColor_Specified, specifiedColor=<rgba 0 0 0 0>}
       cornerRadii=<0 0 0 0>
      }
      tintColor {colorUseRule=UseColor_Specified, specifiedColor=<rgba 1 1 1 1>}
      uVRegion {bIsValid=false, max=<0 0>, min=<0 0>}
     }
     horizontalBackgroundImage {
      bIsDynamicallyLoaded=false
      drawAs=NoDrawType
      imageType=NoImage
      mirroring=NoMirror
      resourceName=None
      resourceObject=null
      tiling=NoTile
      imageSize=<0 0>
      margin=<ltrb 0 0 0 0>
      outlineSettings {
       bUseBrushTransparency=false
       roundingType=HalfHeightRadius
       width=0
       color {colorUseRule=UseColor_Specified, specifiedColor=<rgba 0 0 0 0>}
       cornerRadii=<0 0 0 0>
      }
      tintColor {colorUseRule=UseColor_Specified, specifiedColor=<rgba 1 1 1 1>}
      uVRegion {bIsValid=false, max=<0 0>, min=<0 0>}
     }
     horizontalBottomSlotImage {
      bIsDynamicallyLoaded=false
      drawAs=Image
      imageType=FullColor
      mirroring=NoMirror
      resourceName=../../../Engine/Content/Slate/Common/Scrollbar_Background_Horizontal.png
      resourceObject=null
      tiling=NoTile
      imageSize=<8 8>
      margin=<ltrb 0 0 0 0>
      outlineSettings {
       bUseBrushTransparency=false
       roundingType=HalfHeightRadius
       width=0
       color {colorUseRule=UseColor_Specified, specifiedColor=<rgba 0 0 0 0>}
       cornerRadii=<0 0 0 0>
      }
      tintColor {colorUseRule=UseColor_Specified, specifiedColor=<rgba 1 1 1 1>}
      uVRegion {bIsValid=false, max=<0 0>, min=<0 0>}
     }
     horizontalTopSlotImage {
      bIsDynamicallyLoaded=false
      drawAs=Image
      imageType=FullColor
      mirroring=NoMirror
      resourceName=../../../Engine/Content/Slate/Common/Scrollbar_Background_Horizontal.png
      resourceObject=null
      tiling=NoTile
      imageSize=<8 8>
      margin=<ltrb 0 0 0 0>
      outlineSettings {
       bUseBrushTransparency=false
       roundingType=HalfHeightRadius
       width=0
       color {colorUseRule=UseColor_Specified, specifiedColor=<rgba 0 0 0 0>}
       cornerRadii=<0 0 0 0>
      }
      tintColor {colorUseRule=UseColor_Specified, specifiedColor=<rgba 1 1 1 1>}
      uVRegion {bIsValid=false, max=<0 0>, min=<0 0>}
     }
     hoveredThumbImage {
      bIsDynamicallyLoaded=false
      drawAs=Box
      imageType=FullColor
      mirroring=NoMirror
      resourceName=../../../Engine/Content/Slate/Common/Scrollbar_Thumb.png
      resourceObject=null
      tiling=NoTile
      imageSize=<0 0>
      margin=<ltrb 0.25 0.25 0.25 0.25>
      outlineSettings {
       bUseBrushTransparency=false
       roundingType=HalfHeightRadius
       width=0
       color {colorUseRule=UseColor_Specified, specifiedColor=<rgba 0 0 0 0>}
       cornerRadii=<0 0 0 0>
      }
      tintColor {colorUseRule=UseColor_Specified, specifiedColor=<rgba 1 1 1 1>}
      uVRegion {bIsValid=false, max=<0 0>, min=<0 0>}
     }
     normalThumbImage {
      bIsDynamicallyLoaded=false
      drawAs=Box
      imageType=FullColor
      mirroring=NoMirror
      resourceName=../../../Engine/Content/Slate/Common/Scrollbar_Thumb.png
      resourceObject=null
      tiling=NoTile
      imageSize=<0 0>
      margin=<ltrb 0.25 0.25 0.25 0.25>
      outlineSettings {
       bUseBrushTransparency=false
       roundingType=HalfHeightRadius
       width=0
       color {colorUseRule=UseColor_Specified, specifiedColor=<rgba 0 0 0 0>}
       cornerRadii=<0 0 0 0>
      }
      tintColor {colorUseRule=UseColor_Specified, specifiedColor=<rgba 1 1 1 1>}
      uVRegion {bIsValid=false, max=<0 0>, min=<0 0>}
     }
     verticalBackgroundImage {
      bIsDynamicallyLoaded=false
      drawAs=NoDrawType
      imageType=NoImage
      mirroring=NoMirror
      resourceName=None
      resourceObject=null
      tiling=NoTile
      imageSize=<0 0>
      margin=<ltrb 0 0 0 0>
      outlineSettings {
       bUseBrushTransparency=false
       roundingType=HalfHeightRadius
       width=0
       color {colorUseRule=UseColor_Specified, specifiedColor=<rgba 0 0 0 0>}
       cornerRadii=<0 0 0 0>
      }
      tintColor {colorUseRule=UseColor_Specified, specifiedColor=<rgba 1 1 1 1>}
      uVRegion {bIsValid=false, max=<0 0>, min=<0 0>}
     }
     verticalBottomSlotImage {
      bIsDynamicallyLoaded=false
      drawAs=Image
      imageType=FullColor
      mirroring=NoMirror
      resourceName=../../../Engine/Content/Slate/Common/Scrollbar_Background_Vertical.png
      resourceObject=null
      tiling=NoTile
      imageSize=<8 8>
      margin=<ltrb 0 0 0 0>
      outlineSettings {
       bUseBrushTransparency=false
       roundingType=HalfHeightRadius
       width=0
       color {colorUseRule=UseColor_Specified, specifiedColor=<rgba 0 0 0 0>}
       cornerRadii=<0 0 0 0>
      }
      tintColor {colorUseRule=UseColor_Specified, specifiedColor=<rgba 1 1 1 1>}
      uVRegion {bIsValid=false, max=<0 0>, min=<0 0>}
     }
     verticalTopSlotImage {
      bIsDynamicallyLoaded=false
      drawAs=Image
      imageType=FullColor
      mirroring=NoMirror
      resourceName=../../../Engine/Content/Slate/Common/Scrollbar_Background_Vertical.png
      resourceObject=null
      tiling=NoTile
      imageSize=<8 8>
      margin=<ltrb 0 0 0 0>
      outlineSettings {
       bUseBrushTransparency=false
       roundingType=HalfHeightRadius
       width=0
       color {colorUseRule=UseColor_Specified, specifiedColor=<rgba 0 0 0 0>}
       cornerRadii=<0 0 0 0>
      }
      tintColor {colorUseRule=UseColor_Specified, specifiedColor=<rgba 1 1 1 1>}
      uVRegion {bIsValid=false, max=<0 0>, min=<0 0>}
     }
    }
   }
   slot {class=SizeBoxSlot}
  }
 ]
}
