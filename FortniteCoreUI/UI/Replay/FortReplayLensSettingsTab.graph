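# Lens settings tab of the replay menu, controlling spectator camera exposure, focal length, aperture, focus, rack focus and snap zoom.
# Widget events call the parent FortSpectatorLensSettingsTab setters, whose change events sync controls and visibility, with some options gated by camera type or shipping build.
# Grid and aspect-ratio selectors drive a spawned BPA_ReplayAspectRatio actor.
# Piloting light or fog temporarily forces a 24mm focal length, then restores it.
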WidgetBlueprint /FortniteCoreUI/UI/Replay/FortReplayLensSettingsTab
graphs [
 CloseTabGroup {nodes [
   CloseTabGroup() {is=title K2Node_FunctionEntry}
   > SetTabCollapsed!(
    execute
    self:Fort Replay Context Object Reference=Get FortReplayContext!() -> (ReturnValue:Fort Replay Context Object Reference) {is=title K2Node_GetBlueprintContextEx}
    bIsCollapsed:bool=true
   ) -> (then)
  ]}
 EventGraph {nodes [
   RefreshDOFSettings()
   > RefreshSnapZoomSettings()
   n1 = "Selection Changed (Rotator_AutoExposure)"!() -> (OutputDelegate, then, Selected Index:int=0) {is=title K2Node_ComponentBoundEvent}
   > SetAutoExposureEnabled!(
    execute
    self:Fort Spectator Lens Settings Tab Object Reference
    bAutoExposureEnabled:bool="integer > integer"(A=@n1.Selected Index, B=0) {is=title K2Node_CallFunction}
   ) -> (then)
   "Slider Changed (FocusDistanceSlider)"!() -> (OutputDelegate, then, Slider Value:double) {is=title K2Node_ComponentBoundEvent}
   > SetManualFocusDistance!(
    execute
    self:Fort Spectator Lens Settings Tab Object Reference
    ManualFocusDistance:float=GetMappedValue!(self:Options Menu Slider Replay Object Reference=Get FocusDistanceSlider() -> (FocusDistanceSlider:Options Menu Slider Replay Object Reference)) -> (MappedValue:double)
   ) -> (then)
   "On Hovered (FocalLengthSelector)"() {is=title K2Node_ComponentBoundEvent}
   > SetLatestHoveredWidget!(
    execute
    self
    WidgetIn:User Widget Object Reference=Get FocalLengthSelector() -> (FocalLengthSelector:Rotator Selector Replay Settings Object Reference)
   ) -> (then)
   "On Hovered (ApertureSelector)"() {is=title K2Node_ComponentBoundEvent}
   > SetLatestHoveredWidget(WidgetIn=Get ApertureSelector() -> (ApertureSelector:Rotator Selector Replay Settings Object Reference))
   "On Focused (FocusDistanceSlider)"() {is=title K2Node_ComponentBoundEvent}
   > SetLatestHoveredWidget(WidgetIn=Get FocusDistanceSlider())
   "Button Clicked (FocusTargetButton)"() {is=title K2Node_ComponentBoundEvent}
   > n2 = Cast To FortPlayerControllerSpectating!(execute, Object=GetOwningPlayer()) -> (then, CastFailed, AsFort Player Controller Spectating:Fort Player Controller Spectating Object Reference) {is=title K2Node_DynamicCast}
   > n3 = Cast To HoverDronePawn!(execute, Object=GetSpectatorPawn!(self:Player Controller Object Reference=@n2) -> (ReturnValue:Spectator Pawn Object Reference)) -> (then, CastFailed, AsHover Drone Pawn:Hover Drone Pawn Object Reference) {is=title K2Node_DynamicCast}
   > ToggleFocusMode(self:Hover Drone Pawn Object Reference=@n3)
   n4 = "Selection Changed (RackFocusSelector)"!() -> (OutputDelegate, then, Selected Index:int=0) {is=title K2Node_ComponentBoundEvent}
   > SetRackFocusEnabled!(execute, self:Fort Spectator Lens Settings Tab Object Reference, bRackFocusEnabled:bool="integer > integer"(A=@n4.Selected Index, B=0)) -> (then)
   "Slider Changed (StartPointSlider)"!() -> (OutputDelegate, then, Slider Value:double) {is=title K2Node_ComponentBoundEvent}
   n5 = GetMappedValue(self=Get StartPointSlider() -> (StartPointSlider:Options Menu Slider Replay Object Reference))
   > SetStartPointRackFocusDistance!(execute, self:Fort Spectator Lens Settings Tab Object Reference, StartPointRackFocusDistance:float=@n5) -> (then)
   > SetManualFocusDistance(ManualFocusDistance=@n5)
   "Slider Changed (EndPointSlider)"!() -> (OutputDelegate, then, Slider Value:double) {is=title K2Node_ComponentBoundEvent}
   n6 = GetMappedValue(self=Get EndPointSlider() -> (EndPointSlider:Options Menu Slider Replay Object Reference))
   > SetEndPointRackFocusDistance!(execute, self:Fort Spectator Lens Settings Tab Object Reference, EndPointRackFocusDistance:float=@n6) -> (then)
   > SetManualFocusDistance(ManualFocusDistance=@n6)
   "Slider Changed (RackPointStepSlider)"!() -> (OutputDelegate, then, Slider Value:double) {is=title K2Node_ComponentBoundEvent}
   > SetRackFocusDistanceStep!(
    execute
    self:Fort Spectator Lens Settings Tab Object Reference
    RackFocusDistanceStep:float=GetMappedValue(self=Get RackPointStepSlider() -> (RackPointStepSlider:Options Menu Slider Replay Object Reference))
   ) -> (then)
   "On Hovered (RackFocusSelector)"() {is=title K2Node_ComponentBoundEvent}
   > SetLatestHoveredWidget(WidgetIn=Get RackFocusSelector() -> (RackFocusSelector:Rotator Selector Replay Settings Object Reference))
   "On Focused (StartPointSlider)"() {is=title K2Node_ComponentBoundEvent}
   > SetLatestHoveredWidget(WidgetIn=Get StartPointSlider())
   "Slider Changed (FocalLengthSpeed)"!() -> (OutputDelegate, then, Slider Value:double) {is=title K2Node_ComponentBoundEvent}
   > SetFocalLengthSpeed!(
    execute
    self:Fort Spectator Lens Settings Tab Object Reference
    InFocalSpeed:float=GetMappedValue(self=Get FocalLengthSpeed() -> (FocalLengthSpeed:Options Menu Slider Replay Object Reference))
   ) -> (then)
   "On Focused (EndPointSlider)"() {is=title K2Node_ComponentBoundEvent}
   > SetLatestHoveredWidget(WidgetIn=Get EndPointSlider())
   "On Focused (RackPointStepSlider)"() {is=title K2Node_ComponentBoundEvent}
   > SetLatestHoveredWidget(WidgetIn=Get RackPointStepSlider())
   n7 = "Selection Changed (ToggleGridSelector)"!() -> (OutputDelegate, then, Selected Index:int=0) {is=title K2Node_ComponentBoundEvent}
   > n8 = Switch on Int!(execute:Exec, Selection:int=@n7.Selected Index) -> (Default:Exec, 0:Exec, 1:Exec)
   Is Valid(exec=@n8.0, InputObject=Get AspectRatioPPV() -> (AspectRatioPPV:BPA Replay Aspect Ratio Object Reference))
   > UpdateGrid!(execute, self:BPA Replay Aspect Ratio Object Reference=Get AspectRatioPPV(), Material:Material Interface Object Reference) -> (then)
   n9 = Is Valid(exec=@n8.1, InputObject=Get AspectRatioPPV())
   n10 = SpawnActor BPA Replay Aspect Ratio!(
    execute:Exec=@n9.Is Not Valid
    Class:Actor Class Reference=/FortniteCoreUI/UI/Replay/Materials/BPA_ReplayAspectRatio.BPA_ReplayAspectRatio_C
    SpawnTransform:Transform=Get Actor Transform(self=GetPlayerCameraManager!(PlayerIndex:int=0) -> (ReturnValue:Player Camera Manager Object Reference))
    CollisionHandlingOverride:ESpawnActorCollisionHandlingMethod Enum=AlwaysSpawn
    TransformScaleMethod:ESpawnActorScaleMethod Enum=MultiplyWithRoot
    Owner:Actor Object Reference
    Instigator:Pawn Object Reference
   ) -> (then:Exec, ReturnValue:BPA Replay Aspect Ratio Object Reference) {is=title K2Node_SpawnActorFromClass}
   > n11 = Set AspectRatioPPV(AspectRatioPPV:BPA Replay Aspect Ratio Object Reference=@n10) -> (Output_Get:BPA Replay Aspect Ratio Object Reference)
   UpdateGrid(execute=[@n9.Is Valid, @n11.then], self=Get AspectRatioPPV(), Material=/FortniteCoreUI/UI/Replay/Materials/M_ReplayGrid)
   n12 = "Selection Changed (FocusPlaneVisibility)"!() -> (OutputDelegate, then, Selected Index:int=0) {is=title K2Node_ComponentBoundEvent}
   > SetRackFocusPlaneVisibility!(execute, self:Fort Spectator Lens Settings Tab Object Reference, bShowRackFocusPlane:bool="integer > integer"(A=@n12.Selected Index, B=0)) -> (then)
   "On Hovered (FocusPlaneVisibility)"() {is=title K2Node_ComponentBoundEvent}
   > SetLatestHoveredWidget(WidgetIn=Get FocusPlaneVisibility() -> (FocusPlaneVisibility:Rotator Selector Replay Settings Object Reference))
   n13 = "On Value Commited (StartPointSlider)"!() -> (OutputDelegate, then, Value:double) {is=title K2Node_ComponentBoundEvent}
   > SetStartPointRackFocusDistance(StartPointRackFocusDistance=@n13.Value)
   > SetManualFocusDistance(ManualFocusDistance=@n13.Value)
   n14 = "Selection Changed (FocusSelector)"!() -> (OutputDelegate, then, Selected Index:int=0) {is=title K2Node_ComponentBoundEvent}
   > n15 = Switch on Int!(execute, Selection=@n14.Selected Index) -> (Default, 0, 1, 2:Exec)
   n16 = Branch(
    execute=@n15.2
    Condition="Equal (Enum)"(
     A:ESpectatorCameraType Enum=Get CurrentSpectatorCameraType() -> (CurrentSpectatorCameraType:ESpectatorCameraType Enum=ThirdPerson)
     B:ESpectatorCameraType Enum=DroneFree
    )
   )
   n17 = GetRuntimeOptions()
   > n18 = Branch(Condition=NOT Boolean(A=IsShippingBuild!(self:Fort Runtime Options Object Reference=@n17) -> (ReturnValue:bool)))
   > SetRackFocusEnabled(bRackFocusEnabled=false)
   > SetFreeDroneFocusMode!(execute, self:Fort Spectator Lens Settings Tab Object Reference, InFreeDroneFocusMode:EFreeDroneFocusMode Enum=Target) -> (then)
   > SetVisibility(self=Get FocusTargetButton() -> (FocusTargetButton:Options Menu Action Replay Object Reference), InVisibility=Visible)
   > SetVisibility(self=Get RackFocusSelector(), InVisibility=Collapsed)
   > Set CurrentFocusMode(CurrentFocusMode:int=2) -> (Output_Get:int=0)
   n19 = Branch(execute=[@n16.else, @n18.else], Condition="Equal (Integer)"(A=Get CurrentFocusMode() -> (CurrentFocusMode:int=0)))
   > n20 = Update Row Selector!(
    execute
    self:Rotator Selector Replay Settings Object Reference=Get FocusSelector() -> (FocusSelector:Rotator Selector Replay Settings Object Reference)
    Currently Selected:int=1
   ) -> (then)
   SetFreeDroneFocusMode(execute=[@n20.then, @n15.1], InFreeDroneFocusMode=Manual)
   > SetVisibility(self=Get FocusTargetButton(), InVisibility=Collapsed)
   > Set CurrentFocusMode(CurrentFocusMode=1) -> (Output_Get=0)
   n21 = Get CurrentSpectatorCameraType() -> (CurrentSpectatorCameraType=ThirdPerson)
   > Branch(Condition=NOT Boolean(A="Equal (Enum)"(A=@n21, B=Gameplay)))
   > n22 = Branch(Condition=IsShippingBuild(self=GetRuntimeOptions()))
   SetVisibility(execute=@n22.else, self=Get RackFocusSelector(), InVisibility=Visible)
   n23 = Update Row Selector(execute=@n19.else, self=Get FocusSelector(), Currently Selected=0)
   SetRackFocusEnabled(execute=[@n23.then, @n15.0], bRackFocusEnabled=false)
   > SetFreeDroneFocusMode(InFreeDroneFocusMode=Auto)
   > SetVisibility(self=[Get FocusTargetButton(), Get RackFocusSelector()], InVisibility=Collapsed)
   > Set CurrentFocusMode(CurrentFocusMode=0) -> (Output_Get=0)
   "Slider Changed (SnapZoomStartPoint)"!() -> (OutputDelegate, then, Slider Value:double) {is=title K2Node_ComponentBoundEvent}
   n24 = GetMappedValue(self=Get SnapZoomStartPoint() -> (SnapZoomStartPoint:Options Menu Slider Replay Object Reference))
   > SetSnapZoomLengthStart!(execute, self:Fort Spectator Lens Settings Tab Object Reference, IsSnapZoomLengthStart:float=@n24) -> (then)
   > n25 = Branch(Condition=Get bIsFirst() -> (bIsFirst:bool))
   > SetFocalLengthForceValue!(execute, self:Fort Spectator Lens Settings Tab Object Reference, InFocalLength:float=@n24) -> (then)
   Set bIsFirst(execute=@n25.else, bIsFirst:bool=true) -> (Output_Get:bool)
   "Slider Changed (SnapZoomEndPoint)"!() -> (OutputDelegate, then, Slider Value:double) {is=title K2Node_ComponentBoundEvent}
   n26 = GetMappedValue(self=Get SnapZoomEndPoint() -> (SnapZoomEndPoint:Options Menu Slider Replay Object Reference))
   > SetSnapZoomLengthEnd!(execute, self:Fort Spectator Lens Settings Tab Object Reference, IsSnapZoomLengthEnd:float=@n26) -> (then)
   > SetFocalLengthForceValue(InFocalLength=@n26)
   n27 = "On Text Committed (FocalLengthSelector)"!() -> (OutputDelegate, then, Text:double) {is=title K2Node_ComponentBoundEvent}
   > SetFocalLengthForceValue(InFocalLength=@n27.Text)
   n28 = "Selection Changed (FocalLengthSelector)"!() -> (OutputDelegate, then, Selected Index:int=0) {is=title K2Node_ComponentBoundEvent}
   > SetFocalLengthFromIdx!(execute, self:Fort Spectator Lens Settings Tab Object Reference, FocalLengthIdx:int=@n28.Selected Index) -> (then)
   n29 = "On Text Committed (ApertureSelector)"!() -> (OutputDelegate, then, Text:double) {is=title K2Node_ComponentBoundEvent}
   > SetApertureForceValue!(execute, self:Fort Spectator Lens Settings Tab Object Reference, InAperture:float=@n29.Text) -> (then)
   n30 = "On Value Commited (FocusDistanceSlider)"!() -> (OutputDelegate, then, Value:double) {is=title K2Node_ComponentBoundEvent}
   > SetManualFocusDistance(ManualFocusDistance=@n30.Value)
   n31 = "On Value Commited (FocalLengthSpeed)"!() -> (OutputDelegate, then, Value:double) {is=title K2Node_ComponentBoundEvent}
   > SetFocalLengthSpeed(InFocalSpeed=@n31.Value)
   n32 = "On Value Commited (SnapZoomStartPoint)"!() -> (OutputDelegate, then, Value:double) {is=title K2Node_ComponentBoundEvent}
   > SetSnapZoomLengthStart(IsSnapZoomLengthStart=@n32.Value)
   > SetFocalLengthForceValue(InFocalLength=@n32.Value)
   n33 = "On Value Commited (SnapZoomEndPoint)"!() -> (OutputDelegate, then, Value:double) {is=title K2Node_ComponentBoundEvent}
   > SetSnapZoomLengthEnd(IsSnapZoomLengthEnd=@n33.Value)
   > SetFocalLengthForceValue(InFocalLength=@n33.Value)
   "On Hovered (SnapZoomSelector)"() {is=title K2Node_ComponentBoundEvent}
   > SetLatestHoveredWidget(WidgetIn=Get SnapZoomSelector() -> (SnapZoomSelector:Rotator Selector Replay Settings Object Reference))
   "On Focused (SnapZoomStartPoint)"() {is=title K2Node_ComponentBoundEvent}
   > SetLatestHoveredWidget(WidgetIn=Get SnapZoomStartPoint())
   "On Focused (SnapZoomEndPoint)"() {is=title K2Node_ComponentBoundEvent}
   > SetLatestHoveredWidget(WidgetIn=Get SnapZoomEndPoint())
   "On Focused (FocalLengthSpeed)"() {is=title K2Node_ComponentBoundEvent}
   > SetLatestHoveredWidget(WidgetIn=Get FocalLengthSpeed())
   "Slider Changed (ExposureSlider)"!() -> (OutputDelegate, then, Slider Value:double) {is=title K2Node_ComponentBoundEvent}
   > SetManualExposureBias!(
    execute
    self:Fort Spectator Lens Settings Tab Object Reference
    ManualExposureBias:float=GetMappedValue(self=Get ExposureSlider() -> (ExposureSlider:Options Menu Slider Replay Object Reference))
   ) -> (then)
   n34 = "On Value Commited (EndPointSlider)"!() -> (OutputDelegate, then, Value:double) {is=title K2Node_ComponentBoundEvent}
   > SetEndPointRackFocusDistance(EndPointRackFocusDistance=@n34.Value)
   > SetManualFocusDistance(ManualFocusDistance=@n34.Value)
   n35 = "On Value Commited (ExposureSlider)"!() -> (OutputDelegate, then, Value:double) {is=title K2Node_ComponentBoundEvent}
   > SetManualExposureBias(ManualExposureBias=@n35.Value)
   "Button Hovered (FocusTargetButton)"() {is=title K2Node_ComponentBoundEvent}
   > SetLatestHoveredWidget(WidgetIn=Get FocusTargetButton())
   "On Hovered (ColorComponent_RackPlaneColor)"() {is=title K2Node_ComponentBoundEvent}
   > SetLatestHoveredWidget(WidgetIn=Get ColorComponent_RackPlaneColor() -> (ColorComponent_RackPlaneColor:Replay Color Picker Component Object Reference))
   "On Hovered (ColorComponent_RackPlaneAlpha)"() {is=title K2Node_ComponentBoundEvent}
   > SetLatestHoveredWidget(WidgetIn=Get ColorComponent_RackPlaneAlpha() -> (ColorComponent_RackPlaneAlpha:Replay Color Picker Component Object Reference))
   n36 = "Color Changed (ColorComponent_RackPlaneColor)"!() -> (OutputDelegate, then, Value:double) {is=title K2Node_ComponentBoundEvent}
   CloseTabGroup(execute=[@n36.then, @n37.then]) {is=title K2Node_CallFunction}
   n37 = "Color Changed (ColorComponent_RackPlaneAlpha)"!() -> (OutputDelegate, then, Value:double) {is=title K2Node_ComponentBoundEvent}
   "On Focused (Rotator_AutoExposure)"() {is=title K2Node_ComponentBoundEvent}
   > n38 = Cast To RotatorSelectorReplaySettings!(execute, Object=Get Rotator_AutoExposure() -> (Rotator_AutoExposure:Common User Widget Object Reference)) -> (then, CastFailed, AsRotator Selector Replay Settings:Rotator Selector Replay Settings Object Reference) {is=title K2Node_DynamicCast}
   > SetLatestHoveredWidget(WidgetIn=@n38)
   "On Focused (FocalLengthSelector)"() {is=title K2Node_ComponentBoundEvent}
   > SetLatestHoveredWidget(WidgetIn=Get FocalLengthSelector())
   n39 = "Selection Changed (SnapZoomSelector)"!() -> (OutputDelegate, then, Selected Index:int=0) {is=title K2Node_ComponentBoundEvent}
   > SetSnapZoomEnabled!(execute, self:Fort Spectator Lens Settings Tab Object Reference, IsSnapZoomEnabled:bool="integer > integer"(A=@n39.Selected Index, B=0)) -> (then)
   "On Focused (SnapZoomSelector)"() {is=title K2Node_ComponentBoundEvent}
   > SetLatestHoveredWidget(WidgetIn=Get SnapZoomSelector())
   "On Focused (ApertureSelector)"() {is=title K2Node_ComponentBoundEvent}
   > SetLatestHoveredWidget(WidgetIn=Get ApertureSelector())
   "On Focused (FocusSelector)"() {is=title K2Node_ComponentBoundEvent}
   > SetLatestHoveredWidget(WidgetIn=Get FocusSelector())
   "On Focused (RackFocusSelector)"() {is=title K2Node_ComponentBoundEvent}
   > SetLatestHoveredWidget(WidgetIn=Get RackFocusSelector())
   "On Focused (FocusTargetButton)"() {is=title K2Node_ComponentBoundEvent}
   > SetLatestHoveredWidget(WidgetIn=Get FocusTargetButton())
   "On Focused (FocusPlaneVisibility)"() {is=title K2Node_ComponentBoundEvent}
   > SetLatestHoveredWidget(WidgetIn=Get FocusPlaneVisibility())
   "On Focused (ToggleGridSelector)"() {is=title K2Node_ComponentBoundEvent}
   > SetLatestHoveredWidget(WidgetIn=Get ToggleGridSelector() -> (ToggleGridSelector:Rotator Selector Replay Settings Object Reference))
   n40 = "On Value Commited (RackPointStepSlider)"!() -> (OutputDelegate, then, Value:double) {is=title K2Node_ComponentBoundEvent}
   > SetRackFocusDistanceStep(RackFocusDistanceStep=@n40.Value)
   "On Hovered (Rotator_AutoExposure)"() {is=title K2Node_ComponentBoundEvent}
   > SetLatestHoveredWidget(WidgetIn=Get Rotator_AutoExposure())
   n41 = "Selection Changed (ApertureSelector)"!() -> (OutputDelegate, then, Selected Index:int=0) {is=title K2Node_ComponentBoundEvent}
   > SetApertureFromIdx!(execute, self:Fort Spectator Lens Settings Tab Object Reference, ApertureIdx:int=@n41.Selected Index) -> (then)
   n42 = "Selection Changed (AspectRatio)"!() -> (OutputDelegate, then, Selected Index:int=0) {is=title K2Node_ComponentBoundEvent}
   > n43 = Is Valid(InputObject=Get AspectRatioPPV())
   n44 = SpawnActor BPA Replay Aspect Ratio(
    execute=@n43.Is Not Valid
    Class=&BPA_ReplayAspectRatio_C
    SpawnTransform=Get Actor Transform(self=GetPlayerCameraManager(PlayerIndex=0))
    CollisionHandlingOverride=AlwaysSpawn
    TransformScaleMethod=MultiplyWithRoot
   )
   > n45 = Set AspectRatioPPV(AspectRatioPPV=@n44)
   n46 = Switch on Int!(execute=[@n43.Is Valid, @n45.then], Selection=@n42.Selected Index) -> (Default, 0, 1, 2, 3:Exec, 4:Exec, 5:Exec, 6:Exec)
   UpdateAspectRatio!(
    execute=@n46.0
    self:BPA Replay Aspect Ratio Object Reference=Get AspectRatioPPV()
    Material:Material Interface Object Reference
    BarSize:float=0
    Letterboxing:float=0
   ) -> (then)
   UpdateAspectRatio(execute=@n46.5, self=Get AspectRatioPPV(), Material=/FortniteCoreUI/UI/Replay/Materials/M_AspectRatio_Inst, BarSize=0.125, Letterboxing=1)
   UpdateAspectRatio(execute=@n46.6, self=Get AspectRatioPPV(), Material=&M_AspectRatio_Inst, BarSize=0.18, Letterboxing=1)
   UpdateAspectRatio(execute=@n46.1, self=Get AspectRatioPPV(), Material=&M_AspectRatio_Inst, BarSize=0.34, Letterboxing=0)
   UpdateAspectRatio(execute=@n46.2, self=Get AspectRatioPPV(), Material=&M_AspectRatio_Inst, BarSize=0.22, Letterboxing=0)
   UpdateAspectRatio(execute=@n46.3, self=Get AspectRatioPPV(), Material=&M_AspectRatio_Inst, BarSize=0.125, Letterboxing=0)
   UpdateAspectRatio(execute=@n46.4, self=Get AspectRatioPPV(), Material=&M_AspectRatio_Inst, BarSize=0.02, Letterboxing=1)
   "On Focused (ExposureSlider)"() {is=title K2Node_ComponentBoundEvent}
   > SetLatestHoveredWidget(WidgetIn=Get ExposureSlider())
   n47 = AdjustFocalLengthForPiloting() {is=title K2Node_CustomEvent}
   n48 = Cast To FortPlayerControllerSpectating(execute=[@n47.then, @n60.then], Object=GetPlayerController!(PlayerIndex:int=0) -> (ReturnValue:Player Controller Object Reference))
   > Set StoredFocalLength(StoredFocalLength:float=Get CurrentFocalLength!(self:Cine Camera Component Object Reference=GetCurrentSpectatorCameraComponent!(self:Fort Player Controller Spectating Object Reference=@n48) -> (ReturnValue:Fort Spectator Camera Component Object Reference)) -> (CurrentFocalLength:float)) -> (Output_Get:float)
   > SetFocalLengthForceValue(InFocalLength=24)
   > n49 = Cast To BP_ReplayPC_Athena!(execute, Object=GetPlayerController(PlayerIndex=0)) -> (then, CastFailed, AsBP Replay PC Athena:BP Replay PC Athena Object Reference) {is=title K2Node_DynamicCast}
   > ShowNotification!(execute, self:Fort Player Controller Spectating Object Reference=@n49, NotificationText:Text="LOCTEXT(\"Focal Length 24mm\")") -> (then)
   n50 = RevertFocalLengthAfterPiloting() {is=title K2Node_CustomEvent}
   n51 = Get StoredFocalLength() -> (StoredFocalLength:float)
   SetFocalLengthForceValue(execute=[@n50.then, @n61.then], InFocalLength=@n51)
   > n52 = Cast To BP_ReplayPC_Athena(Object=GetPlayerController(PlayerIndex=0))
   > ShowNotification(
    self=@n52
    NotificationText=Format Text!(
     Format="LOCTEXT(\"{Title} {Value}{mm}\")"
     Title:Text=Get FocalLengthText() -> (FocalLengthText:Text)
     Value:Text="To Text (Float)"(Value=@n51, MinimumIntegralDigits=1, MaximumIntegralDigits=10, MinimumFractionalDigits=0, MaximumFractionalDigits=0)
     mm:Text=Get MillimeterText() -> (MillimeterText:Text)
    ) -> (Result)
   )
   n53 = ScrollBarBehavior() {is=title K2Node_CustomEvent}
   n54 = Get CommonInputSubsystem!() -> (ReturnValue:Common Input Object Reference) {is=title K2Node_GetBlueprintContextEx}
   > Is Valid(InputObject=@n54)
   n55 = GetCurrentInputType(self=@n54)
   > n56 = Branch(Condition="Equal (Enum)"(A:ECommonInputType Enum=@n55, B:ECommonInputType Enum=Gamepad))
   > SetScrollWhenFocusChanges!(
    execute
    self:Scroll Box Object Reference=Get ScrollBox_Tabs() -> (ScrollBox_Tabs:Scroll Box Object Reference)
    NewScrollWhenFocusChanges:EScrollWhenFocusChanges Enum=AnimatedScroll
   ) -> (then)
   n57 = Get LatestHoveredWidget() -> (LatestHoveredWidget:User Widget Object Reference)
   > n58 = Is Valid(InputObject=@n57)
   n59 = Cast To RotatorSelectorReplaySettings(execute=@n58.Is Not Valid, Object=Get Rotator_AutoExposure())
   > SetFocus(self=@n59)
   SetFocus(execute=@n58.Is Valid, self=@n57)
   Branch(execute=@n56.else, Condition="Equal (Enum)"(A=@n55, B=MouseAndKeyboard))
   > SetScrollWhenFocusChanges(self=Get ScrollBox_Tabs(), NewScrollWhenFocusChanges=NoScroll)
   n60 = AdjustFocalLengthForFog() {is=title K2Node_CustomEvent}
   n61 = RevertFocalLengthFromFog() {is=title K2Node_CustomEvent}
   n62 = PilotLightChangedEvent!() -> (OutputDelegate, then, bIsEnabled:bool=~) {is=title K2Node_CustomEvent}
   > n63 = Branch(Condition=@n62.bIsEnabled)
   > AdjustFocalLengthForPiloting() {is=title K2Node_CallFunction}
   RevertFocalLengthAfterPiloting(execute=@n63.else) {is=title K2Node_CallFunction}
   n64 = PilotFogChangedEvent!() -> (OutputDelegate, then, bIsEnabled:bool=~) {is=title K2Node_CustomEvent}
   > n65 = Branch(Condition=@n64.bIsEnabled)
   > AdjustFocalLengthForFog() {is=title K2Node_CallFunction}
   RevertFocalLengthFromFog(execute=@n65.else) {is=title K2Node_CallFunction}
   n66 = HandleCloseTab!() -> (OutputDelegate, then, Selected Index:int) {is=title K2Node_CustomEvent}
   > CloseTabGroup()
   n67 = HandleCloseTabSlider!() -> (OutputDelegate, then, Slider Value:double=~) {is=title K2Node_CustomEvent}
   > CloseTabGroup()
   n68 = Close() {is=title K2Node_CustomEvent}
   > CloseTabGroup()
   n69 = Cast To BP_ReplayPC_Athena(Object=GetPlayerController(PlayerIndex=0))
   > Is Valid(InputObject=@n69)
   > Call Collapse Tab!(execute:Exec, self:BP Replay PC Athena Object Reference=@n69, Collapsed:bool=true) -> (then:Exec) {is=title K2Node_CallDelegate}
   Event Construct()
   n70 = Get FortReplayContext()
   > Bind Event to On Enable Pilot Light Event(self:Fort Replay Context Object Reference=@n70, Delegate=@n62) {is=title K2Node_AddDelegate}
   > Bind Event to On Enable Pilot Fog Event(self:Fort Replay Context Object Reference=@n70, Delegate=@n64) {is=title K2Node_AddDelegate}
   > SetNavigationRules()
   > n71 = Cast To RotatorSelectorReplaySettings(Object=Get Rotator_AutoExposure())
   > Bind Event to Selection Changed(
    self:Rotator Selector Replay Settings Object Reference=[
     @n71
     Get RackFocusSelector()
     Get FocusPlaneVisibility()
     Get ToggleGridSelector()
     Get FocalLengthSelector()
     Get SnapZoomSelector()
     Get ApertureSelector()
     Get FocusSelector()
    ]
    Delegate=@n66
   ) {is=title K2Node_AddDelegate}
   > Bind Event to Slider Changed(
    self:Options Menu Slider Replay Object Reference=[
     Get ExposureSlider()
     Get SnapZoomStartPoint()
     Get SnapZoomEndPoint()
     Get FocalLengthSpeed()
     Get FocusDistanceSlider()
     Get StartPointSlider()
     Get EndPointSlider()
     Get RackPointStepSlider()
    ]
    Delegate=@n67
   ) {is=title K2Node_AddDelegate}
   > Bind Event to Button Clicked(self:Options Menu Action Replay Object Reference=Get FocusTargetButton(), Delegate=@n68) {is=title K2Node_AddDelegate}
   n72 = Event OnApertureChanged!() -> (OutputDelegate, then, PresetApertureIdx:int=0) {is=title K2Node_Event}
   > Update Row Selector(self=Get ApertureSelector(), Currently Selected=@n72.PresetApertureIdx)
   n73 = Event OnFocalLengthChanged!() -> (OutputDelegate, then, FocalLengthTexts:Array of Texts, PresetFocalLengthIdx:int=0) {is=title K2Node_Event}
   > Update Row Selector(self=Get FocalLengthSelector(), Currently Selected=@n73.PresetFocalLengthIdx)
   Event PreConstruct()
   > InitializeWidgets()
   > SetPanelPadding()
   n74 = Event OnEndPointRackFocusDistanceChanged!() -> (OutputDelegate, then, EndPointRackFocusDistance:float) {is=title K2Node_Event}
   > UpdateSliderValueUnmapped!(execute, self:Options Menu Slider Replay Object Reference=Get EndPointSlider(), UnmappedValue:double=@n74.EndPointRackFocusDistance) -> (then)
   n75 = Event OnRackFocusDistanceStepChanged!() -> (OutputDelegate, then, RackFocusDistanceStep:float) {is=title K2Node_Event}
   > UpdateSliderValueUnmapped(self=Get RackPointStepSlider(), UnmappedValue=@n75.RackFocusDistanceStep)
   n76 = Event OnRackFocusPlaneVisibilityChanged!() -> (OutputDelegate, then, bRackFocusPlaneVisibilityEnabled:bool) {is=title K2Node_Event}
   > Update Row Selector(
    self=Get FocusPlaneVisibility()
    Currently Selected=Select(Option 0:int=0, Option 1:int=1, Index=@n76.bRackFocusPlaneVisibilityEnabled) -> (ReturnValue:int)
   )
   > SetVisibility(
    self=[Get ColorComponent_RackPlaneColor(), Get ColorComponent_RackPlaneAlpha()]
    InVisibility=Select(Option 0:ESlateVisibility Enum=Collapsed, Option 1:ESlateVisibility Enum=Visible, Index=@n76.bRackFocusPlaneVisibilityEnabled) -> (ReturnValue:ESlateVisibility Enum)
   )
   n77 = Event OnFreeDroneFocusModeChanged!() -> (OutputDelegate, then, FreeDroneFocusMode:EFreeDroneFocusMode Enum=Manual) {is=title K2Node_Event}
   > n78 = Switch on EFreeDroneFocusMode!(execute, Selection:EFreeDroneFocusMode Enum=@n77.FreeDroneFocusMode) -> (Manual:Exec, Auto:Exec, Target:Exec) {is=title K2Node_SwitchEnum}
   > Update Row Selector(self=Get FocusSelector(), Currently Selected=1)
   > UpdateFocusModeVisibility!(execute, self, FocusMode:int=1) -> (then)
   Update Row Selector(execute=@n78.Target, self=Get FocusSelector(), Currently Selected=2)
   > UpdateFocusModeVisibility(FocusMode=2)
   Update Row Selector(execute=@n78.Auto, self=Get FocusSelector(), Currently Selected=0)
   > UpdateFocusModeVisibility(FocusMode=0)
   n79 = Event OnSnapZoomEnabledChanged!() -> (OutputDelegate, then, IsSnapZoomEnabledEnabled:bool) {is=title K2Node_Event}
   > Update Row Selector(
    self=Get SnapZoomSelector()
    Currently Selected=Select(Option 0:int=0, Option 1:int=1, Index=@n79.IsSnapZoomEnabledEnabled) -> (ReturnValue:int)
   )
   > SetVisibility(
    self=[Get SnapZoomStartPoint(), Get SnapZoomEndPoint(), Get FocalLengthSpeed()]
    InVisibility=Select(Option 0:ESlateVisibility Enum=Collapsed, Option 1:ESlateVisibility Enum=Visible, Index=@n79.IsSnapZoomEnabledEnabled) -> (ReturnValue:ESlateVisibility Enum)
   )
   n80 = Event OnFocalLengthSpeedChanged!() -> (OutputDelegate, then, FocalLengthSpeed:float) {is=title K2Node_Event}
   > UpdateSliderValueUnmapped(self=Get FocalLengthSpeed(), UnmappedValue=@n80.FocalLengthSpeed)
   n81 = Event OnSnapZoomLengthStartChanged!() -> (OutputDelegate, then, IsSnapZoomLengthStart:float) {is=title K2Node_Event}
   > UpdateSliderValueUnmapped(self=Get SnapZoomStartPoint(), UnmappedValue=@n81.IsSnapZoomLengthStart)
   n82 = Event OnSnapZoomLengthEndChanged!() -> (OutputDelegate, then, IsSnapZoomLengthEnd:float) {is=title K2Node_Event}
   > UpdateSliderValueUnmapped(self=Get SnapZoomEndPoint(), UnmappedValue=@n82.IsSnapZoomLengthEnd)
   n83 = Event OnManualExposureBiasChanged!() -> (OutputDelegate, then, ManualExposureBias:float) {is=title K2Node_Event}
   > UpdateSliderValueUnmapped(self=Get ExposureSlider(), UnmappedValue=@n83.ManualExposureBias)
   n84 = Event OnCameraTypeChanged!() -> (
    OutputDelegate
    then
    FortPlayerControllerSpectating:Fort Player Controller Spectating Object Reference
    CameraType:ESpectatorCameraType Enum=ThirdPerson
   ) {is=title K2Node_Event}
   > n85 = Set CurrentSpectatorCameraType(CurrentSpectatorCameraType:ESpectatorCameraType Enum=@n84.CameraType) -> (Output_Get:ESpectatorCameraType Enum=ThirdPerson)
   > n86 = Branch(Condition="Equal (Enum)"(A:ESpectatorCameraType Enum=@n85, B:ESpectatorCameraType Enum=Gameplay))
   > Update Row Selector(self=Get FocusSelector(), Currently Selected=0)
   > n87 = Branch(Condition=Get IsRackFocusEnabled() -> (IsRackFocusEnabled:bool))
   > n88 = SetRackFocusEnabled(bRackFocusEnabled=false)
   SetFreeDroneFocusMode(execute=[@n88.then, @n87.else], InFreeDroneFocusMode=Auto)
   > SetVisibility(
    self=[
     Get RackFocusSelector()
     Get StartPointSlider()
     Get EndPointSlider()
     Get RackPointStepSlider()
     Get FocusDistanceSlider()
     Get FocusTargetButton()
    ]
    InVisibility=Collapsed
   )
   > Set CurrentFocusMode(CurrentFocusMode=0) -> (Output_Get=0)
   n89 = Switch on Int!(execute=@n86.else, Selection=Get CurrentFocusMode() -> (CurrentFocusMode=0)) -> (Default, 0, 1, 2)
   Update Row Selector(execute=@n89.1, self=Get FocusSelector(), Currently Selected=1)
   > SetVisibility(self=Get FocusDistanceSlider(), InVisibility=Visible)
   > n90 = Branch(Condition=IsShippingBuild(self=GetRuntimeOptions()))
   SetVisibility(execute=@n90.else, self=Get RackFocusSelector(), InVisibility=Visible)
   > Branch(Condition=Get IsRackFocusEnabled())
   > SetVisibility(self=[Get StartPointSlider(), Get EndPointSlider(), Get RackPointStepSlider()], InVisibility=Visible)
   SetVisibility(execute=@n89.2, self=Get FocusDistanceSlider(), InVisibility=Collapsed)
   > Update Row Selector(self=Get FocusSelector(), Currently Selected=2)
   > n91 = Branch(Condition=Get IsRackFocusEnabled())
   > n92 = SetRackFocusEnabled(bRackFocusEnabled=false)
   SetFreeDroneFocusMode(execute=[@n92.then, @n91.else], InFreeDroneFocusMode=Auto)
   > SetVisibility(
    self=[
     Get StartPointSlider()
     Get EndPointSlider()
     Get RackPointStepSlider()
     Get FocusDistanceSlider()
     Get FocusTargetButton()
     Get RackFocusSelector()
    ]
    InVisibility=Collapsed
   )
   > Set CurrentFocusMode(CurrentFocusMode=0) -> (Output_Get=0)
   Update Row Selector(execute=@n89.0, self=Get FocusSelector(), Currently Selected=0)
   > SetVisibility(self=Get FocusDistanceSlider(), InVisibility=Collapsed)
   n93 = GetRuntimeOptions()
   > Branch(Condition=NOT Boolean(A=IsShippingBuild(self=@n93)))
   > n94 = Branch(Condition=Get IsRackFocusEnabled())
   > n95 = SetRackFocusEnabled(bRackFocusEnabled=false)
   SetVisibility(
    execute=[@n95.then, @n94.else]
    self=[Get StartPointSlider(), Get EndPointSlider(), Get RackPointStepSlider(), Get RackFocusSelector()]
    InVisibility=Collapsed
   )
   Event On Activated()
   > Parent: On Activated() {is=title K2Node_CallParentFunction}
   > n96 = Branch(Condition=Get bIsFirstActivation() -> (bIsFirstActivation:bool))
   > n97 = Cast To RotatorSelectorReplaySettings(Object=Get Rotator_AutoExposure())
   > SetFocus(self=@n97)
   > n98 = Set bIsFirstActivation(bIsFirstActivation:bool=false) -> (Output_Get:bool)
   SetNavigationRules(execute=[@n98.then, @n96.else])
   > ScrollBarBehavior() {is=title K2Node_CallFunction}
   > Bind Event to On Keybinds Changed(
    self:Fort Global UIContext Object Reference=Get FortGlobalUIContext!() -> (ReturnValue:Fort Global UIContext Object Reference) {is=title K2Node_GetBlueprintContextEx}
    Delegate=@n53
   ) {is=title K2Node_AddDelegate}
   n99 = GetRuntimeOptions()
   > n100 = Branch(Condition=NOT Boolean(A=IsShippingBuild(self=@n99)))
   > SetVisibility(self=Get SnapZoomSelector(), InVisibility=SelfHitTestInvisible)
   SetVisibility(execute=@n100.else, self=Get SnapZoomSelector(), InVisibility=Collapsed)
   n101 = Event OnAutoExposureEnabledChanged!() -> (OutputDelegate, then, bAutoExposureEnabled:bool) {is=title K2Node_Event}
   > Update Row Selector(
    self=Cast To RotatorSelectorReplaySettings!(Object=Get Rotator_AutoExposure()) -> (AsRotator Selector Replay Settings, bSuccess:bool=~)
    Currently Selected=Select(Option 0:int=0, Option 1:int=1, Index=@n101.bAutoExposureEnabled) -> (ReturnValue:int)
   )
   > SetVisibility(
    self=Get ExposureSlider()
    InVisibility=Select(Option 0:ESlateVisibility Enum=Visible, Option 1:ESlateVisibility Enum=Collapsed, Index=@n101.bAutoExposureEnabled) -> (ReturnValue:ESlateVisibility Enum)
   )
   Event CenterOnTab() {is=title K2Node_Event}
   > Center On Widget(self:Rotator Selector Replay Settings Object Reference=Cast To RotatorSelectorReplaySettings(Object=Get Rotator_AutoExposure()) -> (bSuccess=~))
   n102 = Event OnAutoFocusEnabledChanged!() -> (OutputDelegate, then, bAutoFocusEnabled:bool) {is=title K2Node_Event}
   > n103 = Set IsAutoFocusEnabled(IsAutoFocusEnabled:bool=@n102.bAutoFocusEnabled) -> (Output_Get:bool)
   > SetVisibility(self=Get FocusDistanceSlider(), InVisibility=Select(Option 0=Visible, Option 1=Collapsed, Index=@n103))
   n104 = Get CurrentSpectatorCameraType() -> (CurrentSpectatorCameraType=ThirdPerson)
   > n105 = Branch(Condition=NOT Boolean(A="Equal (Enum)"(A=@n104, B=Gameplay)))
   Update Row Selector(
    execute=@n105.else
    self=Get FocusSelector()
    Currently Selected=Select(Option 0:int=1, Option 1:int=0, Index=Get IsAutoFocusEnabled() -> (IsAutoFocusEnabled:bool)) -> (ReturnValue:int)
   )
   n106 = Event OnRackFocusEnabledChanged!() -> (OutputDelegate, then, bRackFocusEnabled:bool) {is=title K2Node_Event}
   > Update Row Selector(self=Get RackFocusSelector(), Currently Selected=Select(Option 0=0, Option 1=1, Index=@n106.bRackFocusEnabled))
   > Set IsRackFocusEnabled(IsRackFocusEnabled:bool=@n106.bRackFocusEnabled) -> (Output_Get:bool)
   > SetVisibility(
    self=[Get StartPointSlider(), Get EndPointSlider(), Get RackPointStepSlider()]
    InVisibility=Select(Option 0:ESlateVisibility Enum=Collapsed, Option 1:ESlateVisibility Enum=Visible, Index=@n106.bRackFocusEnabled) -> (ReturnValue:ESlateVisibility Enum)
   )
   n107 = Event OnStartPointRackFocusDistanceChanged!() -> (OutputDelegate, then, StartPointRackFocusDistance:float) {is=title K2Node_Event}
   > UpdateSliderValueUnmapped(self=Get StartPointSlider(), UnmappedValue=@n107.StartPointRackFocusDistance)
   n108 = Event OnManualFocusDistanceChanged!() -> (OutputDelegate, then, ManualFocusDistance:float) {is=title K2Node_Event}
   > UpdateSliderValueUnmapped(self=Get FocusDistanceSlider(), UnmappedValue=@n108.ManualFocusDistance)
   Event ResetFocusedElement() {is=title K2Node_Event}
   > Is Valid(InputObject=Get LatestHoveredWidget())
   > n109 = Branch(Condition="Equal (Object)"(A=Get Rotator_AutoExposure(), B=Get LatestHoveredWidget()))
   > ResetAutoExposureEnabled(self:Fort Spectator Lens Settings Tab Object Reference)
   n110 = Branch(execute=@n109.else, Condition="Equal (Object)"(A=Get ExposureSlider(), B=Get LatestHoveredWidget()))
   > ResetManualExposureBias(self:Fort Spectator Lens Settings Tab Object Reference)
   n111 = Branch(execute=@n110.else, Condition="Equal (Object)"(A=Get FocalLengthSelector(), B=Get LatestHoveredWidget()))
   > ResetFocalLength(self:Fort Spectator Lens Settings Tab Object Reference)
   n112 = Branch(execute=@n111.else, Condition="Equal (Object)"(A=Get ApertureSelector(), B=Get LatestHoveredWidget()))
   > ResetAperture(self:Fort Spectator Lens Settings Tab Object Reference)
   n113 = Branch(execute=@n112.else, Condition="Equal (Object)"(B=Get LatestHoveredWidget()))
   > ResetAutoFocusEnabled(self:Fort Spectator Lens Settings Tab Object Reference)
   n114 = Branch(execute=@n113.else, Condition="Equal (Object)"(A=Get FocusDistanceSlider(), B=Get LatestHoveredWidget()))
   > ResetManualFocusDistance(self:Fort Spectator Lens Settings Tab Object Reference)
   n115 = Branch(execute=@n114.else, Condition="Equal (Object)"(A=Get RackFocusSelector(), B=Get LatestHoveredWidget()))
   > ResetRackFocusEnabled(self:Fort Spectator Lens Settings Tab Object Reference)
   n116 = Branch(execute=@n115.else, Condition="Equal (Object)"(A=Get StartPointSlider(), B=Get LatestHoveredWidget()))
   > ResetStartPointRackFocusDistance(self:Fort Spectator Lens Settings Tab Object Reference)
   n117 = Branch(execute=@n116.else, Condition="Equal (Object)"(A=Get EndPointSlider(), B=Get LatestHoveredWidget()))
   > ResetEndPointRackFocusDistance(self:Fort Spectator Lens Settings Tab Object Reference)
   n118 = Branch(execute=@n117.else, Condition="Equal (Object)"(A=Get RackPointStepSlider(), B=Get LatestHoveredWidget()))
   > ResetRackFocusDistanceStep(self:Fort Spectator Lens Settings Tab Object Reference)
   Branch(execute=@n118.else, Condition="Equal (Object)"(A=Get FocusPlaneVisibility(), B=Get LatestHoveredWidget()))
   > ResetRackFocusPlaneVisibility(self:Fort Spectator Lens Settings Tab Object Reference)
   n119 = Switch on Int!(execute, Selection=0) -> (Default, 0, 1)
   n120 = Is Valid(exec=@n119.1, InputObject=Get GridPPV() -> (GridPPV:BPA Replay Grid Object Reference))
   n121 = SpawnActor BPA Replay Grid!(
    execute:Exec=@n120.Is Not Valid
    Class:Actor Class Reference=/FortniteCoreUI/UI/Replay/Materials/BPA_ReplayGrid.BPA_ReplayGrid_C
    SpawnTransform:Transform=Get Actor Transform(self=GetPlayerCameraManager(PlayerIndex=0))
    CollisionHandlingOverride:ESpawnActorCollisionHandlingMethod Enum=AlwaysSpawn
    TransformScaleMethod:ESpawnActorScaleMethod Enum=MultiplyWithRoot
    Owner:Actor Object Reference
    Instigator:Pawn Object Reference
   ) -> (then:Exec, ReturnValue:BPA Replay Grid Object Reference) {is=title K2Node_SpawnActorFromClass}
   > Set GridPPV(GridPPV:BPA Replay Grid Object Reference=@n121) -> (Output_Get:BPA Replay Grid Object Reference)
   Is Valid(exec=@n119.0, InputObject=Get GridPPV())
   > Destroy Actor(self=Get GridPPV())
   SetLatestHoveredWidget(WidgetIn=Get FocusSelector())
   SetLatestHoveredWidget(WidgetIn=Get ToggleGridSelector())
   "On Hovered (FocusSelector)"() {is=title K2Node_ComponentBoundEvent}
   "On Hovered (ToggleGridSelector)"() {is=title K2Node_ComponentBoundEvent}
  ]}
 InitializeWidgets {nodes [
   InitializeWidgets() {is=title K2Node_FunctionEntry}
   n1 = Cast To RotatorSelectorReplaySettings(Object=Get Rotator_AutoExposure()) -> (bSuccess=~)
   > Initialize!(
    execute
    self:Rotator Selector Replay Settings Object Reference=@n1
    Display Text:Text="LOCTEXT(\"Auto Exposure\")"
    Options:Array of Texts=Make Array!("[0]":Text=Get OffText() -> (OffText:Text), "[1]":Text=Get OnText() -> (OnText:Text)) -> (Array:Array of Texts)
    Hover Text:Text
    Tab Tooltip:Common Text Object Reference
   ) -> (then)
   > SetLatestHoveredWidget(WidgetIn=@n1)
   n2 = GetMinMaxExposure!(self:Fort Spectator Lens Settings Tab Object Reference) -> (MinExposure:float, MaxExposure:float)
   > Update Slider!(
    execute
    self:Options Menu Slider Replay Object Reference=Get ExposureSlider()
    Slider Text:Text="LOCTEXT(\"Exposure\")"
    Slider Value:double=0.5
    Hover Text:Text
    Tooltip Text Block:Common Text Object Reference
    Min Max Value_X:double=@n2
    Min Max Value_Y:double=@n2.MaxExposure
   ) -> (then)
   > Initialize(
    self=Get ApertureSelector()
    Display Text="LOCTEXT(\"Aperture\")"
    Options=GetPresetApertureTexts!(self:Fort Spectator Lens Settings Tab Object Reference) -> (PresetApertureTexts:Array of Texts)
   )
   n3 = GetMinMaxFocusDistance!(self:Fort Spectator Lens Settings Tab Object Reference) -> (MinFocusDistance:float, MaxFocusDistance:float)
   n4 = Get FocusDistanceSlider()
   > Update Slider(self=@n4, Slider Text="LOCTEXT(\"Distance\")", Slider Value=0.5, Min Max Value_X=@n3, Min Max Value_Y=@n3.MaxFocusDistance)
   > UpdateSliderStep!(execute, self:Options Menu Slider Replay Object Reference=@n4, Step:double=25) -> (then)
   > Initialize(self=Get RackFocusSelector(), Display Text="LOCTEXT(\"Rack Focus\")", Options=Make Array("[0]"=Get OffText(), "[1]"=Get OnText()))
   n5 = GetMinMaxFocusDistance()
   > Update Slider(
    self=Get StartPointSlider()
    Slider Text="LOCTEXT(\"Start Point\")"
    Slider Value=0
    Min Max Value_X=@n5
    Min Max Value_Y=@n5.MaxFocusDistance
   )
   n6 = GetMinMaxFocusDistance()
   > Update Slider(self=Get EndPointSlider(), Slider Text="LOCTEXT(\"End Point\")", Slider Value=1, Min Max Value_X=@n6, Min Max Value_Y=@n6.MaxFocusDistance)
   > Update Slider(self=Get RackPointStepSlider(), Slider Text="LOCTEXT(\"Rack Speed\")", Slider Value=0.5, Min Max Value_X=1, Min Max Value_Y=800)
   > Initialize(self=Get ToggleGridSelector(), Display Text="LOCTEXT(\"Grid\")", Options=Make Array("[0]"=Get OffText(), "[1]"=Get OnText()))
   > Initialize(
    self=Get FocusPlaneVisibility()
    Display Text="LOCTEXT(\"Focus Plane\")"
    Options=Make Array("[0]"="LOCTEXT(\"Hidden\")", "[1]"="LOCTEXT(\"Visible\")")
   )
   > SetupComponent!(
    execute
    self:Replay Color Picker Component Object Reference=Get ColorComponent_RackPlaneColor()
    Name:"Text (by ref)"=Get RackFocusPlaneColorText() -> (RackFocusPlaneColorText:Text)
    SliderMaterial:Material Interface Object Reference=/FortniteCoreUI/UI/Replay/Materials/M_UI_Replay_ColorSlider_Hue
    InitialValue:float=146
    InMin:float=0
    InMax:float=360
    InStep:float=1
   ) -> (then)
   > SetupComponent(
    self=Get ColorComponent_RackPlaneAlpha()
    Name=Get RackFocusPlaneAlphaText() -> (RackFocusPlaneAlphaText:Text)
    InitialValue=0.4
    InMin=0
    InMax=1
    InStep=0.01
   )
   > Initialize(
    self=Get FocusSelector()
    Display Text="LOCTEXT(\"Focus\")"
    Options=Make Array!("[0]"="LOCTEXT(\"Auto\")", "[1]"="LOCTEXT(\"Manual\")", "[2]":Text="LOCTEXT(\"Target\")") -> (Array)
   )
   > Initialize(
    self=Get FocalLengthSelector()
    Display Text="LOCTEXT(\"Focal Length\")"
    Options=Make Array!(
     "[0]"="LOCTEXT(\"10 mm\")"
     "[1]"="LOCTEXT(\"14 mm\")"
     "[2]"="LOCTEXT(\"21 mm\")"
     "[3]":Text="LOCTEXT(\"24 mm\")"
     "[4]":Text="LOCTEXT(\"35 mm\")"
     "[5]":Text="LOCTEXT(\"50 mm\")"
     "[6]":Text="LOCTEXT(\"85 mm\")"
     "[7]":Text="LOCTEXT(\"100 mm\")"
     "[8]":Text="LOCTEXT(\"135 mm\")"
     "[9]":Text="LOCTEXT(\"200 mm\")"
    ) -> (Array)
   )
   > Update Slider!(
    execute
    self=Get FocalLengthSpeed()
    Slider Text="LOCTEXT(\"Zoom Speed\")"
    Slider Value=0
    Hover Text
    Tooltip Text Block
    Min Max Value:Vector 2D Structure=(X=0,Y=100)
   ) -> (then)
   > Initialize(self=Get SnapZoomSelector(), Display Text="LOCTEXT(\"Snap Zoom\")", Options=Make Array!("[0]"=Get OffText(), "[1]"=Get OnText()) -> (Array))
   n7 = GetMinMaxFocalLength!(self:Fort Spectator Lens Settings Tab Object Reference) -> (MinFocalLength:float, MaxFocalLength:float)
   > Update Slider!(
    execute
    self=Get SnapZoomEndPoint()
    Slider Text="LOCTEXT(\"End Point\")"
    Slider Value=1
    Hover Text
    Tooltip Text Block
    Min Max Value_X:double=@n7
    Min Max Value_Y:double=@n7.MaxFocalLength
   ) -> (then)
  ]}
 RefreshDOFSettings {nodes [
   RefreshDOFSettings() {is=title K2Node_FunctionEntry}
   > n1 = IsDepthOfFieldActive!(execute, self:Fort Replay Context Object Reference=Get FortReplayContext()) -> (then, ReturnValue:bool)
   > SetVisibility(InVisibility=Select(Option 0=Collapsed, Option 1=Visible, Index=@n1))
  ]}
 RefreshSnapZoomSettings {nodes [
   RefreshSnapZoomSettings() {is=title K2Node_FunctionEntry}
   > n1 = GetConsoleVariableBoolValue!(execute, VariableName:String=spectator.EnableSnapZoom) -> (then, ReturnValue:bool)
   > SetVisibility(InVisibility=Select(Option 0=Collapsed, Option 1=SelfHitTestInvisible, Index=@n1))
  ]}
 SetLatestHoveredWidget {nodes [
   n1 = SetLatestHoveredWidget!() -> (then, WidgetIn:User Widget Object Reference) {is=title K2Node_FunctionEntry}
   > Set LatestHoveredWidget(LatestHoveredWidget:User Widget Object Reference=@n1) -> (Output_Get:User Widget Object Reference)
  ]}
 SetNavigationRules {nodes [
   SetNavigationRules() {is=title K2Node_FunctionEntry}
   n1 = GetAllChildren(self=Get ScrollBox_Tabs())
   n2 = "Get (a copy)"(Array:Array of Widget Object References=@n1, Dimension 1=int - int(A=Length(TargetArray:Array of Widget Object References=@n1))) -> (Output:Widget Object Reference)
   n3 = "Get (a copy)"(Array=GetAllChildren(self=Get ScrollBox_Tabs()))
   > SetNavigationRuleExplicit!(execute, self:Widget Object Reference=@n3, Direction:EUINavigation Enum=Up, InWidget:Widget Object Reference=@n2) -> (then)
   > SetNavigationRuleExplicit(self=@n2, Direction=Down, InWidget=@n3)
  ]}
 SetPanelPadding {nodes [
   SetPanelPadding() {is=title K2Node_FunctionEntry}
   > n1 = For Each Loop(Array:Array of Widget Object References=GetAllChildren(self=Get ScrollBox_Tabs())) -> (Array Element:Widget Object Reference)
   > Set Loc_ScrollSlot(Loc_ScrollSlot:Scroll Box Slot Object Reference=SlotAsScrollBoxSlot!(Widget:Widget Object Reference=@n1) -> (ReturnValue:Scroll Box Slot Object Reference)) -> (Output_Get:Scroll Box Slot Object Reference)
   > n2 = Branch(Condition="Equal (Integer)"(A=@n1.Array Index))
   > SetPadding!(
    execute
    self:Scroll Box Slot Object Reference=Get Loc_ScrollSlot() -> (Loc_ScrollSlot:Scroll Box Slot Object Reference)
    InPadding_Left:float=15
    InPadding_Top:float=15
    InPadding_Right:float=15
    InPadding_Bottom:float=12
   ) -> (then)
   n3 = Length(TargetArray=GetAllChildren(self=Get ScrollBox_Tabs()))
   n4 = Branch(execute=@n2.else, Condition="Equal (Integer)"(A=@n1.Array Index, B=int - int(A=@n3)))
   > SetPadding(self=Get Loc_ScrollSlot(), InPadding_Left=15, InPadding_Top=0, InPadding_Right=15, InPadding_Bottom=15)
   SetPadding(execute=@n4.else, self=Get Loc_ScrollSlot(), InPadding_Left=15, InPadding_Top=0, InPadding_Right=15, InPadding_Bottom=12)
   > SetPadding(
    self:User Widget Object Reference=Get ColorComponent_RackPlaneColor()
    InPadding_Left=0
    InPadding_Top=0
    InPadding_Right=0
    InPadding_Bottom=40
   )
  ]}
 UpdateFocusModeVisibility {nodes [
   n1 = UpdateFocusModeVisibility!() -> (then, FocusMode:int) {is=title K2Node_FunctionEntry}
   > n2 = Switch on Int!(execute, Selection=@n1) -> (Default, 0, 1, 2:Exec)
   SetRackFocusEnabled(execute=@n2.0, bRackFocusEnabled=false)
   > SetVisibility(self=Get FocusTargetButton(), InVisibility=Collapsed)
   > SetVisibility(self=Get RackFocusSelector(), InVisibility=Collapsed)
   > Set CurrentFocusMode(CurrentFocusMode=0) -> (Output_Get=0)
   SetVisibility(execute=@n2.1, self=Get FocusTargetButton(), InVisibility=Collapsed)
   > Set CurrentFocusMode(CurrentFocusMode=1) -> (Output_Get=0)
   n3 = Get CurrentSpectatorCameraType() -> (CurrentSpectatorCameraType=ThirdPerson)
   > Branch(Condition=NOT Boolean(A="Equal (Enum)"(A=@n3, B=Gameplay)))
   > n4 = Branch(Condition=IsShippingBuild(self=GetRuntimeOptions()))
   SetVisibility(execute=@n4.else, self=Get RackFocusSelector(), InVisibility=Visible)
   SetRackFocusEnabled(execute=@n2.2, bRackFocusEnabled=false)
   > SetVisibility(self=Get FocusTargetButton(), InVisibility=Visible)
   > SetVisibility(self=Get RackFocusSelector(), InVisibility=Collapsed)
   > Set CurrentFocusMode(CurrentFocusMode=2) -> (Output_Get=0)
  ]}
]
variables [
 AspectRatioPPV {type=BPA_ReplayAspectRatio_C, default=null}
 AspectRatioWidget {type=Object, default=null}
 bIsFirst {type=boolean}
 bIsFirstActivation {type=boolean, default=true}
 CurrentFocusMode {type=integer}
 CurrentSpectatorCameraType {type=ESpectatorCameraType, default=ThirdPerson}
 FocalLengthText {type=string, default=Focal Length}
 GridPPV {type=BPA_ReplayGrid_C, default=null}
 IsAutoFocusEnabled {type=boolean}
 IsRackFocusEnabled {type=boolean}
 LatestHoveredWidget {type=UserWidget, default=null}
 MillimeterText {type=string, default=mm}
 OffText {type=string, default=Off}
 OnText {type=string, default=On}
 StoredFocalLength {type=number}
]
widgets {
 info {inheritedWidgetCount=3, parentClass=/Script/FortniteUI.FortSpectatorLensSettingsTab, rootWidgetClass=&Overlay, widgetCount=24}
 widgets [
  Overlay_0 {widgetClassPath=&Overlay, widget=Overlay_0}
  CommonBorder_0 {
   widgetClassPath=&CommonBorder
   parent=Overlay_0
   widget=CommonBorder_0
   properties {horizontalAlignment=HAlign_Center, padding=<ltrb 0 0 0 0>}
   slot {class=OverlaySlot}
  }
  SizeBox_0 {
   widgetClassPath=&SizeBox
   parent=CommonBorder_0
   widget=SizeBox_0
   properties {bOverride_HeightOverride=true, bOverride_WidthOverride=true, heightOverride=470, widthOverride=700}
   slot {class=BorderSlot, properties {horizontalAlignment=HAlign_Center, padding=<ltrb 0 0 0 0>}}
  }
  ScrollBox_Tabs {
   widgetClassPath=&ScrollBox
   bIsVariable=true
   parent=SizeBox_0
   widget=ScrollBox_Tabs
   properties {
    allowOverscroll=false
    bAnimateWheelScrolling=true
    consumeMouseWheel=Always
    navigationDestination=Center
    scrollbarPadding=<ltrb 0 10 10 10>
    scrollbarThickness=<15 15>
    widgetBarStyle {
     thickness=12
     draggedThumbImage {
      bIsDynamicallyLoaded=false
      drawAs=RoundedBox
      imageType=FullColor
      mirroring=NoMirror
      resourceName=../../../Engine/Content/Slate/Common/Scrollbar_Thumb.png
      resourceObject=null
      tiling=NoTile
      imageSize=<12 12>
      margin=<ltrb 0.25 0.25 0.25 0.25>
      outlineSettings {
       bUseBrushTransparency=false
       roundingType=FixedRadius
       width=2
       color {colorUseRule=UseColor_Specified, specifiedColor=<rgba 0 0 0 0>}
       cornerRadii=<6 6 6 6>
      }
      tintColor {colorUseRule=UseColor_Specified, specifiedColor=<rgba 1 1 1 1>}
      uVRegion {bIsValid=false, max=<0 0>, min=<0 0>}
     }
     horizontalBackgroundImage {
      bIsDynamicallyLoaded=false
      drawAs=Box
      imageType=FullColor
      mirroring=NoMirror
      resourceName=None
      resourceObject=/FortUILibrary/UIKit/ScrollBar/MaterialInstances/MI_UIKit_ScrollbarBackground
      tiling=NoTile
      imageSize=<12 12>
      margin=<ltrb 0.5 0.5 0.5 0.5>
      outlineSettings {
       bUseBrushTransparency=false
       roundingType=HalfHeightRadius
       width=0
       color {colorUseRule=UseColor_Specified, specifiedColor=<rgba 0 0 0 0>}
       cornerRadii=<0 0 0 0>
      }
      tintColor {colorUseRule=UseColor_Specified, specifiedColor=<rgba 1 1 1 1>}
      uVRegion {bIsValid=false, max=<0 0>, min=<0 0>}
     }
     horizontalBottomSlotImage {
      bIsDynamicallyLoaded=false
      drawAs=NoDrawType
      imageType=FullColor
      mirroring=NoMirror
      resourceName=../../../Engine/Content/Slate/Common/Scrollbar_Background_Horizontal.png
      resourceObject=null
      tiling=NoTile
      imageSize=<8 8>
      margin=<ltrb 0 0 0 0>
      outlineSettings {
       bUseBrushTransparency=false
       roundingType=HalfHeightRadius
       width=0
       color {colorUseRule=UseColor_Specified, specifiedColor=<rgba 0 0 0 0>}
       cornerRadii=<0 0 0 0>
      }
      tintColor {colorUseRule=UseColor_Specified, specifiedColor=<rgba 1 1 1 1>}
      uVRegion {bIsValid=false, max=<0 0>, min=<0 0>}
     }
     horizontalTopSlotImage {
      bIsDynamicallyLoaded=false
      drawAs=NoDrawType
      imageType=FullColor
      mirroring=NoMirror
      resourceName=../../../Engine/Content/Slate/Common/Scrollbar_Background_Horizontal.png
      resourceObject=null
      tiling=NoTile
      imageSize=<8 8>
      margin=<ltrb 0 0 0 0>
      outlineSettings {
       bUseBrushTransparency=false
       roundingType=HalfHeightRadius
       width=0
       color {colorUseRule=UseColor_Specified, specifiedColor=<rgba 0 0 0 0>}
       cornerRadii=<0 0 0 0>
      }
      tintColor {colorUseRule=UseColor_Specified, specifiedColor=<rgba 1 1 1 1>}
      uVRegion {bIsValid=false, max=<0 0>, min=<0 0>}
     }
     hoveredThumbImage {
      bIsDynamicallyLoaded=false
      drawAs=RoundedBox
      imageType=FullColor
      mirroring=NoMirror
      resourceName=../../../Engine/Content/Slate/Common/Scrollbar_Thumb.png
      resourceObject=null
      tiling=NoTile
      imageSize=<12 12>
      margin=<ltrb 0.25 0.25 0.25 0.25>
      outlineSettings {
       bUseBrushTransparency=false
       roundingType=FixedRadius
       width=2
       color {colorUseRule=UseColor_Specified, specifiedColor=<rgba 0 0 0 0>}
       cornerRadii=<6 6 6 6>
      }
      tintColor {colorUseRule=UseColor_Specified, specifiedColor=<rgba 1 1 1 1>}
      uVRegion {bIsValid=false, max=<0 0>, min=<0 0>}
     }
     normalThumbImage {
      bIsDynamicallyLoaded=false
      drawAs=RoundedBox
      imageType=FullColor
      mirroring=NoMirror
      resourceName=../../../Engine/Content/Slate/Common/Scrollbar_Thumb.png
      resourceObject=null
      tiling=NoTile
      imageSize=<12 12>
      margin=<ltrb 0.25 0.25 0.25 0.25>
      outlineSettings {
       bUseBrushTransparency=false
       roundingType=FixedRadius
       width=2
       color {colorUseRule=UseColor_Specified, specifiedColor=<rgba 0 0 0 0>}
       cornerRadii=<6 6 6 6>
      }
      tintColor {colorUseRule=UseColor_Specified, specifiedColor=<rgba 1 1 1 0.75>}
      uVRegion {bIsValid=false, max=<0 0>, min=<0 0>}
     }
     verticalBackgroundImage {
      bIsDynamicallyLoaded=false
      drawAs=Box
      imageType=NoImage
      mirroring=NoMirror
      resourceName=None
      resourceObject=&MI_UIKit_ScrollbarBackground
      tiling=NoTile
      imageSize=<12 12>
      margin=<ltrb 0.5 0.5 0.5 0.5>
      outlineSettings {
       bUseBrushTransparency=false
       roundingType=HalfHeightRadius
       width=0
       color {colorUseRule=UseColor_Specified, specifiedColor=<rgba 0 0 0 0>}
       cornerRadii=<0 0 0 0>
      }
      tintColor {colorUseRule=UseColor_Specified, specifiedColor=<rgba 1 1 1 1>}
      uVRegion {bIsValid=false, max=<0 0>, min=<0 0>}
     }
     verticalBottomSlotImage {
      bIsDynamicallyLoaded=false
      drawAs=NoDrawType
      imageType=FullColor
      mirroring=NoMirror
      resourceName=../../../Engine/Content/Slate/Common/Scrollbar_Background_Vertical.png
      resourceObject=null
      tiling=NoTile
      imageSize=<8 8>
      margin=<ltrb 0 0 0 0>
      outlineSettings {
       bUseBrushTransparency=false
       roundingType=HalfHeightRadius
       width=0
       color {colorUseRule=UseColor_Specified, specifiedColor=<rgba 0 0 0 0>}
       cornerRadii=<0 0 0 0>
      }
      tintColor {colorUseRule=UseColor_Specified, specifiedColor=<rgba 1 1 1 1>}
      uVRegion {bIsValid=false, max=<0 0>, min=<0 0>}
     }
     verticalTopSlotImage {
      bIsDynamicallyLoaded=false
      drawAs=NoDrawType
      imageType=FullColor
      mirroring=NoMirror
      resourceName=../../../Engine/Content/Slate/Common/Scrollbar_Background_Vertical.png
      resourceObject=null
      tiling=NoTile
      imageSize=<8 8>
      margin=<ltrb 0 0 0 0>
      outlineSettings {
       bUseBrushTransparency=false
       roundingType=HalfHeightRadius
       width=0
       color {colorUseRule=UseColor_Specified, specifiedColor=<rgba 0 0 0 0>}
       cornerRadii=<0 0 0 0>
      }
      tintColor {colorUseRule=UseColor_Specified, specifiedColor=<rgba 1 1 1 1>}
      uVRegion {bIsValid=false, max=<0 0>, min=<0 0>}
     }
    }
    widgetStyle {
     barThickness=12
     bottomShadowBrush {
      bIsDynamicallyLoaded=false
      drawAs=NoDrawType
      imageType=FullColor
      mirroring=NoMirror
      resourceName=../../../Engine/Content/Slate/Common/ScrollBoxShadowBottom.png
      resourceObject=null
      tiling=NoTile
      imageSize=<16 8>
      margin=<ltrb 0.5 0 0.5 1>
      outlineSettings {
       bUseBrushTransparency=false
       roundingType=HalfHeightRadius
       width=0
       color {colorUseRule=UseColor_Specified, specifiedColor=<rgba 0 0 0 0>}
       cornerRadii=<0 0 0 0>
      }
      tintColor {colorUseRule=UseColor_Specified, specifiedColor=<rgba 1 1 1 1>}
      uVRegion {bIsValid=false, max=<0 0>, min=<0 0>}
     }
     horizontalScrolledContentPadding=<ltrb 0 0 1 0>
     leftShadowBrush {
      bIsDynamicallyLoaded=false
      drawAs=NoDrawType
      imageType=FullColor
      mirroring=NoMirror
      resourceName=../../../Engine/Content/Slate/Common/ScrollBoxShadowLeft.png
      resourceObject=null
      tiling=NoTile
      imageSize=<8 16>
      margin=<ltrb 1 0.5 0 0.5>
      outlineSettings {
       bUseBrushTransparency=false
       roundingType=HalfHeightRadius
       width=0
       color {colorUseRule=UseColor_Specified, specifiedColor=<rgba 0 0 0 0>}
       cornerRadii=<0 0 0 0>
      }
      tintColor {colorUseRule=UseColor_Specified, specifiedColor=<rgba 1 1 1 1>}
      uVRegion {bIsValid=false, max=<0 0>, min=<0 0>}
     }
     rightShadowBrush {
      bIsDynamicallyLoaded=false
      drawAs=NoDrawType
      imageType=FullColor
      mirroring=NoMirror
      resourceName=../../../Engine/Content/Slate/Common/ScrollBoxShadowRight.png
      resourceObject=null
      tiling=NoTile
      imageSize=<8 16>
      margin=<ltrb 0 0.5 1 0.5>
      outlineSettings {
       bUseBrushTransparency=false
       roundingType=HalfHeightRadius
       width=0
       color {colorUseRule=UseColor_Specified, specifiedColor=<rgba 0 0 0 0>}
       cornerRadii=<0 0 0 0>
      }
      tintColor {colorUseRule=UseColor_Specified, specifiedColor=<rgba 1 1 1 1>}
      uVRegion {bIsValid=false, max=<0 0>, min=<0 0>}
     }
     topShadowBrush {
      bIsDynamicallyLoaded=false
      drawAs=NoDrawType
      imageType=FullColor
      mirroring=NoMirror
      resourceName=../../../Engine/Content/Slate/Common/ScrollBoxShadowTop.png
      resourceObject=null
      tiling=NoTile
      imageSize=<16 8>
      margin=<ltrb 0.5 1 0.5 0>
      outlineSettings {
       bUseBrushTransparency=false
       roundingType=HalfHeightRadius
       width=0
       color {colorUseRule=UseColor_Specified, specifiedColor=<rgba 0 0 0 0>}
       cornerRadii=<0 0 0 0>
      }
      tintColor {colorUseRule=UseColor_Specified, specifiedColor=<rgba 1 1 1 1>}
      uVRegion {bIsValid=false, max=<0 0>, min=<0 0>}
     }
     verticalScrolledContentPadding=<ltrb 0 0 0 1>
    }
   }
   slot {class=SizeBoxSlot}
  }
  Rotator_AutoExposure {
   widgetClassPath=&/RotatorSelectorReplaySettings
   bInherited=true
   bIsVariable=true
   parent=ScrollBox_Tabs
   widget=Rotator_AutoExposure
   properties {description="When enabled, the camera's settings will automatically adjust to optimize the brightness of the image."}
   slot {class=ScrollBoxSlot, properties {padding=<ltrb 15 15 15 12>}}
  }
  ExposureSlider {
   widgetClassPath=&/OptionsMenuSliderReplay
   bIsVariable=true
   parent=ScrollBox_Tabs
   widget=ExposureSlider
   properties {bUseCheckBox=true, formatting=Single Indent, title=Exposure}
   slot {class=ScrollBoxSlot, properties {padding=<ltrb 15 0 15 12>}}
  }
  FocalLengthSelector {
   widgetClassPath=&RotatorSelectorReplaySettings
   bIsVariable=true
   parent=ScrollBox_Tabs
   widget=FocalLengthSelector
   properties {
    bUseDefaultOption=true
    defaultOption=4
    description="The distance between the lens' optical center and the camera's sensor, measured in millimeters (mm). Determines how much of a scene is captured in a photograph, and how wide or tight the shot is."
    selectedIndex=4
    title=Focal Length
   }
   slot {class=ScrollBoxSlot, properties {padding=<ltrb 15 0 15 12>}}
  }
  SnapZoomSelector {
   widgetClassPath=&RotatorSelectorReplaySettings
   bIsVariable=true
   parent=ScrollBox_Tabs
   widget=SnapZoomSelector
   properties {description="A camera technique that quickly zooms in or out to create a dramatic effect (also known as a whip zoom or crash zoom)."}
   slot {class=ScrollBoxSlot, properties {padding=<ltrb 15 0 15 12>}}
  }
  SnapZoomStartPoint {
   widgetClassPath=&OptionsMenuSliderReplay
   bIsVariable=true
   parent=ScrollBox_Tabs
   widget=SnapZoomStartPoint
   properties {
    bUseCheckBox=true
    formatting=Single Indent
    maxFractionalDigits=0
    maxIntegralDigits=3
    minFractionalDigits=0
    slider Value=0.13
    title=Start Point
    valueSuffixText=" mm"
    visibility=Collapsed
    minMaxValue=<10 200>
   }
   slot {class=ScrollBoxSlot, properties {padding=<ltrb 15 0 15 12>}}
  }
  SnapZoomEndPoint {
   widgetClassPath=&OptionsMenuSliderReplay
   bIsVariable=true
   parent=ScrollBox_Tabs
   widget=SnapZoomEndPoint
   properties {
    bUseCheckBox=true
    formatting=Single Indent
    maxFractionalDigits=0
    maxIntegralDigits=3
    minFractionalDigits=0
    title=End Point
    valueSuffixText=" mm"
    visibility=Collapsed
   }
   slot {class=ScrollBoxSlot, properties {padding=<ltrb 15 0 15 12>}}
  }
  FocalLengthSpeed {
   widgetClassPath=&OptionsMenuSliderReplay
   bIsVariable=true
   parent=ScrollBox_Tabs
   widget=FocalLengthSpeed
   properties {
    bUseCheckBox=true
    description=The speed at which the camera will zoom from the Start Point to the End Point.
    formatting=Single Indent
    maxFractionalDigits=0
    maxIntegralDigits=3
    minFractionalDigits=0
    valueSuffixText=" mm/sec"
    visibility=Collapsed
   }
   slot {class=ScrollBoxSlot, properties {padding=<ltrb 15 0 15 12>}}
  }
  ApertureSelector {
   widgetClassPath=&RotatorSelectorReplaySettings
   bIsVariable=true
   parent=ScrollBox_Tabs
   widget=ApertureSelector
   properties {
    bUseDefaultOption=true
    defaultOption=4
    description=Controls how much light enters the camera. The size of the aperture can be adjusted to control the exposure and depth of field in a photograph.
    selectedIndex=4
    title=Aperture
   }
   slot {class=ScrollBoxSlot, properties {padding=<ltrb 15 0 15 12>}}
  }
  FocusSelector {
   widgetClassPath=&RotatorSelectorReplaySettings
   bIsVariable=true
   parent=ScrollBox_Tabs
   widget=FocusSelector
   properties {description="Determines how sharply the camera lens will render the subject of the frame. This will happen automatically, unless otherwise specified. "}
   slot {class=ScrollBoxSlot, properties {padding=<ltrb 15 0 15 12>}}
  }
  FocusDistanceSlider {
   widgetClassPath=&OptionsMenuSliderReplay
   bIsVariable=true
   parent=ScrollBox_Tabs
   widget=FocusDistanceSlider
   properties {
    bUseCheckBox=true
    formatting=Single Indent
    maxFractionalDigits=0
    maxIntegralDigits=10
    minFractionalDigits=0
    title=Distance
    valueSuffixText=" cm"
    visibility=Collapsed
   }
   slot {class=ScrollBoxSlot, properties {padding=<ltrb 15 0 15 12>}}
  }
  RackFocusSelector {
   widgetClassPath=&RotatorSelectorReplaySettings
   bIsVariable=true
   parent=ScrollBox_Tabs
   widget=RackFocusSelector
   properties {
    description="Change the focus of the lens from one subject (Start Point) to another (End Point) during a shot."
    formatting=Single Indent
    title=Rack Focus
    visibility=Collapsed
   }
   slot {class=ScrollBoxSlot, properties {padding=<ltrb 15 0 15 12>}}
  }
  StartPointSlider {
   widgetClassPath=&OptionsMenuSliderReplay
   bIsVariable=true
   parent=ScrollBox_Tabs
   widget=StartPointSlider
   properties {
    bUseCheckBox=true
    formatting=Double Indent
    maxFractionalDigits=0
    maxIntegralDigits=10
    minFractionalDigits=0
    valueSuffixText=" cm"
    visibility=Collapsed
   }
   slot {class=ScrollBoxSlot, properties {padding=<ltrb 15 0 15 12>}}
  }
  EndPointSlider {
   widgetClassPath=&OptionsMenuSliderReplay
   bIsVariable=true
   parent=ScrollBox_Tabs
   widget=EndPointSlider
   properties {
    bUseCheckBox=true
    formatting=Double Indent
    maxFractionalDigits=0
    maxIntegralDigits=10
    minFractionalDigits=0
    valueSuffixText=" cm"
    visibility=Collapsed
   }
   slot {class=ScrollBoxSlot, properties {padding=<ltrb 15 0 15 12>}}
  }
  RackPointStepSlider {
   widgetClassPath=&OptionsMenuSliderReplay
   bIsVariable=true
   parent=ScrollBox_Tabs
   widget=RackPointStepSlider
   properties {
    bUseCheckBox=true
    description=The speed at which the camera will shift focus from the Start Rack Point to the End Rack Point.
    formatting=Double Indent
    maxFractionalDigits=0
    maxIntegralDigits=10
    minFractionalDigits=0
    valueSuffixText=" cm/sec"
    visibility=Collapsed
   }
   slot {class=ScrollBoxSlot, properties {padding=<ltrb 15 0 15 12>}}
  }
  FocusTargetButton {
   widgetClassPath=&/OptionsMenuActionReplay
   bIsVariable=true
   parent=ScrollBox_Tabs
   widget=FocusTargetButton
   properties {
    buttonText=Select
    description="Point the reticle at a character, then press the button to select them. This will make them the focus of the camera."
    formatting=Single Indent
    title=Focus Target
    visibility=Collapsed
   }
   slot {class=ScrollBoxSlot, properties {padding=<ltrb 15 0 15 12>}}
  }
  FocusPlaneVisibility {
   widgetClassPath=&RotatorSelectorReplaySettings
   bIsVariable=true
   parent=ScrollBox_Tabs
   widget=FocusPlaneVisibility
   properties {description=Visualize where the focus plane is in world space.}
   slot {class=ScrollBoxSlot, properties {padding=<ltrb 15 0 15 12>}}
  }
  ColorComponent_RackPlaneColor {
   widgetClassPath=&/ColorPickerReplay
   bInherited=true
   bIsVariable=true
   parent=ScrollBox_Tabs
   widget=ColorComponent_RackPlaneColor
   properties {bIsFocusable=true, visibility=Collapsed}
   slot {class=ScrollBoxSlot, properties {padding=<ltrb 15 0 15 40>}}
  }
  ColorComponent_RackPlaneAlpha {
   widgetClassPath=&ColorPickerReplay
   bInherited=true
   bIsVariable=true
   parent=ScrollBox_Tabs
   widget=ColorComponent_RackPlaneAlpha
   properties {bColor=false, bIsFocusable=true, visibility=Collapsed}
   slot {class=ScrollBoxSlot, properties {padding=<ltrb 15 0 15 12>}}
  }
  ToggleGridSelector {
   widgetClassPath=&RotatorSelectorReplaySettings
   bIsVariable=true
   parent=ScrollBox_Tabs
   widget=ToggleGridSelector
   properties {description="Divides the image utilizing horizontal and vertical lines, a composition technique known as Rule of Thirds."}
   slot {class=ScrollBoxSlot, properties {padding=<ltrb 15 0 15 12>}}
  }
  AspectRatio {
   widgetClassPath=&RotatorSelectorReplaySettings
   bIsVariable=true
   parent=ScrollBox_Tabs
   widget=AspectRatio
   properties {
    description="Alter the aspect ratio for various platforms or for cinematic effect!"
    title=Aspect Ratio
    options [Default, "Portrait (9:16)", "Square (1:1)", "Fullscreen (4:3)", "Widescreen (1.85:1)", "CinemaScope (2.39:1)", "Ultra Panavision (2.76:1)"]
   }
   slot {class=ScrollBoxSlot, properties {padding=<ltrb 15 0 15 12>}}
  }
 ]
}
