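# Replay UI button widget that subclasses the UIKit regular button and adds no widgets or variables of its own.
# On PreConstruct, after the parent call, and on the On Widget Styled event, it runs one shared override chain.
# That chain sets the background to the Quiet fallback CTA material, then gives the inherited Block_Text the Loud text font material.
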WidgetBlueprint /FortniteCoreUI/UI/Replay/WBP_Replay_Button
graphs [EventGraph {nodes [
   n1 = Event PreConstruct()
   > n2 = Parent: PreConstruct(IsDesignTime=@n1.IsDesignTime)
   SetBackgroundMaterial!(
    execute=[@n2.then, @n3.then]
    self:WBP UIKit Button CTA Base Object Reference
    Background:Material Instance Object Reference=/FortUILibrary/UIKit/Buttons/MaterialInstances/CTAButtons/MI_UIKit_BgQuiet_Fallback
   ) -> (then)
   > SetFontMaterial!(
    execute
    self:WBP UIKit Block Text Object Reference=Get Block_Text() -> (Block_Text:WBP UIKit Block Text Object Reference)
    In Material:Material Instance Object Reference=/FortUILibrary/UIKit/Buttons/MaterialInstances/CTAButtons/MI_UIKit_TextLoud
   ) -> (then)
   n3 = Event On Widget Styled() {is=title K2Node_Event}
  ]}]
variables []
widgets {
 info {
  inheritedWidgetCount=0
  parentClass=/FortUILibrary/UIKit/Buttons/WBP_UIKit_Button_Regular.WBP_UIKit_Button_Regular_C
  rootWidgetClass=null
  widgetCount=0
 }
 widgets []
}
