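# WBP_Replay_CoachMark is the replay UI's coach mark widget, a child of the UIKit coach mark base that adds no widgets of its own.
# On PreConstruct it runs the parent logic, then turns off auto-wrap on the inherited description rich text and sets a fixed wrap width instead.
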WidgetBlueprint /FortniteCoreUI/UI/Replay/WBP_Replay_CoachMark
graphs [EventGraph {nodes [
   n1 = Event PreConstruct()
   > Parent: PreConstruct(IsDesignTime=@n1.IsDesignTime)
   n2 = Get RichText_Description() -> (RichText_Description:Common Rich Text Block Object Reference)
   > SetAutoWrapText!(execute, self:Rich Text Block Object Reference=@n2, InAutoTextWrap:bool=false) -> (then)
   > Set WrapTextAt!(execute, WrapTextAt:float=440, self:Text Layout Widget Object Reference=@n2) -> (then, Output_Get:float)
  ]}]
variables []
widgets {
 info {
  inheritedWidgetCount=0
  parentClass=/FortUILibrary/Systems/CoachMark/WBP_UIKit_CoachMark_Base.WBP_UIKit_CoachMark_Base_C
  rootWidgetClass=null
  widgetCount=0
 }
 widgets []
}
