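# WBP_Replay_Tabs is the vertical, icon-sized tab bar for the replay UI, offering Lens, Drone, World, Game and Interface tabs.
# It layers a UIKit TabGroup over a replay-material backplate, using replay tab buttons, hidden prompts and MainTabLeft/Right input actions.
# The graph only relays events through dispatchers: TabSelected passes the tab ID, TabClicked passes nothing, and mouse enter and leave fire OnHover and OnUnhover.
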WidgetBlueprint /FortniteCoreUI/UI/Replay/WBP_Replay_Tabs
graphs [
 EventGraph {nodes [
   n1 = "On UIKit Tab Selected (Tabs)"!() -> (OutputDelegate, then, TabID:Name=None, TabVM:MVVM Base Viewmodel Object Reference) {is=title K2Node_ComponentBoundEvent}
   > Call Tab Selected!(execute:Exec, self:WBP Replay Tabs Object Reference, TabID:Name=@n1.TabID) -> (then:Exec) {is=title K2Node_CallDelegate}
   "On UIKit Tab Clicked (Tabs)"!() -> (OutputDelegate, then, TabID:Name=None, TabVM:MVVM Base Viewmodel Object Reference) {is=title K2Node_ComponentBoundEvent}
   > Call Tab Clicked!(execute:Exec, self:WBP Replay Tabs Object Reference) -> (then:Exec) {is=title K2Node_CallDelegate}
   Event OnMouseEnter()
   > Call On Hover!(execute:Exec, self:WBP Replay Tabs Object Reference) -> (then:Exec) {is=title K2Node_CallDelegate}
   Event OnMouseLeave!() -> (OutputDelegate, then, MouseEvent:"Pointer Event Structure (by ref)") {is=title K2Node_Event}
   > Call On Unhover!(execute:Exec, self:WBP Replay Tabs Object Reference) -> (then:Exec) {is=title K2Node_CallDelegate}
  ]}
 OnHover {nodes [OnHover() {is=title K2Node_FunctionEntry}]}
 OnUnhover {nodes [OnUnhover() {is=title K2Node_FunctionEntry}]}
 TabClicked {nodes [TabClicked() {is=title K2Node_FunctionEntry}]}
 TabSelected {nodes [TabSelected!() -> (then, TabID:Name) {is=title K2Node_FunctionEntry}]}
]
variables [
 OnHover {}
 OnUnhover {}
 TabClicked {}
 TabSelected {}
]
widgets {
 info {inheritedWidgetCount=0, parentClass=&UserWidget, rootWidgetClass=&Overlay, widgetCount=3}
 widgets [
  Overlay_0 {widgetClassPath=&Overlay, widget=Overlay_0}
  Tabs_Backplate {
   widgetClassPath=&WBP_UIKit_Backplate
   parent=Overlay_0
   widget=Tabs_Backplate
   properties {backplateMaterial=/FortniteCoreUI/UI/Replay/Materials/M_Replay_Backplate_Inst, cornerRadius=Tab}
   slot {class=OverlaySlot, properties {horizontalAlignment=HAlign_Fill, verticalAlignment=VAlign_Fill}}
  }
  Tabs {
   widgetClassPath=/FortUILibrary/UIKit/Tabs/WBP_UIKit_TabGroup
   bIsVariable=true
   parent=Overlay_0
   widget=Tabs
   properties {
    animateReservedPromptSpace=false
    arrowButtonsAllowWrap=false
    backplateBrightness=Dark
    backplateClass=null
    buttonClass=&/WBP_Replay_TabButton.WBP_Replay_TabButton_C
    dividerClass=/Locker/MPL/WBP_MPLocker_Sidebar_Divider.WBP_MPLocker_Sidebar_Divider_C
    hidePrompts=true
    initialSelectedTabId=0
    initialSelectedTabIndex=0
    isVertical=true
    promptsBackplateClass=null
    reservedPromptSpace=60
    scrollBarVisibility=Visible
    tabGroupHorizontalAlignment=HAlign_Center
    tabSpacing=21
    timeToHidePrompts=0
    timeToShowPrompts=0
    useBackplate=false
    useInitialTabs=true
    usePromptBackplate=false
    buttonPadding=<ltrb 5 0 5 0>
    iconSize=<36 36>
    initialTabsData [
     {bEnableTabId=true, tabId=LENS, tabViewModel=&self:WidgetTree.Tabs.UIKitSelectableButtonViewModel_0}
     {bEnableTabId=true, tabId=DRONE, tabViewModel=&self:WidgetTree.Tabs.UIKitSelectableButtonViewModel_2}
     {bEnableTabId=true, tabId=WORLD, tabViewModel=&self:WidgetTree.Tabs.UIKitSelectableButtonViewModel_1}
     {bEnableTabId=true, tabId=GAME, tabViewModel=&self:WidgetTree.Tabs.UIKitSelectableButtonViewModel_3}
     {bEnableTabId=true, tabId=INTERFACE, tabViewModel=&self:WidgetTree.Tabs.UIKitSelectableButtonViewModel_4}
    ]
    minButtonSize=<58 58>
    nextTabInputActionData {dataTable=&InputActionDataTable, rowName=MainTabRight}
    previousTabInputActionData {dataTable=&InputActionDataTable, rowName=MainTabLeft}
    tabGroupPadding=<ltrb 15 21 15 21>
   }
   slot {class=OverlaySlot, properties {horizontalAlignment=HAlign_Fill, verticalAlignment=VAlign_Fill}}
  }
 ]
}
